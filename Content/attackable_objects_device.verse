
using { /Fortnite.com/Devices }
using { /Verse.org/Simulation }
using { /UnrealEngine.com/Temporary/Diagnostics }
using { HavocVerse.Combat }
using { HavocVerse.Debug }
using { HavocVerse.Devices }
using { HavocVerse.UI }
using { HavocVerse.Logging }
using { HavocVerse.Util }

# See https://dev.epicgames.com/documentation/en-us/uefn/create-your-own-device-in-verse for how to create a verse device.

# A Verse-authored creative device that can be placed in a level
AttackableObjectHandler<internal> := class() {
    Device : attackable_objects_device
    Index : int

    AttackableObjectHandlerFunction(Agent : agent) : void = {
        Device.OnObjectAttacked(Agent,Index)
    }
}

attackable_objects_device := class(creative_device):

    @editable p_ResourceManager: resource_manager_device = resource_manager_device{}
    @editable AttackableObjects: []attackable_object = array{}

    OnBegin<override>()<suspends>:void=
        var Index: int = 0
        for(AttackableObject:AttackableObjects):
            AttackableObject.PropManipulator.DamagedEvent.Subscribe(AttackableObjectHandler{Device:=Self, Index:=Index}.AttackableObjectHandlerFunction)
            set Index = Index + 1

    OnObjectAttacked(Agent:agent,Index:int) : void =
        if(Object:=AttackableObjects[Index]):
            CooldownCheck:=Object.GetIsOnCooldown()
            RestrictedItemsTest:=Object.PassRestrictedItemCheck(Agent)
            RequiredItemsTest:=Object.PassRequiredItemCheck(Agent)
            if(CooldownCheck? or not RestrictedItemsTest? or not RequiredItemsTest?):
                return
            else:
                Object.StartCooldown()
                Object.TriggerTrigger(Agent)
                Object.PlayAudio(Agent)
                
                GrantResource:=Object.GetGrantResource()
                if(GrantResource?):
                    ResourceID:=Object.GetResourceID()
                    ResourceAmount:=Object.GetResourceAmount()
                    p_ResourceManager.Give(Agent, ResourceID, ResourceAmount, ?SendFeedback := true)
                Object.GiveAccolade(Agent)

attackable_object<internal> := class<final><concrete>:
    @editable PropManipulator: prop_manipulator_device = prop_manipulator_device{}
    @editable AudioPlayer: ?audio_player_device = false
    @editable Trigger: ?trigger_device = false
    @editable GrantResource: logic = false
    @editable RestrictedItems: []conditional_button_device = array{}
    @editable RequiredItems: []conditional_button_device = array{}
    @editable ResourceID: string = ""
    @editable ResourceAmount: int = 5
    @editable CooldownTimer: float = 5.0
    @editable AccoladeDevice: ?accolades_device = false
    var IsOnCooldown: logic = false

    GiveAccolade(Agent:agent):void=
        if(MaybeAccolade:=AccoladeDevice?):
            MaybeAccolade.Award(Agent)

    PassRequiredItemCheck(Agent:agent):logic=
        if(ItemButton:=RequiredItems[0]):
            Results:=CheckForRequiredItems(Agent)
            return Results
        return true

    PassRestrictedItemCheck(Agent:agent):logic=
        if(ItemButton:=RestrictedItems[0]):
            Results:=CheckForRestrictedItems(Agent)
            return Results
        return true

    CheckForRestrictedItems(Agent:agent) : logic =
        for(ItemButton:RestrictedItems):
            if (ItemButton.IsHoldingItem[Agent, 0]): 
                return false
            if (ItemButton.IsHoldingItem[Agent, 1]): 
                return false
            if (ItemButton.IsHoldingItem[Agent, 2]): 
                return false
        return true
    
    CheckForRequiredItems(Agent:agent) : logic =
        for(ItemButton:RequiredItems):
            if (ItemButton.IsHoldingItem[Agent, 0]): 
                return true
            if (ItemButton.IsHoldingItem[Agent, 1]): 
                return true
            if (ItemButton.IsHoldingItem[Agent, 2]): 
                return true
        return false

    PlayAudio(Agent:agent):void=
        if(MaybeAudioPlayer := AudioPlayer?):
            MaybeAudioPlayer.Play(Agent)

    GetIsOnCooldown():logic=
        return IsOnCooldown

    GetGrantResource():logic=
        return GrantResource
    
    GetResourceID():string=
        return ResourceID
    
    GetResourceAmount():int=
        return ResourceAmount

    TriggerTrigger(Agent:agent):void=
        if(MaybeTrigger2 := Trigger?):
            MaybeTrigger2.Trigger(Agent)
    
    StartCooldown():void=
        set IsOnCooldown = true
        spawn:
            DoObjectCooldown()
    
    DoObjectCooldown()<suspends>:void=
        Sleep(CooldownTimer)
        set IsOnCooldown = false