
using { /Fortnite.com/Devices }
using { /Verse.org/Simulation }
using { /UnrealEngine.com/Temporary/Diagnostics }

Handler1<internal> := class() {
    Device : boost_notifier_device
    Parameter1 : int

    TriggerHandlerFunction(Agent : ?agent) : void = {
        if(MaybeAgent:=Agent?):
            Device.OnTriggerTriggered(MaybeAgent,Parameter1)
    }
}

boost_notifier_device := class(creative_device):
    @editable p_BoostNotifiers: []boost_notifier = array{}
    @editable HudMessage: hud_message_device = hud_message_device{}

    OnBegin<override>()<suspends>:void=
        var Index: int = 0
        for(Boost:p_BoostNotifiers):
            for(Trigger:Boost.Triggers):
                Trigger.TriggeredEvent.Subscribe(Handler1{Device:=Self, Parameter1:=Index}.TriggerHandlerFunction)
            set Index += 1

            
    OnTriggerTriggered(Agent:agent,Parameter1:int) : void =
        if(Boost:=p_BoostNotifiers[Parameter1]):
            HudMessage.SetText(StringToMessage(Boost.GetMessage()))
            HudMessage.Show(Agent)

    StringToMessage<localizes>(value:string) : message = "{value}"

boost_notifier<internal> := class<final><concrete>:
    @editable Triggers: []trigger_device = array{}
    @editable Message: string = ""

    GetMessage() : string =
        return Message