using { /Fortnite.com/Devices }
using { /Verse.org/Simulation }
using { /UnrealEngine.com/Temporary/SpatialMath }
using { /Verse.org/Random }
using { /Fortnite.com/Characters }
using { /Fortnite.com/Vehicles }
using { /Fortnite.com/Game }
using { /UnrealEngine.com/Temporary/UI }

using { HavocVerse.Combat }
using { HavocVerse.Debug }
using { HavocVerse.Devices }
using { HavocVerse.UI }
using { HavocVerse.Logging }
using { HavocVerse.Util }

chair_reward_device := class(creative_device):
    @editable p_ResourceManager: resource_manager_device = resource_manager_device{}

    @editable Chairs: []chairs = array{}

    OnBegin<override>()<suspends>:void=
        for(Chair:Chairs):
            ChairInstance := chair_instance{
                Device := Self
                Config := Chair
            }
            ChairInstance.Init()

chairs<internal> := class<final><concrete>:
    @editable ChairDevice: chair_device = chair_device{}
    @editable DelayBetweenRewards: float = 3.0
    @editable AccoladeDevice: accolades_device = accolades_device{}

    @editable ResourceID: string = ""
    @editable ResourceAmount: int = 5

chair_instance<internal> := class<final> {
    Device: chair_reward_device
    Config: chairs

    var m_SeatedAgent: ?tuple(agent,event()) = false

    Init<internal>(): void ={
        Config.ChairDevice.SeatedEvent.Subscribe(OnPlayerSeated)
        Config.ChairDevice.ExitedEvent.Subscribe(OnPlayerSeatExited)
    }

    OnPlayerSeated<private>(Agent:agent): void ={
        if (not m_SeatedAgent?, Player := player[Agent]) {
            AbortEvent := event(){}
            set m_SeatedAgent = Some(Player, AbortEvent)
            spawn:
                StartChairLoop(Player, AbortEvent)
        }
    }

    OnPlayerSeatExited<private>(Agent:agent): void ={
        if (PrevSeated := m_SeatedAgent?, PrevSeated(0) = Agent) {
            set m_SeatedAgent = None()
            PrevSeated(1).Signal()
        }
    }

    StartChairLoop<private>(Agent:player, AbortEvent:event())<suspends>: void ={
        race:
            AbortEvent.Await()
            loop {
                Sleep(Config.DelayBetweenRewards)
                if (m_SeatedAgent?(0) = Agent, Agent.IsActive[]) {
                    Device.p_ResourceManager.Give(Agent, Config.ResourceID, Config.ResourceAmount, ?SendFeedback := true)
                    Config.AccoladeDevice.Award(Agent)
                } else {
                    set m_SeatedAgent = None()
                    break
                }
            }
    }
}
