
using { /Fortnite.com/Devices }
using { /Verse.org/Simulation }
using { /UnrealEngine.com/Temporary/Diagnostics }

# See https://dev.epicgames.com/documentation/en-us/uefn/create-your-own-device-in-verse for how to create a verse device.

# A Verse-authored creative device that can be placed in a level
game_start_trigger_device := class(creative_device):

    @editable StartTrigger: trigger_device = trigger_device{}

    # Runs when the device is started in a running game
    OnBegin<override>()<suspends>:void=
        StartTrigger.Trigger()