using { /Fortnite.com/Devices }
using { /Verse.org/Simulation }
using { /UnrealEngine.com/Temporary/Diagnostics }
using { /UnrealEngine.com/Temporary/SpatialMath }
using { /Verse.org/Random }
using { /Fortnite.com/Characters }
using { /Fortnite.com/Vehicles }
using { /Fortnite.com/Game }
using { /UnrealEngine.com/Temporary/UI }

using { HavocVerse.Combat }
using { HavocVerse.Debug }
using { HavocVerse.Devices }
using { HavocVerse.UI }
using { HavocVerse.Logging }
using { HavocVerse.Util }
# See https://dev.epicgames.com/documentation/en-us/uefn/create-your-own-device-in-verse for how to create a verse device.

# A Verse-authored creative device that can be placed in a level
ghost_girl_device := class(creative_device):

    @editable HuntingVolume: volume_device = volume_device{}
    var m_HuntablePlayers: [player]void = map{}

    @editable DelayTime: float = 2.0

    var CurrentlyHuntingPlayers: logic = false
    var GhostAtHome: logic = true
    var CurrentlyCapturing: logic = false

    @editable GhostGirlProp: creative_prop = creative_prop{}
    @editable MoveToProp: creative_prop = creative_prop{}
    @editable HomeProp: creative_prop = creative_prop{}
    
    @editable CapturedVolume: volume_device = volume_device{}
    @editable CaptureCines: []cinematic_sequence_device = array{}
    @editable CaptureCineTime: float = 5.0
    var CaptureCinematicIndex: int = 0

    @editable HeartbeatAudio: audio_player_device = audio_player_device{}

    # Runs when the device is started in a running game
    OnBegin<override>()<suspends>:void=
        HuntingVolume.AgentEntersEvent.Subscribe(OnPlayerEnterArena)
        HuntingVolume.AgentExitsEvent.Subscribe(OnPlayerExitsArena)
        
        CapturedVolume.AgentEntersEvent.Subscribe(OnPlayerCaptured)

    OnPlayerEnterArena(Agent : agent): void =
        if (Player := player[Agent]) {
            if. set m_HuntablePlayers[Player] = void
        }

        HeartbeatAudio.Play(Agent)

        if(not CurrentlyHuntingPlayers?):
            set CurrentlyHuntingPlayers = true
            if(GhostAtHome?):
                spawn:
                    HuntLoop()

    OnPlayerExitsArena(Agent : agent): void =
        if (Player := player[Agent]) {
            set m_HuntablePlayers = m_HuntablePlayers.RemoveByKey(Player)(0)
        }

        HeartbeatAudio.Stop(Agent)

        if(CurrentlyHuntingPlayers?):
            CurrentPlayerCount := GetPlayerCount()
            if(CurrentPlayerCount<=0):
                set CurrentlyHuntingPlayers = false

    OnPlayerCaptured(Agent : agent): void =
        spawn:
            OnPlayerCaptured0(Agent)

    OnPlayerCaptured0(Agent : agent)<suspends>: void =
        if(not CurrentlyCapturing?):
            HeartbeatAudio.Stop(Agent)
            set CurrentlyCapturing = true
            PlayCaptureCinematic(Agent)
            Sleep(CaptureCineTime)
            if (FortCharacter := Agent.GetFortCharacter[]):
                FortCharacter.Damage(1000.0)
            set CurrentlyCapturing = false

    PlayCaptureCinematic(Agent : agent):void=
        if (Cinematic := CaptureCines[CaptureCinematicIndex]):
            Cinematic.Play(Agent)
            set CaptureCinematicIndex += 1
        else if (Cinematic := CaptureCines[0]):
            Cinematic.Play(Agent)
            set CaptureCinematicIndex = 1

    GetPlayerCount<public>()<transacts>:int=
        m_HuntablePlayers.Keys().Length

    HuntLoop()<suspends>:void=
        loop:
            MoveGhost()
            CheckGhostAtHome()
            if(GhostAtHome? and not CurrentlyHuntingPlayers?):
                break

    CheckGhostAtHome():void=
        GhostLocation := GhostGirlProp.GetTransform().Translation
        HomeLocation := HomeProp.GetTransform().Translation

        DistanceX : float = GhostLocation.X - HomeLocation.X
        DistanceY : float = GhostLocation.Y - HomeLocation.Y

        HomeDistance := Sqrt((DistanceX * DistanceX) + (DistanceY * DistanceY))
        
        if (HomeDistance < 1000.0):
            set GhostAtHome = true
        else:
            set GhostAtHome = false

    MoveGhost()<suspends>:void=
        var TargetLocation : vector3 = vector3{X:=0.0, Y:=0.0, Z:=0.0}
        var Roll : float = 0.0

        if(CurrentlyHuntingPlayers?):
            var TargettablePlayers:[]player = array{}
            for (Player:m_HuntablePlayers.Keys()) {
                set TargettablePlayers = TargettablePlayers + array{Player}
            }
            if(Player:=TargettablePlayers[0], Agent := agent[Player], FortCharacter := Agent.GetFortCharacter[]):
                set TargetLocation = FortCharacter.GetTransform().Translation

        else:
            set TargetLocation = HomeProp.GetTransform().Translation

        GhostLocation := GhostGirlProp.GetTransform().Translation
                    
        DistanceX : float = GhostLocation.X - TargetLocation.X
        DistanceY : float = GhostLocation.Y - TargetLocation.Y


        var NewRotation : rotation = MakeRotationFromYawPitchRollDegrees(0.0, 0.0, 0.0)

        if (LookDirection := (TargetLocation - GhostLocation).MakeUnitVector[]):
    
            Yaw := RadiansToDegrees(ArcTan(LookDirection.Y, LookDirection.X)) - 90.0
            Pitch := RadiansToDegrees(ArcTan(LookDirection.Z, Sqrt((LookDirection.X * LookDirection.X) + (LookDirection.Y * LookDirection.Y))))
            set Roll = RadiansToDegrees(ArcTan(LookDirection.X, LookDirection.Z)) - 90.0
    
            if (Roll <= 0.0):
                set Roll = Roll * -1
    
            if (Roll >= 10.0):
                set Roll = 10.0
    
            set NewRotation = MakeRotationFromYawPitchRollDegrees(Yaw, 0.0, 0.0)

                            
        GhostDistance := Sqrt((DistanceX * DistanceX) + (DistanceY * DistanceY))

        MoveToLocation := MoveToProp.GetTransform().Translation
                    
        var NewTargetLocation : vector3 = vector3{X:=MoveToLocation.X, Y:=MoveToLocation.Y, Z:=GhostLocation.Z}
                    
        GhostGirlProp.MoveTo(NewTargetLocation,NewRotation,DelayTime)
    
