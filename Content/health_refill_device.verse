
using { /Fortnite.com/Devices }
using { /Verse.org/Simulation }
using { /UnrealEngine.com/Temporary/Diagnostics }
using { /Fortnite.com/Characters }

# See https://dev.epicgames.com/documentation/en-us/uefn/create-your-own-device-in-verse for how to create a verse device.

# A Verse-authored creative device that can be placed in a level
health_refill_device := class(creative_device):

    @editable Triggers: []trigger_device = array{} 

    OnBegin<override>()<suspends>:void=
        for(Trigger:Triggers):
            Trigger.TriggeredEvent.Subscribe(OnTriggered)

    OnTriggered(Agent:?agent):void=
        if(MaybeAgent:=Agent?):
            spawn:
                OnTriggered0(MaybeAgent)

    OnTriggered0(Agent:agent)<suspends>:void=
        Sleep(0.5)
        if(FortCharacter := Agent.GetFortCharacter[]):
            FortCharacter.SetHealth(400.0)
            FortCharacter.SetShield(400.0)