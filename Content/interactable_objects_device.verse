using { /Fortnite.com/Devices }
using { /Verse.org/Simulation }
using { /UnrealEngine.com/Temporary/Diagnostics }
using { /UnrealEngine.com/Temporary/SpatialMath }
using { /Fortnite.com/Characters }
using { /Fortnite.com/Vehicles }
using { /Fortnite.com/Game }
using { /UnrealEngine.com/Temporary/UI }

using { HavocVerse.Combat }
using { HavocVerse.Debug }
using { HavocVerse.Devices }
using { HavocVerse.UI }
using { HavocVerse.Logging }
using { HavocVerse.Util }
# See https://dev.epicgames.com/documentation/en-us/uefn/create-your-own-device-in-verse for how to create a verse device.

# A Verse-authored creative device that can be placed in a level

InteractableObjectHandler<internal> := class() {
    Device : interactable_objects_device
    Index : int

    ObjectHandlerFunction(Agent : agent) : void = {
        Device.OnObjectInteracted(Agent,Index)
    }
}

interactable_objects_device := class(creative_device):
    
    @editable p_ResourceManager: resource_manager_device = resource_manager_device{}
    @editable DisableButtonsTriggers: []trigger_device = array{}
    @editable InteractableObjects: []interactable_object = array{}

    OnBegin<override>()<suspends>:void=
        var Index: int = 0
        for(Object:InteractableObjects):
            for(Button:Object.Buttons):
                Button.InteractedWithEvent.Subscribe(InteractableObjectHandler{Device:=Self, Index:=Index}.ObjectHandlerFunction)
            set Index = Index + 1

    OnObjectInteracted(Agent:agent,Index:int) : void =
        if(Object:=InteractableObjects[Index]):
            Object.GrantItems(Agent)
            Object.PlayAudio(Agent)
            Object.TriggerTrigger(Agent)
            Object.GiveAccolade(Agent)
            GrantResource:=Object.GetGrantResource()
            if(GrantResource?):
                ResourceID:=Object.GetResourceID()
                ResourceAmount:=Object.GetResourceAmount()
                p_ResourceManager.Give(Agent, ResourceID, ResourceAmount, ?SendFeedback := true)

interactable_object<internal> := class<final><concrete>:
    @editable Buttons: []button_device = array{}
    @editable ItemGranter: ?item_granter_device = false
    @editable AudioPlayer: ?audio_player_device = false
    @editable Trigger: ?trigger_device = false
    @editable GrantResource: logic = false
    @editable ResourceID: string = ""
    @editable ResourceAmount: int = 5
    @editable AccoladeDevice: ?accolades_device = false

    GiveAccolade(Agent:agent):void=
        if(MaybeAccolade:=AccoladeDevice?):
            MaybeAccolade.Award(Agent)

    GetGrantResource():logic=
        return GrantResource

    GetResourceID():string=
        return ResourceID

    GetResourceAmount():int=
        return ResourceAmount

    GrantItems(Agent:agent):void=
        if(MaybeItemGranter := ItemGranter?):
            MaybeItemGranter.GrantItem(Agent)

    PlayAudio(Agent:agent):void=
        if(MaybeAudioPlayer := AudioPlayer?):
            MaybeAudioPlayer.Play(Agent)

    TriggerTrigger(Agent:agent):void=
        if(MaybeTrigger2 := Trigger?):
            MaybeTrigger2.Trigger(Agent)



    
