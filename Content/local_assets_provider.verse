###################################################################
# Automatically generated using `gen_local_assets_provider.py`.
#
# Generated by: Brad @ 2024-10-06 10:14:44 UTC.
###################################################################

using { /Fortnite.com/Devices }
using { /Verse.org/Assets }
using { /UnrealEngine.com/Temporary/SpatialMath }

using { HavocVerse.Util }

CustomHUD<public> := module:
    Icons<public> := module:
    Notify<public> := module:
    Src<public> := module:
        HauntedHouse<public> := module:
Assets<public> := module:
    VFX<public> := module:
        Verse<public> := module:
BuyplateBillboards<public> := module:
DailyRewardsMenuUI<public> := module:
Fonts<public> := module:
    FeastBB<public> := module:
    QuicksandBold<public> := module:
GameplayRewardsMenuUI<public> := module:
HavocVerseUI<public> := module:
TemplateCommon<public> := module:
    BackendTooltips<public> := module:
UI<public> := module:
    Verse<public> := module:
        Rebirths<public> := module:

local_assets_provider_device := class(creative_device, assets_provider_device):
    CreateAssetsProvider<override>():assets_provider=
        TextureAssets := array{
            texture_metadata{Name := "CustomHUD/Icons/pumpkin", Asset := CustomHUD.Icons.pumpkin, Dimensions := option{vector2{X := 512.0, Y := 512.0}}}
            texture_metadata{Name := "CustomHUD/Notify/new_pet_unlocked", Asset := CustomHUD.Notify.new_pet_unlocked, Dimensions := option{vector2{X := 941.0, Y := 186.0}}}
            texture_metadata{Name := "CustomHUD/Notify/output_sheet_01", Asset := CustomHUD.Notify.output_sheet_01, Dimensions := option{vector2{X := 2048.0, Y := 2048.0}}}
            texture_metadata{Name := "CustomHUD/Src/bars_group_bg", Asset := CustomHUD.Src.bars_group_bg, Dimensions := option{vector2{X := 883.0, Y := 185.0}}}
            texture_metadata{Name := "CustomHUD/Src/boost_gradient", Asset := CustomHUD.Src.boost_gradient, Dimensions := option{vector2{X := 512.0, Y := 512.0}}}
            texture_metadata{Name := "CustomHUD/Src/booster_bg_effect_mask_round", Asset := CustomHUD.Src.booster_bg_effect_mask_round, Dimensions := option{vector2{X := 1024.0, Y := 1024.0}}}
            texture_metadata{Name := "CustomHUD/Src/details_bg", Asset := CustomHUD.Src.details_bg, Dimensions := option{vector2{X := 718.0, Y := 39.0}}}
            texture_metadata{Name := "CustomHUD/Src/font_shim", Asset := CustomHUD.Src.font_shim, Dimensions := option{vector2{X := 512.0, Y := 512.0}}}
            texture_metadata{Name := "CustomHUD/Src/full_round_mask", Asset := CustomHUD.Src.full_round_mask, Dimensions := option{vector2{X := 1028.0, Y := 26.0}}}
            texture_metadata{Name := "CustomHUD/Src/gold_gradient", Asset := CustomHUD.Src.gold_gradient, Dimensions := option{vector2{X := 512.0, Y := 512.0}}}
            texture_metadata{Name := "CustomHUD/Src/gray_gradient", Asset := CustomHUD.Src.gray_gradient, Dimensions := option{vector2{X := 512.0, Y := 512.0}}}
            texture_metadata{Name := "CustomHUD/Src/health_bar_mask", Asset := CustomHUD.Src.health_bar_mask, Dimensions := option{vector2{X := 1326.0, Y := 70.0}}}
            texture_metadata{Name := "CustomHUD/Src/heart", Asset := CustomHUD.Src.heart, Dimensions := option{vector2{X := 89.0, Y := 79.0}}}
            texture_metadata{Name := "CustomHUD/Src/level_bar_mask", Asset := CustomHUD.Src.level_bar_mask, Dimensions := option{vector2{X := 1658.0, Y := 44.0}}}
            texture_metadata{Name := "CustomHUD/Src/oblique_mask", Asset := CustomHUD.Src.oblique_mask, Dimensions := option{vector2{X := 1028.0, Y := 26.0}}}
            texture_metadata{Name := "CustomHUD/Src/res2_gradient", Asset := CustomHUD.Src.res2_gradient, Dimensions := option{vector2{X := 512.0, Y := 512.0}}}
            texture_metadata{Name := "CustomHUD/Src/res_jacko_bg", Asset := CustomHUD.Src.res_jacko_bg, Dimensions := option{vector2{X := 401.0, Y := 120.0}}}
            texture_metadata{Name := "CustomHUD/Src/res_tickets_bg", Asset := CustomHUD.Src.res_tickets_bg, Dimensions := option{vector2{X := 466.0, Y := 120.0}}}
            texture_metadata{Name := "CustomHUD/Src/resource_boost_bg", Asset := CustomHUD.Src.resource_boost_bg, Dimensions := option{vector2{X := 318.0, Y := 94.0}}}
            texture_metadata{Name := "CustomHUD/Src/resource_line_bg", Asset := CustomHUD.Src.resource_line_bg, Dimensions := option{vector2{X := 490.0, Y := 85.0}}}
            texture_metadata{Name := "CustomHUD/Src/shield_bar_mask", Asset := CustomHUD.Src.shield_bar_mask, Dimensions := option{vector2{X := 1326.0, Y := 40.0}}}
            texture_metadata{Name := "CustomHUD/Src/shield_gradient", Asset := CustomHUD.Src.shield_gradient, Dimensions := option{vector2{X := 512.0, Y := 512.0}}}
            texture_metadata{Name := "CustomHUD/Src/tombstone", Asset := CustomHUD.Src.tombstone, Dimensions := option{vector2{X := 668.0, Y := 42.0}}}
            texture_metadata{Name := "CustomHUD/Src/HauntedHouse/hh_house_pg_end", Asset := CustomHUD.Src.HauntedHouse.hh_house_pg_end, Dimensions := option{vector2{X := 21.0, Y := 44.0}}}
            texture_metadata{Name := "CustomHUD/Src/HauntedHouse/hh_house_pg_fill", Asset := CustomHUD.Src.HauntedHouse.hh_house_pg_fill, Dimensions := option{vector2{X := 1548.0, Y := 44.0}}}
            texture_metadata{Name := "CustomHUD/Src/HauntedHouse/hh_house_pg_label", Asset := CustomHUD.Src.HauntedHouse.hh_house_pg_label, Dimensions := option{vector2{X := 228.0, Y := 24.0}}}
            texture_metadata{Name := "CustomHUD/Src/HauntedHouse/hh_house_pg_start", Asset := CustomHUD.Src.HauntedHouse.hh_house_pg_start, Dimensions := option{vector2{X := 21.0, Y := 44.0}}}
            texture_metadata{Name := "CustomHUD/Src/HauntedHouse/hh_house_progress_bg", Asset := CustomHUD.Src.HauntedHouse.hh_house_progress_bg, Dimensions := option{vector2{X := 1590.0, Y := 44.0}}}
            texture_metadata{Name := "CustomHUD/Src/HauntedHouse/hh_house_progress_fg", Asset := CustomHUD.Src.HauntedHouse.hh_house_progress_fg, Dimensions := option{vector2{X := 1590.0, Y := 44.0}}}
            texture_metadata{Name := "CustomHUD/Src/HauntedHouse/hh_recap_bg", Asset := CustomHUD.Src.HauntedHouse.hh_recap_bg, Dimensions := option{vector2{X := 1209.0, Y := 1004.0}}}
            texture_metadata{Name := "CustomHUD/Src/HauntedHouse/hh_wave_pg_end", Asset := CustomHUD.Src.HauntedHouse.hh_wave_pg_end, Dimensions := option{vector2{X := 27.0, Y := 57.0}}}
            texture_metadata{Name := "CustomHUD/Src/HauntedHouse/hh_wave_pg_fill", Asset := CustomHUD.Src.HauntedHouse.hh_wave_pg_fill, Dimensions := option{vector2{X := 1534.0, Y := 57.0}}}
            texture_metadata{Name := "CustomHUD/Src/HauntedHouse/hh_wave_pg_start", Asset := CustomHUD.Src.HauntedHouse.hh_wave_pg_start, Dimensions := option{vector2{X := 27.0, Y := 57.0}}}
            texture_metadata{Name := "CustomHUD/Src/HauntedHouse/hh_wave_progress_bg", Asset := CustomHUD.Src.HauntedHouse.hh_wave_progress_bg, Dimensions := option{vector2{X := 1590.0, Y := 57.0}}}
            texture_metadata{Name := "CustomHUD/Src/HauntedHouse/hh_wave_progress_fg", Asset := CustomHUD.Src.HauntedHouse.hh_wave_progress_fg, Dimensions := option{vector2{X := 1588.0, Y := 57.0}}}
            texture_metadata{Name := "CustomHUD/Src/HauntedHouse/hh_wave_progress_shadow", Asset := CustomHUD.Src.HauntedHouse.hh_wave_progress_shadow, Dimensions := option{vector2{X := 1710.0, Y := 185.0}}}
            texture_metadata{Name := "BuyplateBillboards/bp_001", Asset := BuyplateBillboards.bp_001, Dimensions := option{vector2{X := 512.0, Y := 256.0}}}
            texture_metadata{Name := "BuyplateBillboards/bp_002", Asset := BuyplateBillboards.bp_002, Dimensions := option{vector2{X := 512.0, Y := 256.0}}}
            texture_metadata{Name := "BuyplateBillboards/bp_003", Asset := BuyplateBillboards.bp_003, Dimensions := option{vector2{X := 512.0, Y := 256.0}}}
            texture_metadata{Name := "BuyplateBillboards/bp_004", Asset := BuyplateBillboards.bp_004, Dimensions := option{vector2{X := 512.0, Y := 256.0}}}
            texture_metadata{Name := "BuyplateBillboards/bp_005", Asset := BuyplateBillboards.bp_005, Dimensions := option{vector2{X := 512.0, Y := 256.0}}}
            texture_metadata{Name := "BuyplateBillboards/bp_006", Asset := BuyplateBillboards.bp_006, Dimensions := option{vector2{X := 512.0, Y := 256.0}}}
            texture_metadata{Name := "BuyplateBillboards/bp_007", Asset := BuyplateBillboards.bp_007, Dimensions := option{vector2{X := 512.0, Y := 256.0}}}
            texture_metadata{Name := "BuyplateBillboards/bp_008", Asset := BuyplateBillboards.bp_008, Dimensions := option{vector2{X := 512.0, Y := 256.0}}}
            texture_metadata{Name := "BuyplateBillboards/bp_009", Asset := BuyplateBillboards.bp_009, Dimensions := option{vector2{X := 512.0, Y := 256.0}}}
            texture_metadata{Name := "BuyplateBillboards/bp_010", Asset := BuyplateBillboards.bp_010, Dimensions := option{vector2{X := 512.0, Y := 256.0}}}
            texture_metadata{Name := "BuyplateBillboards/bp_011", Asset := BuyplateBillboards.bp_011, Dimensions := option{vector2{X := 512.0, Y := 256.0}}}
            texture_metadata{Name := "BuyplateBillboards/bp_012", Asset := BuyplateBillboards.bp_012, Dimensions := option{vector2{X := 512.0, Y := 256.0}}}
            texture_metadata{Name := "BuyplateBillboards/bp_013", Asset := BuyplateBillboards.bp_013, Dimensions := option{vector2{X := 512.0, Y := 256.0}}}
            texture_metadata{Name := "BuyplateBillboards/bp_014", Asset := BuyplateBillboards.bp_014, Dimensions := option{vector2{X := 512.0, Y := 256.0}}}
            texture_metadata{Name := "BuyplateBillboards/bp_015", Asset := BuyplateBillboards.bp_015, Dimensions := option{vector2{X := 512.0, Y := 256.0}}}
            texture_metadata{Name := "BuyplateBillboards/bp_016", Asset := BuyplateBillboards.bp_016, Dimensions := option{vector2{X := 512.0, Y := 256.0}}}
            texture_metadata{Name := "BuyplateBillboards/bp_017", Asset := BuyplateBillboards.bp_017, Dimensions := option{vector2{X := 512.0, Y := 256.0}}}
            texture_metadata{Name := "BuyplateBillboards/bp_018", Asset := BuyplateBillboards.bp_018, Dimensions := option{vector2{X := 512.0, Y := 256.0}}}
            texture_metadata{Name := "BuyplateBillboards/bp_019", Asset := BuyplateBillboards.bp_019, Dimensions := option{vector2{X := 512.0, Y := 256.0}}}
            texture_metadata{Name := "BuyplateBillboards/bp_020", Asset := BuyplateBillboards.bp_020, Dimensions := option{vector2{X := 512.0, Y := 256.0}}}
            texture_metadata{Name := "BuyplateBillboards/bp_021", Asset := BuyplateBillboards.bp_021, Dimensions := option{vector2{X := 512.0, Y := 256.0}}}
            texture_metadata{Name := "BuyplateBillboards/bp_022", Asset := BuyplateBillboards.bp_022, Dimensions := option{vector2{X := 512.0, Y := 256.0}}}
            texture_metadata{Name := "BuyplateBillboards/bp_023", Asset := BuyplateBillboards.bp_023, Dimensions := option{vector2{X := 512.0, Y := 256.0}}}
            texture_metadata{Name := "BuyplateBillboards/bp_024", Asset := BuyplateBillboards.bp_024, Dimensions := option{vector2{X := 512.0, Y := 256.0}}}
            texture_metadata{Name := "BuyplateBillboards/bp_025", Asset := BuyplateBillboards.bp_025, Dimensions := option{vector2{X := 512.0, Y := 256.0}}}
            texture_metadata{Name := "BuyplateBillboards/bp_026", Asset := BuyplateBillboards.bp_026, Dimensions := option{vector2{X := 512.0, Y := 256.0}}}
            texture_metadata{Name := "BuyplateBillboards/bp_027", Asset := BuyplateBillboards.bp_027, Dimensions := option{vector2{X := 512.0, Y := 256.0}}}
            texture_metadata{Name := "BuyplateBillboards/bp_028", Asset := BuyplateBillboards.bp_028, Dimensions := option{vector2{X := 512.0, Y := 256.0}}}
            texture_metadata{Name := "BuyplateBillboards/bp_029", Asset := BuyplateBillboards.bp_029, Dimensions := option{vector2{X := 512.0, Y := 256.0}}}
            texture_metadata{Name := "BuyplateBillboards/bp_030", Asset := BuyplateBillboards.bp_030, Dimensions := option{vector2{X := 512.0, Y := 256.0}}}
            texture_metadata{Name := "BuyplateBillboards/bp_031", Asset := BuyplateBillboards.bp_031, Dimensions := option{vector2{X := 512.0, Y := 256.0}}}
            texture_metadata{Name := "BuyplateBillboards/bp_032", Asset := BuyplateBillboards.bp_032, Dimensions := option{vector2{X := 512.0, Y := 256.0}}}
            texture_metadata{Name := "BuyplateBillboards/bp_033", Asset := BuyplateBillboards.bp_033, Dimensions := option{vector2{X := 512.0, Y := 256.0}}}
            texture_metadata{Name := "BuyplateBillboards/bp_034", Asset := BuyplateBillboards.bp_034, Dimensions := option{vector2{X := 512.0, Y := 256.0}}}
            texture_metadata{Name := "BuyplateBillboards/bp_035", Asset := BuyplateBillboards.bp_035, Dimensions := option{vector2{X := 512.0, Y := 256.0}}}
            texture_metadata{Name := "BuyplateBillboards/bp_036", Asset := BuyplateBillboards.bp_036, Dimensions := option{vector2{X := 512.0, Y := 256.0}}}
            texture_metadata{Name := "BuyplateBillboards/bp_037", Asset := BuyplateBillboards.bp_037, Dimensions := option{vector2{X := 512.0, Y := 256.0}}}
            texture_metadata{Name := "BuyplateBillboards/bp_038", Asset := BuyplateBillboards.bp_038, Dimensions := option{vector2{X := 512.0, Y := 256.0}}}
            texture_metadata{Name := "BuyplateBillboards/bp_039", Asset := BuyplateBillboards.bp_039, Dimensions := option{vector2{X := 512.0, Y := 256.0}}}
            texture_metadata{Name := "BuyplateBillboards/bp_040", Asset := BuyplateBillboards.bp_040, Dimensions := option{vector2{X := 512.0, Y := 256.0}}}
            texture_metadata{Name := "BuyplateBillboards/bp_041", Asset := BuyplateBillboards.bp_041, Dimensions := option{vector2{X := 512.0, Y := 256.0}}}
            texture_metadata{Name := "BuyplateBillboards/bp_042", Asset := BuyplateBillboards.bp_042, Dimensions := option{vector2{X := 512.0, Y := 256.0}}}
            texture_metadata{Name := "BuyplateBillboards/bp_043", Asset := BuyplateBillboards.bp_043, Dimensions := option{vector2{X := 512.0, Y := 256.0}}}
            texture_metadata{Name := "BuyplateBillboards/bp_044", Asset := BuyplateBillboards.bp_044, Dimensions := option{vector2{X := 512.0, Y := 256.0}}}
            texture_metadata{Name := "BuyplateBillboards/bp_045", Asset := BuyplateBillboards.bp_045, Dimensions := option{vector2{X := 512.0, Y := 256.0}}}
            texture_metadata{Name := "BuyplateBillboards/bp_046", Asset := BuyplateBillboards.bp_046, Dimensions := option{vector2{X := 512.0, Y := 256.0}}}
            texture_metadata{Name := "BuyplateBillboards/bp_047", Asset := BuyplateBillboards.bp_047, Dimensions := option{vector2{X := 512.0, Y := 256.0}}}
            texture_metadata{Name := "BuyplateBillboards/bp_048", Asset := BuyplateBillboards.bp_048, Dimensions := option{vector2{X := 512.0, Y := 256.0}}}
            texture_metadata{Name := "BuyplateBillboards/bp_049", Asset := BuyplateBillboards.bp_049, Dimensions := option{vector2{X := 512.0, Y := 256.0}}}
            texture_metadata{Name := "BuyplateBillboards/bp_050", Asset := BuyplateBillboards.bp_050, Dimensions := option{vector2{X := 512.0, Y := 256.0}}}
            texture_metadata{Name := "BuyplateBillboards/bp_051", Asset := BuyplateBillboards.bp_051, Dimensions := option{vector2{X := 512.0, Y := 256.0}}}
            texture_metadata{Name := "BuyplateBillboards/bp_052", Asset := BuyplateBillboards.bp_052, Dimensions := option{vector2{X := 512.0, Y := 256.0}}}
            texture_metadata{Name := "BuyplateBillboards/bp_053", Asset := BuyplateBillboards.bp_053, Dimensions := option{vector2{X := 512.0, Y := 256.0}}}
            texture_metadata{Name := "BuyplateBillboards/bp_054", Asset := BuyplateBillboards.bp_054, Dimensions := option{vector2{X := 512.0, Y := 256.0}}}
            texture_metadata{Name := "BuyplateBillboards/bp_055", Asset := BuyplateBillboards.bp_055, Dimensions := option{vector2{X := 512.0, Y := 256.0}}}
            texture_metadata{Name := "BuyplateBillboards/bp_056", Asset := BuyplateBillboards.bp_056, Dimensions := option{vector2{X := 512.0, Y := 256.0}}}
            texture_metadata{Name := "BuyplateBillboards/bp_057", Asset := BuyplateBillboards.bp_057, Dimensions := option{vector2{X := 512.0, Y := 256.0}}}
            texture_metadata{Name := "BuyplateBillboards/bp_058", Asset := BuyplateBillboards.bp_058, Dimensions := option{vector2{X := 512.0, Y := 256.0}}}
            texture_metadata{Name := "BuyplateBillboards/bp_059", Asset := BuyplateBillboards.bp_059, Dimensions := option{vector2{X := 512.0, Y := 256.0}}}
            texture_metadata{Name := "BuyplateBillboards/bp_060", Asset := BuyplateBillboards.bp_060, Dimensions := option{vector2{X := 512.0, Y := 256.0}}}
            texture_metadata{Name := "BuyplateBillboards/bp_061", Asset := BuyplateBillboards.bp_061, Dimensions := option{vector2{X := 512.0, Y := 256.0}}}
            texture_metadata{Name := "BuyplateBillboards/bp_062", Asset := BuyplateBillboards.bp_062, Dimensions := option{vector2{X := 512.0, Y := 256.0}}}
            texture_metadata{Name := "BuyplateBillboards/bp_063", Asset := BuyplateBillboards.bp_063, Dimensions := option{vector2{X := 512.0, Y := 256.0}}}
            texture_metadata{Name := "BuyplateBillboards/bp_064", Asset := BuyplateBillboards.bp_064, Dimensions := option{vector2{X := 512.0, Y := 256.0}}}
            texture_metadata{Name := "BuyplateBillboards/bp_065", Asset := BuyplateBillboards.bp_065, Dimensions := option{vector2{X := 512.0, Y := 256.0}}}
            texture_metadata{Name := "BuyplateBillboards/bp_066", Asset := BuyplateBillboards.bp_066, Dimensions := option{vector2{X := 512.0, Y := 256.0}}}
            texture_metadata{Name := "BuyplateBillboards/bp_067", Asset := BuyplateBillboards.bp_067, Dimensions := option{vector2{X := 512.0, Y := 256.0}}}
            texture_metadata{Name := "BuyplateBillboards/bp_068", Asset := BuyplateBillboards.bp_068, Dimensions := option{vector2{X := 512.0, Y := 256.0}}}
            texture_metadata{Name := "BuyplateBillboards/bp_069", Asset := BuyplateBillboards.bp_069, Dimensions := option{vector2{X := 512.0, Y := 256.0}}}
            texture_metadata{Name := "BuyplateBillboards/bp_070", Asset := BuyplateBillboards.bp_070, Dimensions := option{vector2{X := 512.0, Y := 256.0}}}
            texture_metadata{Name := "BuyplateBillboards/bp_071", Asset := BuyplateBillboards.bp_071, Dimensions := option{vector2{X := 512.0, Y := 256.0}}}
            texture_metadata{Name := "BuyplateBillboards/bp_072", Asset := BuyplateBillboards.bp_072, Dimensions := option{vector2{X := 512.0, Y := 256.0}}}
            texture_metadata{Name := "BuyplateBillboards/bp_073", Asset := BuyplateBillboards.bp_073, Dimensions := option{vector2{X := 512.0, Y := 256.0}}}
            texture_metadata{Name := "BuyplateBillboards/bp_074", Asset := BuyplateBillboards.bp_074, Dimensions := option{vector2{X := 512.0, Y := 256.0}}}
            texture_metadata{Name := "BuyplateBillboards/bp_075", Asset := BuyplateBillboards.bp_075, Dimensions := option{vector2{X := 512.0, Y := 256.0}}}
            texture_metadata{Name := "BuyplateBillboards/bp_076", Asset := BuyplateBillboards.bp_076, Dimensions := option{vector2{X := 512.0, Y := 256.0}}}
            texture_metadata{Name := "BuyplateBillboards/bp_077", Asset := BuyplateBillboards.bp_077, Dimensions := option{vector2{X := 512.0, Y := 256.0}}}
            texture_metadata{Name := "BuyplateBillboards/bp_078", Asset := BuyplateBillboards.bp_078, Dimensions := option{vector2{X := 512.0, Y := 256.0}}}
            texture_metadata{Name := "BuyplateBillboards/bp_079", Asset := BuyplateBillboards.bp_079, Dimensions := option{vector2{X := 512.0, Y := 256.0}}}
            texture_metadata{Name := "BuyplateBillboards/bp_080", Asset := BuyplateBillboards.bp_080, Dimensions := option{vector2{X := 512.0, Y := 256.0}}}
            texture_metadata{Name := "BuyplateBillboards/bp_081", Asset := BuyplateBillboards.bp_081, Dimensions := option{vector2{X := 512.0, Y := 256.0}}}
            texture_metadata{Name := "BuyplateBillboards/bp_082", Asset := BuyplateBillboards.bp_082, Dimensions := option{vector2{X := 512.0, Y := 256.0}}}
            texture_metadata{Name := "BuyplateBillboards/bp_083", Asset := BuyplateBillboards.bp_083, Dimensions := option{vector2{X := 512.0, Y := 256.0}}}
            texture_metadata{Name := "BuyplateBillboards/bp_084", Asset := BuyplateBillboards.bp_084, Dimensions := option{vector2{X := 512.0, Y := 256.0}}}
            texture_metadata{Name := "BuyplateBillboards/bp_085", Asset := BuyplateBillboards.bp_085, Dimensions := option{vector2{X := 512.0, Y := 256.0}}}
            texture_metadata{Name := "BuyplateBillboards/bp_086", Asset := BuyplateBillboards.bp_086, Dimensions := option{vector2{X := 512.0, Y := 256.0}}}
            texture_metadata{Name := "BuyplateBillboards/bp_087", Asset := BuyplateBillboards.bp_087, Dimensions := option{vector2{X := 512.0, Y := 256.0}}}
            texture_metadata{Name := "BuyplateBillboards/bp_088", Asset := BuyplateBillboards.bp_088, Dimensions := option{vector2{X := 512.0, Y := 256.0}}}
            texture_metadata{Name := "BuyplateBillboards/bp_089", Asset := BuyplateBillboards.bp_089, Dimensions := option{vector2{X := 512.0, Y := 256.0}}}
            texture_metadata{Name := "BuyplateBillboards/bp_090", Asset := BuyplateBillboards.bp_090, Dimensions := option{vector2{X := 512.0, Y := 256.0}}}
            texture_metadata{Name := "BuyplateBillboards/bp_091", Asset := BuyplateBillboards.bp_091, Dimensions := option{vector2{X := 512.0, Y := 256.0}}}
            texture_metadata{Name := "BuyplateBillboards/bp_092", Asset := BuyplateBillboards.bp_092, Dimensions := option{vector2{X := 512.0, Y := 256.0}}}
            texture_metadata{Name := "BuyplateBillboards/bp_093", Asset := BuyplateBillboards.bp_093, Dimensions := option{vector2{X := 512.0, Y := 256.0}}}
            texture_metadata{Name := "BuyplateBillboards/bp_094", Asset := BuyplateBillboards.bp_094, Dimensions := option{vector2{X := 512.0, Y := 256.0}}}
            texture_metadata{Name := "BuyplateBillboards/bp_095", Asset := BuyplateBillboards.bp_095, Dimensions := option{vector2{X := 512.0, Y := 256.0}}}
            texture_metadata{Name := "BuyplateBillboards/bp_096", Asset := BuyplateBillboards.bp_096, Dimensions := option{vector2{X := 512.0, Y := 256.0}}}
            texture_metadata{Name := "BuyplateBillboards/bp_097", Asset := BuyplateBillboards.bp_097, Dimensions := option{vector2{X := 512.0, Y := 256.0}}}
            texture_metadata{Name := "BuyplateBillboards/bp_098", Asset := BuyplateBillboards.bp_098, Dimensions := option{vector2{X := 512.0, Y := 256.0}}}
            texture_metadata{Name := "BuyplateBillboards/bp_099", Asset := BuyplateBillboards.bp_099, Dimensions := option{vector2{X := 512.0, Y := 256.0}}}
            texture_metadata{Name := "BuyplateBillboards/bp_100", Asset := BuyplateBillboards.bp_100, Dimensions := option{vector2{X := 512.0, Y := 256.0}}}
            texture_metadata{Name := "BuyplateBillboards/bp_101", Asset := BuyplateBillboards.bp_101, Dimensions := option{vector2{X := 512.0, Y := 256.0}}}
            texture_metadata{Name := "BuyplateBillboards/bp_102", Asset := BuyplateBillboards.bp_102, Dimensions := option{vector2{X := 512.0, Y := 256.0}}}
            texture_metadata{Name := "BuyplateBillboards/bp_103", Asset := BuyplateBillboards.bp_103, Dimensions := option{vector2{X := 512.0, Y := 256.0}}}
            texture_metadata{Name := "BuyplateBillboards/bp_104", Asset := BuyplateBillboards.bp_104, Dimensions := option{vector2{X := 512.0, Y := 256.0}}}
            texture_metadata{Name := "BuyplateBillboards/bp_105", Asset := BuyplateBillboards.bp_105, Dimensions := option{vector2{X := 512.0, Y := 256.0}}}
            texture_metadata{Name := "BuyplateBillboards/bp_106", Asset := BuyplateBillboards.bp_106, Dimensions := option{vector2{X := 512.0, Y := 256.0}}}
            texture_metadata{Name := "BuyplateBillboards/bp_107", Asset := BuyplateBillboards.bp_107, Dimensions := option{vector2{X := 512.0, Y := 256.0}}}
            texture_metadata{Name := "BuyplateBillboards/bp_108", Asset := BuyplateBillboards.bp_108, Dimensions := option{vector2{X := 512.0, Y := 256.0}}}
            texture_metadata{Name := "BuyplateBillboards/bp_109", Asset := BuyplateBillboards.bp_109, Dimensions := option{vector2{X := 512.0, Y := 256.0}}}
            texture_metadata{Name := "BuyplateBillboards/bp_110", Asset := BuyplateBillboards.bp_110, Dimensions := option{vector2{X := 512.0, Y := 256.0}}}
            texture_metadata{Name := "BuyplateBillboards/bp_111", Asset := BuyplateBillboards.bp_111, Dimensions := option{vector2{X := 512.0, Y := 256.0}}}
            texture_metadata{Name := "BuyplateBillboards/bp_112", Asset := BuyplateBillboards.bp_112, Dimensions := option{vector2{X := 512.0, Y := 256.0}}}
            texture_metadata{Name := "BuyplateBillboards/bp_113", Asset := BuyplateBillboards.bp_113, Dimensions := option{vector2{X := 512.0, Y := 256.0}}}
            texture_metadata{Name := "BuyplateBillboards/bp_114", Asset := BuyplateBillboards.bp_114, Dimensions := option{vector2{X := 512.0, Y := 256.0}}}
            texture_metadata{Name := "BuyplateBillboards/bp_115", Asset := BuyplateBillboards.bp_115, Dimensions := option{vector2{X := 512.0, Y := 256.0}}}
            texture_metadata{Name := "BuyplateBillboards/bp_116", Asset := BuyplateBillboards.bp_116, Dimensions := option{vector2{X := 512.0, Y := 256.0}}}
            texture_metadata{Name := "BuyplateBillboards/bp_117", Asset := BuyplateBillboards.bp_117, Dimensions := option{vector2{X := 512.0, Y := 256.0}}}
            texture_metadata{Name := "BuyplateBillboards/bp_118", Asset := BuyplateBillboards.bp_118, Dimensions := option{vector2{X := 512.0, Y := 256.0}}}
            texture_metadata{Name := "BuyplateBillboards/bp_119", Asset := BuyplateBillboards.bp_119, Dimensions := option{vector2{X := 512.0, Y := 256.0}}}
            texture_metadata{Name := "BuyplateBillboards/bp_120", Asset := BuyplateBillboards.bp_120, Dimensions := option{vector2{X := 512.0, Y := 256.0}}}
            texture_metadata{Name := "BuyplateBillboards/bp_121", Asset := BuyplateBillboards.bp_121, Dimensions := option{vector2{X := 512.0, Y := 256.0}}}
            texture_metadata{Name := "BuyplateBillboards/bp_122", Asset := BuyplateBillboards.bp_122, Dimensions := option{vector2{X := 512.0, Y := 256.0}}}
            texture_metadata{Name := "BuyplateBillboards/bp_123", Asset := BuyplateBillboards.bp_123, Dimensions := option{vector2{X := 512.0, Y := 256.0}}}
            texture_metadata{Name := "BuyplateBillboards/bp_124", Asset := BuyplateBillboards.bp_124, Dimensions := option{vector2{X := 512.0, Y := 256.0}}}
            texture_metadata{Name := "BuyplateBillboards/bp_125", Asset := BuyplateBillboards.bp_125, Dimensions := option{vector2{X := 512.0, Y := 256.0}}}
            texture_metadata{Name := "BuyplateBillboards/bp_126", Asset := BuyplateBillboards.bp_126, Dimensions := option{vector2{X := 512.0, Y := 256.0}}}
            texture_metadata{Name := "BuyplateBillboards/bp_127", Asset := BuyplateBillboards.bp_127, Dimensions := option{vector2{X := 512.0, Y := 256.0}}}
            texture_metadata{Name := "BuyplateBillboards/bp_128", Asset := BuyplateBillboards.bp_128, Dimensions := option{vector2{X := 512.0, Y := 256.0}}}
            texture_metadata{Name := "BuyplateBillboards/bp_129", Asset := BuyplateBillboards.bp_129, Dimensions := option{vector2{X := 512.0, Y := 256.0}}}
            texture_metadata{Name := "BuyplateBillboards/bp_130", Asset := BuyplateBillboards.bp_130, Dimensions := option{vector2{X := 512.0, Y := 256.0}}}
            texture_metadata{Name := "BuyplateBillboards/bp_131", Asset := BuyplateBillboards.bp_131, Dimensions := option{vector2{X := 512.0, Y := 256.0}}}
            texture_metadata{Name := "BuyplateBillboards/bp_132", Asset := BuyplateBillboards.bp_132, Dimensions := option{vector2{X := 512.0, Y := 256.0}}}
            texture_metadata{Name := "BuyplateBillboards/bp_133", Asset := BuyplateBillboards.bp_133, Dimensions := option{vector2{X := 512.0, Y := 256.0}}}
            texture_metadata{Name := "BuyplateBillboards/bp_134", Asset := BuyplateBillboards.bp_134, Dimensions := option{vector2{X := 512.0, Y := 256.0}}}
            texture_metadata{Name := "BuyplateBillboards/bp_135", Asset := BuyplateBillboards.bp_135, Dimensions := option{vector2{X := 512.0, Y := 256.0}}}
            texture_metadata{Name := "BuyplateBillboards/bp_136", Asset := BuyplateBillboards.bp_136, Dimensions := option{vector2{X := 512.0, Y := 256.0}}}
            texture_metadata{Name := "BuyplateBillboards/bp_137", Asset := BuyplateBillboards.bp_137, Dimensions := option{vector2{X := 512.0, Y := 256.0}}}
            texture_metadata{Name := "BuyplateBillboards/bp_138", Asset := BuyplateBillboards.bp_138, Dimensions := option{vector2{X := 512.0, Y := 256.0}}}
            texture_metadata{Name := "BuyplateBillboards/bp_139", Asset := BuyplateBillboards.bp_139, Dimensions := option{vector2{X := 512.0, Y := 256.0}}}
            texture_metadata{Name := "BuyplateBillboards/bp_140", Asset := BuyplateBillboards.bp_140, Dimensions := option{vector2{X := 512.0, Y := 256.0}}}
            texture_metadata{Name := "BuyplateBillboards/bp_141", Asset := BuyplateBillboards.bp_141, Dimensions := option{vector2{X := 512.0, Y := 256.0}}}
            texture_metadata{Name := "BuyplateBillboards/bp_142", Asset := BuyplateBillboards.bp_142, Dimensions := option{vector2{X := 512.0, Y := 256.0}}}
            texture_metadata{Name := "BuyplateBillboards/bp_143", Asset := BuyplateBillboards.bp_143, Dimensions := option{vector2{X := 512.0, Y := 256.0}}}
            texture_metadata{Name := "BuyplateBillboards/bp_200", Asset := BuyplateBillboards.bp_200, Dimensions := option{vector2{X := 512.0, Y := 256.0}}}
            texture_metadata{Name := "BuyplateBillboards/bp_201", Asset := BuyplateBillboards.bp_201, Dimensions := option{vector2{X := 512.0, Y := 256.0}}}
            texture_metadata{Name := "BuyplateBillboards/bp_202", Asset := BuyplateBillboards.bp_202, Dimensions := option{vector2{X := 512.0, Y := 256.0}}}
            texture_metadata{Name := "BuyplateBillboards/bp_203", Asset := BuyplateBillboards.bp_203, Dimensions := option{vector2{X := 512.0, Y := 256.0}}}
            texture_metadata{Name := "BuyplateBillboards/bp_204", Asset := BuyplateBillboards.bp_204, Dimensions := option{vector2{X := 512.0, Y := 256.0}}}
            texture_metadata{Name := "BuyplateBillboards/bp_205", Asset := BuyplateBillboards.bp_205, Dimensions := option{vector2{X := 512.0, Y := 256.0}}}
            texture_metadata{Name := "BuyplateBillboards/bp_206", Asset := BuyplateBillboards.bp_206, Dimensions := option{vector2{X := 512.0, Y := 256.0}}}
            texture_metadata{Name := "BuyplateBillboards/bp_207", Asset := BuyplateBillboards.bp_207, Dimensions := option{vector2{X := 512.0, Y := 256.0}}}
            texture_metadata{Name := "BuyplateBillboards/bp_208", Asset := BuyplateBillboards.bp_208, Dimensions := option{vector2{X := 512.0, Y := 256.0}}}
            texture_metadata{Name := "BuyplateBillboards/bp_209", Asset := BuyplateBillboards.bp_209, Dimensions := option{vector2{X := 512.0, Y := 256.0}}}
            texture_metadata{Name := "BuyplateBillboards/bp_210", Asset := BuyplateBillboards.bp_210, Dimensions := option{vector2{X := 512.0, Y := 256.0}}}
            texture_metadata{Name := "BuyplateBillboards/bp_300", Asset := BuyplateBillboards.bp_300, Dimensions := option{vector2{X := 512.0, Y := 256.0}}}
            texture_metadata{Name := "BuyplateBillboards/bp_301", Asset := BuyplateBillboards.bp_301, Dimensions := option{vector2{X := 512.0, Y := 256.0}}}
            texture_metadata{Name := "BuyplateBillboards/bp_302", Asset := BuyplateBillboards.bp_302, Dimensions := option{vector2{X := 512.0, Y := 256.0}}}
            texture_metadata{Name := "BuyplateBillboards/bp_303", Asset := BuyplateBillboards.bp_303, Dimensions := option{vector2{X := 512.0, Y := 256.0}}}
            texture_metadata{Name := "BuyplateBillboards/bp_304", Asset := BuyplateBillboards.bp_304, Dimensions := option{vector2{X := 512.0, Y := 256.0}}}
            texture_metadata{Name := "BuyplateBillboards/bp_305", Asset := BuyplateBillboards.bp_305, Dimensions := option{vector2{X := 512.0, Y := 256.0}}}
            texture_metadata{Name := "BuyplateBillboards/bp_306", Asset := BuyplateBillboards.bp_306, Dimensions := option{vector2{X := 512.0, Y := 256.0}}}
            texture_metadata{Name := "BuyplateBillboards/bp_307", Asset := BuyplateBillboards.bp_307, Dimensions := option{vector2{X := 512.0, Y := 256.0}}}
            texture_metadata{Name := "BuyplateBillboards/bp_308", Asset := BuyplateBillboards.bp_308, Dimensions := option{vector2{X := 512.0, Y := 256.0}}}
            texture_metadata{Name := "BuyplateBillboards/bp_309", Asset := BuyplateBillboards.bp_309, Dimensions := option{vector2{X := 512.0, Y := 256.0}}}
            texture_metadata{Name := "BuyplateBillboards/bp_310", Asset := BuyplateBillboards.bp_310, Dimensions := option{vector2{X := 512.0, Y := 256.0}}}
            texture_metadata{Name := "BuyplateBillboards/bp_311", Asset := BuyplateBillboards.bp_311, Dimensions := option{vector2{X := 512.0, Y := 256.0}}}
            texture_metadata{Name := "BuyplateBillboards/bp_312", Asset := BuyplateBillboards.bp_312, Dimensions := option{vector2{X := 512.0, Y := 256.0}}}
            texture_metadata{Name := "BuyplateBillboards/bp_400", Asset := BuyplateBillboards.bp_400, Dimensions := option{vector2{X := 512.0, Y := 256.0}}}
            texture_metadata{Name := "BuyplateBillboards/bp_401", Asset := BuyplateBillboards.bp_401, Dimensions := option{vector2{X := 512.0, Y := 256.0}}}
            texture_metadata{Name := "BuyplateBillboards/bp_402", Asset := BuyplateBillboards.bp_402, Dimensions := option{vector2{X := 512.0, Y := 256.0}}}
            texture_metadata{Name := "BuyplateBillboards/bp_403", Asset := BuyplateBillboards.bp_403, Dimensions := option{vector2{X := 512.0, Y := 256.0}}}
            texture_metadata{Name := "BuyplateBillboards/bp_404", Asset := BuyplateBillboards.bp_404, Dimensions := option{vector2{X := 512.0, Y := 256.0}}}
            texture_metadata{Name := "BuyplateBillboards/bp_405", Asset := BuyplateBillboards.bp_405, Dimensions := option{vector2{X := 512.0, Y := 256.0}}}
            texture_metadata{Name := "BuyplateBillboards/bp_406", Asset := BuyplateBillboards.bp_406, Dimensions := option{vector2{X := 512.0, Y := 256.0}}}
            texture_metadata{Name := "BuyplateBillboards/bp_407", Asset := BuyplateBillboards.bp_407, Dimensions := option{vector2{X := 512.0, Y := 256.0}}}
            texture_metadata{Name := "BuyplateBillboards/bp_408", Asset := BuyplateBillboards.bp_408, Dimensions := option{vector2{X := 512.0, Y := 256.0}}}
            texture_metadata{Name := "BuyplateBillboards/bp_409", Asset := BuyplateBillboards.bp_409, Dimensions := option{vector2{X := 512.0, Y := 256.0}}}
            texture_metadata{Name := "BuyplateBillboards/bp_410", Asset := BuyplateBillboards.bp_410, Dimensions := option{vector2{X := 512.0, Y := 256.0}}}
            texture_metadata{Name := "BuyplateBillboards/bp_411", Asset := BuyplateBillboards.bp_411, Dimensions := option{vector2{X := 512.0, Y := 256.0}}}
            texture_metadata{Name := "BuyplateBillboards/bp_412", Asset := BuyplateBillboards.bp_412, Dimensions := option{vector2{X := 512.0, Y := 256.0}}}
            texture_metadata{Name := "BuyplateBillboards/bp_413", Asset := BuyplateBillboards.bp_413, Dimensions := option{vector2{X := 512.0, Y := 256.0}}}
            texture_metadata{Name := "BuyplateBillboards/bp_500", Asset := BuyplateBillboards.bp_500, Dimensions := option{vector2{X := 512.0, Y := 256.0}}}
            texture_metadata{Name := "BuyplateBillboards/bp_claim", Asset := BuyplateBillboards.bp_claim, Dimensions := option{vector2{X := 512.0, Y := 256.0}}}
            texture_metadata{Name := "DailyRewardsMenuUI/background", Asset := DailyRewardsMenuUI.background, Dimensions := option{vector2{X := 2048.0, Y := 1397.0}}}
            texture_metadata{Name := "DailyRewardsMenuUI/claimed_label", Asset := DailyRewardsMenuUI.claimed_label, Dimensions := option{vector2{X := 268.0, Y := 268.0}}}
            texture_metadata{Name := "DailyRewardsMenuUI/day_bg", Asset := DailyRewardsMenuUI.day_bg, Dimensions := option{vector2{X := 480.0, Y := 508.0}}}
            texture_metadata{Name := "DailyRewardsMenuUI/prev_day_overlay", Asset := DailyRewardsMenuUI.prev_day_overlay, Dimensions := option{vector2{X := 480.0, Y := 508.0}}}
            texture_metadata{Name := "DailyRewardsMenuUI/reward_day1", Asset := DailyRewardsMenuUI.reward_day1, Dimensions := option{vector2{X := 480.0, Y := 508.0}}}
            texture_metadata{Name := "DailyRewardsMenuUI/reward_day2", Asset := DailyRewardsMenuUI.reward_day2, Dimensions := option{vector2{X := 480.0, Y := 508.0}}}
            texture_metadata{Name := "DailyRewardsMenuUI/reward_day3", Asset := DailyRewardsMenuUI.reward_day3, Dimensions := option{vector2{X := 480.0, Y := 508.0}}}
            texture_metadata{Name := "DailyRewardsMenuUI/reward_day4", Asset := DailyRewardsMenuUI.reward_day4, Dimensions := option{vector2{X := 480.0, Y := 508.0}}}
            texture_metadata{Name := "DailyRewardsMenuUI/reward_day5", Asset := DailyRewardsMenuUI.reward_day5, Dimensions := option{vector2{X := 480.0, Y := 508.0}}}
            texture_metadata{Name := "DailyRewardsMenuUI/reward_day6", Asset := DailyRewardsMenuUI.reward_day6, Dimensions := option{vector2{X := 480.0, Y := 508.0}}}
            texture_metadata{Name := "DailyRewardsMenuUI/special_gift_claimed", Asset := DailyRewardsMenuUI.special_gift_claimed, Dimensions := option{vector2{X := 915.0, Y := 1070.0}}}
            texture_metadata{Name := "DailyRewardsMenuUI/special_gift_unclaimed", Asset := DailyRewardsMenuUI.special_gift_unclaimed, Dimensions := option{vector2{X := 915.0, Y := 1070.0}}}
            texture_metadata{Name := "DailyRewardsMenuUI/today_bg", Asset := DailyRewardsMenuUI.today_bg, Dimensions := option{vector2{X := 480.0, Y := 508.0}}}
            texture_metadata{Name := "Fonts/FeastBB/a", Asset := Fonts.FeastBB.a, Dimensions := option{vector2{X := 29.0, Y := 64.0}}}
            texture_metadata{Name := "Fonts/FeastBB/A_upper", Asset := Fonts.FeastBB.A_upper, Dimensions := option{vector2{X := 29.0, Y := 64.0}}}
            texture_metadata{Name := "Fonts/FeastBB/b", Asset := Fonts.FeastBB.b, Dimensions := option{vector2{X := 36.0, Y := 64.0}}}
            texture_metadata{Name := "Fonts/FeastBB/B_upper", Asset := Fonts.FeastBB.B_upper, Dimensions := option{vector2{X := 37.0, Y := 64.0}}}
            texture_metadata{Name := "Fonts/FeastBB/c", Asset := Fonts.FeastBB.c, Dimensions := option{vector2{X := 31.0, Y := 64.0}}}
            texture_metadata{Name := "Fonts/FeastBB/C_upper", Asset := Fonts.FeastBB.C_upper, Dimensions := option{vector2{X := 32.0, Y := 64.0}}}
            texture_metadata{Name := "Fonts/FeastBB/d", Asset := Fonts.FeastBB.d, Dimensions := option{vector2{X := 36.0, Y := 64.0}}}
            texture_metadata{Name := "Fonts/FeastBB/D_upper", Asset := Fonts.FeastBB.D_upper, Dimensions := option{vector2{X := 36.0, Y := 64.0}}}
            texture_metadata{Name := "Fonts/FeastBB/e", Asset := Fonts.FeastBB.e, Dimensions := option{vector2{X := 32.0, Y := 64.0}}}
            texture_metadata{Name := "Fonts/FeastBB/E_upper", Asset := Fonts.FeastBB.E_upper, Dimensions := option{vector2{X := 31.0, Y := 64.0}}}
            texture_metadata{Name := "Fonts/FeastBB/f", Asset := Fonts.FeastBB.f, Dimensions := option{vector2{X := 32.0, Y := 64.0}}}
            texture_metadata{Name := "Fonts/FeastBB/F_upper", Asset := Fonts.FeastBB.F_upper, Dimensions := option{vector2{X := 27.0, Y := 64.0}}}
            texture_metadata{Name := "Fonts/FeastBB/g", Asset := Fonts.FeastBB.g, Dimensions := option{vector2{X := 36.0, Y := 64.0}}}
            texture_metadata{Name := "Fonts/FeastBB/G_upper", Asset := Fonts.FeastBB.G_upper, Dimensions := option{vector2{X := 37.0, Y := 64.0}}}
            texture_metadata{Name := "Fonts/FeastBB/h", Asset := Fonts.FeastBB.h, Dimensions := option{vector2{X := 31.0, Y := 64.0}}}
            texture_metadata{Name := "Fonts/FeastBB/H_upper", Asset := Fonts.FeastBB.H_upper, Dimensions := option{vector2{X := 32.0, Y := 64.0}}}
            texture_metadata{Name := "Fonts/FeastBB/i", Asset := Fonts.FeastBB.i, Dimensions := option{vector2{X := 13.0, Y := 64.0}}}
            texture_metadata{Name := "Fonts/FeastBB/I_upper", Asset := Fonts.FeastBB.I_upper, Dimensions := option{vector2{X := 14.0, Y := 64.0}}}
            texture_metadata{Name := "Fonts/FeastBB/j", Asset := Fonts.FeastBB.j, Dimensions := option{vector2{X := 31.0, Y := 64.0}}}
            texture_metadata{Name := "Fonts/FeastBB/J_upper", Asset := Fonts.FeastBB.J_upper, Dimensions := option{vector2{X := 30.0, Y := 64.0}}}
            texture_metadata{Name := "Fonts/FeastBB/k", Asset := Fonts.FeastBB.k, Dimensions := option{vector2{X := 34.0, Y := 64.0}}}
            texture_metadata{Name := "Fonts/FeastBB/K_upper", Asset := Fonts.FeastBB.K_upper, Dimensions := option{vector2{X := 34.0, Y := 64.0}}}
            texture_metadata{Name := "Fonts/FeastBB/l", Asset := Fonts.FeastBB.l, Dimensions := option{vector2{X := 25.0, Y := 64.0}}}
            texture_metadata{Name := "Fonts/FeastBB/L_upper", Asset := Fonts.FeastBB.L_upper, Dimensions := option{vector2{X := 25.0, Y := 64.0}}}
            texture_metadata{Name := "Fonts/FeastBB/m", Asset := Fonts.FeastBB.m, Dimensions := option{vector2{X := 41.0, Y := 64.0}}}
            texture_metadata{Name := "Fonts/FeastBB/M_upper", Asset := Fonts.FeastBB.M_upper, Dimensions := option{vector2{X := 42.0, Y := 64.0}}}
            texture_metadata{Name := "Fonts/FeastBB/n", Asset := Fonts.FeastBB.n, Dimensions := option{vector2{X := 33.0, Y := 64.0}}}
            texture_metadata{Name := "Fonts/FeastBB/N_upper", Asset := Fonts.FeastBB.N_upper, Dimensions := option{vector2{X := 31.0, Y := 64.0}}}
            texture_metadata{Name := "Fonts/FeastBB/num_0", Asset := Fonts.FeastBB.num_0, Dimensions := option{vector2{X := 42.0, Y := 64.0}}}
            texture_metadata{Name := "Fonts/FeastBB/num_1", Asset := Fonts.FeastBB.num_1, Dimensions := option{vector2{X := 26.0, Y := 64.0}}}
            texture_metadata{Name := "Fonts/FeastBB/num_2", Asset := Fonts.FeastBB.num_2, Dimensions := option{vector2{X := 26.0, Y := 64.0}}}
            texture_metadata{Name := "Fonts/FeastBB/num_3", Asset := Fonts.FeastBB.num_3, Dimensions := option{vector2{X := 27.0, Y := 64.0}}}
            texture_metadata{Name := "Fonts/FeastBB/num_4", Asset := Fonts.FeastBB.num_4, Dimensions := option{vector2{X := 32.0, Y := 64.0}}}
            texture_metadata{Name := "Fonts/FeastBB/num_5", Asset := Fonts.FeastBB.num_5, Dimensions := option{vector2{X := 28.0, Y := 64.0}}}
            texture_metadata{Name := "Fonts/FeastBB/num_6", Asset := Fonts.FeastBB.num_6, Dimensions := option{vector2{X := 31.0, Y := 64.0}}}
            texture_metadata{Name := "Fonts/FeastBB/num_7", Asset := Fonts.FeastBB.num_7, Dimensions := option{vector2{X := 30.0, Y := 64.0}}}
            texture_metadata{Name := "Fonts/FeastBB/num_8", Asset := Fonts.FeastBB.num_8, Dimensions := option{vector2{X := 34.0, Y := 64.0}}}
            texture_metadata{Name := "Fonts/FeastBB/num_9", Asset := Fonts.FeastBB.num_9, Dimensions := option{vector2{X := 31.0, Y := 64.0}}}
            texture_metadata{Name := "Fonts/FeastBB/o", Asset := Fonts.FeastBB.o, Dimensions := option{vector2{X := 42.0, Y := 64.0}}}
            texture_metadata{Name := "Fonts/FeastBB/O_upper", Asset := Fonts.FeastBB.O_upper, Dimensions := option{vector2{X := 42.0, Y := 64.0}}}
            texture_metadata{Name := "Fonts/FeastBB/p", Asset := Fonts.FeastBB.p, Dimensions := option{vector2{X := 36.0, Y := 64.0}}}
            texture_metadata{Name := "Fonts/FeastBB/P_upper", Asset := Fonts.FeastBB.P_upper, Dimensions := option{vector2{X := 36.0, Y := 64.0}}}
            texture_metadata{Name := "Fonts/FeastBB/q", Asset := Fonts.FeastBB.q, Dimensions := option{vector2{X := 41.0, Y := 64.0}}}
            texture_metadata{Name := "Fonts/FeastBB/Q_upper", Asset := Fonts.FeastBB.Q_upper, Dimensions := option{vector2{X := 41.0, Y := 64.0}}}
            texture_metadata{Name := "Fonts/FeastBB/r", Asset := Fonts.FeastBB.r, Dimensions := option{vector2{X := 41.0, Y := 64.0}}}
            texture_metadata{Name := "Fonts/FeastBB/R_upper", Asset := Fonts.FeastBB.R_upper, Dimensions := option{vector2{X := 37.0, Y := 64.0}}}
            texture_metadata{Name := "Fonts/FeastBB/s", Asset := Fonts.FeastBB.s, Dimensions := option{vector2{X := 28.0, Y := 64.0}}}
            texture_metadata{Name := "Fonts/FeastBB/S_upper", Asset := Fonts.FeastBB.S_upper, Dimensions := option{vector2{X := 28.0, Y := 64.0}}}
            texture_metadata{Name := "Fonts/FeastBB/symb_bang", Asset := Fonts.FeastBB.symb_bang, Dimensions := option{vector2{X := 13.0, Y := 64.0}}}
            texture_metadata{Name := "Fonts/FeastBB/symb_colon", Asset := Fonts.FeastBB.symb_colon, Dimensions := option{vector2{X := 15.0, Y := 64.0}}}
            texture_metadata{Name := "Fonts/FeastBB/symb_comma", Asset := Fonts.FeastBB.symb_comma, Dimensions := option{vector2{X := 14.0, Y := 64.0}}}
            texture_metadata{Name := "Fonts/FeastBB/symb_dot", Asset := Fonts.FeastBB.symb_dot, Dimensions := option{vector2{X := 14.0, Y := 64.0}}}
            texture_metadata{Name := "Fonts/FeastBB/symb_eq", Asset := Fonts.FeastBB.symb_eq, Dimensions := option{vector2{X := 24.0, Y := 64.0}}}
            texture_metadata{Name := "Fonts/FeastBB/symb_minus", Asset := Fonts.FeastBB.symb_minus, Dimensions := option{vector2{X := 23.0, Y := 64.0}}}
            texture_metadata{Name := "Fonts/FeastBB/symb_paren_close", Asset := Fonts.FeastBB.symb_paren_close, Dimensions := option{vector2{X := 25.0, Y := 64.0}}}
            texture_metadata{Name := "Fonts/FeastBB/symb_paren_open", Asset := Fonts.FeastBB.symb_paren_open, Dimensions := option{vector2{X := 24.0, Y := 64.0}}}
            texture_metadata{Name := "Fonts/FeastBB/symb_plus", Asset := Fonts.FeastBB.symb_plus, Dimensions := option{vector2{X := 23.0, Y := 64.0}}}
            texture_metadata{Name := "Fonts/FeastBB/symb_question", Asset := Fonts.FeastBB.symb_question, Dimensions := option{vector2{X := 25.0, Y := 64.0}}}
            texture_metadata{Name := "Fonts/FeastBB/symb_semicolon", Asset := Fonts.FeastBB.symb_semicolon, Dimensions := option{vector2{X := 14.0, Y := 64.0}}}
            texture_metadata{Name := "Fonts/FeastBB/symb_slash", Asset := Fonts.FeastBB.symb_slash, Dimensions := option{vector2{X := 36.0, Y := 64.0}}}
            texture_metadata{Name := "Fonts/FeastBB/symb_underscore", Asset := Fonts.FeastBB.symb_underscore, Dimensions := option{vector2{X := 31.0, Y := 64.0}}}
            texture_metadata{Name := "Fonts/FeastBB/t", Asset := Fonts.FeastBB.t, Dimensions := option{vector2{X := 28.0, Y := 64.0}}}
            texture_metadata{Name := "Fonts/FeastBB/T_upper", Asset := Fonts.FeastBB.T_upper, Dimensions := option{vector2{X := 28.0, Y := 64.0}}}
            texture_metadata{Name := "Fonts/FeastBB/u", Asset := Fonts.FeastBB.u, Dimensions := option{vector2{X := 30.0, Y := 64.0}}}
            texture_metadata{Name := "Fonts/FeastBB/U_upper", Asset := Fonts.FeastBB.U_upper, Dimensions := option{vector2{X := 30.0, Y := 64.0}}}
            texture_metadata{Name := "Fonts/FeastBB/v", Asset := Fonts.FeastBB.v, Dimensions := option{vector2{X := 34.0, Y := 64.0}}}
            texture_metadata{Name := "Fonts/FeastBB/V_upper", Asset := Fonts.FeastBB.V_upper, Dimensions := option{vector2{X := 34.0, Y := 64.0}}}
            texture_metadata{Name := "Fonts/FeastBB/w", Asset := Fonts.FeastBB.w, Dimensions := option{vector2{X := 49.0, Y := 64.0}}}
            texture_metadata{Name := "Fonts/FeastBB/W_upper", Asset := Fonts.FeastBB.W_upper, Dimensions := option{vector2{X := 50.0, Y := 64.0}}}
            texture_metadata{Name := "Fonts/FeastBB/x", Asset := Fonts.FeastBB.x, Dimensions := option{vector2{X := 42.0, Y := 64.0}}}
            texture_metadata{Name := "Fonts/FeastBB/X_upper", Asset := Fonts.FeastBB.X_upper, Dimensions := option{vector2{X := 43.0, Y := 64.0}}}
            texture_metadata{Name := "Fonts/FeastBB/y", Asset := Fonts.FeastBB.y, Dimensions := option{vector2{X := 39.0, Y := 64.0}}}
            texture_metadata{Name := "Fonts/FeastBB/Y_upper", Asset := Fonts.FeastBB.Y_upper, Dimensions := option{vector2{X := 36.0, Y := 64.0}}}
            texture_metadata{Name := "Fonts/FeastBB/z", Asset := Fonts.FeastBB.z, Dimensions := option{vector2{X := 32.0, Y := 64.0}}}
            texture_metadata{Name := "Fonts/FeastBB/Z_upper", Asset := Fonts.FeastBB.Z_upper, Dimensions := option{vector2{X := 31.0, Y := 64.0}}}
            texture_metadata{Name := "Fonts/QuicksandBold/a", Asset := Fonts.QuicksandBold.a, Dimensions := option{vector2{X := 15.0, Y := 29.0}}}
            texture_metadata{Name := "Fonts/QuicksandBold/A_upper", Asset := Fonts.QuicksandBold.A_upper, Dimensions := option{vector2{X := 16.0, Y := 29.0}}}
            texture_metadata{Name := "Fonts/QuicksandBold/b", Asset := Fonts.QuicksandBold.b, Dimensions := option{vector2{X := 15.0, Y := 29.0}}}
            texture_metadata{Name := "Fonts/QuicksandBold/B_upper", Asset := Fonts.QuicksandBold.B_upper, Dimensions := option{vector2{X := 16.0, Y := 29.0}}}
            texture_metadata{Name := "Fonts/QuicksandBold/c", Asset := Fonts.QuicksandBold.c, Dimensions := option{vector2{X := 12.0, Y := 29.0}}}
            texture_metadata{Name := "Fonts/QuicksandBold/C_upper", Asset := Fonts.QuicksandBold.C_upper, Dimensions := option{vector2{X := 15.0, Y := 29.0}}}
            texture_metadata{Name := "Fonts/QuicksandBold/d", Asset := Fonts.QuicksandBold.d, Dimensions := option{vector2{X := 15.0, Y := 29.0}}}
            texture_metadata{Name := "Fonts/QuicksandBold/D_upper", Asset := Fonts.QuicksandBold.D_upper, Dimensions := option{vector2{X := 17.0, Y := 29.0}}}
            texture_metadata{Name := "Fonts/QuicksandBold/e", Asset := Fonts.QuicksandBold.e, Dimensions := option{vector2{X := 14.0, Y := 29.0}}}
            texture_metadata{Name := "Fonts/QuicksandBold/E_upper", Asset := Fonts.QuicksandBold.E_upper, Dimensions := option{vector2{X := 14.0, Y := 29.0}}}
            texture_metadata{Name := "Fonts/QuicksandBold/f", Asset := Fonts.QuicksandBold.f, Dimensions := option{vector2{X := 11.0, Y := 29.0}}}
            texture_metadata{Name := "Fonts/QuicksandBold/F_upper", Asset := Fonts.QuicksandBold.F_upper, Dimensions := option{vector2{X := 13.0, Y := 29.0}}}
            texture_metadata{Name := "Fonts/QuicksandBold/g", Asset := Fonts.QuicksandBold.g, Dimensions := option{vector2{X := 15.0, Y := 29.0}}}
            texture_metadata{Name := "Fonts/QuicksandBold/G_upper", Asset := Fonts.QuicksandBold.G_upper, Dimensions := option{vector2{X := 17.0, Y := 29.0}}}
            texture_metadata{Name := "Fonts/QuicksandBold/h", Asset := Fonts.QuicksandBold.h, Dimensions := option{vector2{X := 14.0, Y := 29.0}}}
            texture_metadata{Name := "Fonts/QuicksandBold/H_upper", Asset := Fonts.QuicksandBold.H_upper, Dimensions := option{vector2{X := 17.0, Y := 29.0}}}
            texture_metadata{Name := "Fonts/QuicksandBold/i", Asset := Fonts.QuicksandBold.i, Dimensions := option{vector2{X := 6.0, Y := 29.0}}}
            texture_metadata{Name := "Fonts/QuicksandBold/I_upper", Asset := Fonts.QuicksandBold.I_upper, Dimensions := option{vector2{X := 7.0, Y := 29.0}}}
            texture_metadata{Name := "Fonts/QuicksandBold/j", Asset := Fonts.QuicksandBold.j, Dimensions := option{vector2{X := 7.0, Y := 29.0}}}
            texture_metadata{Name := "Fonts/QuicksandBold/J_upper", Asset := Fonts.QuicksandBold.J_upper, Dimensions := option{vector2{X := 13.0, Y := 29.0}}}
            texture_metadata{Name := "Fonts/QuicksandBold/k", Asset := Fonts.QuicksandBold.k, Dimensions := option{vector2{X := 14.0, Y := 29.0}}}
            texture_metadata{Name := "Fonts/QuicksandBold/K_upper", Asset := Fonts.QuicksandBold.K_upper, Dimensions := option{vector2{X := 16.0, Y := 29.0}}}
            texture_metadata{Name := "Fonts/QuicksandBold/l", Asset := Fonts.QuicksandBold.l, Dimensions := option{vector2{X := 6.0, Y := 29.0}}}
            texture_metadata{Name := "Fonts/QuicksandBold/L_upper", Asset := Fonts.QuicksandBold.L_upper, Dimensions := option{vector2{X := 13.0, Y := 29.0}}}
            texture_metadata{Name := "Fonts/QuicksandBold/m", Asset := Fonts.QuicksandBold.m, Dimensions := option{vector2{X := 22.0, Y := 29.0}}}
            texture_metadata{Name := "Fonts/QuicksandBold/M_upper", Asset := Fonts.QuicksandBold.M_upper, Dimensions := option{vector2{X := 20.0, Y := 29.0}}}
            texture_metadata{Name := "Fonts/QuicksandBold/n", Asset := Fonts.QuicksandBold.n, Dimensions := option{vector2{X := 14.0, Y := 29.0}}}
            texture_metadata{Name := "Fonts/QuicksandBold/N_upper", Asset := Fonts.QuicksandBold.N_upper, Dimensions := option{vector2{X := 18.0, Y := 29.0}}}
            texture_metadata{Name := "Fonts/QuicksandBold/num_0", Asset := Fonts.QuicksandBold.num_0, Dimensions := option{vector2{X := 15.0, Y := 29.0}}}
            texture_metadata{Name := "Fonts/QuicksandBold/num_1", Asset := Fonts.QuicksandBold.num_1, Dimensions := option{vector2{X := 9.0, Y := 29.0}}}
            texture_metadata{Name := "Fonts/QuicksandBold/num_2", Asset := Fonts.QuicksandBold.num_2, Dimensions := option{vector2{X := 13.0, Y := 29.0}}}
            texture_metadata{Name := "Fonts/QuicksandBold/num_3", Asset := Fonts.QuicksandBold.num_3, Dimensions := option{vector2{X := 13.0, Y := 29.0}}}
            texture_metadata{Name := "Fonts/QuicksandBold/num_4", Asset := Fonts.QuicksandBold.num_4, Dimensions := option{vector2{X := 17.0, Y := 29.0}}}
            texture_metadata{Name := "Fonts/QuicksandBold/num_5", Asset := Fonts.QuicksandBold.num_5, Dimensions := option{vector2{X := 14.0, Y := 29.0}}}
            texture_metadata{Name := "Fonts/QuicksandBold/num_6", Asset := Fonts.QuicksandBold.num_6, Dimensions := option{vector2{X := 13.0, Y := 29.0}}}
            texture_metadata{Name := "Fonts/QuicksandBold/num_7", Asset := Fonts.QuicksandBold.num_7, Dimensions := option{vector2{X := 14.0, Y := 29.0}}}
            texture_metadata{Name := "Fonts/QuicksandBold/num_8", Asset := Fonts.QuicksandBold.num_8, Dimensions := option{vector2{X := 14.0, Y := 29.0}}}
            texture_metadata{Name := "Fonts/QuicksandBold/num_9", Asset := Fonts.QuicksandBold.num_9, Dimensions := option{vector2{X := 13.0, Y := 29.0}}}
            texture_metadata{Name := "Fonts/QuicksandBold/o", Asset := Fonts.QuicksandBold.o, Dimensions := option{vector2{X := 15.0, Y := 29.0}}}
            texture_metadata{Name := "Fonts/QuicksandBold/O_upper", Asset := Fonts.QuicksandBold.O_upper, Dimensions := option{vector2{X := 18.0, Y := 29.0}}}
            texture_metadata{Name := "Fonts/QuicksandBold/p", Asset := Fonts.QuicksandBold.p, Dimensions := option{vector2{X := 15.0, Y := 29.0}}}
            texture_metadata{Name := "Fonts/QuicksandBold/P_upper", Asset := Fonts.QuicksandBold.P_upper, Dimensions := option{vector2{X := 14.0, Y := 29.0}}}
            texture_metadata{Name := "Fonts/QuicksandBold/q", Asset := Fonts.QuicksandBold.q, Dimensions := option{vector2{X := 15.0, Y := 29.0}}}
            texture_metadata{Name := "Fonts/QuicksandBold/Q_upper", Asset := Fonts.QuicksandBold.Q_upper, Dimensions := option{vector2{X := 19.0, Y := 29.0}}}
            texture_metadata{Name := "Fonts/QuicksandBold/r", Asset := Fonts.QuicksandBold.r, Dimensions := option{vector2{X := 10.0, Y := 29.0}}}
            texture_metadata{Name := "Fonts/QuicksandBold/R_upper", Asset := Fonts.QuicksandBold.R_upper, Dimensions := option{vector2{X := 16.0, Y := 29.0}}}
            texture_metadata{Name := "Fonts/QuicksandBold/s", Asset := Fonts.QuicksandBold.s, Dimensions := option{vector2{X := 12.0, Y := 29.0}}}
            texture_metadata{Name := "Fonts/QuicksandBold/S_upper", Asset := Fonts.QuicksandBold.S_upper, Dimensions := option{vector2{X := 14.0, Y := 29.0}}}
            texture_metadata{Name := "Fonts/QuicksandBold/symb_bang", Asset := Fonts.QuicksandBold.symb_bang, Dimensions := option{vector2{X := 6.0, Y := 29.0}}}
            texture_metadata{Name := "Fonts/QuicksandBold/symb_comma", Asset := Fonts.QuicksandBold.symb_comma, Dimensions := option{vector2{X := 7.0, Y := 29.0}}}
            texture_metadata{Name := "Fonts/QuicksandBold/symb_dot", Asset := Fonts.QuicksandBold.symb_dot, Dimensions := option{vector2{X := 6.0, Y := 29.0}}}
            texture_metadata{Name := "Fonts/QuicksandBold/symb_eq", Asset := Fonts.QuicksandBold.symb_eq, Dimensions := option{vector2{X := 15.0, Y := 29.0}}}
            texture_metadata{Name := "Fonts/QuicksandBold/symb_minus", Asset := Fonts.QuicksandBold.symb_minus, Dimensions := option{vector2{X := 10.0, Y := 29.0}}}
            texture_metadata{Name := "Fonts/QuicksandBold/symb_paren_close", Asset := Fonts.QuicksandBold.symb_paren_close, Dimensions := option{vector2{X := 9.0, Y := 29.0}}}
            texture_metadata{Name := "Fonts/QuicksandBold/symb_paren_open", Asset := Fonts.QuicksandBold.symb_paren_open, Dimensions := option{vector2{X := 9.0, Y := 29.0}}}
            texture_metadata{Name := "Fonts/QuicksandBold/symb_plus", Asset := Fonts.QuicksandBold.symb_plus, Dimensions := option{vector2{X := 14.0, Y := 29.0}}}
            texture_metadata{Name := "Fonts/QuicksandBold/symb_question", Asset := Fonts.QuicksandBold.symb_question, Dimensions := option{vector2{X := 13.0, Y := 29.0}}}
            texture_metadata{Name := "Fonts/QuicksandBold/symb_slash", Asset := Fonts.QuicksandBold.symb_slash, Dimensions := option{vector2{X := 14.0, Y := 29.0}}}
            texture_metadata{Name := "Fonts/QuicksandBold/t", Asset := Fonts.QuicksandBold.t, Dimensions := option{vector2{X := 10.0, Y := 29.0}}}
            texture_metadata{Name := "Fonts/QuicksandBold/T_upper", Asset := Fonts.QuicksandBold.T_upper, Dimensions := option{vector2{X := 15.0, Y := 29.0}}}
            texture_metadata{Name := "Fonts/QuicksandBold/u", Asset := Fonts.QuicksandBold.u, Dimensions := option{vector2{X := 14.0, Y := 29.0}}}
            texture_metadata{Name := "Fonts/QuicksandBold/U_upper", Asset := Fonts.QuicksandBold.U_upper, Dimensions := option{vector2{X := 17.0, Y := 29.0}}}
            texture_metadata{Name := "Fonts/QuicksandBold/v", Asset := Fonts.QuicksandBold.v, Dimensions := option{vector2{X := 13.0, Y := 29.0}}}
            texture_metadata{Name := "Fonts/QuicksandBold/V_upper", Asset := Fonts.QuicksandBold.V_upper, Dimensions := option{vector2{X := 16.0, Y := 29.0}}}
            texture_metadata{Name := "Fonts/QuicksandBold/w", Asset := Fonts.QuicksandBold.w, Dimensions := option{vector2{X := 18.0, Y := 29.0}}}
            texture_metadata{Name := "Fonts/QuicksandBold/W_upper", Asset := Fonts.QuicksandBold.W_upper, Dimensions := option{vector2{X := 23.0, Y := 29.0}}}
            texture_metadata{Name := "Fonts/QuicksandBold/x", Asset := Fonts.QuicksandBold.x, Dimensions := option{vector2{X := 12.0, Y := 29.0}}}
            texture_metadata{Name := "Fonts/QuicksandBold/X_upper", Asset := Fonts.QuicksandBold.X_upper, Dimensions := option{vector2{X := 16.0, Y := 29.0}}}
            texture_metadata{Name := "Fonts/QuicksandBold/y", Asset := Fonts.QuicksandBold.y, Dimensions := option{vector2{X := 14.0, Y := 29.0}}}
            texture_metadata{Name := "Fonts/QuicksandBold/Y_upper", Asset := Fonts.QuicksandBold.Y_upper, Dimensions := option{vector2{X := 14.0, Y := 29.0}}}
            texture_metadata{Name := "Fonts/QuicksandBold/z", Asset := Fonts.QuicksandBold.z, Dimensions := option{vector2{X := 12.0, Y := 29.0}}}
            texture_metadata{Name := "Fonts/QuicksandBold/Z_upper", Asset := Fonts.QuicksandBold.Z_upper, Dimensions := option{vector2{X := 16.0, Y := 29.0}}}
            texture_metadata{Name := "GameplayRewardsMenuUI/background", Asset := GameplayRewardsMenuUI.background, Dimensions := option{vector2{X := 2048.0, Y := 1425.0}}}
            texture_metadata{Name := "GameplayRewardsMenuUI/card_bg", Asset := GameplayRewardsMenuUI.card_bg, Dimensions := option{vector2{X := 548.0, Y := 622.0}}}
            texture_metadata{Name := "GameplayRewardsMenuUI/gameplay_r1", Asset := GameplayRewardsMenuUI.gameplay_r1, Dimensions := option{vector2{X := 480.0, Y := 407.0}}}
            texture_metadata{Name := "GameplayRewardsMenuUI/gameplay_r1_claimed", Asset := GameplayRewardsMenuUI.gameplay_r1_claimed, Dimensions := option{vector2{X := 480.0, Y := 407.0}}}
            texture_metadata{Name := "GameplayRewardsMenuUI/gameplay_r2", Asset := GameplayRewardsMenuUI.gameplay_r2, Dimensions := option{vector2{X := 480.0, Y := 407.0}}}
            texture_metadata{Name := "GameplayRewardsMenuUI/gameplay_r2_claimed", Asset := GameplayRewardsMenuUI.gameplay_r2_claimed, Dimensions := option{vector2{X := 480.0, Y := 407.0}}}
            texture_metadata{Name := "GameplayRewardsMenuUI/gameplay_r3", Asset := GameplayRewardsMenuUI.gameplay_r3, Dimensions := option{vector2{X := 480.0, Y := 407.0}}}
            texture_metadata{Name := "GameplayRewardsMenuUI/gameplay_r3_claimed", Asset := GameplayRewardsMenuUI.gameplay_r3_claimed, Dimensions := option{vector2{X := 480.0, Y := 407.0}}}
            texture_metadata{Name := "GameplayRewardsMenuUI/gameplay_r4", Asset := GameplayRewardsMenuUI.gameplay_r4, Dimensions := option{vector2{X := 480.0, Y := 407.0}}}
            texture_metadata{Name := "GameplayRewardsMenuUI/gameplay_r4_claimed", Asset := GameplayRewardsMenuUI.gameplay_r4_claimed, Dimensions := option{vector2{X := 480.0, Y := 407.0}}}
            texture_metadata{Name := "GameplayRewardsMenuUI/gameplay_r5", Asset := GameplayRewardsMenuUI.gameplay_r5, Dimensions := option{vector2{X := 480.0, Y := 407.0}}}
            texture_metadata{Name := "GameplayRewardsMenuUI/gameplay_r5_claimed", Asset := GameplayRewardsMenuUI.gameplay_r5_claimed, Dimensions := option{vector2{X := 480.0, Y := 407.0}}}
            texture_metadata{Name := "GameplayRewardsMenuUI/gameplay_r6", Asset := GameplayRewardsMenuUI.gameplay_r6, Dimensions := option{vector2{X := 480.0, Y := 407.0}}}
            texture_metadata{Name := "GameplayRewardsMenuUI/gameplay_r6_claimed", Asset := GameplayRewardsMenuUI.gameplay_r6_claimed, Dimensions := option{vector2{X := 480.0, Y := 407.0}}}
            texture_metadata{Name := "GameplayRewardsMenuUI/gameplay_r7", Asset := GameplayRewardsMenuUI.gameplay_r7, Dimensions := option{vector2{X := 480.0, Y := 407.0}}}
            texture_metadata{Name := "GameplayRewardsMenuUI/gameplay_r7_claimed", Asset := GameplayRewardsMenuUI.gameplay_r7_claimed, Dimensions := option{vector2{X := 480.0, Y := 407.0}}}
            texture_metadata{Name := "GameplayRewardsMenuUI/gameplay_r8", Asset := GameplayRewardsMenuUI.gameplay_r8, Dimensions := option{vector2{X := 480.0, Y := 407.0}}}
            texture_metadata{Name := "GameplayRewardsMenuUI/gameplay_r8_claimed", Asset := GameplayRewardsMenuUI.gameplay_r8_claimed, Dimensions := option{vector2{X := 480.0, Y := 407.0}}}
            texture_metadata{Name := "GameplayRewardsMenuUI/unlocks_in_bg", Asset := GameplayRewardsMenuUI.unlocks_in_bg, Dimensions := option{vector2{X := 468.0, Y := 112.0}}}
            texture_metadata{Name := "GameplayRewardsMenuUI/up_next_label", Asset := GameplayRewardsMenuUI.up_next_label, Dimensions := option{vector2{X := 282.0, Y := 68.0}}}
            texture_metadata{Name := "HavocVerseUI/ui_black_contrast_rounded", Asset := HavocVerseUI.ui_black_contrast_rounded, Dimensions := option{vector2{X := 512.0, Y := 512.0}}}
            texture_metadata{Name := "HavocVerseUI/ui_debug_menu", Asset := HavocVerseUI.ui_debug_menu, Dimensions := option{vector2{X := 1280.0, Y := 720.0}}}
            texture_metadata{Name := "HavocVerseUI/ui_debug_menu_code_prompt", Asset := HavocVerseUI.ui_debug_menu_code_prompt, Dimensions := option{vector2{X := 450.0, Y := 542.0}}}
            texture_metadata{Name := "TemplateCommon/placeholder", Asset := TemplateCommon.placeholder, Dimensions := option{vector2{X := 512.0, Y := 256.0}}}
            texture_metadata{Name := "TemplateCommon/BackendTooltips/extra", Asset := TemplateCommon.BackendTooltips.extra, Dimensions := option{vector2{X := 2048.0, Y := 1024.0}}}
            texture_metadata{Name := "TemplateCommon/BackendTooltips/integration_debug", Asset := TemplateCommon.BackendTooltips.integration_debug, Dimensions := option{vector2{X := 2048.0, Y := 2048.0}}}
            texture_metadata{Name := "TemplateCommon/BackendTooltips/progress_line", Asset := TemplateCommon.BackendTooltips.progress_line, Dimensions := option{vector2{X := 1024.0, Y := 512.0}}}
            texture_metadata{Name := "TemplateCommon/BackendTooltips/tycoon", Asset := TemplateCommon.BackendTooltips.tycoon, Dimensions := option{vector2{X := 2048.0, Y := 1024.0}}}
            texture_metadata{Name := "UI/Verse/ui_keypad", Asset := UI.Verse.ui_keypad, Dimensions := option{vector2{X := 804.0, Y := 1566.0}}}
            texture_metadata{Name := "UI/Verse/ui_mapcode", Asset := UI.Verse.ui_mapcode, Dimensions := option{vector2{X := 592.0, Y := 68.0}}}
            texture_metadata{Name := "UI/Verse/Rebirths/rebirth_01", Asset := UI.Verse.Rebirths.rebirth_01, Dimensions := option{vector2{X := 548.0, Y := 128.0}}}
            texture_metadata{Name := "UI/Verse/Rebirths/rebirth_02", Asset := UI.Verse.Rebirths.rebirth_02, Dimensions := option{vector2{X := 548.0, Y := 128.0}}}
            texture_metadata{Name := "UI/Verse/Rebirths/rebirth_03", Asset := UI.Verse.Rebirths.rebirth_03, Dimensions := option{vector2{X := 548.0, Y := 128.0}}}
            texture_metadata{Name := "UI/Verse/Rebirths/rebirth_04", Asset := UI.Verse.Rebirths.rebirth_04, Dimensions := option{vector2{X := 548.0, Y := 128.0}}}
            texture_metadata{Name := "UI/Verse/Rebirths/rebirth_05", Asset := UI.Verse.Rebirths.rebirth_05, Dimensions := option{vector2{X := 548.0, Y := 128.0}}}
            texture_metadata{Name := "UI/Verse/Rebirths/rebirth_06", Asset := UI.Verse.Rebirths.rebirth_06, Dimensions := option{vector2{X := 548.0, Y := 128.0}}}
            texture_metadata{Name := "UI/Verse/Rebirths/rebirth_07", Asset := UI.Verse.Rebirths.rebirth_07, Dimensions := option{vector2{X := 548.0, Y := 128.0}}}
            texture_metadata{Name := "UI/Verse/Rebirths/rebirth_08", Asset := UI.Verse.Rebirths.rebirth_08, Dimensions := option{vector2{X := 548.0, Y := 128.0}}}
            texture_metadata{Name := "UI/Verse/Rebirths/rebirth_09", Asset := UI.Verse.Rebirths.rebirth_09, Dimensions := option{vector2{X := 548.0, Y := 128.0}}}
            texture_metadata{Name := "UI/Verse/Rebirths/rebirth_10", Asset := UI.Verse.Rebirths.rebirth_10, Dimensions := option{vector2{X := 548.0, Y := 128.0}}}
            texture_metadata{Name := "UI/Verse/Rebirths/rebirth_100", Asset := UI.Verse.Rebirths.rebirth_100, Dimensions := option{vector2{X := 548.0, Y := 128.0}}}
            texture_metadata{Name := "UI/Verse/Rebirths/rebirth_11", Asset := UI.Verse.Rebirths.rebirth_11, Dimensions := option{vector2{X := 548.0, Y := 128.0}}}
            texture_metadata{Name := "UI/Verse/Rebirths/rebirth_12", Asset := UI.Verse.Rebirths.rebirth_12, Dimensions := option{vector2{X := 548.0, Y := 128.0}}}
            texture_metadata{Name := "UI/Verse/Rebirths/rebirth_13", Asset := UI.Verse.Rebirths.rebirth_13, Dimensions := option{vector2{X := 548.0, Y := 128.0}}}
            texture_metadata{Name := "UI/Verse/Rebirths/rebirth_14", Asset := UI.Verse.Rebirths.rebirth_14, Dimensions := option{vector2{X := 548.0, Y := 128.0}}}
            texture_metadata{Name := "UI/Verse/Rebirths/rebirth_15", Asset := UI.Verse.Rebirths.rebirth_15, Dimensions := option{vector2{X := 548.0, Y := 128.0}}}
            texture_metadata{Name := "UI/Verse/Rebirths/rebirth_16", Asset := UI.Verse.Rebirths.rebirth_16, Dimensions := option{vector2{X := 548.0, Y := 128.0}}}
            texture_metadata{Name := "UI/Verse/Rebirths/rebirth_17", Asset := UI.Verse.Rebirths.rebirth_17, Dimensions := option{vector2{X := 548.0, Y := 128.0}}}
            texture_metadata{Name := "UI/Verse/Rebirths/rebirth_18", Asset := UI.Verse.Rebirths.rebirth_18, Dimensions := option{vector2{X := 548.0, Y := 128.0}}}
            texture_metadata{Name := "UI/Verse/Rebirths/rebirth_19", Asset := UI.Verse.Rebirths.rebirth_19, Dimensions := option{vector2{X := 548.0, Y := 128.0}}}
            texture_metadata{Name := "UI/Verse/Rebirths/rebirth_20", Asset := UI.Verse.Rebirths.rebirth_20, Dimensions := option{vector2{X := 548.0, Y := 128.0}}}
            texture_metadata{Name := "UI/Verse/Rebirths/rebirth_200", Asset := UI.Verse.Rebirths.rebirth_200, Dimensions := option{vector2{X := 548.0, Y := 128.0}}}
            texture_metadata{Name := "UI/Verse/Rebirths/rebirth_21", Asset := UI.Verse.Rebirths.rebirth_21, Dimensions := option{vector2{X := 548.0, Y := 128.0}}}
            texture_metadata{Name := "UI/Verse/Rebirths/rebirth_22", Asset := UI.Verse.Rebirths.rebirth_22, Dimensions := option{vector2{X := 548.0, Y := 128.0}}}
            texture_metadata{Name := "UI/Verse/Rebirths/rebirth_23", Asset := UI.Verse.Rebirths.rebirth_23, Dimensions := option{vector2{X := 548.0, Y := 128.0}}}
            texture_metadata{Name := "UI/Verse/Rebirths/rebirth_24", Asset := UI.Verse.Rebirths.rebirth_24, Dimensions := option{vector2{X := 548.0, Y := 128.0}}}
            texture_metadata{Name := "UI/Verse/Rebirths/rebirth_25", Asset := UI.Verse.Rebirths.rebirth_25, Dimensions := option{vector2{X := 548.0, Y := 128.0}}}
            texture_metadata{Name := "UI/Verse/Rebirths/rebirth_26", Asset := UI.Verse.Rebirths.rebirth_26, Dimensions := option{vector2{X := 548.0, Y := 128.0}}}
            texture_metadata{Name := "UI/Verse/Rebirths/rebirth_300", Asset := UI.Verse.Rebirths.rebirth_300, Dimensions := option{vector2{X := 548.0, Y := 128.0}}}
            texture_metadata{Name := "UI/Verse/Rebirths/rebirth_400", Asset := UI.Verse.Rebirths.rebirth_400, Dimensions := option{vector2{X := 548.0, Y := 128.0}}}
            texture_metadata{Name := "UI/Verse/Rebirths/rebirth_50", Asset := UI.Verse.Rebirths.rebirth_50, Dimensions := option{vector2{X := 548.0, Y := 128.0}}}
            texture_metadata{Name := "UI/Verse/Rebirths/rebirth_500", Asset := UI.Verse.Rebirths.rebirth_500, Dimensions := option{vector2{X := 548.0, Y := 128.0}}}
        }

        var MappedTextureAssets: [string]texture_metadata = map{}
        for (
            TextureMeta:TextureAssets
            set MappedTextureAssets[TextureMeta.Name] = TextureMeta
        ) {}

        local_assets_provider{
            TextureAssets := MappedTextureAssets
        }

local_assets_provider := class<final>(assets_provider):
    TextureAssets: [string]texture_metadata

    GetTextureMetaAssets<override>()<transacts>:[]texture_metadata=
        TextureAssets.Values()

    GetTextureAsset<override>(Name:string)<transacts><decides>:texture=
        TextureAssets[Name].Asset

    GetTextureAssetDimensions<override>(Name:string)<transacts><decides>:vector2=
        TextureAssets[Name].Dimensions?

    GetMaterialAsset<override>(Name:string)<transacts><decides>:material=
        case(Name):
            "CustomHUD/FontMat_Base" => CustomHUD.FontMat_Base
            "CustomHUD/FontMat_Booster" => CustomHUD.FontMat_Booster
            "CustomHUD/FontMat_Gold" => CustomHUD.FontMat_Gold
            "CustomHUD/FontMat_Res2" => CustomHUD.FontMat_Res2
            "CustomHUD/UIMat_HealthBar" => CustomHUD.UIMat_HealthBar
            "CustomHUD/UIMat_HealthBarCursor" => CustomHUD.UIMat_HealthBarCursor
            "CustomHUD/UIMat_LevelProgress" => CustomHUD.UIMat_LevelProgress
            "CustomHUD/UIMat_ResourceIconBase" => CustomHUD.UIMat_ResourceIconBase
            "CustomHUD/UIMat_ShieldBar" => CustomHUD.UIMat_ShieldBar
            "CustomHUD/Notify/MI_new_pet_unlocked" => CustomHUD.Notify.MI_new_pet_unlocked
            "CustomHUD/Notify/UIMat_NotifyIcon" => CustomHUD.Notify.UIMat_NotifyIcon
            "BuyplateBillboards/MI_bp_001" => BuyplateBillboards.MI_bp_001
            "BuyplateBillboards/MI_bp_002" => BuyplateBillboards.MI_bp_002
            "BuyplateBillboards/MI_bp_003" => BuyplateBillboards.MI_bp_003
            "BuyplateBillboards/MI_bp_004" => BuyplateBillboards.MI_bp_004
            "BuyplateBillboards/MI_bp_005" => BuyplateBillboards.MI_bp_005
            "BuyplateBillboards/MI_bp_006" => BuyplateBillboards.MI_bp_006
            "BuyplateBillboards/MI_bp_007" => BuyplateBillboards.MI_bp_007
            "BuyplateBillboards/MI_bp_008" => BuyplateBillboards.MI_bp_008
            "BuyplateBillboards/MI_bp_009" => BuyplateBillboards.MI_bp_009
            "BuyplateBillboards/MI_bp_010" => BuyplateBillboards.MI_bp_010
            "BuyplateBillboards/MI_bp_011" => BuyplateBillboards.MI_bp_011
            "BuyplateBillboards/MI_bp_012" => BuyplateBillboards.MI_bp_012
            "BuyplateBillboards/MI_bp_013" => BuyplateBillboards.MI_bp_013
            "BuyplateBillboards/MI_bp_014" => BuyplateBillboards.MI_bp_014
            "BuyplateBillboards/MI_bp_015" => BuyplateBillboards.MI_bp_015
            "BuyplateBillboards/MI_bp_016" => BuyplateBillboards.MI_bp_016
            "BuyplateBillboards/MI_bp_017" => BuyplateBillboards.MI_bp_017
            "BuyplateBillboards/MI_bp_018" => BuyplateBillboards.MI_bp_018
            "BuyplateBillboards/MI_bp_019" => BuyplateBillboards.MI_bp_019
            "BuyplateBillboards/MI_bp_020" => BuyplateBillboards.MI_bp_020
            "BuyplateBillboards/MI_bp_021" => BuyplateBillboards.MI_bp_021
            "BuyplateBillboards/MI_bp_022" => BuyplateBillboards.MI_bp_022
            "BuyplateBillboards/MI_bp_023" => BuyplateBillboards.MI_bp_023
            "BuyplateBillboards/MI_bp_024" => BuyplateBillboards.MI_bp_024
            "BuyplateBillboards/MI_bp_025" => BuyplateBillboards.MI_bp_025
            "BuyplateBillboards/MI_bp_026" => BuyplateBillboards.MI_bp_026
            "BuyplateBillboards/MI_bp_027" => BuyplateBillboards.MI_bp_027
            "BuyplateBillboards/MI_bp_028" => BuyplateBillboards.MI_bp_028
            "BuyplateBillboards/MI_bp_029" => BuyplateBillboards.MI_bp_029
            "BuyplateBillboards/MI_bp_030" => BuyplateBillboards.MI_bp_030
            "BuyplateBillboards/MI_bp_031" => BuyplateBillboards.MI_bp_031
            "BuyplateBillboards/MI_bp_032" => BuyplateBillboards.MI_bp_032
            "BuyplateBillboards/MI_bp_033" => BuyplateBillboards.MI_bp_033
            "BuyplateBillboards/MI_bp_034" => BuyplateBillboards.MI_bp_034
            "BuyplateBillboards/MI_bp_035" => BuyplateBillboards.MI_bp_035
            "BuyplateBillboards/MI_bp_036" => BuyplateBillboards.MI_bp_036
            "BuyplateBillboards/MI_bp_037" => BuyplateBillboards.MI_bp_037
            "BuyplateBillboards/MI_bp_038" => BuyplateBillboards.MI_bp_038
            "BuyplateBillboards/MI_bp_039" => BuyplateBillboards.MI_bp_039
            "BuyplateBillboards/MI_bp_040" => BuyplateBillboards.MI_bp_040
            "BuyplateBillboards/MI_bp_041" => BuyplateBillboards.MI_bp_041
            "BuyplateBillboards/MI_bp_042" => BuyplateBillboards.MI_bp_042
            "BuyplateBillboards/MI_bp_043" => BuyplateBillboards.MI_bp_043
            "BuyplateBillboards/MI_bp_044" => BuyplateBillboards.MI_bp_044
            "BuyplateBillboards/MI_bp_045" => BuyplateBillboards.MI_bp_045
            "BuyplateBillboards/MI_bp_046" => BuyplateBillboards.MI_bp_046
            "BuyplateBillboards/MI_bp_047" => BuyplateBillboards.MI_bp_047
            "BuyplateBillboards/MI_bp_048" => BuyplateBillboards.MI_bp_048
            "BuyplateBillboards/MI_bp_049" => BuyplateBillboards.MI_bp_049
            "BuyplateBillboards/MI_bp_050" => BuyplateBillboards.MI_bp_050
            "BuyplateBillboards/MI_bp_051" => BuyplateBillboards.MI_bp_051
            "BuyplateBillboards/MI_bp_052" => BuyplateBillboards.MI_bp_052
            "BuyplateBillboards/MI_bp_053" => BuyplateBillboards.MI_bp_053
            "BuyplateBillboards/MI_bp_054" => BuyplateBillboards.MI_bp_054
            "BuyplateBillboards/MI_bp_055" => BuyplateBillboards.MI_bp_055
            "BuyplateBillboards/MI_bp_056" => BuyplateBillboards.MI_bp_056
            "BuyplateBillboards/MI_bp_057" => BuyplateBillboards.MI_bp_057
            "BuyplateBillboards/MI_bp_058" => BuyplateBillboards.MI_bp_058
            "BuyplateBillboards/MI_bp_059" => BuyplateBillboards.MI_bp_059
            "BuyplateBillboards/MI_bp_060" => BuyplateBillboards.MI_bp_060
            "BuyplateBillboards/MI_bp_061" => BuyplateBillboards.MI_bp_061
            "BuyplateBillboards/MI_bp_062" => BuyplateBillboards.MI_bp_062
            "BuyplateBillboards/MI_bp_063" => BuyplateBillboards.MI_bp_063
            "BuyplateBillboards/MI_bp_064" => BuyplateBillboards.MI_bp_064
            "BuyplateBillboards/MI_bp_065" => BuyplateBillboards.MI_bp_065
            "BuyplateBillboards/MI_bp_066" => BuyplateBillboards.MI_bp_066
            "BuyplateBillboards/MI_bp_067" => BuyplateBillboards.MI_bp_067
            "BuyplateBillboards/MI_bp_068" => BuyplateBillboards.MI_bp_068
            "BuyplateBillboards/MI_bp_069" => BuyplateBillboards.MI_bp_069
            "BuyplateBillboards/MI_bp_070" => BuyplateBillboards.MI_bp_070
            "BuyplateBillboards/MI_bp_071" => BuyplateBillboards.MI_bp_071
            "BuyplateBillboards/MI_bp_072" => BuyplateBillboards.MI_bp_072
            "BuyplateBillboards/MI_bp_073" => BuyplateBillboards.MI_bp_073
            "BuyplateBillboards/MI_bp_074" => BuyplateBillboards.MI_bp_074
            "BuyplateBillboards/MI_bp_075" => BuyplateBillboards.MI_bp_075
            "BuyplateBillboards/MI_bp_076" => BuyplateBillboards.MI_bp_076
            "BuyplateBillboards/MI_bp_077" => BuyplateBillboards.MI_bp_077
            "BuyplateBillboards/MI_bp_078" => BuyplateBillboards.MI_bp_078
            "BuyplateBillboards/MI_bp_079" => BuyplateBillboards.MI_bp_079
            "BuyplateBillboards/MI_bp_080" => BuyplateBillboards.MI_bp_080
            "BuyplateBillboards/MI_bp_081" => BuyplateBillboards.MI_bp_081
            "BuyplateBillboards/MI_bp_082" => BuyplateBillboards.MI_bp_082
            "BuyplateBillboards/MI_bp_083" => BuyplateBillboards.MI_bp_083
            "BuyplateBillboards/MI_bp_084" => BuyplateBillboards.MI_bp_084
            "BuyplateBillboards/MI_bp_085" => BuyplateBillboards.MI_bp_085
            "BuyplateBillboards/MI_bp_086" => BuyplateBillboards.MI_bp_086
            "BuyplateBillboards/MI_bp_087" => BuyplateBillboards.MI_bp_087
            "BuyplateBillboards/MI_bp_088" => BuyplateBillboards.MI_bp_088
            "BuyplateBillboards/MI_bp_089" => BuyplateBillboards.MI_bp_089
            "BuyplateBillboards/MI_bp_090" => BuyplateBillboards.MI_bp_090
            "BuyplateBillboards/MI_bp_091" => BuyplateBillboards.MI_bp_091
            "BuyplateBillboards/MI_bp_092" => BuyplateBillboards.MI_bp_092
            "BuyplateBillboards/MI_bp_093" => BuyplateBillboards.MI_bp_093
            "BuyplateBillboards/MI_bp_094" => BuyplateBillboards.MI_bp_094
            "BuyplateBillboards/MI_bp_095" => BuyplateBillboards.MI_bp_095
            "BuyplateBillboards/MI_bp_096" => BuyplateBillboards.MI_bp_096
            "BuyplateBillboards/MI_bp_097" => BuyplateBillboards.MI_bp_097
            "BuyplateBillboards/MI_bp_098" => BuyplateBillboards.MI_bp_098
            "BuyplateBillboards/MI_bp_099" => BuyplateBillboards.MI_bp_099
            "BuyplateBillboards/MI_bp_100" => BuyplateBillboards.MI_bp_100
            "BuyplateBillboards/MI_bp_101" => BuyplateBillboards.MI_bp_101
            "BuyplateBillboards/MI_bp_102" => BuyplateBillboards.MI_bp_102
            "BuyplateBillboards/MI_bp_103" => BuyplateBillboards.MI_bp_103
            "BuyplateBillboards/MI_bp_104" => BuyplateBillboards.MI_bp_104
            "BuyplateBillboards/MI_bp_105" => BuyplateBillboards.MI_bp_105
            "BuyplateBillboards/MI_bp_106" => BuyplateBillboards.MI_bp_106
            "BuyplateBillboards/MI_bp_107" => BuyplateBillboards.MI_bp_107
            "BuyplateBillboards/MI_bp_108" => BuyplateBillboards.MI_bp_108
            "BuyplateBillboards/MI_bp_109" => BuyplateBillboards.MI_bp_109
            "BuyplateBillboards/MI_bp_110" => BuyplateBillboards.MI_bp_110
            "BuyplateBillboards/MI_bp_111" => BuyplateBillboards.MI_bp_111
            "BuyplateBillboards/MI_bp_112" => BuyplateBillboards.MI_bp_112
            "BuyplateBillboards/MI_bp_113" => BuyplateBillboards.MI_bp_113
            "BuyplateBillboards/MI_bp_114" => BuyplateBillboards.MI_bp_114
            "BuyplateBillboards/MI_bp_115" => BuyplateBillboards.MI_bp_115
            "BuyplateBillboards/MI_bp_116" => BuyplateBillboards.MI_bp_116
            "BuyplateBillboards/MI_bp_117" => BuyplateBillboards.MI_bp_117
            "BuyplateBillboards/MI_bp_118" => BuyplateBillboards.MI_bp_118
            "BuyplateBillboards/MI_bp_119" => BuyplateBillboards.MI_bp_119
            "BuyplateBillboards/MI_bp_120" => BuyplateBillboards.MI_bp_120
            "BuyplateBillboards/MI_bp_121" => BuyplateBillboards.MI_bp_121
            "BuyplateBillboards/MI_bp_122" => BuyplateBillboards.MI_bp_122
            "BuyplateBillboards/MI_bp_123" => BuyplateBillboards.MI_bp_123
            "BuyplateBillboards/MI_bp_124" => BuyplateBillboards.MI_bp_124
            "BuyplateBillboards/MI_bp_125" => BuyplateBillboards.MI_bp_125
            "BuyplateBillboards/MI_bp_126" => BuyplateBillboards.MI_bp_126
            "BuyplateBillboards/MI_bp_127" => BuyplateBillboards.MI_bp_127
            "BuyplateBillboards/MI_bp_128" => BuyplateBillboards.MI_bp_128
            "BuyplateBillboards/MI_bp_129" => BuyplateBillboards.MI_bp_129
            "BuyplateBillboards/MI_bp_130" => BuyplateBillboards.MI_bp_130
            "BuyplateBillboards/MI_bp_131" => BuyplateBillboards.MI_bp_131
            "BuyplateBillboards/MI_bp_132" => BuyplateBillboards.MI_bp_132
            "BuyplateBillboards/MI_bp_133" => BuyplateBillboards.MI_bp_133
            "BuyplateBillboards/MI_bp_134" => BuyplateBillboards.MI_bp_134
            "BuyplateBillboards/MI_bp_135" => BuyplateBillboards.MI_bp_135
            "BuyplateBillboards/MI_bp_136" => BuyplateBillboards.MI_bp_136
            "BuyplateBillboards/MI_bp_137" => BuyplateBillboards.MI_bp_137
            "BuyplateBillboards/MI_bp_138" => BuyplateBillboards.MI_bp_138
            "BuyplateBillboards/MI_bp_139" => BuyplateBillboards.MI_bp_139
            "BuyplateBillboards/MI_bp_140" => BuyplateBillboards.MI_bp_140
            "BuyplateBillboards/MI_bp_141" => BuyplateBillboards.MI_bp_141
            "BuyplateBillboards/MI_bp_142" => BuyplateBillboards.MI_bp_142
            "BuyplateBillboards/MI_bp_143" => BuyplateBillboards.MI_bp_143
            "BuyplateBillboards/MI_bp_200" => BuyplateBillboards.MI_bp_200
            "BuyplateBillboards/MI_bp_201" => BuyplateBillboards.MI_bp_201
            "BuyplateBillboards/MI_bp_202" => BuyplateBillboards.MI_bp_202
            "BuyplateBillboards/MI_bp_203" => BuyplateBillboards.MI_bp_203
            "BuyplateBillboards/MI_bp_204" => BuyplateBillboards.MI_bp_204
            "BuyplateBillboards/MI_bp_205" => BuyplateBillboards.MI_bp_205
            "BuyplateBillboards/MI_bp_206" => BuyplateBillboards.MI_bp_206
            "BuyplateBillboards/MI_bp_207" => BuyplateBillboards.MI_bp_207
            "BuyplateBillboards/MI_bp_208" => BuyplateBillboards.MI_bp_208
            "BuyplateBillboards/MI_bp_209" => BuyplateBillboards.MI_bp_209
            "BuyplateBillboards/MI_bp_210" => BuyplateBillboards.MI_bp_210
            "BuyplateBillboards/MI_bp_300" => BuyplateBillboards.MI_bp_300
            "BuyplateBillboards/MI_bp_301" => BuyplateBillboards.MI_bp_301
            "BuyplateBillboards/MI_bp_302" => BuyplateBillboards.MI_bp_302
            "BuyplateBillboards/MI_bp_303" => BuyplateBillboards.MI_bp_303
            "BuyplateBillboards/MI_bp_304" => BuyplateBillboards.MI_bp_304
            "BuyplateBillboards/MI_bp_305" => BuyplateBillboards.MI_bp_305
            "BuyplateBillboards/MI_bp_306" => BuyplateBillboards.MI_bp_306
            "BuyplateBillboards/MI_bp_307" => BuyplateBillboards.MI_bp_307
            "BuyplateBillboards/MI_bp_308" => BuyplateBillboards.MI_bp_308
            "BuyplateBillboards/MI_bp_309" => BuyplateBillboards.MI_bp_309
            "BuyplateBillboards/MI_bp_310" => BuyplateBillboards.MI_bp_310
            "BuyplateBillboards/MI_bp_311" => BuyplateBillboards.MI_bp_311
            "BuyplateBillboards/MI_bp_312" => BuyplateBillboards.MI_bp_312
            "BuyplateBillboards/MI_bp_400" => BuyplateBillboards.MI_bp_400
            "BuyplateBillboards/MI_bp_401" => BuyplateBillboards.MI_bp_401
            "BuyplateBillboards/MI_bp_402" => BuyplateBillboards.MI_bp_402
            "BuyplateBillboards/MI_bp_403" => BuyplateBillboards.MI_bp_403
            "BuyplateBillboards/MI_bp_404" => BuyplateBillboards.MI_bp_404
            "BuyplateBillboards/MI_bp_405" => BuyplateBillboards.MI_bp_405
            "BuyplateBillboards/MI_bp_406" => BuyplateBillboards.MI_bp_406
            "BuyplateBillboards/MI_bp_407" => BuyplateBillboards.MI_bp_407
            "BuyplateBillboards/MI_bp_408" => BuyplateBillboards.MI_bp_408
            "BuyplateBillboards/MI_bp_409" => BuyplateBillboards.MI_bp_409
            "BuyplateBillboards/MI_bp_410" => BuyplateBillboards.MI_bp_410
            "BuyplateBillboards/MI_bp_411" => BuyplateBillboards.MI_bp_411
            "BuyplateBillboards/MI_bp_412" => BuyplateBillboards.MI_bp_412
            "BuyplateBillboards/MI_bp_413" => BuyplateBillboards.MI_bp_413
            "BuyplateBillboards/MI_bp_500" => BuyplateBillboards.MI_bp_500
            "BuyplateBillboards/MI_bp_claim" => BuyplateBillboards.MI_bp_claim
            "TemplateCommon/BillboardMat_Master" => TemplateCommon.BillboardMat_Master
            "TemplateCommon/BuyplateDiscMat_Master" => TemplateCommon.BuyplateDiscMat_Master
            "TemplateCommon/BuyplateDiscMatTransparent_Master" => TemplateCommon.BuyplateDiscMatTransparent_Master
            "TemplateCommon/BuyplateMatCostNotMet" => TemplateCommon.BuyplateMatCostNotMet
            "TemplateCommon/BuyplateMatMain" => TemplateCommon.BuyplateMatMain
            "TemplateCommon/BuyplateMatMainSub" => TemplateCommon.BuyplateMatMainSub
            "TemplateCommon/BuyplateMatOrange" => TemplateCommon.BuyplateMatOrange
            "TemplateCommon/BackendTooltips/extra_Mat" => TemplateCommon.BackendTooltips.extra_Mat
            "TemplateCommon/BackendTooltips/integration_debug_Mat" => TemplateCommon.BackendTooltips.integration_debug_Mat
            "TemplateCommon/BackendTooltips/progress_line_Mat" => TemplateCommon.BackendTooltips.progress_line_Mat
            "TemplateCommon/BackendTooltips/tycoon_Mat" => TemplateCommon.BackendTooltips.tycoon_Mat

    GetMeshAsset<override>(Name:string)<transacts><decides>:mesh=
        case(Name):
            "TemplateCommon/BuyplateDiscMesh2" => TemplateCommon.BuyplateDiscMesh2
            "TemplateCommon/BuyplateImageBackgroundMesh" => TemplateCommon.BuyplateImageBackgroundMesh
            "TemplateCommon/SM_BuyplateDevice" => TemplateCommon.SM_BuyplateDevice
            "TemplateCommon/BackendTooltips/TooltipPlane" => TemplateCommon.BackendTooltips.TooltipPlane

    GetAnimationSequenceAsset<override>(Name:string)<transacts><decides>:animation_sequence=
       None()?

    GetParticleSystemAsset<override>(Name:string)<transacts><decides>:particle_system=
        case(Name):
            "Assets/VFX/Verse/BuyplateUnlockEffect" => Assets.VFX.Verse.BuyplateUnlockEffect
