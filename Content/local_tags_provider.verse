using { /Fortnite.com/Devices }
using { /Verse.org/Simulation/Tags }

using { HavocVerse.Util }

buyplate_disc_claim_tag := class(tag):
buyplate_disc_green_tag := class(tag):
buyplate_disc_red_tag := class(tag):
buyplate_billboard_tag := class(tag):
collectible_tag := class(tag):

local_tags_provider := class(creative_device, tags_provider):
    GetTag<override>(Name:string)<transacts><decides>:tag=
        case(Name):
            "buyplate_disc_claim" => buyplate_disc_claim_tag{}
            "buyplate_disc_green" => buyplate_disc_green_tag{}
            "buyplate_disc_red" => buyplate_disc_red_tag{}
            "buyplate_billboard" => buyplate_billboard_tag{}
            "collectible" => collectible_tag{}
