
using { /Fortnite.com/Devices }
using { /Verse.org/Simulation }
using { /UnrealEngine.com/Temporary/Diagnostics }

# See https://dev.epicgames.com/documentation/en-us/uefn/create-your-own-device-in-verse for how to create a verse device.

# A Verse-authored creative device that can be placed in a level
multi_chair_manager_device := class(creative_device):

    @editable Chairs: []chair_device = array{}
    @editable Buttons: []button_device = array{}
    
    @editable Cinematic: ?cinematic_sequence_device = false

    OnBegin<override>()<suspends>:void=
        for(Button:Buttons):
            Button.InteractedWithEvent.Subscribe(OnButtonUsed)
        for(Chair:Chairs):
            Chair.ExitedEvent.Subscribe(OnPlayerExitsSeat)

    OnButtonUsed(Agent:agent):void=
        spawn:
            OnButtonUsed0(Agent)

    OnButtonUsed0(Agent:agent)<suspends>:void=
        if(Chair1:=Chairs[0]):
            if(Chair2:=Chairs[1]):
                if(not Chair1.IsOccupied[]):
                    if(not Chair2.IsOccupied[]):
                        if(MaybeCinematic:=Cinematic?):
                            MaybeCinematic.Stop()
                            Sleep(0.05)
                            MaybeCinematic.Play()
                        Chair1.Seat(Agent)
                    else:
                        Chair1.Seat(Agent)
                        for(Button:Buttons):
                            Button.Disable()
                else:
                    if(not Chair2.IsOccupied[]):
                        Chair2.Seat(Agent)
                        for(Button:Buttons):
                            Button.Disable()
                    
    OnPlayerExitsSeat(Agent:agent):void=
        for(Button:Buttons):
            Button.Enable()

    
