
using { /Fortnite.com/Devices }
using { /Verse.org/Simulation }
using { /UnrealEngine.com/Temporary/Diagnostics }

# See https://dev.epicgames.com/documentation/en-us/uefn/create-your-own-device-in-verse for how to create a verse device.

# A Verse-authored creative device that can be placed in a level
player_barrier_device := class(creative_device):

    @editable Switches: []switch_device = array{}
    @editable Barriers: []barrier_device = array{}
    @editable DamageVolumes: []damage_volume_device = array{}
    @editable GateCinematic: cinematic_sequence_device = cinematic_sequence_device{}
    @editable RollbackTrigger: trigger_device = trigger_device{}
    @editable ClaimTrigger: trigger_device = trigger_device{}
    var GateCurrentlyOpen: logic = false

    OnBegin<override>()<suspends>:void=
        for(Switch:Switches):
            Switch.TurnedOnEvent.Subscribe(DisableBarriers)
        for(Switch:Switches):
            Switch.TurnedOffEvent.Subscribe(EnableBarriers)

        RollbackTrigger.TriggeredEvent.Subscribe(OnRollback)
        ClaimTrigger.TriggeredEvent.Subscribe(OnClaim)

    EnableBarriers(Agent:agent):void=
        for(DamageVolume:DamageVolumes):
            DamageVolume.Enable()
        for(Barrier:Barriers):
            Barrier.Enable()
        if(GateCurrentlyOpen?):
            GateCinematic.PlayReverse()
            set GateCurrentlyOpen = false

    DisableBarriers(Agent:agent):void=
        for(DamageVolume:DamageVolumes):
            DamageVolume.Disable()
        for(Barrier:Barriers):
            Barrier.Disable()
        if(not GateCurrentlyOpen?):
            GateCinematic.Play()
            set GateCurrentlyOpen = true

    OnRollback(Agent:?agent):void=
        for(DamageVolume:DamageVolumes):
            DamageVolume.Disable()
        for(Barrier:Barriers):
            Barrier.Disable()
        if(GateCurrentlyOpen?):
            GateCinematic.PlayReverse()
            set GateCurrentlyOpen = false
        for(Switch:Switches):
            Switch.SetState(false)
    
    OnClaim(Agent:?agent):void=
        for(DamageVolume:DamageVolumes):
            DamageVolume.Enable()
        for(Barrier:Barriers):
            Barrier.Enable()
        if(not GateCurrentlyOpen?):
            GateCinematic.Play()
            set GateCurrentlyOpen = true
        for(Switch:Switches):
            Switch.SetState(false)
        