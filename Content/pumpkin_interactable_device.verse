
using { /Fortnite.com/Devices }
using { /Verse.org/Simulation }
using { /UnrealEngine.com/Temporary/Diagnostics }
using { HavocVerse.Combat }
using { HavocVerse.Debug }
using { HavocVerse.Devices }
using { HavocVerse.UI }
using { HavocVerse.Logging }
using { HavocVerse.Util }
# See https://dev.epicgames.com/documentation/en-us/uefn/create-your-own-device-in-verse for how to create a verse device.

# A Verse-authored creative device that can be placed in a level
pumpkin_interactable_device := class(creative_device):

    @editable Buttons: []button_device = array{}
    @editable p_ResourceManager: resource_manager_device = resource_manager_device{}
    @editable ResourceID: string = ""
    @editable ResourceAmount: int = 5
    @editable AccoladeDevice: accolades_device = accolades_device{}

    OnBegin<override>()<suspends>: void ={
        for (Button:Buttons) {
            Instance := pumpkin_button_instance{
                Device := Self
                Button := Button
            }
            Button.InteractedWithEvent.Subscribe(Instance.OnInteract)
        }
    }

pumpkin_button_instance<internal> := class<final> {
    Device: pumpkin_interactable_device
    Button: button_device

    OnInteract<internal>(Agent:agent): void ={
        Device.AccoladeDevice.Award(Agent)
        Device.p_ResourceManager.Give(Agent, Device.ResourceID, Device.ResourceAmount, ?SendFeedback := true)
        Button.Disable()
    }
}
