
using { /Fortnite.com/Devices }
using { /Verse.org/Simulation }
using { /UnrealEngine.com/Temporary/Diagnostics }
using { /UnrealEngine.com/Temporary/SpatialMath }

using { HavocVerse.Util }

ride_teacups_device := class(creative_device):

    @editable p_InputTrigger: input_trigger_device = input_trigger_device{}
    @editable p_Chair: chair_device = chair_device{}
    @editable p_TeacupCinematic: cinematic_sequence_device = cinematic_sequence_device{}
    @editable p_Acceleration : float = 0.15
    @editable p_Deceleration : float = 0.15
    @editable p_StartRate : float = 0.15
    @editable p_MaxRate : float = 20.0
    var PlayRate : float = 0.15

    var m_HoldingPlayers: [player]event() = map{}

    OnBegin<override>()<suspends>:void=
        p_Chair.SeatedEvent.Subscribe(OnEnterChair)
        p_Chair.ExitedEvent.Subscribe(OnExitChair)
        p_InputTrigger.PressedEvent.Subscribe(OnInputTriggerPressed)
        set PlayRate = p_StartRate
    
    OnEnterChair(Agent:agent):void =
        p_InputTrigger.Register(Agent)

    OnExitChair(Agent:agent):void =
        p_InputTrigger.Unregister(Agent)
        if (Player := player[Agent]) {
            RemoveRes := m_HoldingPlayers.RemoveByKey(Player)
            set m_HoldingPlayers = RemoveRes(0)
            if (AbortEvent := RemoveRes(1)?) {
                AbortEvent.Signal()
            }
        }

    OnInputTriggerPressed<private>(Agent:agent):void=
        if (Player := player[Agent], not m_HoldingPlayers[Player]) {
            spawn:
                StartHoldLoop(p_InputTrigger, Player)
        }

    StartHoldLoop<private>(InputTrigger:input_trigger_device, Player:player)<suspends>:void=
        AbortEvent := event(){}
        set m_HoldingPlayers = AddToMap(m_HoldingPlayers, Player, AbortEvent)
        p_TeacupCinematic.Play()
        race:
            AbortEvent.Await()
            loop {
                Sleep(0.05)
                SeatedAgent := p_Chair.GetSeatedAgent()
                if ((not InputTrigger.IsHeld[Player]) or SeatedAgent <> Some(Player)) {
                    if(PlayRate > 3.0):
                        DecreaseRate(PlayRate - p_Deceleration)
                    else if(PlayRate > 0.3):
                        DecreaseRate(PlayRate - p_Deceleration)
                        Sleep(0.1)
                    else:
                        p_TeacupCinematic.GoToEndAndStop()
                        set m_HoldingPlayers = m_HoldingPlayers.RemoveByKey(Player)(0)
                        break
                }
                
                if (InputTrigger.IsHeld[Player]) {
                    IncreaseRate(PlayRate + p_Acceleration)
                }
                p_TeacupCinematic.SetPlayRate(PlayRate)
            }

    IncreaseRate(NewRate: float):void =
        if(NewRate < p_MaxRate):
            set PlayRate = NewRate
        else:
            set PlayRate = p_MaxRate

    DecreaseRate(NewRate: float):void =
        if(NewRate > 0.3):
            set PlayRate = NewRate
        else:
            set PlayRate = 0.3
