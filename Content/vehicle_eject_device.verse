
using { /Fortnite.com/Devices }
using { /Verse.org/Simulation }
using { /UnrealEngine.com/Temporary/Diagnostics }
using { /Fortnite.com/Characters }
using { /Fortnite.com/Vehicles }
using { /Fortnite.com/Game }

using { HavocVerse.Debug }
using { HavocVerse.Devices }
using { HavocVerse.Logging }
using { HavocVerse.Util }

vehicle_eject_device := class(creative_device):

    @editable EjectTrigger: trigger_device = trigger_device{}

    OnBegin<override>()<suspends>:void=
        EjectTrigger.TriggeredEvent.Subscribe(EjectPlayer)

    EjectPlayer(Agent : ?agent) : void =
        if (
            Player := player[Agent?]
            FortPlayer := Player.GetFortCharacter[]
            Vehicle := FortPlayer.GetVehicle[]
        ) {
            Log("Ejecting '{Player}' from their vehicle", ?Level := log_level.Debug)
            Vehicle.Damage(damage_args{
                Instigator := Some(FortPlayer)
                Amount := 10000.0
            })
        }