using { /Fortnite.com/Devices }
using { /Verse.org/Simulation }
using { /UnrealEngine.com/Temporary/SpatialMath }
using { /Verse.org/Random }
using { /Fortnite.com/Characters }
using { /Fortnite.com/Vehicles }
using { /Fortnite.com/Game }
using { /UnrealEngine.com/Temporary/UI }

using { HavocVerse.Combat }
using { HavocVerse.Debug }
using { HavocVerse.Devices }
using { HavocVerse.UI }
using { HavocVerse.Logging }
using { HavocVerse.Util } 

volume_reward_device := class(creative_device):
    @editable p_ResourceManager: resource_manager_device = resource_manager_device{}

    @editable Volumes: []volumes = array{}

    OnBegin<override>()<suspends>:void=
        for(Volume:Volumes):
            VolumeInstance := volume_instance{
                Device := Self
                Config := Volume
            }
            VolumeInstance.Init()

volumes<internal> := class<final><concrete>:
    
    @editable GiveOnEnter: logic = false
    @editable VolumeDevice: volume_device = volume_device{}

    @editable EnableTrigger: trigger_device = trigger_device{}
    @editable DisableTrigger: trigger_device = trigger_device{}
    @editable AccoladeDevice: accolades_device = accolades_device{}

    @editable DelayBetweenRewards: float = 3.0
        
    @editable ResourceID: string = ""
    @editable ResourceAmount: int = 5
    
volume_instance<internal> := class<final> {
    Device: volume_reward_device
    Config: volumes

    var CurrentlyEnabled: logic = false
    var m_InsideVolume: [agent]event() = map{}
    var m_OnEnterCooldown: [agent]void = map{}

    Init<internal>(): void ={
        Config.VolumeDevice.AgentEntersEvent.Subscribe(OnPlayerEnteredVolume)
        Config.VolumeDevice.AgentExitsEvent.Subscribe(OnPlayerExitsVolume)
        Config.EnableTrigger.TriggeredEvent.Subscribe(OnEnableTrigger)
        Config.DisableTrigger.TriggeredEvent.Subscribe(OnDisableTrigger)
    }

    OnEnableTrigger(Agent:?agent):void=
        set CurrentlyEnabled = true

    OnDisableTrigger(Agent:?agent):void=
        set CurrentlyEnabled = false

    GrantReward<private>(Agent:agent): void ={
        Device.p_ResourceManager.Give(Agent, Config.ResourceID, Config.ResourceAmount, ?SendFeedback := true)
        Config.AccoladeDevice.Award(Agent)
    }

    OnPlayerEnteredVolume<private>(Agent:agent): void ={
        if (not CurrentlyEnabled?) {
            return
        }
        if (Config.GiveOnEnter?, not m_OnEnterCooldown[Agent]) {
            set m_OnEnterCooldown = AddToMap(m_OnEnterCooldown, Agent, void)
            GrantReward(Agent)
            spawn:
                ScheduleCooldownRemoval(Agent)
        } else if (not Config.GiveOnEnter?, not m_InsideVolume[Agent], Player := player[Agent]) {
            AbortEvent := event(){}
            set m_InsideVolume = AddToMap(m_InsideVolume, Agent, AbortEvent)
            spawn:
                StartVolumeLoop(Player, AbortEvent)
        }
    }

    ScheduleCooldownRemoval<private>(Agent:agent)<suspends>: void ={
        Sleep(Config.DelayBetweenRewards)
        set m_OnEnterCooldown = m_OnEnterCooldown.RemoveByKey(Agent)(0)
    }

    OnPlayerExitsVolume<private>(Agent:agent): void ={
        RemoveRes := m_InsideVolume.RemoveByKey(Agent)
        set m_InsideVolume = RemoveRes(0)
        if (AbortEvent := RemoveRes(1)?) {
            AbortEvent.Signal()
        }
    }

    StartVolumeLoop<private>(Player:player, AbortEvent:event())<suspends>: void ={
        race:
            AbortEvent.Await()
            loop {
                Sleep(Config.DelayBetweenRewards)
                if (CurrentlyEnabled?, Player.IsActive[]) {
                    GrantReward(Player)
                } else {
                    set m_InsideVolume = m_InsideVolume.RemoveByKey(Player)(0)
                    break
                }
            }
    }
}
