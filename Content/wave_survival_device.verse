using { /Fortnite.com/Devices }
using { /Verse.org/Simulation }
using { /UnrealEngine.com/Temporary/Diagnostics }
using { /UnrealEngine.com/Temporary/SpatialMath }
using { /Verse.org/Random }
using { /Fortnite.com/Playspaces }
using { /Fortnite.com/Characters }
using { /Fortnite.com/Vehicles }
using { /Fortnite.com/Game }
using { /UnrealEngine.com/Temporary/UI }

using { HavocVerse.Combat }
using { HavocVerse.Debug }
using { HavocVerse.Devices }
using { HavocVerse.UI }
using { HavocVerse.Logging }
using { HavocVerse.Util }

# See https://dev.epicgames.com/documentation/en-us/uefn/create-your-own-device-in-verse for how to create a verse device.

# A Verse-authored creative device that can be placed in a level
wave_survival_device := class(creative_device) {
    @editable GSIProp: creative_prop = creative_prop{}
    @editable p_ResourceManager: resource_manager_device = resource_manager_device{}

    @editable JoinButtons: []button_device = array{}
    var CanBeJoined: logic = true
    @editable JoinGameDialog: popup_dialog_device = popup_dialog_device{}
    @editable CloseMenusTrigger: trigger_device = trigger_device{}
    @editable InProgressDialog: popup_dialog_device = popup_dialog_device{}

    var m_PlayersInGame: [player]hh_player_stats = map{}
    var m_GameIsGaming: logic = false

    @editable FullArenaVolume: volume_device = volume_device{}
    @editable SpawnTeleporters: []teleporter_device = array{}
    var TeleporterIndex: int = 0

    @editable MaxPlayers: int = 4
    @editable PregameTimer: timer_device = timer_device{}
    var InPregame: logic = false
    @editable CooldownTimer: timer_device = timer_device{}
    @editable StartTimer: timer_device = timer_device{}
    var SendDelayedNoti: logic = true

    @editable Waves: []wave_manager = array{}
    var CurrentWave: int = 0
    var LastWave: int = 0

    @editable NotificationsToggleSwitch: switch_device = switch_device{}
    @editable NotificationsTrigger: trigger_device = trigger_device{}

    @editable InitialLoadCinema: cinematic_sequence_device = cinematic_sequence_device{}
    @editable UnloadCinematics: []cinematic_sequence_device = array{}

    @editable WaveCompleteTrigger: trigger_device = trigger_device{}
    @editable PerPlayerWaveCompleteTrigger: trigger_device = trigger_device{}
    @editable AllCreaturesEliminatedTrigger: trigger_device = trigger_device{}
    
    @editable LobbyMusic: audio_player_device = audio_player_device{}

    @editable WinVolume: volume_device = volume_device{}
    @editable GameWonAudio: audio_player_device = audio_player_device{}
    @editable RespawnTrigger: trigger_device = trigger_device{}
    @editable WinConditionCreatureSpawners: []creature_spawner_device = array{}
    @editable AllCreatureSpawners: []creature_spawner_device = array{}

    @editable CreatureElimBoostAmnt: int = 5
    @editable WaveCompleteBoostAmnt: int = 50
    @editable VictoryBoostAmnt: int = 500
    @editable PlayerWonTrigger: trigger_device = trigger_device{}

    @editable ItemRemover: item_remover_device = item_remover_device{}
    @editable BroomSwitch: switch_device = switch_device{}

    
    Stopwatch: stopwatch = stopwatch{}
    @editable p_StatsManager: stats_manager_device = stats_manager_device{}

    var m_UpdatePlayerHudFn: ?(player->void) = false # f verse

    ### DOMAIN
    OnBegin<override>()<suspends>:void=
        for(Button:JoinButtons):
            Button.InteractedWithEvent.Subscribe(OnJoinButtonUsed)

        NotificationsTrigger.TriggeredEvent.Subscribe(OnNotiPopupsOn)

        JoinGameDialog.RespondingButtonEvent.Subscribe(OnJoinPopUpSelected)

        CooldownTimer.SuccessEvent.Subscribe(OnCooldownTimerSuccess)
        StartTimer.SuccessEvent.Subscribe(OnStartTimerSuccess)
        PregameTimer.SuccessEvent.Subscribe(OnPregameTimerSuccess)

        FullArenaVolume.AgentEntersEvent.Subscribe(OnPlayerEnterArena)
        FullArenaVolume.AgentExitsEvent.Subscribe(OnPlayerExitsArena)

        WaveCompleteTrigger.TriggeredEvent.Subscribe(OnWaveCompleteTrigger)
        AllCreaturesEliminatedTrigger.TriggeredEvent.Subscribe(OnAllCreaturesEliminatedTrigger)

        for (Wave:Waves) {
            Wave.SetDeviceRef(Self)
            Wave.CreatureManagesSubscribe()
        }

        for(CreatureSpawner:AllCreatureSpawners):
            CreatureSpawner.EliminatedEvent.Subscribe(OnCreatureEliminated)

        set LastWave = Waves.Length

        WinVolume.AgentEntersEvent.Subscribe(OnPlayerWins)

        InitialLoadCinema.Play()

    OnNotiPopupsOn(Agent : ?agent): void =
        if(MaybeAgent:=Agent?):
            spawn:
                OnNotiPopupsOn0(MaybeAgent)

    OnNotiPopupsOn0(Agent : agent)<suspends>: void =
        CloseMenusTrigger.Trigger(Agent)
        Sleep(0.2)
        JoinGameDialog.Show(Agent)

    OnCreatureEliminated(Interaction:device_ai_interaction_result): void =
        if (Player := player[Interaction.Source?]):
            GiveJackos(Player, CreatureElimBoostAmnt)

    OnJoinButtonUsed(Agent:agent): void =
        if(CanBeJoined?):
            JoinGameDialog.Show(Agent)
        else:
            InProgressDialog.Show(Agent)

    OnJoinPopUpSelected(Agent:agent, ResponseIndex:int):void=
        if(ResponseIndex = 0):
            spawn:
                OnJoinPopUpSelected0(Agent)
                
    OnJoinPopUpSelected0(Agent:agent)<suspends> :void=
        Sleep(0.25)
        TeleportPlayerToArena(Agent)

    SchedulePopupToAll()<suspends>:void=
        Sleep(3.0)
        ShowPopupToAll()

    ShowPopupToAll():void=
        for(Player:GetPlayspace().GetPlayers()){
            ShowPopupToPlayer(Player)
        }

    ShowPopupToPlayer(Player:player):void=
        for (Player2:m_PlayersInGame.Keys()):
            if(Player=Player2):
                return
        NotificationsToggleSwitch.CheckState(Player)
        Print("Checking Notifs Switch for {Player}")

    TeleportPlayerToArena(Agent : agent):void=
        if (Teleporter := SpawnTeleporters[TeleporterIndex]):
            Teleporter.Teleport(Agent)
            set TeleporterIndex += 1
        else if (Teleporter := SpawnTeleporters[0]):
            Teleporter.Teleport(Agent)
            set TeleporterIndex = 1

    OnPlayerEnterArena(Agent : agent): void =
        GSIProp.Show()
        if (Player := player[Agent]) {
            set m_PlayersInGame = AddToMap(m_PlayersInGame, Player, hh_player_stats{
                m_StartAt := GetSimulationElapsedTime()
            })
            UpdateWavesPlayerCounts()
        }

        ItemRemover.Remove(Agent)
        BroomSwitch.TurnOff(Agent)
        Print("Items removed for {Agent}")
        
        LobbyMusic.Play(Agent)
        if (not InPregame?):
            set InPregame = true
            StartTimer.Pause()
            if(SendDelayedNoti?):
                spawn:
                    SchedulePopupToAll()
                Print("Scheduled Popup To All")
                
            PregameTimer.Reset()
            PregameTimer.Start()
        
        CurPlayerCount := GetPlayerCount()
        if (CurPlayerCount>=MaxPlayers):
            set CanBeJoined = false

    OnAllCreaturesEliminatedTrigger(Agent : ?agent): void =
        if(Wave:=Waves[CurrentWave]):
            for (Player:m_PlayersInGame.Keys()) {
                PerPlayerWaveCompleteTrigger.Trigger(Player)
            }
            RestoreHealth := Wave.GetShieldRestoreAmount()
            RestoreShield := Wave.GetShieldRestoreAmount()
            for (Player:m_PlayersInGame.Keys()) {
                Wave.PlayAllCreaturesElimnatedAudio(Player)
                if(FortCharacter := Player.GetFortCharacter[]):
                    FortCharacter.SetHealth(FortCharacter.GetHealth()+RestoreHealth)
                    FortCharacter.SetShield(FortCharacter.GetShield()+RestoreShield)
            }

    OnWaveCompleteTrigger(Agent : ?agent): void =

        if(Wave:=Waves[CurrentWave]):
            for (Player:m_PlayersInGame.Keys()) {
                GiveJackos(Player, WaveCompleteBoostAmnt)
            }

        set CurrentWave = CurrentWave + 1
        if(CurrentWave < LastWave, Wave:=Waves[CurrentWave]):
            for (Player:m_PlayersInGame.Keys()) {
                spawn: 
                    Wave.PlayStartWave(Player)
            }

            PlayerCount := GetPlayerCount()
            spawn:
                Wave.StartWave(PlayerCount)
            OnWaveStart(CurrentWave)

            if (Fn := m_UpdatePlayerHudFn?) {
                for (Player->__:m_PlayersInGame) {
                    Fn(Player)
                }
            }
        else:
            EnableWinCondition()

    EnableWinCondition():void=
        Print("Win Condition Enabled")
        for(WinConditionCreatureSpawner:WinConditionCreatureSpawners):
            WinConditionCreatureSpawner.Enable()

    OnPlayerExitsArena(Agent : agent): void =
        if (Player := player[Agent]) {
            LobbyMusic.Stop(Player)
            RemoveRes := m_PlayersInGame.RemoveByKey(Player)
            set m_PlayersInGame = RemoveRes(0)
            if (PlayerStats := RemoveRes(1)?) {
                PlayerStats.SendRecap(Player, false)
            }

            for(Wave:Waves):
                Wave.StopAudio(Agent)

            UpdateWavesPlayerCounts()

            var PlayerCount: int = GetPlayerCount()
            if(PlayerCount<=0):
                GameLost()
        }

    UpdateWavesPlayerCounts():void=
        var PlayerCount: int = GetPlayerCount()
        for(Wave:Waves):
            Wave.SetCurrentPlayerCount(PlayerCount)

    GetPlayerCount<public>()<transacts>:int=
        m_PlayersInGame.Keys().Length

    OnStartTimerSuccess(Agent : ?agent): void =
        set SendDelayedNoti = false
        ShowPopupToAll()
        spawn:
            DelayedTimerRestart()
    
    DelayedTimerRestart()<suspends>:void=
        Sleep(5.0)
        if(not InPregame?):
            StartTimer.Reset()
            StartTimer.Start()

    OnCooldownTimerSuccess(Agent : ?agent): void =
        InitialLoadCinema.Play()
        set CanBeJoined = true

    OnPregameTimerSuccess(Agent : ?agent): void =
        var PlayerCount: int = GetPlayerCount()
        set InPregame = false
        set CanBeJoined = false
        set SendDelayedNoti = true
    
        if(PlayerCount>0):
            if(not m_GameIsGaming?):
                spawn:
                    StartGame()
        else:
            spawn:
                GameCanceled()

    StartGame()<suspends>:void=
        
        GSIProp.Hide()
        for (Player->PlayerStats:m_PlayersInGame) {
            PlayerStats.Reset()
            set PlayerStats.m_StartAt = GetSimulationElapsedTime()
            Stopwatch.Start(Player)

            LobbyMusic.Stop(Player)
        }

        var PlayerCount: int = GetPlayerCount()

        set InPregame = false
        set CanBeJoined = false
        set CurrentWave = 0
        set m_GameIsGaming = true

        if(Wave:=Waves[CurrentWave]):
            for (Player:m_PlayersInGame.Keys()) {
                spawn: 
                    Wave.PlayStartWave(Player)
            }
            Wave.StartWave(PlayerCount)
            OnWaveStart(CurrentWave)

    OnWaveStart<private>(WaveIndex:int): void ={
        for (__->PlayerStats:m_PlayersInGame) {
            set PlayerStats.m_MaxWave = Max(PlayerStats.m_MaxWave, WaveIndex + 1)
        }
    }

    GameCanceled()<suspends>:void=
        Print("Game Canceled")

        set m_GameIsGaming = false

        CooldownTimer.Reset()
        StartTimer.Reset()
        PregameTimer.Reset()
        CooldownTimer.Start()
        StartTimer.Start()
        Stopwatch.Reset()

        for (Player:m_PlayersInGame.Keys()) {
                LobbyMusic.Stop(Player)
            }
        set InPregame = false
        set CanBeJoined = false

    GameLost():void=
        Print("Game Lost")

        set m_GameIsGaming = false
        set InPregame = false
        set CanBeJoined = false

        CooldownTimer.Reset()
        StartTimer.Reset()
        PregameTimer.Reset()
        CooldownTimer.Start()
        StartTimer.Start()
        Stopwatch.Reset()

        for(Wave:Waves):
            Wave.GameLost()

        for (Player->PlayerStats:m_PlayersInGame) {
            set PlayerStats.m_EndAt = Some(GetSimulationElapsedTime())
            PlayerStats.SendRecap(Player, false)
        }

        spawn:
            PlayGameOverCinematics()

    OnPlayerWins(Agent:agent):void=
        if (Player := player[Agent]) {
            PlayerWonTrigger.Trigger(Player)
            GiveJackos(Player, VictoryBoostAmnt)
            RemoveRes := m_PlayersInGame.RemoveByKey(Player)
            set m_PlayersInGame = RemoveRes(0)
            StopwatchRes := Stopwatch.Stop(Player)
            if (ElapsedTime := StopwatchRes?) {
                p_StatsManager.RegisterEventScore("hauntedhouse", Player, ElapsedTime)
            }
            if (PlayerStats := RemoveRes(1)?) {
                set PlayerStats.m_EndAt = Some(GetSimulationElapsedTime())
                PlayerStats.SendRecap(Player, true)
            }
        }

        RespawnTrigger.Trigger(Agent)
        GameWonAudio.Play(Agent)

    PlayGameOverCinematics()<suspends>:void=
        Sleep(10.0)
        for(UnloadCinematic:UnloadCinematics):
            UnloadCinematic.Play()

    GiveJackos<internal>(Player:player, Amount:int): void ={
        p_ResourceManager.Give(Player, "boost", Amount, ?SendFeedback := true)
        if (PlayerStats := m_PlayersInGame[Player]) {
            set PlayerStats.m_EarnedJackos +=  Amount
        }
    }

    ### API
    SetUpdatePlayerHudFn<public>(Fn:(player->void)): void ={
        set m_UpdatePlayerHudFn = Some(Fn)
    }

    IsPlaying<public>(Player:player)<decides><transacts>: void ={
        m_PlayersInGame[Player]
    }

    GetProgressInfo<public>()<decides><transacts>: hh_progress_info ={
        m_GameIsGaming?
        WaveProgress := (
            Wave := Waves[CurrentWave]
            ToFloat(Wave.CreaturesKilled) / ToFloat(Wave.TargetCreatureKills)
        ) or 0.0
        hh_progress_info{
            CurrentWaveIndex := CurrentWave
            TotalWaves := LastWave
            WaveProgress := WaveProgress
        }
    }
}

hh_progress_info<public> := class<final> {
    CurrentWaveIndex<public>: int
    TotalWaves<public>: int
    WaveProgress<public>: float
}

wave_creature_manager<internal> := class<final><concrete>:
    @editable CreatureSpawners: []creature_spawner_device = array{}
    @editable WaveTotal1Player: int = 1
    @editable WaveTotal2Players: int = 1
    @editable WaveTotal3Players: int = 1
    @editable WaveTotal4Players: int = 1
    @editable CutOffLevel: float = 0.5

    var WaveActive: logic = false

    var CutOffAmountSurpassed: logic = false

    var AmountSpawned: int = 0
    var WaveTotal: int = 0
    var AllCreaturesSpawned: logic = true

    GameLost():void=
        set WaveActive = false
        set CutOffAmountSurpassed = false
        set AmountSpawned = 0
        set WaveTotal = 0
        set AllCreaturesSpawned = false

    CreatureManagesSubscribe():void=
        for(CreatureSpawner:CreatureSpawners):
            CreatureSpawner.SpawnedEvent.Subscribe(OnCreatureSpawned)

    EliminateCreatures():void=
        set AllCreaturesSpawned = true
        for(CreatureSpawner:CreatureSpawners):
            CreatureSpawner.EliminateCreatures()

    OnCreatureSpawned(Agent:agent): void =
        if(WaveActive?):
            set AmountSpawned = AmountSpawned + 1

            if(AmountSpawned>=WaveTotal):
                DisableSpawners()
                set AllCreaturesSpawned = true
                set WaveTotal = AmountSpawned

    GetCutOffLevel():float=
        return CutOffLevel

    SetCutOffAmountSurpassed(CutOffSurpassed:logic):void=
        set CutOffAmountSurpassed = CutOffSurpassed

    GetCutOffAmountSurpassed():logic=
        return CutOffAmountSurpassed

    GetWaveActive():logic=
        return WaveActive

    SetWaveActive(IsWaveActive:logic):void=
        set WaveActive = IsWaveActive

    GetWaveTotal():int=
        return WaveTotal

    GetAmountSpawned():int=
        return AmountSpawned

    EnableSpawners():void=
        for(CreatureSpawner:CreatureSpawners):
            CreatureSpawner.Enable()

    DisableSpawners():void=
        for(CreatureSpawner:CreatureSpawners):
            CreatureSpawner.Disable()

    SpawnCreatures():void=
        if(not AllCreaturesSpawned?):
            if(CutOffAmountSurpassed?):
                EnableSpawners()

    InitiateCreatures(PlayerCount:int):void=
        set AmountSpawned = 0
        set WaveTotal = 0
        set AllCreaturesSpawned = false
        set CutOffAmountSurpassed = false

        if(CutOffLevel>0.0):
            set CutOffAmountSurpassed = false
        if(PlayerCount>3):
            set WaveTotal = WaveTotal4Players
        else if(PlayerCount>2):
            set WaveTotal = WaveTotal3Players
        else if(PlayerCount>1):
            set WaveTotal = WaveTotal2Players
        else:
            set WaveTotal = WaveTotal1Player

        if(CutOffLevel<=0.0):
            set CutOffAmountSurpassed = true
     
wave_manager<internal> := class<final><concrete>:
    var IsWaveActive: logic = false
    @editable IsLastWave: logic = false
    @editable HealthRestoreAmount: float = 0.0
    @editable ShieldRestoreAmount: float = 0.0

    @editable CreatureManagers: []wave_creature_manager = array{}
    @editable PreWaveDelay: float = 2.0
    @editable PostWaveDelay: float = 3.0
    var SpawnedCreatures: int = 0
    var CreaturesKilled: int = 0
    var TargetCreatureKills: int = 0
    
    @editable WaveCompleteTrigger: trigger_device = trigger_device{}
    @editable AllCreaturesEliminatedTrigger: trigger_device = trigger_device{}

    @editable WaveStartAudio: ?audio_player_device = false
    @editable TrackStartDelay: float = 5.0
    @editable WaveTrack: ?audio_player_device = false
    @editable WaveCreaturesEliminatedAudio: ?audio_player_device = false

    @editable NewArenaOpenCinematic: ?cinematic_sequence_device = false
    @editable PreviousArenaClosedCinematic: ?cinematic_sequence_device = false
    @editable NextArenaVolumes: []volume_device = array{}

    var FailSafeCount: int = 0
    var CurrentCreaturesInZone: int = 0
    var FailSafeFix: logic = false
    @editable CreatureCheckVolume: volume_device = volume_device{}

    var PlayersInNextArena: int = 0

    var m_DeviceRef: ?wave_survival_device = false
    var CurrentPlayerCount: int = 0

    SetDeviceRef<internal>(Device:wave_survival_device): void ={
        set m_DeviceRef = Some(Device)
    }

    StartWave(PlayerCount:int)<suspends>:void=
        if(not IsWaveActive?):
            set SpawnedCreatures = 0
            set CreaturesKilled = 0
            set TargetCreatureKills = 0
            set PlayersInNextArena = 0
            set FailSafeCount = 0 
            set FailSafeFix = false

            for(CreatureManager:CreatureManagers):
                CreatureManager.InitiateCreatures(PlayerCount)

                CreatureManager.SetWaveActive(true)

                NewCreatueKills:=CreatureManager.GetWaveTotal()
                set TargetCreatureKills = TargetCreatureKills + NewCreatueKills
            
            Sleep(PreWaveDelay)    
            set IsWaveActive = true

            spawn:
                FailSafeLoop()

            SpawnCreatures()
            

    FailSafeLoop()<suspends>:void=
        loop:
            Sleep(5.0)
            if(not IsWaveActive?):
                break
            else if(FailSafeFix?):
                break  
            else:
                Print("Fail Safe Check current creatures = {CurrentCreaturesInZone}")
                if(CurrentCreaturesInZone<=0):
                    set FailSafeCount = FailSafeCount + 1
                    if(FailSafeCount>=2):
                        Print("Fail Safe!!!!!")
                        WaveCompleted()

    GetHealthRestoreAmount():float=
        return HealthRestoreAmount
     
    GetShieldRestoreAmount():float=
        return ShieldRestoreAmount

    PlayAllCreaturesElimnatedAudio(Agent:agent):void=
        if(MaybeWaveTrack:=WaveTrack?):
            if(not IsLastWave?):
                MaybeWaveTrack.Stop(Agent)
        if(MaybeWaveCreaturesEliminatedAudio:=WaveCreaturesEliminatedAudio?):
            MaybeWaveCreaturesEliminatedAudio.Play(Agent)

    PlayStartWave(Agent:agent)<suspends>:void=
        if(MaybeStartWaveAudio:=WaveStartAudio?):
            MaybeStartWaveAudio.Play(Agent)

        Sleep(TrackStartDelay)

        if(MaybeWaveTrack:=WaveTrack?):
            MaybeWaveTrack.Play(Agent)

    StopAudio(Agent:agent):void=
        if(MaybeWaveTrack:=WaveTrack?):
            MaybeWaveTrack.Stop(Agent)

    SetCurrentPlayerCount(PlayerCount:int):void=
        set CurrentPlayerCount = PlayerCount

    SpawnCreatures():void=
        for(CreatureManager:CreatureManagers):
            CreatureManager.SpawnCreatures()

    CreatureManagesSubscribe():void=
        for(CreatureManager:CreatureManagers):
            CreatureManager.CreatureManagesSubscribe()
            for(CreaturePlacer:CreatureManager.CreatureSpawners):
                CreaturePlacer.SpawnedEvent.Subscribe(OnCreatureSpawned)
                CreaturePlacer.EliminatedEvent.Subscribe(OnCreatureEliminated)

        for(Volume:NextArenaVolumes):
            Volume.AgentEntersEvent.Subscribe(EnterNextArena)
            Volume.AgentExitsEvent.Subscribe(ExitNextArena)

        CreatureCheckVolume.AgentEntersEvent.Subscribe(IncreaseCreatureSpawnCount)
        CreatureCheckVolume.AgentExitsEvent.Subscribe(DecreaseCreatureSpawnCount)

    IncreaseCreatureSpawnCount(Agent:agent):void=
        if(IsWaveActive?):
            set CurrentCreaturesInZone = CurrentCreaturesInZone + 1
            Print("TestyTesttest")
        
    DecreaseCreatureSpawnCount(Agent:agent):void=
        if(IsWaveActive?):
            set CurrentCreaturesInZone = CurrentCreaturesInZone - 1

    EnterNextArena(Agent:agent):void=
        spawn:
            EnterNextArena0(Agent)

    EnterNextArena0(Agent:agent)<suspends>:void=
        if(IsWaveActive?):
            Sleep(0.5)
            set PlayersInNextArena = PlayersInNextArena + 1
            if(PlayersInNextArena>=CurrentPlayerCount):
                if(MaybePreviousArenaClosedCinematic:=PreviousArenaClosedCinematic?):
                    MaybePreviousArenaClosedCinematic.Play()
                    spawn:
                        WaveCompleted()

    ExitNextArena(Agent:agent):void=
        if(IsWaveActive?):
            set PlayersInNextArena = PlayersInNextArena - 1

    OnCreatureSpawned(Agent:agent): void =
        var NewWaveTotal: int = 0

        if(IsWaveActive?):
            for(CreatureManager:CreatureManagers):

                NewCreatueKills:=CreatureManager.GetWaveTotal()
                set NewWaveTotal = NewWaveTotal + NewCreatueKills

            if(NewWaveTotal>TargetCreatureKills):
                set TargetCreatureKills = NewWaveTotal

    OnCreatureEliminated(Res:device_ai_interaction_result): void =
        if(IsWaveActive?):
            set CreaturesKilled = CreaturesKilled + 1

            UpdateCutoffAmounts()

            if(CreaturesKilled>=TargetCreatureKills):
                
                AllCreaturesEliminatedTrigger.Trigger()

                if(MaybeNextArenaOpenCinematic:=NewArenaOpenCinematic?):
                    MaybeNextArenaOpenCinematic.Play()
                if(NextArenaVolumes.Length>0):
                    set FailSafeFix = true
                else:
                    Print("This run")
                    spawn:
                        WaveCompleted()
        
            if (
                Device := m_DeviceRef?
                Killer := player[Res.Source?]
                PlayerStats := Device.m_PlayersInGame[Killer]
            ) {
                set PlayerStats.m_ZombieKills += 1
            }

            if (Device := m_DeviceRef?, Fn := Device.m_UpdatePlayerHudFn?) {
                for (Player->__:Device.m_PlayersInGame) {
                    Fn(Player)
                }
            }

    UpdateCutoffAmounts():void =
        for(CreatureManager:CreatureManagers):
            CutOffSurpassed:=CreatureManager.GetCutOffAmountSurpassed()
            if(not CutOffSurpassed?):
                CutOffLevel:=CreatureManager.GetCutOffLevel()
                CutOffAmount:=CutOffLevel*TargetCreatureKills
                if(CreaturesKilled*1.0>CutOffAmount):
                    CreatureManager.SetCutOffAmountSurpassed(true)
                    CreatureManager.SpawnCreatures()

    GameLost():void=
        set IsWaveActive = false
        set SpawnedCreatures = 0
        set CreaturesKilled = 0
        set TargetCreatureKills = 0
        set PlayersInNextArena = 0
        set CurrentCreaturesInZone = 0
        set FailSafeCount = 0 
        set FailSafeFix = false
        for(CreatureManager:CreatureManagers):
            CreatureManager.SetWaveActive(false)
            CreatureManager.SetCutOffAmountSurpassed(false)
            CreatureManager.EliminateCreatures()
            CreatureManager.DisableSpawners()
            CreatureManager.GameLost()

    WaveCompleted()<suspends>:void=
        if(IsWaveActive?):
            set IsWaveActive = false
            if(NextArenaVolumes.Length>0):
            else:
                Sleep(PostWaveDelay)

            for(CreatureManager:CreatureManagers):
                CreatureManager.SetWaveActive(false)

            WaveCompleteTrigger.Trigger()

hh_player_stats<internal> := class<final> {
    var m_StartAt: float
    var m_EndAt: ?float = false
    var m_MaxWave: int = 0
    var m_ZombieKills: int = 0
    var m_EarnedJackos: int = 0 # TODO : when are we giving jackos?

    var m_RecapSent: logic = false

    Reset<internal>(): void ={
        set m_EndAt = None()
        set m_MaxWave = 0
        set m_ZombieKills = 0
        set m_EarnedJackos = 0
        set m_RecapSent = false
    }

    CreateRecapMenu<private>(Won:logic): ui_menu_component ={
        Duration := (m_EndAt? or GetSimulationElapsedTime()) - m_StartAt
        CustomHUD.hh_recap_menu_component{
            Won := Won
            MaxWave := m_MaxWave
            AliveDuration := Duration
            ZombieKills := m_ZombieKills
            EarnedJackos := m_EarnedJackos
        }
    }

    SendRecap<internal>(Player:player, Won:logic): void ={
        if (m_RecapSent? or (m_MaxWave < 1 and m_ZombieKills < 1 and m_EarnedJackos < 1)) {
            return
        }
        set m_RecapSent = true
        Menu := CreateRecapMenu(Won)
        spawn:
            ScheduleSendRecap(Player, Menu)
    }

    ScheduleSendRecap<private>(Player:player, Menu:ui_menu_component)<suspends>: void ={
        # the player might be dead so send it whenever they're alive
        loop {
            if (not Player.IsActive[]) {
                return
            }
            if (FortCharacter := Player.GetFortCharacter[], FortCharacter.IsActive[]) {
                Sleep(0.3)
                if (FortCharacter.IsActive[]) {
                    OpenSimpleMenu("hh_recap", Player, Menu)
                }
                return
            }
            Sleep(0.0)
        }
    }
}     
