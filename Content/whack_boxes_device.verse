using { /Fortnite.com/Devices }
using { /Verse.org/Simulation }
using { /UnrealEngine.com/Temporary/Diagnostics }
using { /UnrealEngine.com/Temporary/SpatialMath }

using { HavocVerse.Debug }
using { HavocVerse.Devices }
using { HavocVerse.Resource }
using { HavocVerse.Storage }
using { HavocVerse.Logging }
using { HavocVerse.Util }

box_hit_handler<internal> := class() {
    Device : whack_boxes_device
    Parameter1 : int

    BoxHit(Agent : agent) : void = {
        Device.OnBoxHit(Agent,Parameter1)
    }

    EnableBox(Agent : ?agent) : void = {
        Device.OnBoxEnabled(Parameter1)
    }
    
    DisableBox(Agent : ?agent) : void = {
        Device.OnBoxDisabled(Parameter1)
    }
}

whack_boxes_device := class(creative_device, aware_device):

    @editable p_WhackBoxes: []whack_box = array{}

    DeviceMeta<override>()<transacts>:aware_device_meta=
        aware_device_meta{
            DeviceName := "whack_boxes_device"
        }  
        
    OnBegin<override>()<suspends>:void=
        var Index: int = 0
        for(Box:p_WhackBoxes):
            Box.PropMani.DamagedEvent.Subscribe(box_hit_handler{Device:=Self, Parameter1:=Index}.BoxHit)
            Box.EnableTrigger.TriggeredEvent.Subscribe(box_hit_handler{Device:=Self, Parameter1:=Index}.EnableBox)
            Box.DisableTrigger.TriggeredEvent.Subscribe(box_hit_handler{Device:=Self, Parameter1:=Index}.DisableBox)
            Box.DisableBox()
            set Index = Index + 1

    OnBoxHit(Agent:agent,Parameter1:int) : void =
        if(Box:=p_WhackBoxes[Parameter1]):
            if (Box.IsOnDelay[]) {
                return
            }

            Box.BoxHit(Agent)

    OnBoxEnabled(Parameter1:int) : void =
        if(Box:=p_WhackBoxes[Parameter1]):
            Box.EnableBox()

    OnBoxDisabled(Parameter1:int) : void =
        if(Box:=p_WhackBoxes[Parameter1]):
            Box.DisableBox()

whack_box<internal> := class<final><concrete>:
    @editable PropMani : prop_manipulator_device = prop_manipulator_device{}
    @editable RewardTrigger: trigger_device = trigger_device{}
    @editable EnableTrigger: trigger_device = trigger_device{}
    @editable DisableTrigger: trigger_device = trigger_device{}
    @editable p_OnCollectPropEffects: []whack_box_effect_prop_config = array{}
    @editable p_AudioPlayer: audio_player_device = audio_player_device{}
    var HitDelay: float = 0.5
    var OnDelay: logic = false

    PlayPropEffect<internal>(Index:int, Config:whack_box_effect_prop_config)<suspends>:void=
        if (Config.Marker.IsValid[]) {
            EffectTransform := Config.Marker.GetTransform().WithScale(MakeUniformVec3(1.0))
            if (EffectProp := SpawnProp(Config.PropAsset, EffectTransform)(0)?) {
                Sleep(Config.EffectDuration)
                EffectProp.Dispose()
            }
        } else {
            Print("Failed")
        }

    IsOnDelay()<decides><transacts>:void=
        OnDelay?

    BoxHit(Agent:agent):void=
        RewardTrigger.Trigger(Agent)
        for (Index->PropEffectConfig:p_OnCollectPropEffects) {
            spawn:
                PlayPropEffect(Index, PropEffectConfig)
        }
        p_AudioPlayer.Play(Agent)
        set OnDelay = true
        spawn:
            ApplyCooldown()

    EnableBox():void=
        PropMani.ShowProps()

    DisableBox():void=
        PropMani.HideProps()

    ApplyCooldown()<suspends>:void=
        set OnDelay = true
        Sleep(HitDelay)
        set OnDelay = false

whack_box_effect_prop_config := class<final><concrete>:
    @editable Marker: creative_prop = creative_prop{}
    @editable PropAsset: creative_prop_asset = DefaultCreativePropAsset
    @editable EffectDuration: float = 0.5
        