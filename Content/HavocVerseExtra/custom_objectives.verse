using { /Fortnite.com/Devices }
using { /Verse.org/Simulation }
using { /Verse.org/Colors }
using { /UnrealEngine.com/Temporary/Diagnostics }
using { /UnrealEngine.com/Temporary/SpatialMath }
using { /UnrealEngine.com/Temporary/UI }

using { HavocVerse.Debug }
using { HavocVerse.Devices }
using { HavocVerse.Storage }
using { HavocVerse.Persistence }
using { HavocVerse.UI }
using { HavocVerse.Logging }
using { HavocVerse.Util }
using { HavocVerse.Util.Task }

custom_objective<public> := class:
    ObjectiveId<public>: string
    Image<public>: string
    AwardId<public>: ?string
    Goal<public>: ?int
    Priority<public>: int = 0

## Objective Handle {{{
custom_objective_result<public> := enum:
    Completed
    Cancelled

custom_objective_handle<public> := class<final><unique>:
    Player<public>: player
    Objective<public>: custom_objective
    Controller<public>: custom_objectives_device
    AutoCompletes<public>: logic

    var m_Completed<private>: ?custom_objective_result = false
    var m_Value<internal>: int = 0

    CompletedEvent<private>: event(custom_objective_result) = event(custom_objective_result){}

    AwaitComplete<public>()<suspends>:custom_objective_result=
        CompletedEvent.Await()

    GetResult<public>()<transacts><decides>:custom_objective_result=
        m_Completed?

    # Completes the objective.
    Complete<public>(?Result:custom_objective_result = custom_objective_result.Completed):void=
        if (m_Completed?) {
            return
        }
        set m_Completed = Some(Result)
        Controller.HandleObjectiveCompleted(Player, Objective, Result)
        Controller.UpdatePlayer(Player)
        LogMessage := if (Result = custom_objective_result.Completed) {
            if (Goal := Objective.Goal?) {
                Controller.MsgObjectiveCompletedGoal(Player, Objective.ObjectiveId, m_Value, Goal)
            } else {
                Controller.MsgObjectiveCompleted(Player, Objective.ObjectiveId)
            }
        } else {
            Controller.MsgObjectiveCancelled(Player, Objective.ObjectiveId)
        }
        Log(LogMessage, ?Level := log_level.Debug)
        CompletedEvent.Signal(Result)
    
    # Cancels the objective.
    Cancel<public>():void=
        Complete(?Result := custom_objective_result.Cancelled)

    # Sets the value of the objective.
    SetValue<public>(Value:int):void=
        if ((m_Completed?) or (m_Value = Value)) {
            return
        }
        if (Goal := Objective.Goal?) {
            Delta := Value - m_Value
            set m_Value = Value
            Log(Controller.MsgObjectiveValueUpdated(Player, Objective.ObjectiveId, Value, Delta), ?Level := log_level.Verbose)
            if (AutoCompletes?, Value >= Goal) {
                Complete()
            } else {
                Controller.UpdatePlayer(Player)
            }
        } else if (AutoCompletes?, Value > 0) {
            Complete()
        }

    # Adds to the value of the objective.
    Add<public>(Delta:int):void=
        SetValue(m_Value + Delta)
## }}}

custom_objective_config := class<concrete>:
    @editable ObjectiveId: string = ""
    @editable ImageOverride: string = ""
    @editable AwardId: string = ""
    @editable Goal: int = -1
    @editable Priority: int = 0

TooltipBeginAfter<internal><localizes>:message = "Begins the objective line after an objective completed, deferred from game start"
TooltipLineObjectives<internal><localizes>:message = "Objectives to enable one after the other"

custom_objective_line_config := class<concrete>:
    @editable{ToolTip := TooltipBeginAfter}
    BeginAfter: ?string = false
    @editable{ToolTip := TooltipLineObjectives}
    Objectives: []string = array{}

custom_objectives_player_state := class:
    PlayerQuitEvent: event() = event(){}

    var m_Active: [string]custom_objective_handle = map{}
    var m_Completed: [string]void = map{}

TooltipObjectives<internal><localizes>:message = "All known objectives"
TooltipObjectiveLines<internal><localizes>:message = "Automatic objective lines. Use in combination with AwardIds"
TooltipObjectiveLinesStartAfterPersistenceInit<internal><localizes>:message = "Begins objective lines after a persistence recorder has initialized a player.\nIf not set, will begin when the player joins"

custom_objectives_device<public> := class(creative_device, aware_device, aware_debug_device, persistence_recordable):
    @editable p_DevId: string = "N/A"

    @editable{ToolTip := TooltipObjectives}
    p_Objectives: []custom_objective_config = array{}
    @editable{ToolTip := TooltipObjectiveLines}
    p_ObjectiveLines: []custom_objective_line_config = array{}
    @editable{ToolTip := TooltipObjectiveLinesStartAfterPersistenceInit}
    p_ObjectivesLinesStartAfterPersistenceInit: ?persistence_recorder_device = false

    @editable p_DefaultImagePath: string = "ObjectivesUI/objective_%Id%"
    @editable p_ImagesScale: float = 1.0
    @editable p_StackGap: float = 30.0

    @editable p_HUDAccessControl: ?access_control_device = false
    @editable p_HUDManagerConfig: hud_manager_base_config = hud_manager_base_config{
        RefreshRate := -1.0
    }
    @editable p_HUDPlacement: ui_canvas_placement_config = ui_canvas_placement_config{}

    var m_Objectives: [string]custom_objective = map{}
    var m_Players: [player]custom_objectives_player_state = map{}

    var m_HudManager: ?custom_objectives_hud_manager = false

    var m_PersistenceControl: persistence_recordable_control = persistence_recordable_control{}

    DeviceMeta<override>()<transacts>:aware_device_meta=
        aware_device_meta{
            DeviceName := "custom_objectives"
            DevId := p_DevId
        }

    DebugStats<override>()<transacts>:aware_debug_stats=
        aware_debug_stats{
            Length := map{
                "objectives" => m_Objectives.Length
                "players" => m_Players.Length
            }
        }

    OnBegin<override>()<suspends>:void=
        for (
            ObjectiveConfig:p_Objectives
            Objective := custom_objective{
                ObjectiveId := ObjectiveConfig.ObjectiveId,
                Image := if (ObjectiveConfig.ImageOverride.Length > 0) {
                    ObjectiveConfig.ImageOverride
                } else {
                    p_DefaultImagePath.Replace("%Id%", ObjectiveConfig.ObjectiveId)
                },
                AwardId := if (ObjectiveConfig.AwardId.Length > 0) {
                    Some(ObjectiveConfig.AwardId)
                } else {
                    None()
                },
                Goal := if (ObjectiveConfig.Goal > 0) {
                    Some(ObjectiveConfig.Goal)
                } else {
                    None()
                },
                Priority := ObjectiveConfig.Priority
            }
            set m_Objectives[ObjectiveConfig.ObjectiveId] = Objective
        ) {}

        AwareInitDevice(Self, OnInit)
    
    OnInit<private>()<suspends>:void=
        GetPlayspace().PlayerRemovedEvent().Subscribe(OnPlayerRemoved)

        if (PersistenceRecorder := p_ObjectivesLinesStartAfterPersistenceInit?) {
            SubscribeToEvent(PersistenceRecorder.PlayerSaveSlotSelectedEvent, OnPlayerSaveSlotSelected)
        } else {
            GetPlayspace().PlayerAddedEvent().Subscribe(OnPlayerAdded)
        }

        HudManager := custom_objectives_hud_manager{
            AccessControl := p_HUDAccessControl
            RefreshRate := p_HUDManagerConfig.RefreshRate
            ZOrder := p_HUDManagerConfig.ZOrder
            Device := Self
        }
        set m_HudManager = Some(HudManager)
        spawn:
            HudManager.Run(GetPlayspace())
        
    OnPlayerRemoved<private>(Player:player):void=
        RemoveRes := m_Players.RemoveByKey(Player)
        set m_Players = RemoveRes(0)
        if (PlayerState := RemoveRes(1)?) {
            PlayerState.PlayerQuitEvent.Signal()
        }

    OnPlayerAdded<private>(Player:player):void=
        if (not (p_ObjectivesLinesStartAfterPersistenceInit?)) {
            BeginObjectiveLines(Player)
        }

    OnPlayerSaveSlotSelected<private>(Event:player_save_slot_selected_event):void=
        if (p_ObjectivesLinesStartAfterPersistenceInit?) {
            BeginObjectiveLines(Event.Player)
        }

    BeginObjectiveLines<private>(Player:player):void=
        for (ObjectiveLine:p_ObjectiveLines, not (ObjectiveLine.BeginAfter?)) {
            spawn:
                RunObjectiveLine(ObjectiveLine, Player)
        }

    RunObjectiveLine<private>(ObjectiveLine:custom_objective_line_config, Player:player)<suspends>:void=
        PlayerState := GetPlayerState(Player)
        race:
            PlayerState.PlayerQuitEvent.Await()
            for (ObjectiveId:ObjectiveLine.Objectives) {
                AwaitObjective(Player, ObjectiveId)
            }

    HandleObjectiveCompleted<internal>(Player:player, Objective:custom_objective, Result:custom_objective_result):void=
        for (
            ObjectiveLine:p_ObjectiveLines
            BeginAfter := ObjectiveLine.BeginAfter?
            BeginAfter = Objective.ObjectiveId
        ) {
            spawn:
                RunObjectiveLine(ObjectiveLine, Player)
        }
    
    GetPlayerState<private>(Player:player):custom_objectives_player_state=
        if (PlayerState := m_Players[Player]) {
            PlayerState
        } else {
            PlayerState := custom_objectives_player_state{}
            if (set m_Players[Player] = PlayerState) {}
            PlayerState
        }

    # Returns the active objectives for the player.
    GetActiveObjectives<public>(Player:player)<transacts>:[]custom_objective_handle=
        if (PlayerState := m_Players[Player]) {
            PlayerState.m_Active.Values()
        } else {
            array{}
        }

    # Awards 1 (usually Complete) to the player's objectives matching the award id.
    Award<public>(Player:player, AwardId:string):void=
        Award(Player, AwardId, 1)

    # Awards a value to the player's objectives matching the award id.
    Award<public>(Player:player, AwardId:string, Value:int):void=
        PlayerState := GetPlayerState(Player)
        for (ActiveHandle:PlayerState.m_Active, ActiveHandle.Objective.AwardId? = AwardId) {
            ActiveHandle.Add(Value)
        }

    # Starts an objective for a player.
    # Returns None() if the objective is already completed, not found, or otherwise couldn't be started.
    StartObjective<public>(Player:player, ObjectiveId:string):?custom_objective_handle=
        if (not Player.IsActive[]) {
            LogPlayerOffline("StartObjective({ObjectiveId})", Player)
            return None()
        }

        PlayerState := GetPlayerState(Player)
        if (Objective := m_Objectives[ObjectiveId]) {
            if (PlayerState.m_Completed[ObjectiveId] or PlayerState.m_Active[ObjectiveId]) {
                return None()
            }
            ActiveHandle := custom_objective_handle{
                Player := Player,
                Objective := Objective,
                Controller := Self,
                AutoCompletes := true
            }
            if:
                set PlayerState.m_Active[ObjectiveId] = ActiveHandle
            if (HudManager := m_HudManager?) {
                HudManager.Refresh(Player)
            }
            Log(MsgObjectiveStarted(Player, ObjectiveId), ?Level := log_level.Debug)
            Some(ActiveHandle)
        } else {
            Log("Attempted to start objective \"{ObjectiveId}\" but it isn't registered on {Self}", ?Level := log_level.Error)
            None()
        }

    # Starts an objective for a player and awaits its completion.
    AwaitObjective<public>(Player:player, ObjectiveId:string)<suspends>:custom_objective_result=
        Handle := StartObjective(Player, ObjectiveId)
        if (l_Handle := Handle?) {
            l_Handle.AwaitComplete()
        } else {
            custom_objective_result.Cancelled
        }

    MsgObjectiveStarted<private><localizes>(Agent:agent, ObjectiveId:string):message=
        "Started objective \"{ObjectiveId}\" for '{Agent}'"
    MsgObjectiveCompleted<internal><localizes>(Agent:agent, ObjectiveId:string):message=
        "Completed objective \"{ObjectiveId}\" for '{Agent}' (bool)"
    MsgObjectiveCompletedGoal<internal><localizes>(Agent:agent, ObjectiveId:string, Value:int, Goal:int):message=
        "Completed objective \"{ObjectiveId}\" for '{Agent}' ({Value}/{Goal})"
    MsgObjectiveCancelled<internal><localizes>(Agent:agent, ObjectiveId:string):message=
        "Cancelled objective \"{ObjectiveId}\" for '{Agent}'"
    MsgObjectiveValueUpdated<internal><localizes>(Agent:agent, ObjectiveId:string, NewValue:int, Delta:int):message=
        "'{Agent}' progress on objective \"{ObjectiveId}\" is now {NewValue} (+{Delta})"

    UpdatePlayer<internal>(Player:player, ?Save:logic = true):void=
        PlayerState := GetPlayerState(Player)
        for (
            ObjectiveId->ActiveHandle:PlayerState.m_Active
            Result := ActiveHandle.GetResult[]
            (Result = custom_objective_result.Cancelled) or (set PlayerState.m_Completed[ActiveHandle.Objective.ObjectiveId] = void)
        ) {
            set PlayerState.m_Active = PlayerState.m_Active.RemoveByKey(ObjectiveId)(0)
        }
        if (Save?) {
            m_PersistenceControl.Save(Player)
        }
        if (HudManager := m_HudManager?) {
            HudManager.Refresh(Player)
        }

## Persistence {{{
    GetPersistenceControl<override>()<transacts><decides>:persistence_recordable_control=
        m_PersistenceControl

    PersistenceReserve<override>(StorageMem:storage_memory)<transacts><decides>:persistence_recordable_state=
        var ObjectiveChunks: [string]storage_memory_chunk = map{}
        for (Objective:m_Objectives) {
            var WantedBits: int = 2
            if (Goal := Objective.Goal?) {
                set WantedBits += BitsLength(Goal)
            }
            MemChunk := StorageMem.Reserve[WantedBits, "objective_{Objective.ObjectiveId}"]
            set ObjectiveChunks = AddToMap(ObjectiveChunks, Objective.ObjectiveId, MemChunk)
        }
        custom_objectives_persistence_state{
            ObjectiveChunks := ObjectiveChunks
        }

    PersistenceSave<override>(i_State:persistence_recordable_state, Agent:agent, SaveCause:persistence_save_cause):void=
        if (SaveCause = persistence_save_cause.AutoSave) {
            return
        }
        if (State := custom_objectives_persistence_state[i_State], PlayerState := m_Players[Agent]) {
            for (ObjectiveId->MemChunk:State.ObjectiveChunks, Objective := m_Objectives[ObjectiveId]) {
                IsCompleted := if (PlayerState.m_Completed[ObjectiveId]). true else false
                Active := option{PlayerState.m_Active[ObjectiveId]}
                IsActive := if (Active?). true else false
                MemChunk.Write(Agent, array{IsCompleted, IsActive})
                if (Goal := Objective.Goal?, ActiveHandle := Active?) {
                    MemChunk.WriteInt(Agent, Clamp(ActiveHandle.m_Value, 0, Goal), ?Offset := 2)
                }
            }
        }

    PersistenceLoad<override>(i_State:persistence_recordable_state, Agent:agent, Query:persistence_load_query):void=
        if (State := custom_objectives_persistence_state[i_State], Player := player[Agent]) {
            PlayerState := GetPlayerState(Player)
            for (ObjectiveId->MemChunk:State.ObjectiveChunks, Objective := m_Objectives[ObjectiveId]) {
                Data := MemChunk.Read(Agent, ?Limit := 2)
                IsCompleted := option{Data[0]}.UnwrapOr(false)
                IsActive := option{Data[1]}.UnwrapOr(true)

                if:
                    IsCompleted?
                    set PlayerState.m_Completed[ObjectiveId] = void
                else if:
                    Query.Overwrite?
                    set PlayerState.m_Completed = PlayerState.m_Completed.RemoveByKey(ObjectiveId)(0)

                if (IsActive?) {
                    ActiveHandle := custom_objective_handle{
                        Player := Player,
                        Objective := Objective,
                        Controller := Self,
                        AutoCompletes := true
                    }
                    if (Goal := Objective.Goal?) {
                        ActiveHandle.SetValue(MemChunk.ReadInt(Agent, ?Offset := 2))
                    }
                    if:
                        set PlayerState.m_Active[ObjectiveId] = ActiveHandle
                } else if (Query.Overwrite?, ActiveHandle := PlayerState.m_Active[ObjectiveId]) {
                    set PlayerState.m_Active = PlayerState.m_Active.RemoveByKey(ObjectiveId)(0)
                    ActiveHandle.Cancel()
                }
            }
            UpdatePlayer(Player, ?Save := false)
        }

custom_objectives_persistence_state := class(persistence_recordable_state):
    ObjectiveChunks: [string]storage_memory_chunk

    Reset<override>(Agent:agent):void=
        for (__->MemChunk:ObjectiveChunks) {
            MemChunk.Reset(Agent)
        }
## }}}

## HUD {{{
custom_objectives_hud_manager := class(hud_manager_base):
    Device: custom_objectives_device

    CreateHUD<override>(Player:player):?hud_manager_player=
        Component := custom_objective_hud_component:
            Device := Device
            Player := Player
        Some(MakeUIFrameworkHUDPlayer(Player, Component, Device.p_HUDPlacement))

custom_objective_hud_component := class(ui_component):
    Device: custom_objectives_device
    Player: player

    var m_LastActive: []custom_objective_handle = array{}

    TryUpdate<override>():logic=
        ActiveObjectives := Device.GetActiveObjectives(Player)
        if (ActiveObjectives = m_LastActive) {
            true
        } else {
            false
        }

    Create<override>():widget=
        Active := Device.GetActiveObjectives(Player)
        set m_LastActive = Active
        OrderedActive := Active.QuickSortDesc(CmpObjectiveHandle)
        StackBoxSlots := for (Index->Slot:OrderedActive) {
            stack_box_slot:
                Widget := AddChild(custom_objective_ui_objective_component{
                    ImagesScale := Device.p_ImagesScale
                    ActiveHandle := Slot
                })
                HorizontalAlignment := horizontal_alignment.Left
                VerticalAlignment := vertical_alignment.Top
                Padding := margin{
                    Top := if (Index > 0) {
                        Device.p_StackGap
                    } else {
                        0.0
                    }
                }
        }
        stack_box:
            Orientation := orientation.Vertical
            Slots := StackBoxSlots

custom_objective_ui_objective_component_progress_bar := class<final>:
    FillStart: widget
    FillStartWidth: float
    FillBlock: color_block
    Width: float

custom_objective_ui_objective_component := class(ui_component):
    ImagesScale: float
    ActiveHandle: custom_objective_handle

    var m_GoalText: ?custom_font_text_block = false
    var m_ProgressBar: ?custom_objective_ui_objective_component_progress_bar = false

    TryUpdate<override>():logic=
        if (GoalText := m_GoalText?, Goal := ActiveHandle.Objective.Goal?) {
            GoalText.SetText("{ActiveHandle.m_Value}/{Goal}")
        }
        if (ProgressBar := m_ProgressBar?, Goal := ActiveHandle.Objective.Goal?) {
            Visibility := if (ActiveHandle.m_Value > 0) {
                widget_visibility.Visible
            } else {
                widget_visibility.Hidden
            }
            ProgressBar.FillStart.SetVisibility(Visibility)
            Progress := Clamp(ToFloat(ActiveHandle.m_Value) / ToFloat(Goal), 0.0, 1.0)
            FillBlockWidth := Max(0.0, (ProgressBar.Width * Progress) - ProgressBar.FillStartWidth)
            ProgressBar.FillBlock.SetDesiredSize(vector2{X := FillBlockWidth, Y := 1.0})
        }
        true

    Create<override>():widget=
        Background := CreateStaticImage(ActiveHandle.Objective.Image, ?UniformScale := ImagesScale)
        Overlay := overlay:
            Slots := array{
                overlay_slot:
                    Widget := Background
                    HorizontalAlignment := horizontal_alignment.Fill
                    VerticalAlignment := vertical_alignment.Fill
            }
        if (Goal := ActiveHandle.Objective.Goal?) {
            ProgressBarFillStart := CreateStaticImage("ObjectivesUI/progress_bar_fill_start", ?UniformScale := ImagesScale)
            ProgressBarFillLeftPadding := GetFixedWidgetDimensions(ProgressBarFillStart).X
            ProgressBarFillStart.SetVisibility(widget_visibility.Hidden)

            ProgressBarFillBlock := color_block{
                DefaultColor := NamedColors.White
                DefaultDesiredSize := vector2{X := 0.0, Y := 1.0}
            }

            ProgressBarBackground := CreateStaticImage("ObjectivesUI/progress_bar_bg", ?UniformScale := ImagesScale)
            ProgressBarWidth := GetFixedWidgetDimensions(ProgressBarBackground).X
            set m_ProgressBar = Some(custom_objective_ui_objective_component_progress_bar{
                FillStart := ProgressBarFillStart
                FillStartWidth := ProgressBarFillLeftPadding
                FillBlock := ProgressBarFillBlock
                Width := ProgressBarWidth
            })
            ProgressBar := overlay:
                Slots := array{
                    overlay_slot:
                        Widget := ProgressBarBackground
                        HorizontalAlignment := horizontal_alignment.Fill
                        VerticalAlignment := vertical_alignment.Fill
                    overlay_slot:
                        Widget := ProgressBarFillStart
                        HorizontalAlignment := horizontal_alignment.Left
                        VerticalAlignment := vertical_alignment.Fill
                    overlay_slot:
                        Widget := ProgressBarFillBlock
                        HorizontalAlignment := horizontal_alignment.Left
                        VerticalAlignment := vertical_alignment.Fill
                        Padding := margin{
                            Left := ProgressBarFillLeftPadding
                        }
                }
            Overlay.AddWidget(overlay_slot{
                Widget := ProgressBar
                HorizontalAlignment := horizontal_alignment.Center
                VerticalAlignment := vertical_alignment.Bottom
                Padding := margin{
                    Bottom := 42.0
                }
            })

            CustomFont := GetCustomFontOrDefault("Quicksand-Bold")
            GoalText := CreateCustomFontTextBlock(CustomFont, "",
                ?TextColor := NamedColors.White, ?FontSize := 19.0)
            set m_GoalText = Some(GoalText)
            Overlay.AddWidget(overlay_slot{
                Widget := GoalText.Widget
                HorizontalAlignment := horizontal_alignment.Center
                VerticalAlignment := vertical_alignment.Bottom
                Padding := margin{
                    Bottom := 14.0
                }
            })

            TryUpdate()
        } else {
            set m_GoalText = None()
            set m_ProgressBar = None()
        }

        Overlay

CmpObjectiveHandle<internal>(A:custom_objective_handle, B:custom_objective_handle)<computes>:int=
    CmpInt(A.Objective.Priority, B.Objective.Priority)
## }}}
