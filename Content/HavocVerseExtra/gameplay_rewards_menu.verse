using { /Fortnite.com/Devices }
using { /Fortnite.com/UI }
using { /Verse.org/Simulation }
using { /Verse.org/Colors }
using { /UnrealEngine.com/Temporary/Diagnostics }
using { /UnrealEngine.com/Temporary/UI }
using { /UnrealEngine.com/Temporary/SpatialMath }

using { HavocVerse.Debug }
using { HavocVerse.Devices }
using { HavocVerse.Devices.GameSpecific }
using { HavocVerse.Storage }
using { HavocVerse.UI }
using { HavocVerse.Logging }
using { HavocVerse.Util }

TooltipGameplayRewardsIconImageClaimedOverride<internal><localizes>:message = "By default, '%IconImage%_claimed' is used as the claimed rewards icon"

gameplay_rewards_row := class<concrete>:
    @editable Entries: []gameplay_rewards_entry = array{}

gameplay_rewards_entry := class<concrete><unique>:
    @editable RewardId: string = ""
    @editable PlayTime: int = 100
    @editable IconImage: string = ""
    @editable{ToolTip := TooltipGameplayRewardsIconImageClaimedOverride} IconImageClaimedOverride: ?string = false

    @editable OnClaimTriggers: []trigger_device = array{}
    @editable OnReadyTriggers: []trigger_device = array{}

    var m_EffectiveId: string = ""

gameplay_rewards_player_state := class:
    var m_Playtime: float = 0.0

    var m_ReadyHUDOnScreen: logic = false
    var m_Ready: [string]void = map{}
    var m_Claimed: [string]void = map{}

    var m_NeedsSync: logic = true
    var m_UpNext: ?gameplay_rewards_entry = false

    var m_OpenedMenu: ?gameplay_rewards_menu_instance = false

    IsUpNext(Entry:gameplay_rewards_entry)<transacts><decides>:void=
        m_UpNext? = Entry

gameplay_rewards_menu_device := class(creative_device, aware_device, aware_debug_device, persistence_recordable, persistence_resetable):
    @editable p_Rows: []gameplay_rewards_row = array{}
    @editable p_RowsPerPage: int = 2

    @editable p_PersistenceKeepAcrossTycoonRebirths: logic = false
    @editable p_ReadyHUDMessage: ?hud_message_device = false

    @editable p_MenuImagesScale: float = 1.0
    @editable p_MenuRewardIconsScale: float = 1.0

    @editable p_CloseMenuOnClaim: logic = true
    @editable p_OpenButtons: []button_device = array{}

    var m_Menu: ?gameplay_rewards_menu = false
    var m_Players: [player]gameplay_rewards_player_state = map{}

    var m_PersistenceControl: persistence_recordable_control = persistence_recordable_control{}

    DeviceMeta<override>()<transacts>:aware_device_meta=
        aware_device_meta{
            DeviceName := "gameplay_rewards_menu"
        }

    DebugStats<override>()<transacts>:aware_debug_stats=
        aware_debug_stats{
            Length := map{
                "players" => m_Players.Length
            }
        }

    OnBegin<override>()<suspends>:void=
        var IdsInUse: [string]void = map{}
        for (RowIndex->Row:p_Rows) {
            for (Index->Entry:Row.Entries) {
                EffectiveId := if (Entry.RewardId.Length > 0) {
                    Entry.RewardId
                } else {
                    "{RowIndex}->{Index}"
                }
                if (
                    not IdsInUse[EffectiveId]
                    set IdsInUse[EffectiveId] = void
                ) {
                    set Entry.m_EffectiveId = EffectiveId
                } else {
                    Log("{Self} has duplicate RewardId \"{EffectiveId}\"", ?Level := log_level.Error)
                }
            }
        }

        AwareInitDevice(Self, OnInit)

    OnInit<private>()<suspends>:void=
        if (AssetsProvider := AwareDefaultAssetsProvider[]) {
            set m_Menu = Some(gameplay_rewards_menu{
                AssetsProvider := AssetsProvider
                Device := Self
            })
        }
        for (Button:p_OpenButtons) {
            Button.InteractedWithEvent.Subscribe(OnButtonInteractedWith)
        }

        GetPlayspace().PlayerAddedEvent().Subscribe(OnPlayerAdded)
        GetPlayspace().PlayerRemovedEvent().Subscribe(OnPlayerRemoved)
        for (Player:GetPlayspace().GetPlayers()) {
            OnPlayerAdded(Player)
        }

        if (Integration := AwareGetIntegration[]) {
            Integration.SubscribeToMenuCloseRequest(OnMenuCloseRequest)
        }

        spawn:
            StartTickingLoop()

    ListEntries<private>()<transacts>:[]gameplay_rewards_entry=
        var Result: []gameplay_rewards_entry = array{}
        for (Row:p_Rows) {
            set Result += Row.Entries
        }
        Result

    StartTickingLoop<private>()<suspends>:void=
        var LastTick: float = GetSimulationElapsedTime()
        loop {
            Sleep(1.0)

            Now := GetSimulationElapsedTime()
            Elapsed := Now - LastTick
            set LastTick = Now

            for (Player->PlayerState:m_Players) {
                set PlayerState.m_Playtime += Elapsed
                SyncPlayer(Player, PlayerState, false)
            }
            if (Menu := m_Menu?) {
                Menu.RefreshActiveInstances()
            }
        }

    SyncPlayer<internal>(Player:player, PlayerState:gameplay_rewards_player_state, Recovered:logic):void=
        if (UpNext := PlayerState.m_UpNext?, PlayerState.m_Playtime >= ToFloat(UpNext.PlayTime)) {
            set PlayerState.m_NeedsSync = true
            set PlayerState.m_UpNext = None()
        }
        if (PlayerState.m_NeedsSync?) {
            AllEntries := ListEntries()
            RemainingEntries := for (Entry:AllEntries, PlayerState.m_Playtime < ToFloat(Entry.PlayTime)) {
                Entry
            }
            var UpNext: ?gameplay_rewards_entry = false
            for (Entry:RemainingEntries, UpNext = false or (UpNext?.PlayTime > Entry.PlayTime)) {
                set UpNext = Some(Entry)
            }
            for (
                Entry:AllEntries
                not PlayerState.m_Claimed[Entry.m_EffectiveId]
                PlayerState.m_Playtime >= ToFloat(Entry.PlayTime)
                not PlayerState.m_Ready[Entry.m_EffectiveId]
                set PlayerState.m_Ready[Entry.m_EffectiveId] = void
            ) {
                if (not (Recovered?)) {
                    for (Trigger:Entry.OnReadyTriggers) {
                        Trigger.Trigger(Player)
                    }
                }
            }
            if (PlayerState.m_Ready.Length > 0, not (PlayerState.m_ReadyHUDOnScreen?)) {
                if (HUDMessage := p_ReadyHUDMessage?) {
                    HUDMessage.Show(Player)
                }
                set PlayerState.m_ReadyHUDOnScreen = true
            } else if (PlayerState.m_Ready.Length <= 0, PlayerState.m_ReadyHUDOnScreen?) {
                if (HUDMessage := p_ReadyHUDMessage?) {
                    HUDMessage.Show(Player, ToMessage(""), ?DisplayTime := 0.01)
                }
                set PlayerState.m_ReadyHUDOnScreen = false

            }
            set PlayerState.m_UpNext = UpNext
            set PlayerState.m_NeedsSync = false
        }

    OnPlayerAdded<private>(Player:player):void=
        PlayerState := GetPlayerState(Player)
        SyncPlayer(Player, PlayerState, false)
    
    OnPlayerRemoved<private>(Player:player):void=
        set m_Players = m_Players.RemoveByKey(Player)(0)

    OnMenuCloseRequest<private>(Player:player):void=
        if (Menu := m_Menu?) {
            Menu.CloseInstance(Player)
        }

    OnButtonInteractedWith<private>(Agent:agent):void=
        if (Player := player[Agent], Menu := m_Menu?) {
            Menu.Open(Player)
        }

    GetPlayerState<internal>(Player:player):gameplay_rewards_player_state=
        if (PlayerState := m_Players[Player]) {
            PlayerState
        } else {
            PlayerState := gameplay_rewards_player_state{}
            if (set m_Players[Player] = PlayerState) {}
            PlayerState
        }
    
    GetPersistenceControl<override>()<transacts><decides>:persistence_recordable_control=
        m_PersistenceControl

    PersistenceReserve<override>(StorageMem:storage_memory)<transacts><decides>:persistence_recordable_state=
        var RewardsChunks: [string]storage_memory_chunk = map{}
        for (Entry:ListEntries()) {
            MemChunk := StorageMem.Reserve[1, "gameplay_reward_{Entry.m_EffectiveId}"]
            set RewardsChunks[Entry.m_EffectiveId] = MemChunk
        }
        gameplay_rewards_persistence_state{
            PlaytimeChunk := StorageMem.Reserve[22, "gameplay_playtime"]
            RewardsChunks := RewardsChunks
        }

    PersistenceSave<override>(i_State:persistence_recordable_state, Agent:agent, SaveCause:persistence_save_cause):void=
        if (State := gameplay_rewards_persistence_state[i_State], PlayerState := m_Players[Agent]) {
            State.PlaytimeChunk.WriteInt(Agent, DefaultingFloor(PlayerState.m_Playtime))
            for (RewardId->MemChunk:State.RewardsChunks) {
                IsClaimed := if (PlayerState.m_Claimed[RewardId]). true else false
                MemChunk.Write(Agent, array{IsClaimed})
            }
        }

    PersistenceLoad<override>(i_State:persistence_recordable_state, Agent:agent, Query:persistence_load_query):void=
        if (State := gameplay_rewards_persistence_state[i_State], Player := player[Agent]) {
            PlayerState := GetPlayerState(Player)
            SavedPlaytime := ToFloat(State.PlaytimeChunk.ReadInt(Player))
            if (Query.Overwrite?) {
                set PlayerState.m_Playtime = SavedPlaytime
            } else if (Query.FirstLoad?) {
                set PlayerState.m_Playtime += SavedPlaytime
            }
            for (RewardId->MemChunk:State.RewardsChunks) {
                Data := MemChunk.Read(Player)
                IsClaimed := option{Data[0]}.UnwrapOr(false)
                if (IsClaimed?) {
                    if:
                        set PlayerState.m_Claimed[RewardId] = void
                } else if (Query.Overwrite?) {
                    set PlayerState.m_Claimed = PlayerState.m_Claimed.RemoveByKey(RewardId)(0)
                }
            }
            SyncPlayer(Player, PlayerState, true)
            if (OpenedMenu := PlayerState.m_OpenedMenu?) {
                OpenedMenu.Refresh()
            }
        }

    PersistenceReset<override>(i_State:persistence_recordable_state, Agent:agent, Query:persistence_reset_query):void=
        if (p_PersistenceKeepAcrossTycoonRebirths?, tycoon_manager_rebirth_persistence_reset_query[Query]) {
            Log("Skipping reset of gameplay rewards persistence state because it is initialized by a rebirth", ?Level := log_level.Debug)
        } else {
            i_State.Reset(Agent)
            PersistenceLoad(i_State, Agent, persistence_load_query{
                Overwrite := true
                FirstLoad := false
            })
        }

gameplay_rewards_persistence_state := class(persistence_recordable_state):
    PlaytimeChunk: storage_memory_chunk
    RewardsChunks: [string]storage_memory_chunk

    Reset<override>(Agent:agent):void=
        PlaytimeChunk.Reset(Agent)
        for (__->MemChunk:RewardsChunks) {
            MemChunk.Reset(Agent)
        }

## Menu {{{
gameplay_rewards_menu := class(ui_menu):
    Device: gameplay_rewards_menu_device

    CreateInstance<override>(Player:player):tuple(ui_menu_instance,canvas)=
        PlayerState := Device.GetPlayerState(Player)
        var PageIndex: int = 0
        loop {
            StartRow := PageIndex * Device.p_RowsPerPage
            Rows := Device.p_Rows.SafeSlice(StartRow, StartRow + Device.p_RowsPerPage)
            if (Rows.Length <= 0) {
                break
            }
            for (Row:Rows) {
                for (Entry:Row.Entries) {
                    if (not PlayerState.m_Ready[Entry.m_EffectiveId], not PlayerState.m_Claimed[Entry.m_EffectiveId]) {
                        break
                    }
                }
            }
            set PageIndex += 1
        }
        CreateInstance(Player, PlayerState, PageIndex)

    CreateInstance<private>(Player:player, PlayerState:gameplay_rewards_player_state, Page:int):tuple(ui_menu_instance,canvas)=
        TimerFont := GetCustomFontOrDefault("Quicksand-Bold")

        Instance := gameplay_rewards_menu_instance{
            Menu := Self
            Player := Player
            Device := Device
            PlayerState := PlayerState
            TimerFont := TimerFont
        }

        MainGrid := gameplay_rewards_menu_main_grid_component:
            Instance := Instance
        set Instance.m_MainGrid = Some(MainGrid)

        ActionsRow := gameplay_rewards_menu_actions_row_component:
            Instance := Instance

        Background := CreateStaticImage("GameplayRewardsMenuUI/background", ?Scale := vector2{X := 0.7294921875, Y := 0.7294736842})
        Widget := overlay:
            Slots := array{
                overlay_slot:
                    Widget := Background
                    HorizontalAlignment := horizontal_alignment.Center
                    VerticalAlignment := vertical_alignment.Center
                overlay_slot:
                    Widget := Instance.AddComponent(MainGrid)
                    HorizontalAlignment := horizontal_alignment.Center
                    VerticalAlignment := vertical_alignment.Top
                    Padding := margin{
                        Top := 210.0
                    }
                overlay_slot:
                    Widget := Instance.AddComponent(ActionsRow)
                    HorizontalAlignment := horizontal_alignment.Center
                    VerticalAlignment := vertical_alignment.Bottom
                    Padding := margin{
                        Bottom := 70.0
                    }
            }

        Canvas := GetSimpleCanvas(Widget)
        (Instance, Canvas)

gameplay_rewards_menu_timer_state := class:
    UnlocksInText: custom_font_text_block

gameplay_rewards_menu_instance := class(ui_menu_instance):
    Device: gameplay_rewards_menu_device
    PlayerState: gameplay_rewards_player_state
    TimerFont: custom_font

    var m_MainGrid: ?gameplay_rewards_menu_main_grid_component = false

    OnPrevPageClick(Data:widget_message):void=
        if (MainGrid := m_MainGrid?) {
            MainGrid.MovePage(-1)
        }

    OnNextPageClick(Data:widget_message):void=
        if (MainGrid := m_MainGrid?) {
            MainGrid.MovePage(1)
        }

### Main Grid Component {{{
gameplay_rewards_menu_main_grid_component := class(ui_component):
    Instance: gameplay_rewards_menu_instance

    var m_Page<private>: int = 0

    SetPage(Page:int):void=
        PrevPage := m_Page
        Device := Instance.Device
        set m_Page = Clamp(Page, 0, DivDefaultingCeil(Device.p_Rows.Length, Device.p_RowsPerPage)-1)
        if (m_Page <> PrevPage) {
            ForceUpdate()
        }

    MovePage(Delta:int):void=
        SetPage(m_Page+Delta)

    Create<override>():widget=
        Device := Instance.Device
        StartRow := m_Page * Device.p_RowsPerPage
        Rows := Device.p_Rows.SafeSlice(StartRow, StartRow + Device.p_RowsPerPage)

        GridWidgets := for (Row:Rows) {
            for (Entry:Row.Entries, Entry.m_EffectiveId.Length > 0) {
                AddChild(gameplay_rewards_menu_card_component{
                    Instance := Instance
                    Entry := Entry
                })
            }
        }
        CreateUIGrid(GridWidgets, ?GapX := 40.0, ?GapY := 30.0)
### }}}

### Card Component {{{
gameplay_rewards_card_status := enum:
    CountingDown
    Ready
    Claimed

gameplay_rewards_menu_card_component := class(ui_component):
    Instance: gameplay_rewards_menu_instance
    Entry: gameplay_rewards_entry

    var m_Status: ?gameplay_rewards_card_status = false
    var m_UnlocksInText: ?custom_font_text_block = false
    var m_IsUpNext: logic = false

    PollStatus<private>():gameplay_rewards_card_status=
        if (Instance.PlayerState.m_Claimed[Entry.m_EffectiveId]) {
            gameplay_rewards_card_status.Claimed
        } else if (ToFloat(Entry.PlayTime) - Instance.PlayerState.m_Playtime <= 0.0) {
            gameplay_rewards_card_status.Ready
        } else {
            gameplay_rewards_card_status.CountingDown
        }

    TryUpdate<override>():logic=
        Status := PollStatus()
        if ((not (m_Status?)) or (Status <> (m_Status?))) {
            return false
        }
        if (UnlocksInText := m_UnlocksInText?) {
            RemainingSeconds := ToFloat(Entry.PlayTime) - Instance.PlayerState.m_Playtime
            IsUpNext := if (Instance.PlayerState.IsUpNext[Entry]). true else false
            if (IsUpNext <> m_IsUpNext) {
                return false
            }

            UnlocksInText.SetText(ToHumanDurationAutoScale(RemainingSeconds, 2, human_duration_opts{
                Delimiter := " "
                DaysFmt := ("d", "d")
                HoursFmt := ("h", "h")
                MinutesFmt := ("m", "m")
            }))
        }
        true

    Create<override>():widget=
        ImagesScale := Instance.Device.p_MenuImagesScale

        Status := PollStatus()
        set m_Status = Some(Status)
        set m_UnlocksInText = None()
        UnlockRow := case(Status):
            gameplay_rewards_card_status.Claimed =>
                Button := button_regular:
                    DefaultText := ToMessage("Claimed")
                ApplySizeHint(Button, vector2{X := 230.0, Y := 30.0})
            gameplay_rewards_card_status.Ready =>
                Button := button_loud:
                    DefaultText := ToMessage("Claim")
                Button.OnClick().Subscribe(OnClaimClick)
                ApplySizeHint(Button, vector2{X := 230.0, Y := 30.0})
            gameplay_rewards_card_status.CountingDown =>
                UnlocksInText := CreateCustomFontTextBlock(Instance.TimerFont, "",
                    ?TextColor := NamedColors.AliceBlue,
                    ?FontSize := 19.0)
                set m_UnlocksInText = Some(UnlocksInText)
                overlay:
                    Slots := array{
                        overlay_slot:
                            Widget := CreateStaticImage("GameplayRewardsMenuUI/unlocks_in_bg", ?UniformScale := ImagesScale)
                            HorizontalAlignment := horizontal_alignment.Fill
                            VerticalAlignment := vertical_alignment.Fill
                        overlay_slot:
                            Widget := UnlocksInText.Widget
                            HorizontalAlignment := horizontal_alignment.Left
                            VerticalAlignment := vertical_alignment.Top
                            Padding := margin{
                                Top := 14.0
                                Left := 138.0
                            }
                    }

        RewardsIcon := CreateStaticImage(if (Status = gameplay_rewards_card_status.Claimed) {
            Entry.IconImageClaimedOverride.UnwrapOr(Entry.IconImage + "_claimed")
                .ReplaceAll("%IconImage%", Entry.IconImage)
        } else {
            Entry.IconImage
        }, ?UniformScale := Instance.Device.p_MenuRewardIconsScale)
        RewardsIconDimensions := GetFixedWidgetDimensions(RewardsIcon)

        ContentStack := stack_box:
            Orientation := orientation.Vertical
            Slots := array{
                stack_box_slot:
                    Widget := RewardsIcon
                    HorizontalAlignment := horizontal_alignment.Center
                    VerticalAlignment := vertical_alignment.Top
                stack_box_slot:
                    Widget := UnlockRow
                    HorizontalAlignment := horizontal_alignment.Center
                    VerticalAlignment := vertical_alignment.Top
                    Padding := margin{
                        Top := 12.0
                    }
            }

        Card := overlay:
            Slots := array{
                overlay_slot:
                    Widget := CreateStaticImage("GameplayRewardsMenuUI/card_bg", ?UniformScale := ImagesScale)
                    HorizontalAlignment := horizontal_alignment.Fill
                    VerticalAlignment := vertical_alignment.Fill
                overlay_slot:
                    Widget := ContentStack
                    HorizontalAlignment := horizontal_alignment.Center
                    VerticalAlignment := vertical_alignment.Top
                    Padding := margin{
                        Top := 216.0 - RewardsIconDimensions.Y
                    }
            }

        IsUpNext := if (Status = gameplay_rewards_card_status.CountingDown, Instance.PlayerState.IsUpNext[Entry]) {
            true
        } else {
            false
        }
        set m_IsUpNext = IsUpNext
        if (IsUpNext?) {
            Card.AddWidget(overlay_slot{
                Widget := CreateStaticImage("GameplayRewardsMenuUI/up_next_label", ?UniformScale := ImagesScale)
                HorizontalAlignment := horizontal_alignment.Center
                VerticalAlignment := vertical_alignment.Top
                Padding := margin{
                    Top := -10.0
                }
            })
        }

        TryUpdate()

        Card

    OnClaimClick(Data:widget_message):void=
        if (
            not Instance.PlayerState.m_Claimed[Entry.m_EffectiveId]
            set Instance.PlayerState.m_Claimed[Entry.m_EffectiveId] = void
        ) {
            set Instance.PlayerState.m_Ready = Instance.PlayerState.m_Ready.RemoveByKey(Entry.m_EffectiveId)(0)
            set Instance.PlayerState.m_NeedsSync = true
            Instance.Device.SyncPlayer(Instance.Player, Instance.PlayerState, false)
            Instance.Device.m_PersistenceControl.Save(Instance.Player)
            if (Instance.Device.p_CloseMenuOnClaim?) {
                Instance.Close()
            }
            for (Trigger:Entry.OnClaimTriggers) {
                Trigger.Trigger(Instance.Player)
            }
            Update()
        }
### }}}

### Actions Row Component {{{
gameplay_rewards_menu_actions_row_component := class(ui_component):
    Instance: gameplay_rewards_menu_instance

    Create<override>():widget=
        CloseButton := button_regular:
            DefaultText := ToMessage("Close")
        CloseButton.OnClick().Subscribe(Instance.OnCloseClick)

        PrevPageButton := button_regular:
            DefaultText := ToMessage("<")
        PrevPageButton.OnClick().Subscribe(Instance.OnPrevPageClick)
        NextPageButton := button_regular:
            DefaultText := ToMessage(">")
        NextPageButton.OnClick().Subscribe(Instance.OnNextPageClick)

        ActionsRowContent := stack_box:
            Orientation := orientation.Horizontal
            Slots := array{
                stack_box_slot:
                    Widget := ApplySizeHint(PrevPageButton, vector2{X := 80.0, Y := 50.0})
                    HorizontalAlignment := horizontal_alignment.Left
                    VerticalAlignment := vertical_alignment.Top
                stack_box_slot:
                    Widget := ApplySizeHint(CloseButton, vector2{X := 760.0, Y := 50.0})
                    HorizontalAlignment := horizontal_alignment.Fill
                    VerticalAlignment := vertical_alignment.Top
                    Padding := margin{
                        Left := 15.0
                    }
                stack_box_slot:
                    Widget := ApplySizeHint(NextPageButton, vector2{X := 80.0, Y := 50.0})
                    HorizontalAlignment := horizontal_alignment.Left
                    VerticalAlignment := vertical_alignment.Top
                    Padding := margin{
                        Left := 15.0
                    }
            }

        overlay:
            Slots := array{
                overlay_slot:
                    Widget := EmptyRect(vector2{X := 950.0, Y := 66.0})
                    HorizontalAlignment := horizontal_alignment.Fill
                    VerticalAlignment := vertical_alignment.Fill
                overlay_slot:
                    Widget := ActionsRowContent
                    HorizontalAlignment := horizontal_alignment.Fill
                    VerticalAlignment := vertical_alignment.Top
                    Padding := margin{
                        Bottom := 16.0
                    }
            }
### }}}
## }}}
