using { /Fortnite.com/Devices }
using { /Fortnite.com/UI }
using { /Verse.org/Simulation }
using { /Verse.org/Colors }
using { /UnrealEngine.com/Temporary/Diagnostics }
using { /UnrealEngine.com/Temporary/UI }
using { /UnrealEngine.com/Temporary/SpatialMath }

using { HavocVerse.Debug }
using { HavocVerse.Devices }
using { HavocVerse.Logging }
using { HavocVerse.UI }
using { HavocVerse.Util }

home_configurator_category_config := class<final><concrete>:
    @editable Name: string = ""
    @editable AttachmentKindId: string = ""
    @editable StyleLabelTemplate: string = "Style %Current%/%Total%"
    @editable Styles: []home_configurator_style_config = array{}

    ExpectsAutoFraming<internal>()<decides><transacts>:void=
        var RelyingOnAutoFrame: logic = false
        for (Style:Styles, not RelyingOnAutoFrame?) {
            if (not Style.Cinematic?) {
                set RelyingOnAutoFrame = true
            }
        }
        RelyingOnAutoFrame?

home_configurator_category_instance_config := class<final><concrete>:
    @editable PreviewCamera: gameplay_camera_fixed_point_device = gameplay_camera_fixed_point_device{}
    @editable AutoFramingSeqController: ?cinematic_sequence_device = false

    var m_AppliedStyleIndex: int = -1

home_configurator_style_config := class<final><concrete>:
    @editable Attachment: string = ""
    @editable LockedLabelTextOverride: ?string = false
    @editable AutoFrameOverride: ?int = false
    @editable Cinematic: ?cinematic_sequence_device = false

home_configurator_menu_config := class<final><concrete>:
    @editable BackgroundImage: string = ""
    @editable BackgroundImageOffset: vector2 = vector2{}
    @editable BackgroundImageScale: float = 1.0
    @editable LockedLabelIconSize: vector2 = vector2{X := 48.0, Y := 48.0}
    @editable LockedLabelIconTextSpacing: float = 6.0
    @editable LockedLabelIconLocked: string = ""
    @editable LockedLabelIconUnlocked: string = ""
    @editable LockedLabelLockedTextColor: color = NamedColors.Red
    @editable LockedLabelUnlockedTextColor: color = NamedColors.Green
    @editable LockedLabelLockedDefaultText: string = "Locked"
    @editable LockedLabelUnlockedText: string = "Unlocked!"

home_configurator_model_device := class<final><concrete>(creative_device, aware_device):
    @editable p_DevId: string = "N/A"

    @editable p_AttachmentsManager: attachments_manager_device = attachments_manager_device{}
    @editable p_Categories: []home_configurator_category_config = array{}

    @editable p_MenuConfig: home_configurator_menu_config = home_configurator_menu_config{}

    var m_InstanceBackrefs<internal>: []home_configurator_instance_device = array{}

    DeviceMeta<override>()<transacts>:aware_device_meta=
        aware_device_meta{
            DeviceName := "home_configurator_model"
            DevId := p_DevId
        }

    ConfigCheck<private>():void=
        for (Index->Category:p_Categories) {
            var Issues: []string = array{}
            if (Category.Name.IsEmpty[]) {
                set Issues += array{"missing Name"}
            }
            if (Category.AttachmentKindId.IsEmpty[]) {
                set Issues += array{"missing AttachmentKindId"}
            }
            if (Issues.Length > 0) {
                Log(Self, "Category index {Index} is misconfigured: {Issues.Join(", ")}", ?Level := log_level.Error)
            }
        }

    OnBegin<override>()<suspends>:void=
        AwareInitDevice(Self, OnInit)
    
    OnInit<private>()<suspends>:void=
        ConfigCheck()

        p_AttachmentsManager.RegisterHook(home_configurator_player_attachment_hook{
            ModelDevice := Self
        })

    FindInstanceForPlayer<internal>(Player:player)<decides><transacts>:home_configurator_instance_device=
        var Found: ?home_configurator_instance_device = None()
        for (Instance:m_InstanceBackrefs, not Found?, Instance.p_AccessControl.IsAllowed[Player]) {
            set Found = Some(Instance)
        }
        Found?

    FindCurrentStyleIndex<internal>(Player:player, CategoryIndex:int)<decides><transacts>:int=
        Category := p_Categories[CategoryIndex]
        Attachment := p_AttachmentsManager.GetActiveAttachment[Player, Category.AttachmentKindId]
        var Found: ?int = false
        for (StyleIndex->Style:Category.Styles, not Found?, Style.Attachment = Attachment) {
            set Found = Some(StyleIndex)
        }
        Found?

TooltipHomeConfiguratorCategoryInstances<internal><localizes>:message = "Must be in the same order as the model's categories"

home_configurator_instance_device := class<final><concrete>(creative_device, aware_device, resetable, custom_device):
    @editable p_DevId: string = "N/A"
    @editable{ToolTip := TooltipRoutingIdOverride} p_RoutingIdOverride: ?string = false

    @editable p_Model: home_configurator_model_device = home_configurator_model_device{}
    @editable{ToolTip := TooltipHomeConfiguratorCategoryInstances}
    p_CategoryInstances: []home_configurator_category_instance_config = array{}

    @editable p_AccessControl: access_control_device = access_control_device{}
    @editable p_StartEnabled: logic = true

    @editable p_OpenButtons: []button_device = array{}
    @editable p_HideHudAccessControl: ?access_control_device = false

    var m_Enabled: logic = false
    var m_Subscriptions: []cancelable = array{}
    var m_SyncedPlayer: logic = false

    DeviceMeta<override>()<transacts>:aware_device_meta=
        aware_device_meta{
            DeviceName := "home_configurator"
            DevId := p_DevId
            RoutingIdOverride := p_RoutingIdOverride
        }

    OnBegin<override>()<suspends>:void=
        set p_Model.m_InstanceBackrefs += array{Self}
        AwareInitDevice(Self, OnInit)

    ConfigCheck<private>():void=
        if (p_CategoryInstances.Length <> p_Model.p_Categories.Length) {
            Log(Self, "Category instance count ({p_CategoryInstances.Length}) does not match model category count (p_Model.p_Categories.Length)", ?Level := log_level.Error)
        }

        for (Index->CategoryInstance:p_CategoryInstances, Category := p_Model.p_Categories[Index]) {
            if (not CategoryInstance.AutoFramingSeqController?, Category.ExpectsAutoFraming[]) {
                Log(Self, "Category index {Index} is missing an auto framing sequence: configured styles depend on it", ?Level := log_level.Error)
            }
        }

        if (HudAccessControl := p_HideHudAccessControl?, HudAccessControl.p_Mode <> access_control_mode.Whitelist) {
            Log(Self, "p_HideHudAccessControl must be in whitelist mode", ?Level := log_level.Error)
        }

    OnInit<private>()<suspends>:void=
        ConfigCheck()

        p_AccessControl.RegisterUsingDevice(Self)
        p_AccessControl.SubscribeToPlayerJoinEvents(OnPlayerAdded, OnPlayerRemoved)

        if (p_StartEnabled?) {
            Enable()
        } else {
            Disable()
        }

    OnPlayerAdded<private>(Player:player):void=
        if (not m_SyncedPlayer?) {
            SyncPlayer(Player)
            set m_SyncedPlayer = true
        }

    OnPlayerRemoved<private>(Player:player, GameQuit:logic):void=
        {}

    SyncPlayer<internal>(Player:player):void=
        for (CategoryIndex->__:p_CategoryInstances) {
            StyleIndex := p_Model.FindCurrentStyleIndex[Player, CategoryIndex] or 0
            ApplyStyle(CategoryIndex, StyleIndex)
        }

    Enable<private>():void=
        if (m_Enabled?) {
            return
        }
        set m_Enabled = true

        for (OpenButton:p_OpenButtons) {
            OpenButton.Enable()
            set m_Subscriptions += array{
                OpenButton.InteractedWithEvent.Subscribe(OnOpenButtonInteractedWith)
            }
        }

    Disable<private>():void=
        set m_Enabled = false
        for (OpenButton:p_OpenButtons) {
            OpenButton.Disable()
        }
        for (Subscription:m_Subscriptions) {
            Subscription.Cancel()
        }
        set m_Subscriptions = array{}

    Reset<override>(Agent:?agent):void=
        for (
            CategoryIndex->CategoryInstance:p_CategoryInstances
            Category := p_Model.p_Categories[CategoryIndex]
            DefaultStyle := Category.Styles[0]
        ) {
            ApplyStyle(Category, CategoryInstance, 0, DefaultStyle)
        }
        set m_SyncedPlayer = false

    OnOpenButtonInteractedWith<private>(Agent:agent):void=
        OpenSimpleMenu("home_configurator", Agent, home_configurator_menu_component{
            Device := Self
        })

    ApplyStyle<internal>(CategoryIndex:int, StyleIndex:int):void=
        if (
            Category := p_Model.p_Categories[CategoryIndex]
            CategoryInstance := p_CategoryInstances[CategoryIndex]
            CategoryInstance.m_AppliedStyleIndex <> StyleIndex
            Style := Category.Styles[StyleIndex]
        ) {
            ApplyStyle(Category, CategoryInstance, StyleIndex, Style)
        }

    ApplyStyle<internal>(
        Category:home_configurator_category_config,
        CategoryInstance:home_configurator_category_instance_config,
        StyleIndex:int,
        Style:home_configurator_style_config
    ):void=
        if (CinematicDevice := Style.Cinematic?) {
            CinematicDevice.Play()
        } else if (SeqController := CategoryInstance.AutoFramingSeqController?) {
            Frame := Style.AutoFrameOverride? or StyleIndex
            # TODO : maybe need to cold start at 0?
            SeqController.SetPlaybackFrame(Frame)
            SeqController.Play()
            SeqController.Pause()
        }

    CustomTrigger<override>(Action:string, Agent:?agent):logic=
        if (Action = "enable") {
            Enable()
            return true
        } else if (Action = "disable") {
            Disable()
            return true
        }
        false

    GetCurrentStyleIndex<internal>(CategoryIndex:int)<transacts>:int=
        if (
            Player := player[p_AccessControl.ListAllowedAgents()[0]]
            CurrentStyleIndex := p_Model.FindCurrentStyleIndex[Player, CategoryIndex]
        ) {
            CurrentStyleIndex
        } else {
            0
        }

home_configurator_player_attachment_hook<internal> := class<final>(player_attachment_hook):
    ModelDevice: home_configurator_model_device

    EquipAttachment<override>(Player:player, AttachmentKindId:string, Attachment:string, Recall:logic):void=
        if (InstanceDevice := ModelDevice.FindInstanceForPlayer[Player]) {
            for (
                CategoryIndex->Category:ModelDevice.p_Categories
                Category.AttachmentKindId = AttachmentKindId
                CategoryInstance := InstanceDevice.p_CategoryInstances[CategoryIndex]
            ) {
                for (StyleIndex->Style:Category.Styles, Style.Attachment = Attachment) {
                    InstanceDevice.ApplyStyle(Category, CategoryInstance, StyleIndex, Style)
                    return
                }
            }
        }

    UnequipAttachment<override>(Player:player, AttachmentKindId:string, Attachment:string, ReplacingWith:?string):void=
        if (not ReplacingWith?, InstanceDevice := ModelDevice.FindInstanceForPlayer[Player]) {
            for (
                CategoryIndex->Category:ModelDevice.p_Categories
                Category.AttachmentKindId = AttachmentKindId
                CategoryInstance := InstanceDevice.p_CategoryInstances[CategoryIndex]
            ) {
                # find category containing prev attachmentKindId->attachment matching pair
                for (Style:Category.Styles, Style.Attachment = Attachment, DefaultStyle := Category.Styles[0]) {
                    InstanceDevice.ApplyStyle(Category, CategoryInstance, 0, DefaultStyle)
                    return
                }
            }
        }

    ShouldRecallOnRespawn<override>()<transacts>:logic = false

## Menu {{{
home_configurator_menu_component_props := class<final>:
    StyleLabel: cached_text_block
    StyleInfoImage: defaulting_texture_block
    StyleInfoLabel: cached_text_block

home_configurator_menu_component := class<final>(ui_menu_component):
    RootWidget<override>: logic = true

    Device: home_configurator_instance_device

    var m_PrevHudState: [player]void = map{}

    var m_ActiveCategoryIndex: int = 0
    var m_ActiveCameraIndex: int = -1
    var m_ActiveStyleIndex: int = -1

    var m_CategoryButtons: []shapeshifting_fort_button = array{}
    var m_Props: ?home_configurator_menu_component_props = false

    TryUpdate<override>():logic=
        # category buttons
        for (Index->CategoryButton:m_CategoryButtons) {
            ButtonType := if (Index = m_ActiveCategoryIndex) {
                fort_button_type.Loud
            } else {
                fort_button_type.Regular
            }
            CategoryButton.SetButtonType(ButtonType)
        }

        # active style
        if (m_ActiveStyleIndex < 0) {
            CurrentStyleIndex := Device.GetCurrentStyleIndex(m_ActiveCategoryIndex)
            set m_ActiveStyleIndex = CurrentStyleIndex
        }
        Device.ApplyStyle(m_ActiveCategoryIndex, m_ActiveStyleIndex)
        if (Props := m_Props?, Category := Device.p_Model.p_Categories[m_ActiveCategoryIndex]) {
            Text := Category.StyleLabelTemplate
                .ReplaceAll("%Current%", ToString(m_ActiveStyleIndex + 1))
                .ReplaceAll("%Total%", ToString(Category.Styles.Length))
            Props.StyleLabel.SetText(Text)

            if (
                Style := Category.Styles[m_ActiveStyleIndex]
                Player := player[Device.p_AccessControl.ListAllowedAgents()[0]]
                Device.p_Model.p_AttachmentsManager.IsUnlocked[Player, Category.AttachmentKindId, Style.Attachment]
            ) {
                Props.StyleInfoLabel.SetText(Device.p_Model.p_MenuConfig.LockedLabelUnlockedText)
                Props.StyleInfoLabel.SetTextColor(Device.p_Model.p_MenuConfig.LockedLabelUnlockedTextColor)
                Props.StyleInfoImage.SetCachingTexture(Some(Device.p_Model.p_MenuConfig.LockedLabelIconUnlocked))
            } else {
                if (Style := Category.Styles[m_ActiveStyleIndex], LockedText := Style.LockedLabelTextOverride?) {
                    Props.StyleInfoLabel.SetText(LockedText)
                } else {
                    Props.StyleInfoLabel.SetText(Device.p_Model.p_MenuConfig.LockedLabelLockedDefaultText)
                }
                Props.StyleInfoLabel.SetTextColor(Device.p_Model.p_MenuConfig.LockedLabelLockedTextColor)
                Props.StyleInfoImage.SetCachingTexture(Some(Device.p_Model.p_MenuConfig.LockedLabelIconLocked))
            }
        }

        # active preview
        if (m_ActiveCategoryIndex <> m_ActiveCameraIndex) {
            if (PrevCamera := Device.p_CategoryInstances[m_ActiveCameraIndex].PreviewCamera) {
                for (Viewer:ListViewingPlayers()) {
                    PrevCamera.RemoveFrom(Viewer)
                }
            }
            set m_ActiveCameraIndex = m_ActiveCategoryIndex
            if (NewCamera := Device.p_CategoryInstances[m_ActiveCategoryIndex].PreviewCamera) {
                for (Viewer:ListViewingPlayers()) {
                    NewCamera.AddTo(Viewer)
                }
            }
        }

        true

    Create<override>():widget=
        ButtonSize := vector2{X := 40.0, Y := 600.0}
        LeftButton := CreateFortButton(fort_button_type.Regular, ToMessage("<"))
        LeftButton.OnClick().Subscribe(OnLeftClick)
        RightButton := CreateFortButton(fort_button_type.Regular, ToMessage(">"))
        RightButton.OnClick().Subscribe(OnRightClick)

        MenuWidth := 1400.0 + (ButtonSize.X * 2)
        ScrollButtons := overlay:
            Slots := array{
                overlay_slot:
                    Widget := color_block:
                        DefaultOpacity := 0.0
                        DefaultDesiredSize := vector2{X := MenuWidth}
                overlay_slot:
                    Widget := ApplySizeHint(LeftButton, ButtonSize)
                    HorizontalAlignment := horizontal_alignment.Left
                    VerticalAlignment := vertical_alignment.Fill
                overlay_slot:
                    Widget := ApplySizeHint(RightButton, ButtonSize)
                    HorizontalAlignment := horizontal_alignment.Right
                    VerticalAlignment := vertical_alignment.Fill
            }

        CloseButton := CreateFortButton(fort_button_type.Regular, ToMessage("Back"))
        BindToViewerClose(CloseButton.OnClick())

        CategoryButtonsGap := 20.0
        CategoryCount := Max(1, Device.p_Model.p_Categories.Length)
        CategoryButtonWidth := (MenuWidth - (CategoryButtonsGap * (CategoryCount - 1))) / ToFloat(CategoryCount)
        set m_CategoryButtons = array{}
        LastIndex := Device.p_Model.p_Categories.Length-1
        CategoriesSlots := for (Index->Category:Device.p_Model.p_Categories) {
            Closure := home_configurator_category_button_closure:
                Menu := Self
                CategoryIndex := Index
            CategoryButton := CreateShapeshiftingFortButton(fort_button_type.Regular, ToMessage(Category.Name))
            set m_CategoryButtons += array{CategoryButton}
            CategoryButton.AddOnClickListener(Closure.OnClick)
            stack_box_slot:
                Widget := ApplySizeHint(CategoryButton.ParentWidget(), vector2{X := CategoryButtonWidth, Y := 40.0})
                HorizontalAlignment := horizontal_alignment.Left
                VerticalAlignment := vertical_alignment.Top
                Padding := margin{
                    Right := if (Index < LastIndex) {
                        CategoryButtonsGap
                    } else {
                        0.0
                    }
                }
        }
        CategoriesRow := stack_box:
            Orientation := orientation.Horizontal
            Slots := CategoriesSlots

        ActionButtons := stack_box:
            Orientation := orientation.Vertical
            Slots := array{
                stack_box_slot:
                    Widget := CategoriesRow
                    HorizontalAlignment := horizontal_alignment.Fill
                    VerticalAlignment := vertical_alignment.Top
                stack_box_slot:
                    Widget := ApplySizeHint(CloseButton, vector2{Y := 40.0})
                    HorizontalAlignment := horizontal_alignment.Fill
                    VerticalAlignment := vertical_alignment.Top
                    Padding := margin{
                        Top := 10.0
                    }
            }

        ControlsWidget := stack_box:
            Orientation := orientation.Vertical
            Slots := array{
                stack_box_slot:
                    Widget := ScrollButtons
                    HorizontalAlignment := horizontal_alignment.Center
                    VerticalAlignment := vertical_alignment.Top
                stack_box_slot:
                    Widget := ActionButtons
                    HorizontalAlignment := horizontal_alignment.Fill
                    VerticalAlignment := vertical_alignment.Top
                    Padding := margin{
                        Top := 20.0
                    }
            }

        MenuConfig := Device.p_Model.p_MenuConfig
        Props := home_configurator_menu_component_props:
            StyleLabel := MakeCachedTextBlock(CreateTextBlock(""))
            StyleInfoImage := CreateDefaultingTextureBlock(MenuConfig.LockedLabelIconSize, None())
            StyleInfoLabel := MakeCachedTextBlock(CreateTextBlock(""))
        set m_Props = Some(Props)

        StyleInfoWidget := stack_box:
            Orientation := orientation.Vertical
            Slots := array{
                stack_box_slot:
                    Widget := Props.StyleLabel.Widget
                    HorizontalAlignment := horizontal_alignment.Center
                    VerticalAlignment := vertical_alignment.Top
                stack_box_slot:
                    Widget := stack_box:
                        Orientation := orientation.Horizontal
                        Slots := array{
                            stack_box_slot:
                                Widget := Props.StyleInfoImage.Widget()
                                HorizontalAlignment := horizontal_alignment.Left
                                VerticalAlignment := vertical_alignment.Center
                            stack_box_slot:
                                Widget := Props.StyleInfoLabel.Widget
                                HorizontalAlignment := horizontal_alignment.Left
                                VerticalAlignment := vertical_alignment.Center
                                Padding := margin{
                                    Left := MenuConfig.LockedLabelIconTextSpacing
                                }
                        }
                    HorizontalAlignment := horizontal_alignment.Center
                    VerticalAlignment := vertical_alignment.Top
            }

        Widget := overlay:
            Slots := array{
                overlay_slot:
                    Widget := CreateStaticImage(MenuConfig.BackgroundImage, ?UniformScale := MenuConfig.BackgroundImageScale)
                    HorizontalAlignment := horizontal_alignment.Center
                    VerticalAlignment := vertical_alignment.Center
                    Padding := margin{
                        Top := MenuConfig.BackgroundImageOffset.Y
                        Left := MenuConfig.BackgroundImageOffset.X
                    }
                overlay_slot:
                    Widget := StyleInfoWidget
                    HorizontalAlignment := horizontal_alignment.Center
                    VerticalAlignment := vertical_alignment.Center
                    Padding := margin{
                        Top := 150.0
                    }
                overlay_slot:
                    Widget := ControlsWidget
                    HorizontalAlignment := horizontal_alignment.Center
                    VerticalAlignment := vertical_alignment.Center
            }

        TryUpdate()
        GetSimpleCanvas(Widget, ?Offsets := Some(margin{
            Top := 60.0
        }))

    OnViewEnter<override>(Player:player, SingleViewer:logic):void=
        if (
            HudAccessControl := Device.p_HideHudAccessControl?
            HudAccessControl.IsAllowed[Player]
            set m_PrevHudState[Player] = void
        ) {
            HudAccessControl.RemoveFromWhitelist(Player, false)
        }

        if (PreviewCamera := Device.p_CategoryInstances[m_ActiveCameraIndex].PreviewCamera) {
            PreviewCamera.AddTo(Player)
        }

    OnViewLeave<override>(Player:player):void=
        RemoveRes := m_PrevHudState.RemoveByKey(Player)
        set m_PrevHudState = RemoveRes(0)

        if (PreviewCamera := Device.p_CategoryInstances[m_ActiveCameraIndex].PreviewCamera) {
            PreviewCamera.RemoveFrom(Player)
        }
        if (RemoveRes(1)?, HudAccessControl := Device.p_HideHudAccessControl?) {
            HudAccessControl.AddToWhitelist(Player)
        }

        # reset active styles to what the player actually owns
        if (m_PrevHudState.Length = 0) {
            Owner := player[Device.p_AccessControl.ListAllowedAgents()[0]] or Player
            Device.SyncPlayer(Owner)
            set m_ActiveStyleIndex = -1
        }

    OnLeftClick<private>(Msg:widget_message):void=
        CycleStyle(-1)

    OnRightClick<private>(Msg:widget_message):void=
        CycleStyle(1)

    CycleStyle<private>(Delta:int):void=
        Category := if (l_Category := Device.p_Model.p_Categories[m_ActiveCategoryIndex]) {
            l_Category
        } else {
            return
        }

        StyleCount := Category.Styles.Length
        var w_StyleIndex: int = Max(0, m_ActiveStyleIndex) + Delta
        if (w_StyleIndex < 0) {
            set w_StyleIndex = StyleCount - DefaultingMod(-w_StyleIndex, StyleCount)
        } else if (w_StyleIndex >= StyleCount) {
            set w_StyleIndex = DefaultingMod(w_StyleIndex, StyleCount)
        }
        set m_ActiveStyleIndex = w_StyleIndex

        if (
            Style := Category.Styles[m_ActiveStyleIndex]
            Player := player[Device.p_AccessControl.ListAllowedAgents()[0]]
            Device.p_Model.p_AttachmentsManager.IsUnlocked[Player, Category.AttachmentKindId, Style.Attachment]
        ) {
            Device.p_Model.p_AttachmentsManager.SetActiveAttachment(Player, Category.AttachmentKindId, Some(Style.Attachment))
        }

        Update()

home_configurator_category_button_closure := class<final>:
    Menu: home_configurator_menu_component
    CategoryIndex: int

    OnClick(__:widget_message):void=
        set Menu.m_ActiveCategoryIndex = CategoryIndex
        set Menu.m_ActiveStyleIndex = -1
        Menu.Update()
## }}}
