using { /Fortnite.com/Devices }
using { /Verse.org/Simulation }
using { /UnrealEngine.com/Temporary/Diagnostics }
using { /UnrealEngine.com/Temporary/SpatialMath }
using { /UnrealEngine.com/Temporary/UI }

using { HavocVerse.Debug }
using { HavocVerse.UI }
using { HavocVerse.Util }

hud_img_frame_config := class<final><concrete> {
	@editable Image: string = ""
	@editable Duration: float = 1.0 # seconds
}

hud_img_player_state := class<final> {
	Canvas: canvas
	AbortEvent: event() = event(){}
}

hud_img_sequence_device := class<final><concrete>(creative_device) {
	@editable p_BeginTrigger: []trigger_device = array{}
	@editable p_StopTrigger: []trigger_device = array{}

	@editable p_Frames: []hud_img_frame_config = array{}

    @editable p_ImagePlacement: ui_canvas_placement_config = ui_canvas_placement_config{}
	@editable p_ImageSize: vector2 = vector2{X := 1920.0, Y := 1080.0}
    @editable p_ImageZOrder: type{_X:int where 0 <= _X, _X <= 2147483647} = 0

	var m_Playing: [player]hud_img_player_state = map{}

	### DOMAIN
	OnBegin<override>()<suspends>: void ={
		for (Trigger:p_BeginTrigger) {
			Trigger.TriggeredEvent.Subscribe(OnBeginTrigger)
		}
		for (Trigger:p_StopTrigger) {
			Trigger.TriggeredEvent.Subscribe(OnStopTrigger)
		}
	}

	OnBeginTrigger<private>(Agent:?agent): void = if (Player := player[Agent?]) {
		Start(Player, ?Restart := true)
	}

	OnStopTrigger<private>(Agent:?agent): void = if (Player := player[Agent?]) {
		Stop(Player)
	}

	PlaySequence<private>(Player:player, Placeholder:placeholder_widget, AbortEvent:event())<suspends>: void ={
		race:
			AbortEvent.Await()
			block:
				for (Frame:p_Frames) {
					ImageWidget := CreateStaticImage(Frame.Image, p_ImageSize)
					Placeholder.SetWidget(ImageWidget)
					Sleep(Frame.Duration)
				}
	}

	### API
	# Begins the image sequence for a player.
	Start<public>(Player:player, ?Restart:logic=false): void ={
		if (m_Playing[Player]) {
			if (Restart?) {
				Stop(Player)
			} else {
				return
			}
		} else if (p_Frames.Length = 0) {
			return
		}

		Placeholder := placeholder_widget{}
		Canvas := canvas:
			Slots := array{
				p_ImagePlacement.ToCanvasSlot(Placeholder.ParentWidget())
			}
		State := hud_img_player_state{
			Canvas := Canvas
		}
		if (PlayerUI := GetPlayerUI[Player], set m_Playing[Player] = State) {
			spawn:
				PlaySequence(Player, Placeholder, State.AbortEvent)
			PlayerUI.AddWidget(Canvas, player_ui_slot{
				ZOrder := p_ImageZOrder
			})
		}
	}

	# Stops the image sequence for a player.
	Stop<public>(Player:player): void ={
		RemoveRes := m_Playing.RemoveByKey(Player)
		set m_Playing = RemoveRes(0)
		if (State := RemoveRes(1)?) {
			State.AbortEvent.Signal()
			if (PlayerUI := GetPlayerUI[Player]) {
				PlayerUI.RemoveWidget(State.Canvas)
			}
		}
	}
}
