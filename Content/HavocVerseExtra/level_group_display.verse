using { /Fortnite.com/Devices }
using { /Fortnite.com/Characters }
using { /Fortnite.com/Vehicles }
using { /Fortnite.com/UI }
using { /UnrealEngine.com/Temporary/Diagnostics }
using { /UnrealEngine.com/Temporary/SpatialMath }
using { /UnrealEngine.com/Temporary/UI }
using { /Verse.org/Colors }
using { /Verse.org/Simulation }

using { HavocVerse.Debug }
using { HavocVerse.Devices }
using { HavocVerse.UI }
using { HavocVerse.Util }
using { HavocVerse.Logging }

level_group_vfx_config := class<final><concrete>:
    @editable Level: int = 0
    @editable VFX: visual_effect_powerup_device = visual_effect_powerup_device{}

level_group_display_hud_config := class<final><concrete>:
    @editable ScalePadding: logic = true

    @editable LevelImageScale: float = 0.5
    @editable LevelImagePath: string = "UI/Levels/level_%Level%"
    @editable LevelImageFallback: ?string = false

    @editable LevelLabelTemplate: string = "Level %Level%"
    @editable LevelLabelShowFromLevel: int = 0
    @editable LevelLabelTextStyle: ui_text_style_config = ui_text_style_config{}

    @editable ProgressBarImageScale: float = 0.5
    @editable ProgressBarBackground: string = "UI/Levels/progress_bar_background"
    @editable ProgressBarPadding: ui_margin_config = ui_margin_config{
        Left := 10.0
        Top := 10.0
        Right := 14.0
        Bottom := 14.0
    }
    @editable ProgressBarFill: ?string = option{"UI/Levels/progress_bar_fill"}
    @editable ProgressBarFillColor: color = NamedColors.White
    @editable ProgressBarFillStart: ?string = option{"UI/Levels/progress_bar_fill_start"}
    @editable ProgressBarFillEnd: ?string = option{"UI/Levels/progress_bar_fill_end"}
    @editable ProgressBarLabelTemplate: ?string = option{"%Current%/%Total%"}
    @editable ProgressBarLabelNumberFormatter: number_formatter_config = number_formatter_config{}
    @editable ProgressBarLabelTextStyle: ui_text_style_config = ui_text_style_config{}

level_group_display_device := class<concrete>(creative_device, aware_device, player_level_hook):
    @editable p_LevelGroupsManager: level_groups_manager_device = level_groups_manager_device{}
    @editable p_LevelGroupId: string = ""

    @editable p_VFXPerLevel: []level_group_vfx_config = array{}
    @editable p_VFXPerLevelDefaultLast: logic = true
    @editable p_VFXPerLevelClear: ?visual_effect_powerup_device = false
    @editable p_HideVFXWhileInVehicle: logic = true

    @editable p_HUDEnabled: logic = true
    @editable p_HUDConfig: level_group_display_hud_config = level_group_display_hud_config{}
    @editable p_HUDAccessControl: ?access_control_device = false
    @editable p_HUDManagerConfig: hud_manager_base_config = hud_manager_base_config{
        RefreshRate := 1.0
    }
    @editable p_HUDReactiveRefresh: logic = true

    @editable p_HUDLevelImage: ?ui_canvas_placement_config = false
    @editable p_HUDLevelLabel: ?ui_canvas_placement_config = false
    @editable p_HUDProgressBar: ?ui_canvas_placement_config = false

    var m_HudManager: ?level_group_display_hud_manager = false
    var m_HideForPlayers: [player]void = map{}

    DeviceMeta<override>()<transacts>:aware_device_meta=
        aware_device_meta{
            DeviceName := "level_group_display"
            DevId := p_LevelGroupId
        }

    OnBegin<override>()<suspends>:void=
        if (p_LevelGroupId.Length = 0) {
            Log(Self, "p_LevelGroupId is not set!", ?Level := log_level.Error)
            return
        }

        AwareInitDevice(Self, OnInit)
    
    OnInit<private>()<suspends>:void=
        p_LevelGroupsManager.RegisterHook(Self, ?GroupId := p_LevelGroupId)

        if (p_HUDEnabled?, p_HUDLevelImage? or p_HUDLevelLabel? or p_HUDProgressBar?) {
            HudManager := level_group_display_hud_manager{
                AccessControl := p_HUDAccessControl
                RefreshRate := p_HUDManagerConfig.RefreshRate
                ZOrder := p_HUDManagerConfig.ZOrder
                Device := Self
            }
            set m_HudManager = Some(HudManager)
            spawn:
                HudManager.Run(GetPlayspace())

            if (p_HUDReactiveRefresh?) {
                p_LevelGroupsManager.SubscribeToPlayerUpdateEvent(OnPlayerExpStateChanged)
            }
        }

        if (p_HideVFXWhileInVehicle?, p_VFXPerLevel.Length > 0) {
            GetPlayspace().PlayerRemovedEvent().Subscribe(OnPlayerRemoved)

            spawn:
                StartVFXHideCheckLoop()
        }

    OnPlayerRemoved<private>(Player:player):void=
        set m_HideForPlayers = m_HideForPlayers.RemoveByKey(Player)(0)

    StartVFXHideCheckLoop<private>()<suspends>:void=
        loop {
            Sleep(0.1)
            for (Player:GetPlayspace().GetPlayers(), FortCharacter := Player.GetFortCharacter[]) {
                if (FortCharacter.GetVehicle[]) {
                    if (
                        not m_HideForPlayers[Player]
                        set m_HideForPlayers[Player] = void
                    ) {
                        VFXHide(Player)
                    }
                } else if (m_HideForPlayers[Player]) {
                    set m_HideForPlayers = m_HideForPlayers.RemoveByKey(Player)(0)
                    VFXShow(Player)
                }
            }
        }

    VFXShow<private>(Player:player):void=
        Level := p_LevelGroupsManager.GetLevel(Player, p_LevelGroupId)
        HandlePlayerLevel(Player, p_LevelGroupId, None(), Level)
        Log(Self, "Showing level VFX for '{Player}' after leaving vehicle", ?Level := log_level.Debug)

    VFXHide<private>(Player:player):void=
        if (VFXClear := p_VFXPerLevelClear?) {
            VFXClear.Pickup(Player)
        }
        Log(Self, "Hiding level VFX for '{Player}' because they entered a vehicle", ?Level := log_level.Debug)

    OnPlayerExpStateChanged<private>(Player:player, LevelExpPair:level_exp_pair):void=
        if (HudManager := m_HudManager?, Player.IsActive[]) {
            HudManager.Refresh(Player)
        }

    HandlePlayerLevel<override>(Player:player, GroupId:string, PrevLevel:?int, Level:int):void=
        if (GroupId <> p_LevelGroupId) {
            return
        } else if (m_HideForPlayers[Player]) {
            return
        }

        if (l_PrevLevel := PrevLevel?, VFXClear := p_VFXPerLevelClear?) {
            VFXClear.Pickup(Player)
        }

        var VFXLevel: ?int = None()
        for (Config:p_VFXPerLevel, Config.Level = Level) {
            Config.VFX.Pickup(Player)
            set VFXLevel = Some(Level)
        }
        if (not VFXLevel?, p_VFXPerLevelDefaultLast?) {
            var It: int = Level - 1
            loop {
                if (It <= 0 or VFXLevel?) {
                    break
                }
                for (Config:p_VFXPerLevel, Config.Level = It) {
                    Config.VFX.Pickup(Player)
                    set VFXLevel = Some(Level)
                }
                set It -= 1
            }
        }

        NewStateStr := if (l_VFXLevel := VFXLevel?) {
            "Level {l_VFXLevel}"
        } else {
            "None"
        }
        Log(Self, MsgHandlePlayerLevel(Player, NewStateStr), ?Level := log_level.Verbose)

    MsgHandlePlayerLevel<private><localizes>(Agent:agent, NewStateStr:string):message=
        "Updated level display VFX for '{Agent}'. Applied VFX: {NewStateStr}"

## HUD {{{
level_group_display_hud_manager := class<final>(hud_manager_base):
    Device: level_group_display_device

    CreateHUD<override>(Player:player):?hud_manager_player=
        var Elements: []tuple(ui_component,ui_canvas_placement_config) = array{}
        if (Placement := Device.p_HUDLevelImage?) {
            Component := level_group_display_ui_level_image_component{
                Device := Device
                Player := Player
            }
            set Elements += array{(Component, Placement)}
        }
        if (Placement := Device.p_HUDLevelLabel?) {
            Component := level_group_display_ui_level_label_component{
                Device := Device
                Player := Player
            }
            set Elements += array{(Component, Placement)}
        }
        if (Placement := Device.p_HUDProgressBar?) {
            Component := level_group_display_ui_progress_bar_component{
                Device := Device
                Player := Player
            }
            set Elements += array{(Component, Placement)}
        }

        Some(MakeUIFrameworkHUDPlayer(Player, Elements))

### Level Image {{{
level_group_display_ui_level_image_component := class<final>(ui_component):
    Device: level_group_display_device
    Player: player

    var m_LastLevel: ?int = false

    TryUpdate<override>():logic=
        Level := Device.p_LevelGroupsManager.GetLevel(Player, Device.p_LevelGroupId)
        if (m_LastLevel? <> Level) {
            false
        } else {
            true
        }

    Create<override>():widget=
        Level := Device.p_LevelGroupsManager.GetLevel(Player, Device.p_LevelGroupId)
        Texture := Device.p_HUDConfig.LevelImagePath
            .ReplaceAll("%Level%", ToString(Level))
        set m_LastLevel = Some(Level)
        CreateStaticImage(Texture, ?UniformScale := Device.p_HUDConfig.LevelImageScale)
### }}}

### Level Label {{{
level_group_display_ui_level_label_component := class<final>(ui_component):
    Device: level_group_display_device
    Player: player

    var m_LastLevel: ?int = false
    var m_TextBlock: ?text_block = false

    TryUpdate<override>():logic=
        Level := Device.p_LevelGroupsManager.GetLevel(Player, Device.p_LevelGroupId)
        if (m_LastLevel? = Level) {
            return true
        }
        if (TextBlock := m_TextBlock?) {
            set m_LastLevel = Some(Level)
            Text := if (Level >= Device.p_HUDConfig.LevelLabelShowFromLevel) {
                Device.p_HUDConfig.LevelLabelTemplate
                    .ReplaceAll("%Level%", ToString(Level))
            } else {
                ""
            }
            TextBlock.SetText(ToMessage(Text))
            true
        } else {
            false
        }

    Create<override>():widget=
        TextBlock := Device.p_HUDConfig.LevelLabelTextStyle.ToTextBlock("")
        set m_TextBlock = Some(TextBlock)
        TextBlock
### }}}

### Progress Bar {{{
level_group_display_ui_progress_bar_component_state := class<final>:
    FillTail: ?tuple(widget,float)
    FillHead: ?tuple(widget,float)
    FillWidget: widget
    Width: float

level_group_display_ui_progress_bar_component := class<final>(ui_component):
    Device: level_group_display_device
    Player: player

    var m_State: ?level_group_display_ui_progress_bar_component_state = false
    var m_Label: ?cached_text_block = false

    TryUpdate<override>():logic=
        LevelExpPair := Device.p_LevelGroupsManager.GetLevelExp(Player, Device.p_LevelGroupId)
        ExpectedExp := Device.p_LevelGroupsManager.GetTargetExp[Device.p_LevelGroupId, LevelExpPair] or -1

        if (State := m_State?) {
            Progress := Clamp(ToFloat(LevelExpPair.ExpRest) / ToFloat(ExpectedExp), 0.0, 1.0)

            var TargetWidth: float = State.Width * Progress
            if (Tail := State.FillTail?) {
                Tail(0).SetVisibility(if (TargetWidth > 0.0) {
                    widget_visibility.Visible
                } else {
                    widget_visibility.Collapsed
                })
                set TargetWidth -= Tail(1)
            }
            if (Head := State.FillHead?) {
                Head(0).SetVisibility(if (TargetWidth > 0.0) {
                    widget_visibility.Visible
                } else {
                    widget_visibility.Collapsed
                })
                set TargetWidth -= Head(1)
            }

            FillWidth := Max(TargetWidth, 0.0)
            if (ColorBlock := color_block[State.FillWidget]) {
                Size := ColorBlock.GetDesiredSize().WithX(FillWidth)
                ColorBlock.SetDesiredSize(Size)
            } else if (TextureBlock := texture_block[State.FillWidget]) {
                Size := TextureBlock.GetDesiredSize().WithX(FillWidth)
                TextureBlock.SetDesiredSize(Size)
            }
        }
        if (Label := m_Label?, Template := Device.p_HUDConfig.ProgressBarLabelTemplate?) {
            Formatter := Device.p_HUDConfig.ProgressBarLabelNumberFormatter
            Text := Template
                .ReplaceAll("%Level%", Formatter.Format(LevelExpPair.Level))
                .ReplaceAll("%Current%", Formatter.Format(LevelExpPair.ExpRest))
                .ReplaceAll("%Total%", Formatter.Format(ExpectedExp))
            Label.SetText(Text)
        }
        true

    Create<override>():widget=
        Config := Device.p_HUDConfig
        ImageScale := Config.ProgressBarImageScale

        Background := CreateStaticImage(Config.ProgressBarBackground, ?UniformScale := ImageScale)
        Width := GetFixedWidgetDimensions(Background).X

        ProgressBar := stack_box:
            Orientation := orientation.Horizontal

        FillTail := if (Texture := Config.ProgressBarFillStart?) {
            Image := CreateStaticImage(Texture, ?UniformScale := ImageScale)
            ProgressBar.AddWidget(stack_box_slot{
                Widget := Image
                HorizontalAlignment := horizontal_alignment.Left
                VerticalAlignment := vertical_alignment.Center
            })
            Some(Image, GetFixedWidgetDimensions(Image).X)
        } else {
            None()
        }

        FillWidget := if (Texture := Config.ProgressBarFill?) {
            CreateStaticImage(Texture, ?UniformScale := ImageScale)
        } else {
            color_block{
                DefaultColor := Config.ProgressBarFillColor
                DefaultDesiredSize := vector2{X := 0.0, Y := 1.0}
            }
        }
        ProgressBar.AddWidget(stack_box_slot{
            Widget := FillWidget
            HorizontalAlignment := horizontal_alignment.Left
            VerticalAlignment := vertical_alignment.Center
        })

        FillHead := if (Texture := Config.ProgressBarFillEnd?) {
            Image := CreateStaticImage(Texture, ?UniformScale := ImageScale)
            ProgressBar.AddWidget(stack_box_slot{
                Widget := Image
                HorizontalAlignment := horizontal_alignment.Left
                VerticalAlignment := vertical_alignment.Center
            })
            Some(Image, GetFixedWidgetDimensions(Image).X)
        } else {
            None()
        }

        Label := if (Template := Config.ProgressBarLabelTemplate?) {
            TextBlock := MakeCachedTextBlock of Config.ProgressBarLabelTextStyle.ToTextBlock("")
            Some(TextBlock)
        } else {
            None()
        }

        set m_State = Some(level_group_display_ui_progress_bar_component_state{
            FillTail := FillTail
            FillHead := FillHead
            FillWidget := FillWidget
            Width := Width
        })
        set m_Label = Label
        TryUpdate()

        Overlay := overlay:
            Slots := array{
                overlay_slot:
                    Widget := Background
                    HorizontalAlignment := horizontal_alignment.Fill
                    VerticalAlignment := vertical_alignment.Fill
                overlay_slot:
                    Widget := ProgressBar
                    HorizontalAlignment := horizontal_alignment.Fill
                    VerticalAlignment := vertical_alignment.Fill
                    Padding := if (Config.ScalePadding?) {
                        Config.ProgressBarPadding.ToMargin(?UniformScale := ImageScale)
                    } else {
                        Config.ProgressBarPadding.ToMargin()
                    }
            }
        if (l_Label := Label?) {
            Overlay.AddWidget(overlay_slot{
                Widget := l_Label.Widget
                HorizontalAlignment := horizontal_alignment.Center
                VerticalAlignment := vertical_alignment.Center
            })
        }
        Overlay
### }}}
## }}}
