using { /Fortnite.com/Devices }
using { /Verse.org/Simulation }
using { /UnrealEngine.com/Temporary/Diagnostics }
using { /UnrealEngine.com/Temporary/SpatialMath }

using { HavocVerse.Debug }
using { HavocVerse.Logging }
using { HavocVerse.Util }

real_time_rig_source_device := class(creative_device, aware_device, real_time_provider):
	@editable p_RigProp: creative_prop = creative_prop{}
	@editable p_SelfIsOrigin: logic = true

	DeviceMeta<override>()<transacts>:aware_device_meta=
		aware_device_meta{
			DeviceName := "real_time_source"
		}

	OnBegin<override>()<suspends>:void=
		AwareInitDevice(Self, OnInit)
	
	OnInit<private>()<suspends>:void=
		if (not p_RigProp.IsValid[]) {
			Log(Self, "Rig prop is not set or invalid. Real time will not be exposed", ?Level := log_level.Error)
		}

	UtcNow<override>():date_time=
		GetDateTime[] or date_time{
			Year := 1970
			Month := 0
			Day := 0
			Hour := 0
			Minute := 0
			Second := 0.0
		}

	GetDateTime<private>()<decides><transacts>:date_time=
		p_RigProp.IsValid[]
		RigTransform := p_RigProp.GetTransform()
		Translation := if (p_SelfIsOrigin?) {
			RigTransform.Translation - GetTransform().Translation
		} else {
			RigTransform.Translation
		}
		date_time{
			Year := Int[Translation.X]
			Month := Int[Translation.Y]
			Day := Int[Translation.Z]
			Hour := Int[RigTransform.Scale.X]
			Minute := Int[RigTransform.Scale.Y]
			Second := RigTransform.Scale.Z
		}
