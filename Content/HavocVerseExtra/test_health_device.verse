using { /Fortnite.com/Devices }
using { /Fortnite.com/Characters }
using { /Verse.org/Simulation }
using { /UnrealEngine.com/Temporary/Diagnostics }

test_health_device := class<final><concrete>(creative_device) {
	@editable p_HealthPlusButton: button_device = button_device{}
	@editable p_HealthMinusButton: button_device = button_device{}
	@editable p_ShieldPlusButton: button_device = button_device{}
	@editable p_ShieldMinusButton: button_device = button_device{}

	@editable p_HealthDelta: float = 5.0
	@editable p_ShieldDelta: float = 10.0

	OnBegin<override>()<suspends>: void ={
		p_HealthPlusButton.InteractedWithEvent.Subscribe(OnHealthPlusPress)
		p_HealthMinusButton.InteractedWithEvent.Subscribe(OnHealthMinusPress)
		p_ShieldPlusButton.InteractedWithEvent.Subscribe(OnShieldPlusPress)
		p_ShieldMinusButton.InteractedWithEvent.Subscribe(OnShieldMinusPress)
	}

	Modify<private>(Agent:agent, Health:logic, Delta:float): void ={
		if (FortCharacter := Agent.GetFortCharacter[]) {
			if (Health?) {
				FortCharacter.SetHealth(Clamp(FortCharacter.GetHealth() + Delta, 0.0, FortCharacter.GetMaxHealth()))
			} else {
				FortCharacter.SetShield(Clamp(FortCharacter.GetShield() + Delta, 0.0, FortCharacter.GetMaxShield()))
			}
		}
	}

	OnHealthPlusPress<private>(Agent:agent): void ={
		Modify(Agent, true, p_HealthDelta)
	}

	OnHealthMinusPress<private>(Agent:agent): void ={
		Modify(Agent, true, -p_HealthDelta)
	}

	OnShieldPlusPress<private>(Agent:agent): void ={
		Modify(Agent, false, p_ShieldDelta)
	}

	OnShieldMinusPress<private>(Agent:agent): void ={
		Modify(Agent, false, -p_ShieldDelta)
	}
}