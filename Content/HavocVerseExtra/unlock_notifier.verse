using { /Fortnite.com/Devices }
using { /Verse.org/Simulation }
using { /UnrealEngine.com/Temporary/Diagnostics }

using { HavocVerse.Debug }
using { HavocVerse.Devices }
using { HavocVerse.Devices.GameSpecific }
using { HavocVerse.Util }
using { HavocVerse.Logging }

unlock_notifier_idx_mapping := class<abstract>:
    @editable ImageIndex: int = -1
    @editable DirectHUDMessage: ?hud_message_device = false

unlock_notifier_pet_idx_mapping := class<final><concrete>(unlock_notifier_idx_mapping):
    @editable PetId: string = ""

unlock_notifier_attachment_idx_mapping := class<final><concrete>(unlock_notifier_idx_mapping):
    @editable Attachment: string = ""

unlock_notifier_device := class<final><concrete>(creative_device, aware_device):
    @editable p_DevId: string = "N/A"

    @editable p_NotifyController: notify_controller_device = notify_controller_device{}

    @editable p_PetController: ?pet_controller_device = false
    @editable p_PetMappings: []unlock_notifier_pet_idx_mapping = array{}
    @editable p_PetNotifyTemplateId: string = "unlock_pet"

    @editable p_TycoonManager: ?tycoon_manager_device = false
    @editable p_TycoonIgnoreVirtualRebirths: logic = true

    @editable p_AttachmentsManager: ?attachments_manager_device = false
    @editable p_AttachmentKindId: string = ""
    @editable p_AttachmentMappings: []unlock_notifier_attachment_idx_mapping = array{}
    @editable p_AttachmentNotifyTemplateId: string = "unlock_default"

    DeviceMeta<override>()<transacts>:aware_device_meta=
        aware_device_meta{
            DeviceName := "unlock_notifier"
            DevId := p_DevId
        }

    OnBegin<override>()<suspends>:void=
        AwareInitDevice(Self, OnInit)
    
    OnInit<private>()<suspends>:void=
        if (p_PetMappings.Length > 0, not p_PetController?) {
            Log(Self, "Pet mappings are configured, but p_PetController is not set", ?Level := log_level.Error)
        }
        if (p_PetMappings.Length > 0, PetsController := p_PetController?) {
            PetsController.SubscribeToUnlockEvent(OnPetUnlocked)
        }

        if (TycoonManager := p_TycoonManager?) {
            TycoonManager.SubscribeToRebirthEvent(OnTycoonRebirth)
        }

        if (p_AttachmentMappings.Length > 0, not p_AttachmentsManager?) {
            Log(Self, "Attachment mappings are configured, but p_AttachmentsManager is not set", ?Level := log_level.Error)
        }
        if (p_AttachmentMappings.Length > 0, AttachmentsManager := p_AttachmentsManager?) {
            AttachmentsManager.AddUnlockListener(OnAttachmentUnlocked)
        }

    Notify<private>(Player:player, TemplateId:string, Mapping:unlock_notifier_idx_mapping):void=
        if (HUDMessage := Mapping.DirectHUDMessage?) {
            # custom override
            HUDMessage.Show(Player)
        } else {
            p_NotifyController.Notify(Player, ?TemplateId := TemplateId,
                ?ImageIndex := Some(Mapping.ImageIndex))
        }

    OnPetUnlocked<private>(Player:player, PetId:string):void=
        for (Mapping:p_PetMappings, Mapping.PetId = PetId) {
            Notify(Player, p_PetNotifyTemplateId, Mapping)
            return
        }

    OnAttachmentUnlocked<private>(Event:attachment_unlocked_event):void=
        if (Event.AttachmentKindId <> p_AttachmentKindId) {
            return
        }
        for (Mapping:p_AttachmentMappings, Mapping.Attachment = Event.Attachment) {
            Notify(Event.Player, p_AttachmentNotifyTemplateId, Mapping)
            return
        }

    OnTycoonRebirth<private>(Event:tycoon_rebirth_event):void=
        if (Event.Virtual?, p_TycoonIgnoreVirtualRebirths?) {
            return
        }
        if (TycoonManager := p_TycoonManager?) {
            for (
                RebirthPetConfig:TycoonManager.p_RebirthPets
                RebirthPetConfig.RebirthLevel = Event.NewRebirthLevel
            ) {
                OnPetUnlocked(Event.Player, RebirthPetConfig.PetId)
            }
            for (
                RebirthAttachmentConfig:TycoonManager.p_RebirthAttachments
                RebirthAttachmentConfig.RebirthLevel = Event.NewRebirthLevel
            ) {
                OnAttachmentUnlocked(attachment_unlocked_event{
                    Player := Event.Player
                    AttachmentKindId := RebirthAttachmentConfig.AttachmentKindId
                    Attachment := RebirthAttachmentConfig.Attachment
                })
            }
        }
