using { /Verse.org/Simulation }

agent_callback<public> := type{__(:agent):void} 
maybe_agent_callback<public> := type{__(:?agent):void} 

(HavocVerse:)Combat<public> := module:
(HavocVerse:)Debug<public> := module:
(HavocVerse:)Devices<public> := module:
    (HavocVerse.Devices:)GameSpecific<public> := module:
(HavocVerse:)JSON<public> := module:
(HavocVerse:)Math<public> := module:
(HavocVerse:)Persistence<public> := module:
(HavocVerse:)Resource<public> := module:
(HavocVerse:)Storage<public> := module:
(HavocVerse:)UI<public> := module:
    (HavocVerse.UI:)Defaults<public> := module:
(HavocVerse:)Util<public> := module:
    (HavocVerse.Util:)Task<public> := module:

(HavocVerse:)Logging<public> := module:
    using { /UnrealEngine.com/Temporary/Diagnostics }

    havoc_log_chan<public> := class(log_channel) {}
