using { /Verse.org/Assets }
using { /Verse.org/Simulation/Tags }
using { /UnrealEngine.com/Temporary/SpatialMath }

texture_metadata<public> := class:
    Name<public>: string
    Asset<public>: texture
    Dimensions<public>: ?vector2

# Abstraction to provide access to assets by name.
assets_provider<public> := interface:
    # Returns all texture assets with their metadata.
    GetTextureMetaAssets<public>()<transacts>:[]texture_metadata

    # Gets a texture asset by name.
    GetTextureAsset<public>(Name:string)<transacts><decides>:texture

    # Gets a texture asset's default dimensions by name.
    GetTextureAssetDimensions<public>(Name:string)<transacts><decides>:vector2

    # Gets a material asset by name.
    GetMaterialAsset<public>(Name:string)<transacts><decides>:material

    # Gets a mesh asset by name.
    GetMeshAsset<public>(Name:string)<transacts><decides>:mesh

    # Gets an animation sequence by name.
    GetAnimationSequenceAsset<public>(Name:string)<transacts><decides>:animation_sequence

    # Gets a particle system by name.
    GetParticleSystemAsset<public>(Name:string)<transacts><decides>:particle_system

# Device that can create an assets provider.
assets_provider_device<public> := interface:
    CreateAssetsProvider<public>():assets_provider

# Abstraction to provide access to tags by name.
tags_provider<public> := interface:
    # Gets a tag by name.
    GetTag<public>(Name:string)<transacts><decides>:tag
