# Returns the bits of a number as an array of booleans.
ToBits<public>(Val:int, ?PadLength:int = 0)<transacts>:[]logic=
    var WorkingVal: int = Val
    var Result: []logic = array{}
    loop {
        if (WorkingVal <= 0) {
            break
        }
        Bit := if (Mod[WorkingVal, 2] = 1). true else false
        set Result += array{Bit}
        set WorkingVal = DivDefaultingFloor(WorkingVal, 2)
    }
    MaxIndex := Max(PadLength, Result.Length)-1
    for (Index := 0..MaxIndex) {
        if (Bit := Result[MaxIndex-Index]) {
            Bit
        } else {
            false
        }
    }

# Returns the number represented by the bits from an array of booleans.
FromBits<public>(Bits:[]logic)<transacts>:int=
    var Result: int = 0
    var CurrentValue: int = 1
    MaxIndex := Bits.Length-1
    for (Index := 0..MaxIndex) {
        if (Index > 0) {
            set CurrentValue *= 2
        }
        if (Bits[MaxIndex-Index]?) {
            set Result += CurrentValue
        }
    }
    Result

# Returns a string representation of a bit array.
BitsToString<public>(Val:[]logic)<computes>:string=
    for (Bit:Val) {
        if (Bit?) {
            '1'
        } else {
            '0'
        }
    }

Min<public>(A:logic, B:logic):logic=
    if (A = false or B = false) {
        false
    } else {
        true
    }

Max<public>(A:logic, B:logic):logic=
    if (A = true or B = true) {
        true
    } else {
        false
    }

# Returns the number of required bits to store a number.
BitsLength<public>(Val:int)<transacts>:int=
    if (Val = 0) {
        return 1
    }
    var WorkingVal: int =  if (Val < 0) {
        -Val
    } else {
        Val
    }
    var BitCount: int = 0
    loop {
        if (WorkingVal > 0, Res := Floor(WorkingVal / 2)) {
            set BitCount += 1
            set WorkingVal = Res
        } else {
            break
        }
    }

    # sanity check because Floor is broken in Verse sometimes...
    if (FromBits(array{true}.Repeat(BitCount)) >= Val) {
        BitCount
    } else {
        set BitCount = 1
        loop {
            if (FromBits(array{true}.Repeat(BitCount)) >= Val) {
                break
            }
            set BitCount += 1
        }
        BitCount
    }

InvertLogic<public>(Val:logic)<transacts>:logic=
    if (Val?). false else true

LogicToInt<public>(Val:logic)<transacts>:int=
    if (Val?). 1 else 0

IntToLogic<public>(Val:int)<transacts>:logic=
    if (Val = 0). false else true
