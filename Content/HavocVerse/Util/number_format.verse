number_format_style<public> := enum:
    Literal
    ThousandsSeparator
    KMBT
    KMGT
    TenKMBT
    BT

number_format_options<public> := class<concrete>:
    MinDecimals<public>: int = 0
    MinDecimalsForShortened<public>: int = 0
    MaxDecimals<public>: int = 1

FormatNumber<public>(Val:int, ?Style:number_format_style = number_format_style.ThousandsSeparator)<transacts>:string=
    FormatNumber(Val*1.0, Opts := number_format_options{}, ?Style := Style)

FormatNumber<public>(Val:float, ?Style:number_format_style = number_format_style.ThousandsSeparator)<transacts>:string=
    FormatNumber(Val, Opts := number_format_options{}, ?Style := Style)

FormatNumber<public>(Val:float, Opts:number_format_options, ?Style:number_format_style = number_format_style.ThousandsSeparator)<transacts>:string=
    IntPart: string = case(Style):
        number_format_style.Literal => ToString(FormattingFloor(Val))
        number_format_style.ThousandsSeparator => FormatNumberThousandsSeparated(FormattingFloor(Val))
        number_format_style.KMBT => return FormatNumberWithSuffix(Val, kMBTMagnitudes, kMBTSuffixList, Opts)
        number_format_style.KMGT => return FormatNumberWithSuffix(Val, kMBTMagnitudes, kMGTSuffixList, Opts)
        number_format_style.TenKMBT =>
            if (Val >= 10000.0) {
                return FormatNumberWithSuffix(Val, kMBTMagnitudes, kMBTSuffixList, Opts)
            }
            FormatNumberThousandsSeparated(FormattingFloor(Val))
        number_format_style.BT =>
            if (Val >= BTMagnitudes[BTMagnitudes.Length-1]) {
                return FormatNumberWithSuffix(Val, BTMagnitudes, BTSuffixList, Opts)
            }
            FormatNumberThousandsSeparated(FormattingFloor(Val))
    Decimals := (Val - ToFloat(FormattingFloor(Val))) * Pow(10.0, Max(Opts.MinDecimals, Opts.MaxDecimals) * 1.0)
    if ((Opts.MinDecimals > 0) or (Decimals > 0.0 and Opts.MaxDecimals > 0)) {
        var DecimalPart: string = ToString(FormattingFloor(Decimals))
        loop {
            if (DecimalPart.Length >= Opts.MinDecimals) {
                break
            }
            set DecimalPart = "0" + DecimalPart
        }
        "{IntPart}.{DecimalPart}"
    } else {
        IntPart
    }

FormattingFloor<internal>(Val:float)<transacts>:int=
    DefaultingFloor(Val + 1e-13)

FormatNumberWithSuffix<internal>(Val:float, Magnitudes:[]float, SuffixList:[]string, Opts:number_format_options)<transacts>:string=
    SuffixRes := GetNumberSuffix(Val, Magnitudes, SuffixList)
    l_Opts := number_format_options{
        MinDecimals := if (SuffixRes.Magnitude > 1.0) {
            Opts.MinDecimalsForShortened
        } else {
            Opts.MinDecimals
        }
        MinDecimalsForShortened := Opts.MinDecimalsForShortened
        MaxDecimals := Opts.MaxDecimals
    }
    FormattedPrefix := FormatNumber(Val / SuffixRes.Magnitude, l_Opts, ?Style := number_format_style.Literal)
    "{FormattedPrefix}{SuffixRes.Suffix}"

FormatNumberThousandsSeparated<internal>(Val:int, ?Separator:string = ",")<transacts>:string=
    ConvertedString := ToString(Val)
    var Result: string = ""
    for (Index := 0..ConvertedString.Length-1) {
        if (Mod[ConvertedString.Length - Index, 3] = 0 and Index <> 0) {
            set Result += Separator
        }
        if (Char := ConvertedString[Index]) {
            set Result += ToString(Char)
        }
    }
    Result

kMBTMagnitudes<internal>: []float = array{1000000000000000.0, 1000000000000.0, 1000000000.0, 1000000.0, 1000.0}
kMBTSuffixList<internal>: []string = array{"P", "T", "B", "M", "k"}
kMGTSuffixList<internal>: []string = array{"P", "T", "G", "M", "k"}

BTMagnitudes<internal>: []float = array{1000000000000000.0, 1000000000000.0, 1000000000.0}
BTSuffixList<internal>: []string = array{"P", "T", "B"}

number_suffix_result<internal> := struct:
    Magnitude: float
    Suffix: string

GetNumberSuffix<internal>(Val:float, Magnitudes:[]float, SuffixList:[]string)<transacts>:number_suffix_result=
    MagnitudeIndex := FindNumberSuffixMagnitude(Magnitudes, Val)
    if (
      MagnitudeIndex >= 0
      l_Suffix := SuffixList[MagnitudeIndex]
      Magnitude := Magnitudes[MagnitudeIndex]
    ) {
        number_suffix_result{
            Magnitude := Magnitude
            Suffix := l_Suffix
        }
    } else {
        number_suffix_result{
            Magnitude := 1.0
            Suffix := ""
        }
    }

FindNumberSuffixMagnitude<internal>(Magnitudes:[]float, Val:float)<transacts>:int=
    for (Index := 0..Magnitudes.Length-1, Magnitude := Magnitudes[Index]) {
        if (Val >= Magnitude) {
            return Index
        }
    }
    -1
