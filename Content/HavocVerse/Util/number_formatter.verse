using { /Verse.org/Simulation }

using { HavocVerse.Util }

number_formatter_config<public> := class<concrete>:
    Placeholder<public>: string = "%Val%"

    @editable Template<public>: string = "%Val%"
    @editable Style<public>: number_format_style = number_format_style.ThousandsSeparator
    @editable MinDecimals<public>: int = 0
    @editable MinDecimalsForShortened<public>: int = 0
    @editable MaxDecimals<public>: int = 1

    Format<public>(Val:int)<transacts>:string=
        Format(ToFloat(Val))

    Format<public>(Val:float)<transacts>:string=
        Opts := number_format_options{
            MinDecimals := MinDecimals
            MinDecimalsForShortened := MinDecimalsForShortened
            MaxDecimals := MaxDecimals
        }
        Template.ReplaceAll(Placeholder, FormatNumber(Val, Opts, ?Style := Style))

    CreateFormatter<public>()<transacts>:number_formatter=
        number_formatter{
            Placeholder := Placeholder
            Template := Template
            Style := Style
            Opts := number_format_options{
                MinDecimals := MinDecimals
                MinDecimalsForShortened := MinDecimalsForShortened
                MaxDecimals := MaxDecimals
            }
        }

number_formatter<public> := class:
    Placeholder<public>: string
    Template<public>: string
    Style<public>: number_format_style
    Opts<public>: number_format_options

    Format<public>(Val:int)<transacts>:string=
        Format(ToFloat(Val))

    Format<public>(Val:float)<transacts>:string=
        Template.ReplaceAll(Placeholder, FormatNumber(Val, Opts, ?Style := Style))

CreateNumberFormatter<public>(Config:?number_formatter_config):number_formatter=
    if (l_Config := Config?) {
        l_Config
    } else {
        number_formatter_config{}
    }.CreateFormatter()
