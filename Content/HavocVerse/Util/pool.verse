using { /Fortnite.com/Devices }
using { /Verse.org/Simulation }
using { /Verse.org/Simulation/Tags }
using { /UnrealEngine.com/Temporary/SpatialMath }

object_pool<public> := class:
    ReturnHook<public>: ?type{__(:creative_object):void} = false

    var m_Entries: []object_pool_entry = array{}

    MemLength<public>()<transacts>:int=
        m_Entries.Length

    Add<public>(Object:creative_object):void=
        set m_Entries += array{object_pool_entry{
            Pool := Self
            Object := Object
        }}

    AddFromTag<public>(Tag:tag, Filter:type{__(:creative_object)<transacts><decides>:void}):void=
        for (
            FoundObject:GetCreativeObjectsWithTag(Tag)
            Object := creative_object[FoundObject]
            Filter[Object]
        ) {
            Add(Object)
        }

    Reserve<public>()<transacts><decides>:object_pool_entry=
        var Reserved: ?object_pool_entry = None()
        for (Entry:m_Entries, not (Reserved?), not (Entry.m_Used?)) {
            set Entry.m_Used = true
            set Reserved = Some(Entry)
        }
        Reserved?

    FilteredReserve<public>(Filter:type{__(:creative_object)<transacts><decides>:void})<transacts><decides>:object_pool_entry=
        var Reserved: ?object_pool_entry = None()
        for (Entry:m_Entries, not (Reserved?), not (Entry.m_Used?), Filter[Entry.Object]) {
            set Entry.m_Used = true
            set Reserved = Some(Entry)
        }
        Reserved?

object_pool_entry<public> := class:
    Pool<public>: object_pool
    Object<public>: creative_object

    var m_Used: logic = false

    Release<public>():void=
        if (ReturnHook := Pool.ReturnHook?) {
            ReturnHook(Object)
        }
        set m_Used = false

FilterCreativeProps<public>(Object:creative_object)<transacts><decides>:void = creative_prop[Object]
FilterAudioPlayers<public>(Object:creative_object)<transacts><decides>:void = audio_player_device[Object]

prop_pool<public> := class:
    Hide<public>: logic = true
    TeleportBack<public>: logic = true
    ScaleDownBeforeTeleport<public>: logic = true
    ReturnHook<public>: ?type{__(:creative_prop):void} = false

    var m_Entries: []prop_pool_entry = array{}

    MemLength<public>()<transacts>:int=
        m_Entries.Length

    Size<public>()<transacts>:int=
        for (Entry:m_Entries, Entry.Prop.IsValid[]) {
            Entry
        }.Length

    Add<public>(Prop:creative_prop):void=
        if (not Prop.IsValid[]) {
            return
        }
        set m_Entries += array{prop_pool_entry{
            Pool := Self
            Prop := Prop
            InitialTransform := Prop.GetTransform()
        }}
        if (Self.Hide?) {
            Prop.Hide()
        }

    AddFromTag<public>(Tag:tag):void=
        for (
            FoundObject:GetCreativeObjectsWithTag(Tag)
            Prop := creative_prop[FoundObject]
            Prop.IsValid[]
        ) {
            Add(Prop)
        }

    Reserve<public>()<transacts><decides>:prop_pool_entry=
        var Reserved: ?prop_pool_entry = None()
        for (Entry:m_Entries, not (Reserved?), not (Entry.m_Used?), Entry.Prop.IsValid[]) {
            set Entry.m_Used = true
            set Reserved = Some(Entry)
        }
        Reserved?

    # Reserves a prop from the pool, teleporting it to a given position while keeping original
    # rotation and scale.
    # If Hide is enabled, the prop will still be hidden.
    Reserve<public>(TeleportTo:vector3)<transacts><decides>:prop_pool_entry=
        Reserve0[Some(TeleportTo), None()]

    # Reserves a prop from the pool, teleporting it to a given transform.
    # If Hide is enabled, the prop will still be hidden.
    Reserve<public>(TeleportTo:transform)<transacts><decides>:prop_pool_entry=
        Reserve0[None(), Some(TeleportTo)]

    Reserve0<private>(Position:?vector3, Transform:?transform)<transacts><decides>:prop_pool_entry=
        var Reserved: ?prop_pool_entry = None()
        for (
            i_Entry:m_Entries
            not (Reserved?)
            not (i_Entry.m_Used?)
            i_Entry.Prop.IsValid[]
        ) {
            set Reserved = Some(i_Entry)
        }
        Entry := Reserved?
        if (l_Position := Position?) {
            NewTransform := Entry.InitialTransform.WithTranslation(l_Position)
            Entry.Prop.TeleportTo[NewTransform]
        } else {
            Entry.Prop.TeleportTo[Transform?]
        }
        set Entry.m_Used = true
        Entry

prop_pool_entry<public> := class:
    Pool<public>: prop_pool
    Prop<public>: creative_prop
    InitialTransform<public>: transform

    var m_Used: logic = false

    Release<public>():void=
        if ((not (m_Used?)) or (not Prop.IsValid[])) {
            return
        }
        if (Pool.Hide?) {
            Prop.Hide()
        }
        if:
            Pool.TeleportBack?
            Pool.ScaleDownBeforeTeleport?
            Prop.TeleportTo[Prop.GetTransform().WithScale(SafeScale(MakeUniformVec3(0.0)))]
        if:
            Pool.TeleportBack?
            Prop.TeleportTo[InitialTransform]
        if (ReturnHook := Pool.ReturnHook?) {
            ReturnHook(Prop)
        }
        set m_Used = false
