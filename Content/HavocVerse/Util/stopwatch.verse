using { /Verse.org/Simulation }

stopwatch<public> := class<final> {
	var m_Running: [comparable]float = map{}

	# Resets the stopwatch for all running actors, if any.
	Reset<public>(): void ={
		set m_Running = map{}
	}

	# Starts the stopwatch for an actor.
	Start<public>(Key:comparable, ?StartedAt:float=-1.0): void ={
		if (m_Running[Key], StartedAt < 0.0) {
			return
		}
		l_StartedAt := if (StartedAt < 0.0) {
			GetSimulationElapsedTime()
		} else {
			StartedAt
		}
		if. set m_Running[Key] = l_StartedAt
	}

	# Stops the stopwatch for an actor.
	# Returns the elapsed duration in seconds if it was running.
	Stop<public>(Key:comparable): ?float ={
		RemoveRes := m_Running.RemoveByKey(Key)
		set m_Running = RemoveRes(0)
		if (StartedAt := RemoveRes(1)?) {
			Some(GetSimulationElapsedTime() - StartedAt)
		} else {
			None()
		}
	}
}
