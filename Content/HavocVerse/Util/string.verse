# Checks if a string contains another string.
# If the substring is empty, it will always succeed.
# Fails if the substring is not found.
(Input:string).Contains<public>(Substr:string)<transacts><decides>:void= StringContains0(Input, Substr)?

# Internal string contains util to return early if the pattern is found.
StringContains0<internal>(Input:string, Substr:string)<transacts>:logic=
    if (SubstrStart := Substr[0]) {
        for (Index := 0..Input.Length-1, Char := Input[Index], Char = SubstrStart) {
            var FoundSubstr: logic = true
            for (SubIndex := 1..Substr.Length-1, FoundSubstr?, SubChar := Substr[SubIndex]) {
                if (Input[Index + SubIndex] <> SubChar) {
                    set FoundSubstr = false
                }
            }
            if (FoundSubstr?) {
                return true
            }
        }
        false
    } else {
        true
    }

# Checks if a string is empty. That is, if with all whitespace removed,
# it has a length greater than 0.
(Input:string).IsEmpty<public>()<decides><transacts>:void=
    Input.Length = 0 or Input.ReplaceAll(" ", "").Length = 0

# Replaces all instances of a pattern in a string by another string.
# If you want to know whether the pattern was replaced, use `TryReplace`, which is failable.
(Input:string).Replace<public>(Pattern:string, Replacement:string, ?MaxReplacements:int = 0)<transacts>:string=
    if (ReplacedStr := Input.TryReplace[Pattern, Replacement, ?MaxReplacements := MaxReplacements]) {
        ReplacedStr
    } else {
        Input
    }

try_replace_state<internal> := class<final>:
    var Result: string = ""
    var ReplacedTimes: int = 0

# Replaces all instances of a pattern in a string by another string.
# Fails if the input or pattern is empty, or if the pattern is not found.
(Input:string).TryReplace<public>(Pattern:string, Replacement:string, ?MaxReplacements:int = 0)<decides><transacts>:string=
    Input.Length > 0
    PatternStart := Pattern[0]

    State := try_replace_state{}
    UpperIndex := Input.Length-1

    var Index: int = 0
    loop {
        Char := if (l_Char := Input[Index]) {
            l_Char
        } else {
            break
        }

        if (Char = PatternStart and (MaxReplacements <= 0 or State.ReplacedTimes < MaxReplacements)) {
            var IsPattern: logic = true
            for (PatIndex := 1..Pattern.Length-1, IsPattern?, PatChar := Pattern[PatIndex]) {
                if (not (Input[Index + PatIndex] = PatChar)) {
                    set IsPattern = false
                }
            }
            if (IsPattern?) {
                set State.Result = State.Result + Replacement
                set Index = Index + Pattern.Length
                set State.ReplacedTimes = State.ReplacedTimes + 1
            } else {
                set State.Result = State.Result + ToString(Char)
            }
        } else {
            set State.Result = State.Result + ToString(Char)
        }

        if (Index < UpperIndex) {
            set Index = Index + 1
        } else {
            break
        }
    }

    State.ReplacedTimes > 0
    State.Result

# Serializes a `logic` value to a string.
LogicToString<public>(Val:logic)<computes>:string=
    if (Val?) {
        "true"
    } else {
        "false"
    }

StringArrayToString<public>(Input:[]string)<transacts>:string=
    var Result: string = "["
    for (Index->Entry:Input) {
        if (Index > 0) {
            set Result = Result + ", "
        }
        set Result = Result + "\"{Entry}\""
    }
    set Result = Result + "]"
    Result

(Input:[]t where t:type).ArrayToString<public>(MappingFn:type{__(:t):string}):string=
    var Result: string = "["
    for (Index->Entry:Input) {
        if (Index > 0) {
            set Result = Result + ", "
        }
        set Result = Result + MappingFn(Entry)
    }
    set Result = Result + "]"
    Result

StringEmptyXOR<public>(A:string, B:string)<transacts><decides>:void=
    (A.Length = 0 or A.Length = 0) and not (A.Length = 0 and B.Length = 0)

## Casing {{{
(Input:string).ToLowercase<public>()<computes>:string=
    for (Char:Input) {
        case(Char):
            'A' => 'a'
            'B' => 'b'
            'C' => 'c'
            'D' => 'd'
            'E' => 'e'
            'F' => 'f'
            'G' => 'g'
            'H' => 'h'
            'I' => 'i'
            'J' => 'j'
            'K' => 'k'
            'L' => 'l'
            'M' => 'm'
            'N' => 'n'
            'O' => 'o'
            'P' => 'p'
            'Q' => 'q'
            'R' => 'r'
            'S' => 's'
            'T' => 't'
            'U' => 'u'
            'V' => 'v'
            'W' => 'w'
            'X' => 'x'
            'Y' => 'y'
            'Z' => 'z'
            _ => Char
    }

(Input:string).ToUppercase<public>()<computes>:string=
    for (Char:Input) {
        case(Char):
            'a' => 'A'
            'b' => 'B'
            'c' => 'C'
            'd' => 'D'
            'e' => 'E'
            'f' => 'F'
            'g' => 'G'
            'h' => 'H'
            'i' => 'I'
            'j' => 'J'
            'k' => 'K'
            'l' => 'L'
            'm' => 'M'
            'n' => 'N'
            'o' => 'O'
            'p' => 'P'
            'q' => 'Q'
            'r' => 'R'
            's' => 'S'
            't' => 'T'
            'u' => 'U'
            'v' => 'V'
            'w' => 'W'
            'x' => 'X'
            'y' => 'Y'
            'z' => 'Z'
            _ => Char
    }
## }}}

(Input:string).PadStart<public>(TargetLength:int, ?Filler:char = ' ')<transacts>: string ={
    if (N := TargetLength - Input.Length, N > 0) {
        Prefix := for (__ := 1..N). Filler
        Prefix + Input
    } else {
        Input
    }
}
