using { /Fortnite.com/Devices }
using { /UnrealEngine.com/Temporary/SpatialMath }
using { /UnrealEngine.com/Temporary/UI }

ToString<public>(Transform:transform)<transacts>:string=
    "\{Translation: {Transform.Translation}, Rotation: {Transform.Rotation}, Scale: {Transform.Scale}\}"

ToString<public>(Visibility:widget_visibility)<transacts>:string=
    case(Visibility):
        widget_visibility.Visible => "Visible"
        widget_visibility.Hidden => "Hidden"
        widget_visibility.Collapsed => "Collapsed"

ToString<public>(InputMode:ui_input_mode)<transacts>:string=
    case(InputMode):
        ui_input_mode.None => "None"
        ui_input_mode.All => "All"

ToString<public>(MoveToResult:move_to_result)<transacts>:string=
    case(MoveToResult):
        move_to_result.DestinationReached => "DestinationReached"
        move_to_result.WillNotReachDestination => "WillNotReachDestination"

ToString<public>(SpawnResult:spawn_prop_result)<transacts>:string=
    case(SpawnResult):
        spawn_prop_result.Ok => "Ok"
        spawn_prop_result.UnknownError => "UnknownError"
        spawn_prop_result.InvalidSpawnPoint => "InvalidSpawnPoint"
        spawn_prop_result.SpawnPointOutOfBounds => "SpawnPointOutOfBounds"
        spawn_prop_result.InvalidAsset => "InvalidAsset"
        spawn_prop_result.TooManyProps => "TooManyProps"

StrArrayToString<public>(Input:[]string)<transacts>:string=
    Result := Join(Input, ", ")
    "[{Result}]"

FloatArrayToString<public>(Input:[]float)<transacts>: string ={
    var Result: string = "["
    for (Index->Element:Input) {
        if (Index <> 0) {
            set Result += ", "
        }
        set Result += ToString(Element)
    }
    Result + "]"
}
