using { /Verse.org/Simulation }
using { /Fortnite.com/Devices }

using { HavocVerse.Util }

dialog_response_type<public> := enum:
    Clicked
    Dismiss
    Timeout

dialog_response_event<public> := class:
    Agent<public>: agent
    ResponseType<public>: dialog_response_type
    ButtonIndex<public>: ?int

dialog_device_wrapper_viewing_state := class:
    var m_Callback: ?type{__(:dialog_response_event):void}

dialog_device_wrapper<public> := class:
    DialogDevice<public>: popup_dialog_device 

    var m_ViewingAgents: [agent]dialog_device_wrapper_viewing_state = map{}
    var m_Subscriptions: []cancelable = array{}

    Show<public>(Agent:agent, Callback:?type{__(:dialog_response_event):void}):logic=
        if (
            set m_ViewingAgents[Agent] = dialog_device_wrapper_viewing_state{
                m_Callback := Callback
            }
        ) {
            if (m_Subscriptions.Length = 0) {
                Closure := dialog_device_controller_closure{
                    Wrapper := Self
                }
                set m_Subscriptions = array{
                    DialogDevice.RespondingButtonEvent.Subscribe(Closure.OnDialogButtonPressed),
                    DialogDevice.DismissedEvent.Subscribe(Closure.OnDialogDismissed),
                    DialogDevice.TimeOutEvent.Subscribe(Closure.OnDialogTimeout)
                }
            }
            DialogDevice.Show(Agent)
            true
        } else {
            false
        }

    Cleanup<public>():void=
        for (Subscription:m_Subscriptions) {
            Subscription.Cancel()
        }
        set m_Subscriptions = array{}

    HandleDialogInteraction<internal>(Agent:agent, ResponseType:dialog_response_type, ButtonIndex:?int):void=
        RemoveRes := m_ViewingAgents.RemoveByKey(Agent)
        set m_ViewingAgents = RemoveRes(0)
        if (State := RemoveRes(1)?, Callback := State.m_Callback?) {
            set State.m_Callback = None() # prevent double calls
            Event := dialog_response_event{
                Agent := Agent
                ResponseType := ResponseType
                ButtonIndex := ButtonIndex
            }
            Callback(Event)
        }

dialog_device_controller_closure := class:
    Wrapper: dialog_device_wrapper

    OnDialogButtonPressed(Agent:agent, ButtonIndex:int):void=
        Wrapper.HandleDialogInteraction(Agent, dialog_response_type.Clicked, Some(ButtonIndex))
    
    OnDialogDismissed(Agent:agent):void=
        Wrapper.HandleDialogInteraction(Agent, dialog_response_type.Dismiss, None())

    OnDialogTimeout(Agent:agent):void=
        Wrapper.HandleDialogInteraction(Agent, dialog_response_type.Timeout, None())

ToString<public>(ResponseType:dialog_response_type)<transacts>:string=
    case(ResponseType):
        dialog_response_type.Clicked => "Clicked"
        dialog_response_type.Dismiss => "Dismiss"
        dialog_response_type.Timeout => "Timeout"

ToString<public>(Event:dialog_response_event)<transacts>:string=
    ButtonIndex := if (l_ButtonIndex := Event.ButtonIndex?) {
        ToString(l_ButtonIndex)
    } else {
        "None"
    }   
    "\{ResponseType: {Event.ResponseType}, ButtonIndex: {ButtonIndex}\}"
