using { /Verse.org/Simulation }
using { /Fortnite.com/Devices }
using { /Fortnite.com/Playspaces }
using { /UnrealEngine.com/Temporary/Diagnostics }
using { /UnrealEngine.com/Temporary/UI }
using { /UnrealEngine.com/Temporary/SpatialMath }

using { HavocVerse.Debug }
using { HavocVerse.Storage }
using { HavocVerse.Logging }
using { HavocVerse.Util }

var g_GlobalUI<internal>: weak_map(session,global_ui_instance) = map{}

GlobalUI_DebugStats<public>()<transacts>:aware_debug_stats=
    var Result: [string]int = map{}
    if (GlobalUIInstance := g_GlobalUI[GetSession()]) {
        var TotalRootWidgets: int = 0
        for (Player:GlobalUIInstance.m_Players) {
            set TotalRootWidgets += Player.m_RootWidgets.Length
        }
        set Result = ConcatenateMaps(Result, map{
            "players" => GlobalUIInstance.m_Players.Length
            "players.rootWidgets:sum" => GlobalUIInstance.m_Players.Length
        })
    }
    aware_debug_stats{
        Length := Result
    }

global_ui_config<public> := class<concrete>:
    @editable DisableGlobalUI: logic = false
    @editable MergeToRootCanvas: logic = true
    @editable SeparateCanvasForInputAll: logic = true
    @editable ResendAfterInputClose: logic = true

global_ui_instance<internal> := class:
    var m_PlayerRemovedSubscription: ?cancelable = false

    var m_Players: [player]global_ui_player = map{}
    var m_Config: global_ui_config = global_ui_config{}

    OnPlayerRemoved<internal>(Player:player):void=
        set m_Players = m_Players.RemoveByKey(Player)(0)

global_ui_player<internal> := class:
    Player: player
    PlayerUI: player_ui

    var RootCanvas: canvas
    var RootCanvasForInputAll: ?canvas

    var m_RootWidgets: []tuple(widget,player_ui_slot) = array{}

    var m_InputMode: ui_input_mode = ui_input_mode.None
    var m_RootCanvasForInputAllOnScreen: logic = false

    EnsureInputMode<internal>():void=
        if (RootCanvasForInputAll?) {
            return
        }
        var RequireInput: logic = false
        for (Entry:m_RootWidgets, not (RequireInput?), Entry(1).InputMode = ui_input_mode.All) {
            set RequireInput = true
        }
        ExpectedInputMode := if (RequireInput?) {
            ui_input_mode.All
        } else {
            ui_input_mode.None
        }
        if (ExpectedInputMode = m_InputMode) {
            return
        }
        set m_InputMode = ExpectedInputMode
        PlayerUI.RemoveWidget(RootCanvas)
        PlayerUI.AddWidget(RootCanvas, player_ui_slot{
            InputMode := ExpectedInputMode
        })

    CreateCanvasSlot<internal>(Widget:widget, Slot:player_ui_slot):canvas_slot=
        canvas_slot{
            Widget := Widget
            ZOrder := Slot.ZOrder
            SizeToContent := false
            Anchors := anchors{
                Minimum := vector2{X := 0.0, Y := 0.0}
                Maximum := vector2{X := 1.0, Y := 1.0}
            }
            Alignment := vector2{X := 0.0, Y := 0.0}
            Offsets := margin{}
        }

GetGlobalUIInstance<internal>():?global_ui_instance=
    if (l_GlobalUIInstance := g_GlobalUI[GetSession()]) {
        Some(l_GlobalUIInstance)
    } else if (
        l_GlobalUIInstance := global_ui_instance{}
        set g_GlobalUI[GetSession()] = l_GlobalUIInstance
    ) {
        Some(l_GlobalUIInstance)
    } else {
        None()
    }

GlobalUISetup<public>(Playspace:fort_playspace, Config:global_ui_config):void=
    InstanceRes := GetGlobalUIInstance()
    if (Instance := InstanceRes?) {
        set Instance.m_Config = Config
        if (not (Instance.m_PlayerRemovedSubscription?)) {
            Subscription := Playspace.PlayerRemovedEvent().Subscribe(Instance.OnPlayerRemoved)
            set Instance.m_PlayerRemovedSubscription = Some(Subscription)
        }
        if (not (Config.MergeToRootCanvas?) and (Config.ResendAfterInputClose?)) {
            Log("Global UI configured with MergeToRootCanvas disabled and ResendAfterInputClose enabled. This is not yet supported.", ?Level := log_level.Error)
        }
    }

GlobalUIAddWidget<public>(Player:player, Widget:widget):void=
    GlobalUIAddWidget(Player, Widget, player_ui_slot{})

GlobalUIAddWidget<public>(Player:player, Widget:widget, Slot:player_ui_slot):void=
    InstanceRes := GetGlobalUIInstance()
    Instance := if (l_Instance := InstanceRes?) {
        l_Instance
    } else {
        return
    }
    if (Instance.m_Config.DisableGlobalUI?) {
        if (PlayerUI := GetPlayerUI[Player]) {
            PlayerUI.AddWidget(Widget, Slot)
        }
        return
    }
    MergeToRootCanvas := Instance.m_Config.MergeToRootCanvas
    UIPlayer := if (l_UIPlayer := Instance.m_Players[Player]) {
        l_UIPlayer
    } else if (
        PlayerUI := GetPlayerUI[Player]
        l_UIPlayer := global_ui_player{
            Player := Player
            PlayerUI := PlayerUI
            RootCanvas := canvas{}
            RootCanvasForInputAll := if (Instance.m_Config.SeparateCanvasForInputAll?) {
                Some(canvas{})
            } else {
                None()
            }
            m_InputMode := Slot.InputMode
        }
        set Instance.m_Players[Player] = l_UIPlayer
    ) {
        if (MergeToRootCanvas?) {
            if (l_UIPlayer.RootCanvasForInputAll?) {
                PlayerUI.AddWidget(l_UIPlayer.RootCanvas, player_ui_slot{
                    InputMode := ui_input_mode.None
                })
            } else {
                PlayerUI.AddWidget(l_UIPlayer.RootCanvas, player_ui_slot{
                    InputMode := Slot.InputMode
                })
            }
        }
        l_UIPlayer
    } else {
        return
    }

    set UIPlayer.m_RootWidgets += array{(Widget, Slot)}
    if (not (MergeToRootCanvas?)) {
        if (PlayerUI := GetPlayerUI[Player]) {
            PlayerUI.AddWidget(Widget, Slot)
        }
        return
    }

    CanvasSlot := UIPlayer.CreateCanvasSlot(Widget, Slot)
    if (Slot.InputMode = ui_input_mode.All, RootCanvasForInputAll := UIPlayer.RootCanvasForInputAll?) {
        RootCanvasForInputAll.AddWidget(CanvasSlot)
        if (not (UIPlayer.m_RootCanvasForInputAllOnScreen?)) {
            UIPlayer.PlayerUI.AddWidget(RootCanvasForInputAll, player_ui_slot{
                InputMode := ui_input_mode.All
            })
            set UIPlayer.m_RootCanvasForInputAllOnScreen = true
        }
        RootCanvasForInputAll.AddWidget(CanvasSlot)
    } else {
        UIPlayer.EnsureInputMode()
        UIPlayer.RootCanvas.AddWidget(CanvasSlot)
    }

GlobalUIRemoveWidget<public>(Player:player, Widget:widget):void=
    InstanceRes := GetGlobalUIInstance()
    if (
        Instance := InstanceRes?
        UIPlayer := Instance.m_Players[Player]
    ) {
        MergeToRootCanvas := Instance.m_Config.MergeToRootCanvas

        var WidgetEntry: ?tuple(widget,player_ui_slot) = None()
        var FilteredRootWidgets: []tuple(widget,player_ui_slot) = array{}
        var AnyRemainingInputRequiring: logic = false
        for (Entry:UIPlayer.m_RootWidgets) {
            if (Entry(0) = Widget) {
                set WidgetEntry = Some(Entry)
            } else {
                set FilteredRootWidgets += array{Entry}
                if (Entry(1).InputMode = ui_input_mode.All) {
                    set AnyRemainingInputRequiring = true
                }
            }
        }
        set UIPlayer.m_RootWidgets = FilteredRootWidgets
        if (RemovedEntry := WidgetEntry?) {
            RemovedWidget := RemovedEntry(0)
            Slot := RemovedEntry(1)

            if (MergeToRootCanvas?) {
                if (Slot.InputMode = ui_input_mode.All, RootCanvasForInputAll := UIPlayer.RootCanvasForInputAll?) {
                    RootCanvasForInputAll.RemoveWidget(RemovedWidget)
                    if (not (AnyRemainingInputRequiring?)) {
                        UIPlayer.PlayerUI.RemoveWidget(RootCanvasForInputAll)
                        set UIPlayer.m_RootCanvasForInputAllOnScreen = false
                    }
                } else {
                    UIPlayer.RootCanvas.RemoveWidget(RemovedWidget)
                    UIPlayer.EnsureInputMode()
                    # switching input mode on single canvas is equivalent to resend after input close
                    return
                }
            } else if (PlayerUI := GetPlayerUI[Player]) {
                PlayerUI.RemoveWidget(Widget)
            }

            if (
                Instance.m_Config.ResendAfterInputClose?
                Slot.InputMode = ui_input_mode.All
                not (AnyRemainingInputRequiring?)
                PlayerUI := GetPlayerUI[Player]
                MergeToRootCanvas?
            ) {
                Log(MsgGlobalUIResendAfterInputClose(Player, UIPlayer.m_RootWidgets.Length), ?Level := log_level.Verbose)
                PlayerUI.RemoveWidget(UIPlayer.RootCanvas)
                set UIPlayer.RootCanvas = canvas{
                    Slots := for (Entry:UIPlayer.m_RootWidgets) {
                        UIPlayer.CreateCanvasSlot(Entry(0), Entry(1))
                    }
                }
                PlayerUI.AddWidget(UIPlayer.RootCanvas)
            }
        }
    } else if (PlayerUI := GetPlayerUI[Player]) {
        PlayerUI.RemoveWidget(Widget)
    }

GlobalUIForceResend<public>(Player:player):void=
    InstanceRes := GetGlobalUIInstance()
    if (
        Instance := InstanceRes?
        UIPlayer := Instance.m_Players[Player]
        Instance.m_Config.MergeToRootCanvas?
        PlayerUI := GetPlayerUI[Player]
    ) {
        Log(MsgGlobalUIResendForced(Player, UIPlayer.m_RootWidgets.Length), ?Level := log_level.Verbose)

        RootCanvasSlots := if (RootCanvasForInputAll := UIPlayer.RootCanvasForInputAll?) {
            if (UIPlayer.m_RootCanvasForInputAllOnScreen?) {
                InputCanvasSlots := for (Entry:UIPlayer.m_RootWidgets, Entry(1).InputMode = ui_input_mode.All) {
                    UIPlayer.CreateCanvasSlot(Entry(0), Entry(1))
                }
                PlayerUI.RemoveWidget(RootCanvasForInputAll)
                NewRootCanvasForInputAll := canvas{
                    Slots := InputCanvasSlots
                }
                set UIPlayer.RootCanvasForInputAll = Some(NewRootCanvasForInputAll)
                PlayerUI.AddWidget(NewRootCanvasForInputAll, player_ui_slot{
                    InputMode := ui_input_mode.All
                })
            }
            for (Entry:UIPlayer.m_RootWidgets, Entry(1).InputMode = ui_input_mode.None) {
                UIPlayer.CreateCanvasSlot(Entry(0), Entry(1))
            }
        } else {
            for (Entry:UIPlayer.m_RootWidgets) {
                UIPlayer.CreateCanvasSlot(Entry(0), Entry(1))
            }
        }
        PlayerUI.RemoveWidget(UIPlayer.RootCanvas)
        set UIPlayer.RootCanvas = canvas{
            Slots := RootCanvasSlots
        }
        PlayerUI.AddWidget(UIPlayer.RootCanvas)
    }

MsgGlobalUIResendAfterInputClose<internal><localizes>(Agent:agent, ElementCount:int):message=
    "Resending {ElementCount} UI elements to '{Agent}' (input close)"

MsgGlobalUIResendForced<internal><localizes>(Agent:agent, ElementCount:int):message=
    "Resending {ElementCount} UI elements to '{Agent}' (forced)"
