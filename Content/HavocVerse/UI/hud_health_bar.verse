using { /Verse.org/Simulation }
using { /Verse.org/Colors }
using { /Fortnite.com/UI }
using { /UnrealEngine.com/Temporary/SpatialMath }
using { /UnrealEngine.com/Temporary/UI }

using { HavocVerse.UI }
using { HavocVerse.Util }

hud_health_bar_config<public> := class<concrete>:
    @editable BarSize<public>: vector2 = vector2{X := 1000.0, Y := 60.0}
    @editable BorderColor<public>: color = NamedColors.Black
    @editable BorderWidth<public>: float = 4.0
    @editable TrackColor<public>: color = NamedColors.Gray
    @editable LastDamageColor<public>: color = NamedColors.Pink
    @editable HealthColor<public>: color = NamedColors.Red
    @editable TextTemplate<public>: string = "%Health%/%MaxHealth% HP"
    @editable TextStyle<public>: ui_text_style_config = ui_text_style_config{}

hud_health_bar<public> := class(hud_element):
    Provider<public>: hud_health_bar_provider
    BarSize<public>: vector2
    TextTemplate<public>: string

    TrackWidget<public>: color_block
    LastDamageWidget<public>: color_block
    HealthWidget<public>: color_block

    TextWidget<public>: ?text_block

    Update<override>():void=
        Health := Provider.GetHealth()
        MaxHealth := Provider.GetMaxHealth()

        PrevFrac := Clamp(Provider.GetPrevHealth() / MaxHealth, 0.0, 1.0)
        CurrentFrac := Clamp(Health / MaxHealth, 0.0, 1.0)

        LastDamageWidget.SetDesiredSize(vector2{
            X := Lerp(0.0, BarSize.X, PrevFrac)
            Y := BarSize.Y
        })
        HealthWidget.SetDesiredSize(vector2{
            X := Lerp(0.0, BarSize.X, CurrentFrac)
            Y := BarSize.Y
        })

        if (l_TextWidget := TextWidget?) {
            HealthText := ToString(DefaultingFloor(Health))
            MaxHealthText := ToString(DefaultingFloor(MaxHealth))
            l_TextWidget.SetText(ToMessage(TextTemplate
                .ReplaceAll("%Health%", HealthText)
                .ReplaceAll("%MaxHealth%", MaxHealthText)))
        }

CreateHUDHealthBar<public>(Config:hud_health_bar_config, Provider:hud_health_bar_provider):hud_health_bar=
    EmptySize := vector2{X := 0.0, Y := Config.BarSize.Y}
    BorderWidget := color_block{
        DefaultColor := Config.BorderColor
        DefaultDesiredSize := Config.BarSize + vector2{X := Config.BorderWidth * 2.0, Y := Config.BorderWidth * 2.0}
    }
    TrackWidget := color_block{
        DefaultColor := Config.TrackColor
        DefaultDesiredSize := Config.BarSize
    }
    LastDamageWidget := color_block{
        DefaultColor := Config.LastDamageColor
        DefaultDesiredSize := EmptySize
    }
    HealthWidget := color_block{
        DefaultColor := Config.HealthColor
        DefaultDesiredSize := EmptySize
    }
    BorderInset := margin{
        Top := Config.BorderWidth
        Left := Config.BorderWidth
        Bottom := Config.BorderWidth
        Right := Config.BorderWidth
    }
    var WidgetSlots: []overlay_slot = array{
        overlay_slot{
            Widget := BorderWidget
            HorizontalAlignment := horizontal_alignment.Left
            VerticalAlignment := vertical_alignment.Top
        }
        overlay_slot{
            Widget := TrackWidget
            HorizontalAlignment := horizontal_alignment.Left
            VerticalAlignment := vertical_alignment.Top
            Padding := BorderInset
        }
        overlay_slot{
            Widget := LastDamageWidget
            HorizontalAlignment := horizontal_alignment.Left
            VerticalAlignment := vertical_alignment.Top
            Padding := BorderInset
        }
        overlay_slot{
            Widget := HealthWidget
            HorizontalAlignment := horizontal_alignment.Left
            VerticalAlignment := vertical_alignment.Top
            Padding := BorderInset
        }
    }
    TextWidget := if (Config.TextTemplate.Length > 0) {
        l_TextWidget := Config.TextStyle.ToTextBlock("")
        set WidgetSlots += array{overlay_slot{
            Widget := l_TextWidget
            HorizontalAlignment := horizontal_alignment.Center
            VerticalAlignment := vertical_alignment.Center
        }}
        Some(l_TextWidget)
    } else {
        None()
    }
    Widget := overlay{
        Slots := WidgetSlots
    }
    hud_health_bar{
        Widget := Widget
        Provider := Provider
        BarSize := Config.BarSize
        TextTemplate := Config.TextTemplate
        TrackWidget := TrackWidget
        LastDamageWidget := LastDamageWidget
        HealthWidget := HealthWidget
        TextWidget := TextWidget
    }

hud_health_bar_provider<public> := interface:
    GetMaxHealth<public>():float

    GetPrevHealth<public>():float

    GetHealth<public>():float
