using { /Fortnite.com/Devices }
using { /Fortnite.com/Playspaces }
using { /Verse.org/Simulation }
using { /UnrealEngine.com/Temporary/Diagnostics }
using { /UnrealEngine.com/Temporary/UI }

using { HavocVerse.Devices }
using { HavocVerse.Logging }
using { HavocVerse.Util }
using { HavocVerse.Util.Task }

hud_manager_base_config<public> := class<concrete>:
    @editable RefreshRate<public>: float = 0.25
    @editable ZOrder<public>: type{_X:int where 0 <= _X, _X <= 2147483647} = 0

hud_manager_base<public> := class<abstract>:
    AccessControl<public>: ?access_control_device = false
    RefreshRate<public>: float = 0.25
    ZOrder<public>: type{_X:int where 0 <= _X, _X <= 2147483647} = 0

    var m_Playspace: ?fort_playspace = false
    var m_Enabled: logic = false
    var m_Players: [player]hud_manager_player = map{}

    Init<public>(Playspace:fort_playspace, ?StartEnabled:logic = true, ?PostponePlayersInit:float = 0.0):void=
        if (m_Playspace?) {
            return
        }
        set m_Playspace = Some(Playspace)
        set m_Enabled = StartEnabled
        if (PostponePlayersInit > 0.0) {
            spawn:
                InitSecondPhase(Playspace, PostponePlayersInit)
        } else {
            InitSecondPhase(Playspace)
        }

    InitSecondPhase<private>(Playspace:fort_playspace, Delay:float)<suspends>:void=
        if (Delay > 0.0) {
            Sleep(Delay)
            Log("HUD manager initializing second phase after {Delay}s init delay", ?Level := log_level.Debug)
        }
        InitSecondPhase(Playspace)

    InitSecondPhase<private>(Playspace:fort_playspace):void=
        ActivePlayers := if (l_AccessControl := AccessControl?, l_AccessControl.p_Mode = access_control_mode.Whitelist) {
            l_AccessControl.SubscribeToPlayerWhitelistEvents(OnPlayerAdded, OnPlayerRemovedFromWhitelist)
            for (Agent:l_AccessControl.ListAllowedAgents(), Player := player[Agent]) {
                Player
            }
        } else {
            Playspace.GetPlayers()
        }
        Playspace.PlayerAddedEvent().Subscribe(OnPlayerAdded)
        Playspace.PlayerRemovedEvent().Subscribe(OnPlayerRemoved)
        if (m_Enabled?) {
            for (Player:ActivePlayers) {
                OnPlayerAdded(Player)
            }
        }
    
    RecheckUnmanagedPlayers<public>():void=
        if (not (m_Enabled?)) {
            return
        }
        ActivePlayers := if (l_AccessControl := AccessControl?, l_AccessControl.p_Mode = access_control_mode.Whitelist) {
            for (Agent:l_AccessControl.ListAllowedAgents(), Player := player[Agent]) {
                Player
            }
        } else if (Playspace := m_Playspace?) {
            Playspace.GetPlayers()
        } else {
            return
        }
        for (Player:ActivePlayers, not m_Players[Player]) {
            OnPlayerAdded(Player)
        }

    StartEventLoop<protected>()<suspends>:void=
        if (RefreshRate < 0.0) {
            return
        }
        loop {
            Sleep(RefreshRate)
            Refresh()
        }

    Refresh<public>():void=
        for (Player->PlayerHUD:m_Players, Player.IsActive[]) {
            PlayerHUD.Update()
        }

    Refresh<public>(Player:player):void=
        if (PlayerHUD := m_Players[Player], Player.IsActive[]) {
            PlayerHUD.Update()
        }

    Run<public>(Playspace:fort_playspace, ?StartEnabled:logic = true)<suspends>:void=
        Init(Playspace, ?StartEnabled := StartEnabled)
        StartEventLoop()

    GetPlayers<public>()<transacts>:[]hud_manager_player=
        for (__->Player:m_Players) {
            Player
        }

    OnPlayerAdded<private>(Player:player):void=
        if (not (m_Enabled?) or m_Players[Player]) {
            return
        }
        if (l_AccessControl := AccessControl?, not l_AccessControl.IsAllowed[Player]) {
            return
        }
        PlayerHUDRes := CreateHUD(Player)
        if (PlayerHUD := PlayerHUDRes?, set m_Players[Player] = PlayerHUD) {
            PlayerHUD.InitPreAdd()
            GlobalUIAddWidget(Player, PlayerHUD.Canvas, player_ui_slot{
                ZOrder := ZOrder
            })
            PlayerHUD.InitPostAdd()
        }

    OnPlayerRemoved<private>(Player:player):void=
        RemoveRes := m_Players.RemoveByKey(Player)
        set m_Players = RemoveRes(0)
        if (PlayerHUD := RemoveRes(1)?) {
            GlobalUIRemoveWidget(Player, PlayerHUD.Canvas)
        }

    Enable<public>(?Force:logic = false):void=
        if (not (Force?) and m_Enabled?) {
            return
        }
        set m_Enabled = true
        for (Player:m_Playspace?.GetPlayers()) {
            OnPlayerAdded(Player)
        }

    Disable<public>():void=
        set m_Enabled = false
        for (Player->__:m_Players) {
            OnPlayerRemoved(Player)
        }

    IsEnabled<public>()<transacts><decides>:void=
        m_Enabled?

    OnPlayerRemovedFromWhitelist<private>(Player:player, GameQuit:logic):void=
        OnPlayerRemoved(Player)

    CreateHUD<protected>(Player:player):?hud_manager_player

hud_manager_player<public> := class<abstract>:
    Player<public>: player
    Canvas<public>: canvas

    InitPreAdd<public>():void=
        Update()

    InitPostAdd<public>():void=
        {}

    Update<public>():void

hud_manager_player_default<public> := class(hud_manager_player):
    Elements<public>: []hud_element

    InitPostAdd<override>():void=
        spawn:
            InitPostAddNextTick()

    InitPostAddNextTick<private>()<suspends>:void=
        # if elements have their visibility set to hidden on the same tick (or before) of the
        # canvas being added to the UI, they will remain visible and their Visibility will be bugged
        # this is a workaround to force the glitched elements to be hidden
        Sleep(1.0)
        for (Element:Elements) {
            Element.ForceUpdateVisibility()
        }

    Update<override>():void=
        for (Element:Elements) {
            Element.Update()
        }

MakeSingleElementHUDPlayer<public>(Player:player, Canvas:canvas, Element:hud_element)<transacts>:hud_manager_player=
    hud_manager_player_default{
        Player := Player
        Canvas := Canvas
        Elements := array{Element}
    }

hud_manager_ui_framework_player_default<public> := class(hud_manager_player):
    Elements<public>: []ui_element

    Update<override>():void=
        for (Element:Elements) {
            Element.Update()
        }

MakeUIFrameworkHUDPlayer<public>(Player:player, Component:ui_component, Placement:ui_canvas_placement_config)<transacts>:hud_manager_player=
    Canvas := canvas:
        Slots := array{
            Placement.ToCanvasSlot(Component.ParentWidget())
        }
    hud_manager_ui_framework_player_default{
        Player := Player
        Canvas := Canvas
        Elements := array{Component}
    }

MakeUIFrameworkHUDPlayer<public>(Player:player, Elements:[]tuple(ui_component,ui_canvas_placement_config))<transacts>:hud_manager_player=
    CanvasSlots := for (Element:Elements) {
        Component := Element(0)
        Placement := Element(1)
        Placement.ToCanvasSlot(Component.ParentWidget())
    }
    hud_manager_ui_framework_player_default{
        Player := Player
        Canvas := canvas:
            Slots := CanvasSlots
        Elements := for (Element:Elements). Element(0)
    }

hud_element<public> := class<abstract>:
    var m_Visible<protected>: logic = true

    Widget<public>: widget

    Update<public>():void

    Hide<public>():void=
        if (m_Visible?) {
            set m_Visible = false
            Widget.SetVisibility(widget_visibility.Hidden)
        }

    Show<public>():void=
        if (not (m_Visible?)) {
            set m_Visible = true
            Widget.SetVisibility(widget_visibility.Visible)
        }

    ForceUpdateVisibility<public>()<suspends>:void=
        if (not (m_Visible?)) {
            Widget.SetVisibility(widget_visibility.Visible)
            Sleep(0.0)
            Widget.SetVisibility(widget_visibility.Hidden)
        }
