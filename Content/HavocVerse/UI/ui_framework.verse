using { /Fortnite.com/UI }
using { /UnrealEngine.com/Temporary/SpatialMath }
using { /UnrealEngine.com/Temporary/UI }
using { /Verse.org/Simulation }

using { HavocVerse.Util }

# A UI element.
ui_element<public> := class<abstract>:
    Update<public>():void

    Discard<public>():void=
        {}

ui_component_dynamic_state<internal> := class<final>:
    AbortTasksEvent: event() = event(){}
    var Subscriptions: []cancelable = array{}

ui_framework_offset_setter_closure<internal> := class<final> {
    Component: ui_component
    LeftAligned: logic

    Setter<internal>(Offset:vector2): void ={
        Component.Placeholder.SetOffset(Offset, ?LeftAligned := LeftAligned)
    }
}

# A UI component, that maintains state and updates efficiently.
ui_component<public> := class<abstract>(ui_element):
    Placeholder<public>: placeholder_widget = placeholder_widget{}

    var m_Children<internal>: []ui_element = array{}
    var m_Dirty<private>: logic = false
    var m_DynamicState<private>: ?ui_component_dynamic_state = false
    var m_Opacity<protected>: ?type{_X:float where 0.000000 <= _X, _X <= 1.000000} = false

    ParentWidget<public>()<transacts>:widget=
        Placeholder.BackingOverlay

    OffsetSetter<public>(?LeftAligned:logic=false) : type{__(:vector2):void} ={
        ui_framework_offset_setter_closure{
            Component := Self
            LeftAligned := LeftAligned
        }.Setter
    }

    SetOpacity<public>(Opacity:float):void=
        if (ClampedOpacity := type{_X:float where 0.000000 <= _X, _X <= 1.000000}[Clamp(Opacity, 0.0, 1.0)]) {
            set m_Opacity = Some(ClampedOpacity)
            ApplyOpacity(ClampedOpacity)
        } else {
            # shouldn't happen, ensured by Clamp()
            set m_Opacity = None()
        }

    Update<override>():void=
        if (CurrentWidget := Placeholder.m_Widget?) {
            if (not (m_Dirty?)) {
                # component doesn't need to be recreated due to 'dirty' state
                Updated := TryUpdate()
                if (Updated?) {
                    for (Child:m_Children) {
                        Child.Update()
                    }
                    return
                }
                # component couldn't update, recreate it
            }
            Placeholder.Clear()
        }
        for (Child:m_Children) {
            Child.Discard()
        }
        set m_Dirty = false
        set m_Children = array{}
        if (DynamicState := m_DynamicState?) {
            DynamicState.AbortTasksEvent.Signal()
            for (Subscription:DynamicState.Subscriptions) {
                Subscription.Cancel()
            }
            set DynamicState.Subscriptions = array{}
            # don't clear out m_DynamicState cause this widget will likely keep being dynamic
        }
        NewWidget := Create()
        Placeholder.SetWidget(NewWidget)
        if (Opacity := m_Opacity?) {
            ApplyOpacity(Opacity)
        }

    # Forces the component to be recreated.
    ForceUpdate<public>():void=
        set m_Dirty = true
        Update()

    # Marks this compoment as 'dirty'.
    # It will be recreated on the next update, skipping TryUpdate.
    MarkDirty<public>():void=
        set m_Dirty = true

    # Attempts to update the component.
    # If false is returned, the component will be recreated.
    TryUpdate<public>():logic=
        true

    # Renders the whole component anew.
    Create<public>():widget

    # Adds a child component to this component.
    # Children are cleared every time the component is recreated.
    AddChild<public>(Component:ui_component):widget=
        set m_Children += array{Component}
        Component.Update()
        Component.ParentWidget()

    # Schedules a function to run in the background for this widget.
    # The task will be cancelled when the widget is re-created or discarded.
    #
    # When working within menus, make sure the widget is a ui_menu_component,
    # otherwise the task will be cancelled when a single viewer (out ofpossibly several) closes the menu.
    ScheduleTask<public>(SpawnFn:type{__()<suspends>:void}):void=
        AbortEvent := (m_DynamicState?.AbortTasksEvent) or (
            DynamicState := ui_component_dynamic_state{}
            set m_DynamicState = Some(DynamicState)
            DynamicState.AbortTasksEvent
        )
        spawn:
            ScheduleTask0(SpawnFn, AbortEvent)

    ScheduleTask0<private>(SpawnFn:type{__()<suspends>:void}, AbortEvent:event())<suspends>:void=
        race:
            AbortEvent.Await()
            SpawnFn()

    BindSubscription<public>(Subscription:cancelable):void=
        if (DynamicState := m_DynamicState?) {
            set DynamicState.Subscriptions += array{Subscription}
        } else {
            DynamicState := ui_component_dynamic_state{}
            set m_DynamicState = Some(DynamicState)
            set DynamicState.Subscriptions += array{Subscription}
        }

    ProcessComponentTree<internal>(ProcessFn:type{__(:ui_component):void}):void=
        ProcessFn(Self)
        for (Child:m_Children, ChildComponent := ui_component[Child]) {
            ProcessFn(ChildComponent)
            ChildComponent.ProcessComponentTree(ProcessFn)
        }

    Discard<override>():void=
        for (Child:m_Children) {
            Child.Discard()
        }
        set m_Dirty = false
        set m_Children = array{}
        if (DynamicState := m_DynamicState?) {
            DynamicState.AbortTasksEvent.Signal()
            for (Subscription:DynamicState.Subscriptions) {
                Subscription.Cancel()
            }
            set m_DynamicState = None()
        }
        set m_Opacity = None()
        Placeholder.Clear()

    # Applies a target opacity to compatible widgets.
    # The default implementation walks through the widget tree, which has limitations.
    # When using .AddWidget on containers, this function needs to be overriden to properly support it.
    ApplyOpacity<protected>(Opacity:type{_X:float where 0.000000 <= _X, _X <= 1.000000}):void=
        if (Widget := Placeholder.m_Widget?) {
            ApplyOpacity0(Widget, Opacity)
        }

    ApplyOpacity0<protected>(Widget:widget, Opacity:type{_X:float where 0.000000 <= _X, _X <= 1.000000}):void=
        if (Canvas := canvas[Widget]) {
            for (Slot:Canvas.Slots) {
                ApplyOpacity0(Slot.Widget, Opacity)
            }
        } else if (StackBox := stack_box[Widget]) {
            for (Slot:StackBox.Slots) {
                ApplyOpacity0(Slot.Widget, Opacity)
            }
        } else if (Overlay := overlay[Widget]) {
            for (Slot:Overlay.Slots) {
                ApplyOpacity0(Slot.Widget, Opacity)
            }
        } else if (TextBlock := text_block[Widget]) {
            TextBlock.SetTextOpacity(Opacity)
        } else if (ColorBlock := color_block[Widget]) {
            ColorBlock.SetOpacity(Opacity)
        }

ui_menu_component<public> := class<abstract>(ui_component):
    var m_Viewers<internal>: [ui_menu_instance]void = map{}
    var m_Subscriptions<private>: []cancelable = array{}

    RootWidget<public>: logic = false

    RegisterViewer<public>(Instance:ui_menu_instance):void=
        Closure := register_viewer_closure:
            MenuInstance := Instance
        ProcessComponentTree(Closure.DoRegister)

    UnregisterViewer<public>(Instance:ui_menu_instance):void=
        Closure := register_viewer_closure:
            MenuInstance := Instance
        ProcessComponentTree(Closure.DoUnregister)

    ViewerCount<public>()<transacts>:int=
        m_Viewers.Length

    ListViewingPlayers<public>()<transacts>:[]player=
        for (Viewer->__:m_Viewers) {
            Viewer.Player
        }

    BindToViewerClose<public>(Listenable:listenable(widget_message)):void=
        set m_Subscriptions += array{Listenable.Subscribe(OnCloseClick)}

    OnCloseClick<private>(Data:widget_message):void=
        for (Viewer->__:m_Viewers, Viewer.Player = Data.Player) {
            Viewer.Close()
        }

    AddChild<override>(Component:ui_component):widget=
        ChildWidget := (super:)AddChild(Component)
        if (MenuComponent := ui_menu_component[Component], m_Viewers.Length > 0) {
            ProcessComponentTree(Self.PropagateViewers)
        }
        ChildWidget

    PropagateViewers<private>(Component:ui_component):void=
        if (MenuComponent := ui_menu_component[Component]) {
            for (
                MenuInstance->__:m_Viewers
                not MenuComponent.m_Viewers[MenuInstance]
                set MenuComponent.m_Viewers[MenuInstance] = void
            ) {
                SingleViewer := if (m_Viewers.Length = 1 and MenuComponent.m_Viewers.Length = 1). true else false
                MenuComponent.OnViewEnter(MenuInstance.Player, SingleViewer)
            }
        }

    Discard<override>():void=
        (super:)Discard()
        set m_Viewers = map{}
        for (Subscription:m_Subscriptions) {
            Subscription.Cancel()
        }
        set m_Subscriptions = array{}

    OnViewEnter<public>(Player:player, SingleViewer:logic):void=
        {}

    OnViewLeave<public>(Player:player):void=
        {}

register_viewer_closure<internal> := class<final>:
    MenuInstance<internal>: ui_menu_instance

    DoRegister<internal>(Component:ui_component):void=
        if (MenuComponent := ui_menu_component[Component], not MenuComponent.m_Viewers[MenuInstance]) {
            if. set MenuComponent.m_Viewers[MenuInstance] = void
            SingleViewer := if (MenuComponent.m_Viewers.Length = 1). true else false
            MenuComponent.OnViewEnter(MenuInstance.Player, SingleViewer)
        }

    DoUnregister<internal>(Component:ui_component):void=
        if (MenuComponent := ui_menu_component[Component]) {
            RemoveRes := MenuComponent.m_Viewers.RemoveByKey(MenuInstance)
            set MenuComponent.m_Viewers = RemoveRes(0)
            if (RemoveRes(1)?) {
                MenuComponent.OnViewLeave(MenuInstance.Player)
            }
        }

# Position of a notification on screen.
ui_notification_position<public> := class<final> {
    Anchor<public>: vector2
    Alignment<public>: vector2
    Offsets<public>: margin = margin{}
}

# A UI component that will be displayed as a notification and
# can decide to clear on its own.
ui_notification_component<public> := class<abstract><unique>(ui_component) {
    var m_Viewers: [player]tuple(canvas,?ui_notification_group) = map{}

    ### DOMAIN
    Discard<override>():void=
        (super:)Discard()
        for (Player->State:m_Viewers) {
            Hide0(Player, State)
        }
        set m_Viewers = map{}

    Hide0<private>(Player:player, State:tuple(canvas,?ui_notification_group)): void ={
        if (PlayerUI := GetPlayerUI[Player]) {
            PlayerUI.RemoveWidget(State(0))
        }
        if (Group := State(1)?, Group.m_Active[Player] = Self) {
            set Group.m_Active = Group.m_Active.RemoveByKey(Player)(0)
        }
    }

    ### API
    # Show the notification to a player.
    Show<public>(
        Player:player,
        Position:ui_notification_position,
        ?Group:?ui_notification_group=false
    ): void ={
        if (m_Viewers[Player]) {
            return
        }

        Update()
        Canvas := canvas:
            Slots := array{
                canvas_slot:
                    Widget := ParentWidget()
                    Anchors := anchors:
                        Minimum := Position.Anchor
                        Maximum := Position.Anchor
                    Offsets := Position.Offsets
                    Alignment := Position.Alignment
                    SizeToContent := true
            }
        if (PlayerUI := GetPlayerUI[Player], set m_Viewers[Player] = (Canvas, Group)) {
            PlayerUI.AddWidget(Canvas)
        }
    }

    # Hide the notification from all players it's currently visible to.
    Hide<public>(): void ={
        for (Player->State:m_Viewers) {
            Hide0(Player, State)
        }
        set m_Viewers = map{}
    }

    # Hide the notification from a player.
    Hide<public>(Player:player): void ={
        RemoveRes := m_Viewers.RemoveByKey(Player)
        set m_Viewers = RemoveRes(0)
        if (State := RemoveRes(1)?) {
            Hide0(Player, State)
        }
    }
}

# Notifications group, to ensure exclusivity of a certain kind of notifications
# on screen at a given time.
# TODO : Queue notifications
ui_notification_group<public> := class<final> {
    var m_Active: [player]ui_notification_component = map{}

    ### API
    # Sends a notification to a player. This will clear
    # any previous notification still shown to them.
    Send<public>(Player:player, Notification:ui_notification_component, Position:ui_notification_position): void ={
        Clear(Player)
        if. set m_Active[Player] = Notification
        Notification.Show(Player, Position, ?Group := Some(Self))
    }

    # Clears the currently active notification for a player, if any.
    Clear<public>(Player:player): void ={
        if (ActiveComponent := m_Active[Player]) {
            ActiveComponent.Hide(Player) # should be cleared from m_Active in Hide0
        }
    }
}

placeholder_widget<public> := class<final>:
    BackingOverlay<public>: overlay = overlay{}

    var m_Widget<internal>: ?widget = false

    ParentWidget<public>()<transacts>:widget=
        BackingOverlay

    SetWidget<public>(NewWidget:widget):void=
        if (PrevWidget := m_Widget?) {
            BackingOverlay.RemoveWidget(PrevWidget)
        }
        set m_Widget = Some(NewWidget)
        BackingOverlay.AddWidget(overlay_slot{
            Widget := NewWidget
            HorizontalAlignment := horizontal_alignment.Fill
            VerticalAlignment := vertical_alignment.Fill
        })

    SetOffset<public>(Offset:vector2, ?LeftAligned:logic=true):void=
        if (Widget := m_Widget?) {
            BackingOverlay.RemoveWidget(Widget)
            BackingOverlay.AddWidget(overlay_slot{
                Widget := Widget
                HorizontalAlignment := if (LeftAligned?) {
                    horizontal_alignment.Left
                } else {
                    horizontal_alignment.Right
                }
                VerticalAlignment := vertical_alignment.Top
                Padding := margin{
                    Top := Offset.Y
                    Right := if (not LeftAligned?). Offset.X else 0.0
                    Left := if (LeftAligned?). Offset.X else 0.0
                }
            })
        }

    Clear<public>():void=
        if (Widget := m_Widget?) {
            BackingOverlay.RemoveWidget(Widget)
            set m_Widget = None()
        }

# A placeholder element that may contain a UI component.
#
# Note: Please prefer using a re-usable `placeholder_widget` across
#       components, when the context enables it.
ui_component_placeholder<public> := class(ui_element):
    BackingOverlay<public>: overlay = overlay{}

    var m_Component<private>: ?ui_component = false

    ParentWidget<public>()<transacts>:widget=
        BackingOverlay

    Update<override>():void=
        if (Component := m_Component?) {
            Component.Update()
        }

    Discard<override>():void=
        Clear()

    GetComponent<public>()<decides><transacts>:ui_component=
        m_Component?

    SetComponent<public>(Component:ui_component):void=
        if (PrevComponent := m_Component?) {
            BackingOverlay.RemoveWidget(PrevComponent.ParentWidget())
        }
        set m_Component = Some(Component)
        BackingOverlay.AddWidget(overlay_slot{
            Widget := Component.ParentWidget()
            HorizontalAlignment := horizontal_alignment.Fill
            VerticalAlignment := vertical_alignment.Fill
        })

    Clear<public>():void=
        if (PrevComponent := m_Component?) {
            BackingOverlay.RemoveWidget(PrevComponent.ParentWidget())
        }
        set m_Component = None()
