using { /Verse.org/Assets }
using { /Verse.org/Colors }
using { /Verse.org/Simulation }
using { /UnrealEngine.com/Temporary/Diagnostics }
using { /UnrealEngine.com/Temporary/SpatialMath }
using { /UnrealEngine.com/Temporary/UI }
using { /Fortnite.com/UI }

using { HavocVerse.Debug }
using { HavocVerse.Logging }
using { HavocVerse.Util }

ui_assets_provider<public> := interface:
    GetCustomFont<public>(Name:string)<transacts><decides>:custom_font

DefaultUIAssetsProvider<public>()<transacts><decides>:ui_assets_provider=
    ui_assets_provider[AwareDefaultAssetsProvider[]]

MsgSkipMenuOpenThrottling<internal><localizes>(Player:player):message=
    "Skipping menu open for '{Player}' because double open (<1s) was detected."

ui_menu<public> := class<abstract>:
    AssetsProvider<public>: assets_provider
    ZOrder<public>: type{_X:int where 0 <= _X, _X <= 2147483647} = 1000

    var m_ActiveInstances<public>: [player]ui_menu_instance = map{}

    Open<public>(Player:player):?ui_menu_instance=
        if (
            ActiveInstance := m_ActiveInstances[Player]
            OpenedAt := ActiveInstance.m_OpenedAt?
            GetSimulationElapsedTime() - OpenedAt < 1.0
        ) {
            Log(MsgSkipMenuOpenThrottling(Player), ?Level := log_level.Debug)
            return None()
        }
        CreateRes := CreateInstance(Player)
        CommonOpen(Player, CreateRes(0), CreateRes(1))
        Some(CreateRes(0))

    CreateInstance<public>(Player:player):tuple(ui_menu_instance,canvas)

    CommonOpen<protected>(Player:player, Instance:ui_menu_instance, Canvas:canvas):void=
        if (PrevInstance := m_ActiveInstances[Player]) {
            PrevInstance.Close()
        }
        set Instance.m_Canvas = Some(Canvas)
        set Instance.m_OpenedAt = Some(GetSimulationElapsedTime())
        GlobalUIAddWidget(Player, Canvas, player_ui_slot{
            InputMode := ui_input_mode.All
            ZOrder := ZOrder
        })
        if:
            set m_ActiveInstances[Player] = Instance

    PlaceOnCenteredBackground<protected>(Widget:widget,
        BackgroundTexture:string, BackgroundSize:vector2,
        ?DefaultToColorBlock:logic = false,
        ?DefaultingColor:color = NamedColors.Black,
        ?DefaultingColorOpacity:type{_X:float where 0.0 <= _X, _X <= 1.0} = 0.8
    )<transacts>:widget=
        if (
            BackgroundTexture.Length > 0
            BackgroundImage := AssetsProvider.GetTextureAsset[BackgroundTexture]
        ) {
            overlay{
                Slots := array{
                    overlay_slot{
                        Widget := texture_block{
                            DefaultImage := BackgroundImage
                            DefaultDesiredSize := BackgroundSize
                        }
                        HorizontalAlignment := horizontal_alignment.Center
                        VerticalAlignment := vertical_alignment.Center
                    }
                    overlay_slot{
                        Widget := Widget
                        HorizontalAlignment := horizontal_alignment.Center
                        VerticalAlignment := vertical_alignment.Center
                    }
                }
            }
        } else if (DefaultToColorBlock?) {
            overlay{
                Slots := array{
                    overlay_slot{
                        Widget := color_block{
                            DefaultDesiredSize := BackgroundSize
                            DefaultColor := DefaultingColor
                            DefaultOpacity := DefaultingColorOpacity
                        }
                        HorizontalAlignment := horizontal_alignment.Center
                        VerticalAlignment := vertical_alignment.Center
                    }
                    overlay_slot{
                        Widget := Widget
                        HorizontalAlignment := horizontal_alignment.Center
                        VerticalAlignment := vertical_alignment.Center
                    }
                }
            }
        } else {
            Widget
        }

    CommonCanvas<protected>(
        Widget:widget,
        ?BackgroundTexture:string = "",
        ?BackgroundSize:vector2 = vector2{X := 0.0, Y := 0.0},
        ?DefaultToColorBlock:logic = false,
        ?Offsets:?margin = false
    ):canvas=
        CanvasWidget := PlaceOnCenteredBackground(Widget, BackgroundTexture, BackgroundSize,
            ?DefaultToColorBlock := DefaultToColorBlock)
        GetSimpleCanvas(CanvasWidget, ?Offsets := Offsets)

    RefreshActiveInstances<public>():void=
        for (__->Instance:m_ActiveInstances) {
            Instance.Refresh()
        }

    CloseInstance<public>(Player:player):void=
        if (Instance := m_ActiveInstances[Player]) {
            Instance.Close()
        }

ui_menu_instance<public> := class<unique>:
    Menu<public>: ui_menu
    PrevMenu<public>: ?ui_menu = false

    Player<public>: player

    var m_Canvas: ?canvas = false
    var m_OpenedAt: ?float = false
    var m_RootElements: []ui_element = array{}

    AddRootElement<public>(Element:ui_element):void=
        set m_RootElements += array{Element}
        if (Component := ui_component[Element]) {
            Component.Update()
            if (MenuComponent := ui_menu_component[Element]) {
                MenuComponent.RegisterViewer(Self)
            }
        }

    AddComponent<public>(Component:ui_component):widget=
        AddRootElement(Component)
        Component.ParentWidget()

    ClearRootElements<public>():void=
        for (Element:m_RootElements) {
            Element.Discard()
        }
        set m_RootElements = array{}

    Refresh<public>():void=
        for (Element:m_RootElements) {
            Element.Update()
        }

    Close<public>():void=
        if (Canvas := m_Canvas?) {
            set m_Canvas = None()
            Canvas.SetEnabled(false) # forces a UI refresh
            GlobalUIRemoveWidget(Player, Canvas)

            for (Element:m_RootElements, MenuComponent := ui_menu_component[Element]) {
                MenuComponent.UnregisterViewer(Self)
            }
            for (Element:m_RootElements, Component := ui_component[Element]) {
                Component.ProcessComponentTree(DoDiscardUnviewed)
            }

            set Menu.m_ActiveInstances = Menu.m_ActiveInstances.RemoveByKey(Player)(0)
            SimpleMenuHandleMenuClose(Player, Self)
            OnClose()
        }

    IsActive<public>()<transacts><decides>:void=
        m_Canvas?

    OnClose<protected>():void=
        {}

    OnCloseClick<public>(Data:widget_message):void=
        Close()

    OnGoBackClick<public>(Data:widget_message):void=
        Close()
        if (l_PrevMenu := PrevMenu?) {
            l_PrevMenu.Open(Player)
        }

DoDiscardUnviewed<internal>(Component:ui_component):void=
    if (MenuComponent := ui_menu_component[Component], MenuComponent.ViewerCount() > 0) {
        # still has viewers, do not discard
        return
    }
    Component.Discard()

var g_MenuCache<internal>: weak_map(session, [string]simple_menu_cache_state) = map{}

MenuUI_DebugStats<scoped {HavocVerse}>()<transacts>:aware_debug_stats=
    var Result: [string]int = map{}
    if (MenuCache := g_MenuCache[GetSession()]) {
        var TotalPlayers: int = 0
        for (__->CacheState:MenuCache) {
            set TotalPlayers += CacheState.Players.Length
        }
        set Result = ConcatenateMaps(Result, map{
            "menuCache" => MenuCache.Length
            "menuCache.players:sum" => TotalPlayers
        })
    }
    aware_debug_stats{
        Length := Result
    }

MenuUI_HandlePlayerQuit<scoped {HavocVerse}>(Player:player):void=
    if (MenuCache := g_MenuCache[GetSession()]) {
        for (__->CacheState:MenuCache, PlayerState := CacheState.Players[Player]) {
            if (ActiveMenuInstance := PlayerState.ActiveMenu?) {
                ActiveMenuInstance.Close()
            }
            set CacheState.Players = CacheState.Players.RemoveByKey(Player)(0)
        }
    }

simple_menu_cache_state<internal> := class<final>:
    var Players: [player]simple_menu_cache_player_state = map{}

simple_menu_cache_player_state<internal> := class<final>:
    var LastOpenTime: ?float = false
    var ActiveMenu: ?ui_menu_instance = false

OpenSimpleMenu<public>(MenuId:string, Agent:agent, Factory:type{__():ui_menu_component}, ?Reopen:logic = true):void=
    if (Player := player[Agent], AssetsProvider := AwareDefaultAssetsProvider[]) {
        var MenuCache: [string]simple_menu_cache_state = g_MenuCache[GetSession()] or map{}
        CacheState := MenuCache[MenuId] or simple_menu_cache_state{}
        if (PlayerState := CacheState.Players[Player]) {
            if (not Reopen?, PlayerState.ActiveMenu?) {
                Log("Skipping menu \"{MenuId}\" open for '{Player}' because it is already open.", ?Level := log_level.Debug)
                return
            }

            if (
                LastOpenedAt := PlayerState.LastOpenTime?
                GetSimulationElapsedTime() - LastOpenedAt < 1.0
            ) {
                Log("Skipping menu \"{MenuId}\" open for '{Player}' because double open (<1s) was detected.", ?Level := log_level.Debug)
                return
            }

            if (PrevMenuInstance := PlayerState.ActiveMenu?) {
                set PlayerState.ActiveMenu = None()
                PrevMenuInstance.Close()
                # close leads to player state removal from cache
            }
        }

        if:
            set MenuCache[MenuId] = CacheState
            set g_MenuCache[GetSession()] = MenuCache
        PlayerCacheState := CacheState.Players[Player] or (
            NewState := simple_menu_cache_player_state{}
            set CacheState.Players = ConcatenateMaps(CacheState.Players, map{Player => NewState})
            NewState
        )
        set PlayerCacheState.LastOpenTime = Some(GetSimulationElapsedTime())

        Menu := simple_ui_framework_menu{
            AssetsProvider := AssetsProvider
            Component := Factory()
        }
        OpenRes := Menu.Open(Player)
        if (MenuInstance := OpenRes?) {
            set PlayerCacheState.ActiveMenu = Some(MenuInstance)
        }
    }

simple_menu_component_provider<internal> := class<final>:
    Component: ui_menu_component
    Factory():ui_menu_component = Component

OpenSimpleMenu<public>(MenuId:string, Agent:agent, Component:ui_menu_component, ?Reopen:logic = true):void=
    Provider := simple_menu_component_provider{Component := Component}
    OpenSimpleMenu(MenuId, Agent, Provider.Factory, ?Reopen := Reopen)

IsSimpleMenuOpen<public>(MenuId:string, Player:player)<decides><transacts>:void=
    MenuCache := g_MenuCache[GetSession()]
    MenuCache[MenuId].Players[Player].ActiveMenu?

SimpleCloseAllOpenMenus<public>(Player:player):void=
    if (MenuCache := g_MenuCache[GetSession()]) {
        var ClosedCount: int = 0
        for (
            MenuId->CacheState:MenuCache
            MenuInstance := CacheState.Players[Player].ActiveMenu?
        ) {
            MenuInstance.Close()
        }
        if (ClosedCount > 0) {
            Log("Closed {ClosedCount} open menus for '{Player}'", ?Level := log_level.Debug)
        }
    }

SimpleMenuHandleMenuClose<internal>(Player:player, MenuInstance:ui_menu_instance):void=
    if (MenuCache := g_MenuCache[GetSession()]) {
        for (
            MenuId->CacheState:MenuCache
            CacheState.Players[Player].ActiveMenu? = MenuInstance
        ) {
            set CacheState.Players = CacheState.Players.RemoveByKey(Player)(0)
        }
    }

simple_ui_framework_menu<internal> := class<final>(ui_menu):
    Component: ui_menu_component

    CreateInstance<override>(Player:player):tuple(ui_menu_instance,canvas)=
        Instance := ui_menu_instance:
            Menu := Self
            Player := Player
        Widget := Instance.AddComponent(Component)
        Canvas := if (Component.RootWidget?) {
            canvas{
                Slots := array{
                    canvas_slot{
                        Widget := Widget
                        SizeToContent := false
                        Anchors := anchors{
                            Minimum := vector2{X := 0.0, Y := 0.0}
                            Maximum := vector2{X := 1.0, Y := 1.0}
                        }
                        Alignment := vector2{X := 0.0, Y := 0.0}
                    }
                }
            }
        } else {
            GetSimpleCanvas(Widget)
        }
        (Instance, Canvas)

# DEPRECATED. Use ui_menu_component instead.
dynamic_ui_menu_element<public> := class(ui_element):
    Placeholder<public>: overlay
    WidgetFactory<public>: type{__(:ui_element):widget}

    var m_Widget: ?widget = false

    Update<override>():void=
        if (CurrentWidget := m_Widget?) {
            Placeholder.RemoveWidget(CurrentWidget)
        }
        NewWidget := WidgetFactory(Self)
        set m_Widget = Some(NewWidget)
        Placeholder.AddWidget(overlay_slot{
            Widget := NewWidget
            HorizontalAlignment := horizontal_alignment.Fill
            VerticalAlignment := vertical_alignment.Fill
        })

    Discard<override>():void=
        if (CurrentWidget := m_Widget?) {
            Placeholder.RemoveWidget(CurrentWidget)
            set m_Widget = None()
        }

# DEPRECATED. Use ui_menu_component instead.
CreateDynamicMenuElement<public>(WidgetFactory:type{__(:ui_element):widget}):dynamic_ui_menu_element=
    Placeholder := overlay{}
    Element := dynamic_ui_menu_element{
        Placeholder := Placeholder
        WidgetFactory := WidgetFactory
    }
    Element.Update()
    Element
