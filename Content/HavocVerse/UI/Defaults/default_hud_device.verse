using { /Fortnite.com/Devices }
using { /Fortnite.com/UI }
using { /Verse.org/Simulation }
using { /UnrealEngine.com/Temporary/Diagnostics }
using { /UnrealEngine.com/Temporary/UI }
using { /UnrealEngine.com/Temporary/SpatialMath }

using { HavocVerse.Debug }
using { HavocVerse.Devices }
using { HavocVerse.Logging }
using { HavocVerse.UI }
using { HavocVerse.UI.Defaults }
using { HavocVerse.Util }

# Device to display default HUDs.
default_hud_device := class(creative_device, aware_device):
    @editable p_AccessControl: ?access_control_device = false

    @editable p_ResourceManager: resource_manager_device = resource_manager_device{}

    @editable p_HUDManagerConfig: hud_manager_base_config = hud_manager_base_config{}
    @editable p_ResourceSidebarEnabled: logic = false
    @editable p_ResourceSidebarConfig: hud_resource_sidebar_config = hud_resource_sidebar_config{}
    @editable p_ResourceSidebarLines: []hud_resource_sidebar_line_config = array{}
    @editable p_ResourceSidebarPlacement: ui_canvas_placement_config = ui_canvas_placement_config{
        Anchor := vector2{X := 0.0, Y := 0.5}
        Offsets := ui_margin_config{Left := 10.0}
        Alignment := vector2{X := 0.0, Y := 0.5}
    }

    @editable p_StaticBackgroundImage: string = ""
    @editable p_StaticBackgroundSize: vector2 = vector2{X := 100.0, Y := 50.0}
    @editable p_StaticBackgroundPlacement: ui_canvas_placement_config = ui_canvas_placement_config{
        Anchor := vector2{X := 0.0, Y := 0.5}
        Offsets := ui_margin_config{Left := 10.0}
        Alignment := vector2{X := 0.0, Y := 0.5}
    }

    var m_HudManager: ?default_hud_device_hud_manager = false

    DeviceMeta<override>()<transacts>:aware_device_meta=
        aware_device_meta{
            DeviceName := "default_hud"
        }

    OnBegin<override>()<suspends>:void=
        AwareInitDevice(Self, OnInit)

    OnInit<private>()<suspends>:void=
        HudManager := default_hud_device_hud_manager{
            AccessControl := p_AccessControl
            RefreshRate := 0.1
            ZOrder := p_HUDManagerConfig.ZOrder
            Device := Self
        }
        set m_HudManager = Some(HudManager)
        p_ResourceManager.Setup()
        spawn:
            HudManager.Run(GetPlayspace())

default_hud_device_hud_manager := class(hud_manager_base):
    Device: default_hud_device

    CreateHUD<override>(Player:player):?hud_manager_player=
        var Elements: []hud_element = array{}
        var CanvasSlots: []canvas_slot = array{}

        BackgroundWidget := CreateOptionalImage(Device.p_StaticBackgroundImage, Device.p_StaticBackgroundSize)
        if (l_BackgroundWidget := BackgroundWidget?) {
            set CanvasSlots += array{Device.p_StaticBackgroundPlacement.ToCanvasSlot(l_BackgroundWidget)}
        }

        if (Device.p_ResourceSidebarEnabled?, AssetsProvider := AwareDefaultAssetsProvider[]) {
            ResourceSidebar := CreateHUDResourceSidebar(
                AssetsProvider,
                Device.p_ResourceSidebarConfig,
                Device.p_ResourceSidebarLines,
                Player,
                Device.p_ResourceManager
            )
            set Elements += array{ResourceSidebar}
            set CanvasSlots += array{Device.p_ResourceSidebarPlacement.ToCanvasSlot(ResourceSidebar.Widget)}
        }

        if (Elements.Length > 0) {
            Some(hud_manager_player_default{
                Player := Player
                Canvas := canvas{
                    Slots := CanvasSlots
                }
                Elements := Elements
            })
        } else {
            false
        }
