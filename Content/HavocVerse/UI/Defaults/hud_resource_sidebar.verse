using { /Fortnite.com/UI }
using { /Verse.org/Simulation }
using { /Verse.org/Colors }
using { /UnrealEngine.com/Temporary/Diagnostics }
using { /UnrealEngine.com/Temporary/UI }
using { /UnrealEngine.com/Temporary/SpatialMath }

using { HavocVerse.Devices }
using { HavocVerse.Resource }
using { HavocVerse.UI }
using { HavocVerse.Util }

using { HavocVerse.Logging }

hud_resource_sidebar_config<public> := struct<concrete>:
    @editable InnerPadding<public>: float = 8.0
    @editable LineSize<public>: vector2 = vector2{X := 350.0, Y := 84.0}
    @editable AmountLeftOffset<public>: float = 90.0
    @editable RateRightOffset<public>: float = 14.0
    @editable DefaultRateFormatter: number_formatter_config = number_formatter_config{
        Template := "%Val%/s"
    }

hud_resource_sidebar_line_config<public> := struct<concrete>:
    @editable ResourceId<public>: string = ""
    @editable AmountFormatter<public>: []number_formatter_config = array{}
    @editable ShowGenerationRate<public>: logic = false
    @editable RateFormatter<public>: []number_formatter_config = array{}
    @editable LineTexture<public>: string = ""
    @editable AmountColor<public>: color = NamedColors.White
    @editable RateLeftAligned<public>: ?float = false
    @editable RateColor<public>: color = NamedColors.White
    @editable RateColoringValueSource<public>: hud_resource_sidebar_rate_color_threshold_source = hud_resource_sidebar_rate_color_threshold_source.Rate
    @editable RateColoringThresholds<public>: []hud_resource_sidebar_rate_color_threshold = array{}
    @editable MultiplierIndicatorPlacement<public>: ui_canvas_placement_config = ui_canvas_placement_config{}

hud_resource_sidebar_rate_color_threshold_source<public> := enum:
    Rate
    Multiplier

hud_resource_sidebar_rate_color_threshold<public> := class<concrete>:
    @editable ValueThreshold<public>: float = 0.0
    @editable Color<public>: color = NamedColors.White

CreateHUDResourceSidebar<public>(
    AssetsProvider:assets_provider,
    Config:hud_resource_sidebar_config,
    Lines:[]hud_resource_sidebar_line_config,
    Player:player,
    ResourceManager:resource_manager_device
):hud_resource_sidebar_instance=
    LinesRes := CreateHUDResourceSidebarLines(AssetsProvider, Config, Lines)
    LineInstances := LinesRes(0)
    StackSlots := LinesRes(1)
    hud_resource_sidebar_instance{
        Widget := stack_box{
            Orientation := orientation.Vertical
            Slots := StackSlots
        }
        Player := Player
        ResourceManager := ResourceManager
        Lines := LineInstances
        DefaultRateFormatter := Config.DefaultRateFormatter.CreateFormatter()
    }

CreateHUDResourceSidebarLines<public>(
    AssetsProvider:assets_provider,
    Config:hud_resource_sidebar_config,
    Lines:[]hud_resource_sidebar_line_config
):tuple([]hud_resource_sidebar_instance_line,[]stack_box_slot)=
    LineInstances := for (LineConfig:Lines) {
        CreateHUDResourceSidebarLine(AssetsProvider, Config, LineConfig)
    }
    StackSlots := for (Index->Line:LineInstances) {
        stack_box_slot{
            Widget := Line.Widget
            VerticalAlignment := vertical_alignment.Top
            HorizontalAlignment := horizontal_alignment.Left
            Padding := if (Index > 0) {
                margin{
                    Top := Config.InnerPadding
                }
            } else {
                margin{}
            }
        }
    }
    (LineInstances, StackSlots)

CreateHUDResourceSidebarLine<internal>(
    AssetsProvider:assets_provider,
    SidebarConfig:hud_resource_sidebar_config,
    LineConfig:hud_resource_sidebar_line_config
):hud_resource_sidebar_instance_line=
    AmountWidget := text_block{
        DefaultText := ToMessage("")
        DefaultTextColor := LineConfig.AmountColor
        DefaultJustification := text_justification.Left
    }
    LineBackgroundWidget := if (
        LineConfig.LineTexture.Length > 0
        Texture := AssetsProvider.GetTextureAsset[LineConfig.LineTexture]
    ) {
        texture_block{
            DefaultImage := Texture
            DefaultDesiredSize := SidebarConfig.LineSize
        }
    } else {
        color_block{
            DefaultOpacity := 0.0
            DefaultDesiredSize := SidebarConfig.LineSize
        }
    }
    var OverlaySlots: []overlay_slot = array{
        overlay_slot{
            Widget := LineBackgroundWidget
        }
        overlay_slot{
            HorizontalAlignment := horizontal_alignment.Left
            VerticalAlignment := vertical_alignment.Center
            Padding := margin{
                Left := SidebarConfig.AmountLeftOffset
            }
            Widget := AmountWidget
        }
    }
    RateWidget: ?text_block = if (LineConfig.ShowGenerationRate?) {
        Widget := text_block{
            DefaultText := ToMessage("")
            DefaultTextColor := LineConfig.RateColor
            DefaultJustification := text_justification.Left
        }
        if (LeftPadding := LineConfig.RateLeftAligned?) {
            set OverlaySlots += array{overlay_slot{
                HorizontalAlignment := horizontal_alignment.Left
                VerticalAlignment := vertical_alignment.Center
                Padding := margin{
                    Left := LeftPadding
                }
                Widget := Widget
            }}
        } else {
            set OverlaySlots += array{overlay_slot{
                HorizontalAlignment := horizontal_alignment.Right
                VerticalAlignment := vertical_alignment.Center
                Padding := margin{
                    Right := SidebarConfig.RateRightOffset
                }
                Widget := Widget
            }}
        }
        Some(Widget)
    } else {
        false
    }
    LineWidget := overlay{
        Slots := OverlaySlots
    }
    hud_resource_sidebar_instance_line{
        Widget := LineWidget
        ResourceId := LineConfig.ResourceId
        AmountFormatter := option{LineConfig.AmountFormatter[0].CreateFormatter()}
        AmountWidget := MakeCachedTextBlock(AmountWidget)
        RateFormatter := option{LineConfig.RateFormatter[0].CreateFormatter()}
        RateWidget := MakeCachedTextBlock(RateWidget)
        DefaultRateColor := LineConfig.RateColor
        RateColoringValueSource := LineConfig.RateColoringValueSource
        RateColoringOrderedThresholds := LineConfig.RateColoringThresholds.QuickSortWithOpts(CmpRateColoringThresholds, quick_sort_opts{
            Order := sort_order.Descending
        })
    }

hud_resource_sidebar_instance_line<public> := struct:
    Widget<public>: overlay

    ResourceId<public>: string
    AmountFormatter<public>: ?number_formatter
    AmountWidget<public>: cached_text_block
    RateFormatter<public>: ?number_formatter
    RateWidget<public>: ?cached_text_block

    DefaultRateColor<public>: color
    RateColoringValueSource<public>: hud_resource_sidebar_rate_color_threshold_source
    RateColoringOrderedThresholds<public>: []hud_resource_sidebar_rate_color_threshold

hud_resource_sidebar_instance<public> := class(hud_element):
    Player<public>: player
    ResourceManager<public>: resource_manager_device

    Lines<public>: []hud_resource_sidebar_instance_line
    DefaultRateFormatter<public>: number_formatter

    Update<override>():void=
        for (Line:Self.Lines) {
            ResourceAmountStr := if (AmountFormmatter := Line.AmountFormatter?) {
                ResourceAmount := ResourceManager.GetAmount(Player, Line.ResourceId)
                AmountFormmatter.Format(ResourceAmount)
            } else {
                ResourceManager.GetFormattedAmount(Player, Line.ResourceId)
            }
            Line.AmountWidget.SetText(ResourceAmountStr)

            if (RateWidget := Line.RateWidget?) {
                Rate := ResourceManager.GetAutoGenerationRate(Player, Line.ResourceId)
                RateStr := if (RateFormatter := Line.RateFormatter?) {
                    RateFormatter.Format(Rate)
                } else {
                    DefaultRateFormatter.Format(Rate)
                }
                RateWidget.SetText(RateStr)

                ColoringValue := case(Line.RateColoringValueSource):
                    hud_resource_sidebar_rate_color_threshold_source.Rate => Rate
                    hud_resource_sidebar_rate_color_threshold_source.Multiplier =>
                        Value := ResourceManager.GetEffectiveMultiplier(Player, Line.ResourceId)
                        #Log("Coloring rate of '{Line.ResourceId}' with effective multiplier {Value}", ?Level := log_level.Verbose)
                        Value
                var FoundThreshold: logic = false
                for (Index->Threshold:Line.RateColoringOrderedThresholds, not (FoundThreshold?)) {
                    if (ColoringValue >= Threshold.ValueThreshold) {
                        RateWidget.SetTextColor(Threshold.Color)
                        set FoundThreshold = true
                    }
                }
                if (not (FoundThreshold?)) {
                    RateWidget.SetTextColor(Line.DefaultRateColor)
                }
            }
        }

CmpRateColoringThresholds(
    A:hud_resource_sidebar_rate_color_threshold,
    B:hud_resource_sidebar_rate_color_threshold
)<computes>:int=
    CmpFloat(A.ValueThreshold, B.ValueThreshold)
