using { /Fortnite.com/Devices }
using { /Verse.org/Simulation }
using { /UnrealEngine.com/Temporary/Diagnostics }

using { HavocVerse.Debug }
using { HavocVerse.Devices }
using { HavocVerse.Persistence }
using { HavocVerse.Util }
using { HavocVerse.Logging }

verse_bits_storage_data<public> := class<final><persistable>:
    Bits: []logic = array{}

MakeVerseBitsStorageData<internal><constructor>(CopyOf:verse_bits_storage_data)<transacts> := verse_bits_storage_data:
    Bits := CopyOf.Bits

verse_bits_storage_device<public> := class(creative_device, storage_memory_device, aware_device):
    @editable p_StorageId: string = "verse_storage0"
    @editable p_Size: int = 1000
    @editable p_FixedLayout: storage_memory_fixed_layout_config = storage_memory_fixed_layout_config{}

    var m_StorageMem: ?master_chunk_storage_memory = false

    DeviceMeta<override>()<transacts>:aware_device_meta=
        aware_device_meta{
            DeviceName := "verse_bits_storage"
            DevId := p_StorageId
        }

    OnBegin<override>()<suspends>:void=
        AwareInitDevice(Self, OnInit)

    OnInit<private>()<suspends>:void=
        EnsureReady()

    EnsureReady<override>():void=
        if (not (m_StorageMem?)) {
            MasterChunk := verse_bits_storage_master_chunk{
                Size := Size()
                Device := Self
            }
            StorageMem := master_chunk_storage_memory{
                MasterChunk := MasterChunk
                FixedLayout := p_FixedLayout

                ContextStartSentence := "Verse bits storage \"{p_StorageId}\""
                ContextInSentence := "Verse bits storage \"{p_StorageId}\""
            }
            set m_StorageMem = Some(StorageMem)
            StorageMem.EnsureReady()
        }

    GetID<override>()<computes>:string=
        "\"{p_StorageId}\""

    Size<override>()<computes>:int=
        p_Size

    Reserve<override>(ChunkSize:int, LayoutId:string)<transacts><decides>:storage_memory_chunk=
        if (not (m_StorageMem?)) {
            Log("Verse bits storage {GetID()} is not yet initialized (trying to reserve for chunk layout ID \"{LayoutId}\")", ?Level := log_level.Error)
            false?
        }
        m_StorageMem?.Reserve[ChunkSize, LayoutId]

    GetStorageMem<public>()<transacts><decides>:master_chunk_storage_memory=
        m_StorageMem?

    Dump<override>(Agent:agent, Detailed:logic):void=
        if (StorageMem := m_StorageMem?) {
            StorageMem.Dump(Agent, Detailed)
        }

    Fill<override>(Agent:agent, Value:logic):void=
        if (StorageMem := m_StorageMem?) {
            StorageMem.MasterChunk.Write(Agent, array{Value}.Repeat(StorageMem.MasterChunk.Size))
        }

    Reset<public>(Agent:agent):void=
        if:
            Player := player[Agent]
            PlayerDataSetRawBitsStore(Player, p_StorageId, verse_bits_storage_data{
                Bits := array{false}.Repeat(p_Size)
            })

verse_bits_storage_master_chunk<internal> := class(storage_memory_chunk):
    Device: verse_bits_storage_device

    Read<override>(Agent:agent, ?Offset:int = 0, ?Limit:int = -1, ?Raw:logic = false):[]logic=
        State := if (
            Player := player[Agent]
            Player.IsActive[]
            StorageData := PlayerDataGetRawBitsStore[Player, Device.p_StorageId]
        ) {
            StorageData.Bits
        } else {
            return EmptyState()
        }
        ReadFrom := Offset
        ReadTo := if (Limit < 0). Size else Min(Size, ReadFrom + Limit)
        State.SafeSlice(ReadFrom, ReadTo)

    Write<override>(Agent:agent, Data:[]logic, ?Offset:int = 0):void=
        Player := if (l_Player := player[Agent], l_Player.IsActive[]) {
            l_Player
        } else {
            return
        }
        StorageData := if (l_StorageData := PlayerDataGetRawBitsStore[Player, Device.p_StorageId]) {
            l_StorageData
        } else {
            PlayerDataSetRawBitsStore(Player, Device.p_StorageId, verse_bits_storage_data{
                Bits := array{false}.Repeat(Device.p_Size)
            })
        }

        var AnyUpdate: logic = false
        var NewData: []logic = StorageData.Bits
        if (NewData.Length < Device.p_Size) {
            Log(MsgExpandingStorageSize(Device.GetID(), Agent, NewData.Length, Device.p_Size))
            set NewData = NewData + array{false}.Repeat(Device.p_Size - NewData.Length)
            set AnyUpdate = true
        }

        for (
            Index->Bit:Data
            MemIndex := Offset + Index
            MemIndex < Size
            set NewData[MemIndex] = Bit
        ) {
            if (not (AnyUpdate?)) {
                set AnyUpdate = true
            }
        }

        if (AnyUpdate?) {
            PlayerDataSetRawBitsStore(Player, Device.p_StorageId, verse_bits_storage_data{
                MakeVerseBitsStorageData<constructor>(StorageData)
                Bits := NewData
            })
        }

    MsgExpandingStorageSize<private><localizes>(DevId:string, Agent:agent, PrevSize:int, NewSize:int):message=
        "Expanding storage size on verse bits storage {DevId} from {PrevSize} bits to {NewSize} bits for '{Agent}'"
