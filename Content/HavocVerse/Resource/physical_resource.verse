using { /Fortnite.com/Devices }
using { /Verse.org/Simulation }
using { /UnrealEngine.com/Temporary/Diagnostics }

using { HavocVerse.Devices }
using { HavocVerse.Logging }
using { HavocVerse.Util }

physical_resource<public> := class(resource_base):
    ResourceId<public>: string

    BulkGranter<public>: ?bulk_resource_granter_device
    CondButton<public>: ?conditional_button_device

    GetAmount<override>(Agent:?agent):int=
        if (l_Agent := Agent?) {
            if (l_CondButton := CondButton?) {
                return l_CondButton.GetItemCount(l_Agent, 0)
            } else {
                Log("Physical resource {ResourceId} doesn't support GetAmount operation (missing cond button)", ?Level := log_level.Error)
            }
        } else {
            LogAgentlessOperation("GetAmount")
        }
        -1

    Give<override>(Agent:?agent, Amount:int):logic=
        if (l_Agent := Agent?) {
            if (l_BulkGranter := BulkGranter?) {
                if (Amount = 0) {
                    return true
                }
                l_BulkGranter.Give(l_Agent, Amount)
                return true
            } else {
                Log("Physical resource {ResourceId} doesn't support Give operation (missing bulk granter)", ?Level := log_level.Error)
            }
        } else {
            LogAgentlessOperation("Give")
        }
        false

    Take<override>(Agent:?agent, Amount:int):logic=
        if (l_Agent := Agent?) {
            if (l_CondButton := CondButton?) {
                if (Amount = 0) {
                    return true
                }
                ItemCount := l_CondButton.GetItemCount(l_Agent, 0)
                if (ItemCount >= Amount) {
                    l_CondButton.SetItemCountRequired(0, Amount)
                    l_CondButton.Activate(l_Agent)
                    return true
                }
            } else {
                Log("Physical resource {ResourceId} doesn't support Take operation (bulk granter configuration)", ?Level := log_level.Error)
            }
        } else {
            LogAgentlessOperation("Take")
        }
        false

    LogAgentlessOperation<private>(Op:string):void=
        Log("Physical resources don't support agent-less {Op} (trace from {ResourceId})", ?Level := log_level.Warning)
