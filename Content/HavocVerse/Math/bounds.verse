using { /Fortnite.com/Devices }
using { /Fortnite.com/Game }
using { /Verse.org/Random }
using { /Verse.org/Verse }
using { /Verse.org/Simulation }
using { /Verse.org/Colors }
using { /UnrealEngine.com/Temporary/SpatialMath }
using { /UnrealEngine.com/Temporary/Diagnostics }

using { HavocVerse.Debug }
using { HavocVerse.Logging }
using { HavocVerse.Util }

BoundsFromMarkers<public>(MarkerA:creative_object, MarkerB:creative_object, ?InfiniteZ:logic = false)<transacts>:bounds=
    if (
        PropA := creative_prop[MarkerA]
        PropB := creative_prop[MarkerB]
        (not PropA.IsValid[]) or (not PropB.IsValid[])
    ) {
        Log("Creating bounds from prop markers that are no longer valid!", ?Level := log_level.Warning)
    }
    BoundsFromMarkers(
        MarkerA.GetTransform().Translation,
        MarkerB.GetTransform().Translation,
        ?InfiniteZ := InfiniteZ
    )

BoundsFromMarkers<public>(MarkerA:vector3, MarkerB:vector3, ?InfiniteZ:logic = false)<transacts>:bounds=
    bounds{
        Lower := vector3{
            X := Min(MarkerA.X, MarkerB.X),
            Y := Min(MarkerA.Y, MarkerB.Y),
            Z := Min(MarkerA.Z, MarkerB.Z)
        },
        Upper := vector3{
            X := Max(MarkerA.X, MarkerB.X),
            Y := Max(MarkerA.Y, MarkerB.Y),
            Z := Max(MarkerA.Z, MarkerB.Z)
        },
        InfiniteZ := InfiniteZ
    }

bounds_config<public> := class<concrete>:
    @editable MarkerA<public>: creative_prop = creative_prop{}
    @editable MarkerB<public>: creative_prop = creative_prop{}
    @editable InfiniteZ<public>: logic = true

    ToBounds<public>()<transacts>:bounds=
        BoundsFromMarkers(MarkerA, MarkerB, ?InfiniteZ := InfiniteZ)

bounds<public> := class:
    Lower<public>: vector3 = vector3{}
    Upper<public>: vector3 = vector3{}
    InfiniteZ<public>: logic = false
    
    RandomPoint<public>()<transacts>:vector3=
        vector3{
            X := GetRandomFloat(Lower.X, Upper.X),
            Y := GetRandomFloat(Lower.Y, Upper.Y),
            Z := GetRandomFloat(Lower.Z, Upper.Z)
        }

    CenterPoint<public>()<transacts>:vector3=
        vector3{
            X := (Lower.X + Upper.X) / 2.0,
            Y := (Lower.Y + Upper.Y) / 2.0,
            Z := (Lower.Z + Upper.Z) / 2.0
        }

    IsWithin<public>(Positional:positional)<transacts><decides>:void=
        IsWithin[Positional.GetTransform().Translation]

    IsWithin<public>(Position:vector3)<transacts><decides>:void=
        Position.X >= Lower.X and Position.X <= Upper.X and
            Position.Y >= Lower.Y and Position.Y <= Upper.Y and
            (InfiniteZ? or (Position.Z >= Lower.Z and Position.Z <= Upper.Z))

    Intersects<public>(Other:bounds)<transacts><decides>:void=
        Lower.X <= Other.Upper.X and Upper.X >= Other.Lower.X and
            Lower.Y <= Other.Upper.Y and Upper.Y >= Other.Lower.Y and
            (((InfiniteZ?) or (Other.InfiniteZ?)) or (Lower.Z <= Other.Upper.Z and Upper.Z >= Other.Lower.Z))

    DrawDebugBox<public>(
        DebugDraw:debug_draw,
        ?Color:color = NamedColors.Yellow,
        ?Thickness:float = 1.0,
        ?DrawDurationPolicy:debug_draw_duration_policy = debug_draw_duration_policy.SingleFrame,
        ?Duration:float = 0.0
    )<transacts>:void=
        Center := CenterPoint()
        var HalfExtent: vector3 = (Upper - Lower) / 2.0
        if (InfiniteZ?) {
            set HalfExtent = HalfExtent.WithZ(10000.0)
        }
        DebugDraw.DrawBox(Center, IdentityRotation(), ?Extent := HalfExtent,
            ?Color := Color, ?Thickness := Thickness,
            ?DrawDurationPolicy := DrawDurationPolicy, ?Duration := Duration)

    DrawDebugBox<public>(DebugDraw:debug_draw, Args:debug_draw_args)<transacts>:void=
        DrawDebugBox(DebugDraw, ?Color := Args.Color, ?Thickness := Args.Thickness,
            ?DrawDurationPolicy := Args.DrawDurationPolicy, ?Duration := Args.Duration)

IsWithinAnyBounds<public>(AllBounds:[]bounds, Positional:positional)<transacts><decides>:void=
    IsWithinAnyBounds0(AllBounds, Positional)?

IsWithinAnyBounds<public>(AllBounds:[]bounds, Positional:vector3)<transacts><decides>:void=
    IsWithinAnyBounds0(AllBounds, Positional)?

IsWithinAnyBounds0<internal>(AllBounds:[]bounds, Positional:vector3)<transacts>:logic=
    for (Bounds:AllBounds, Bounds.IsWithin[Positional]) {
        return true
    }
    false

IsWithinAnyBounds0<internal>(AllBounds:[]bounds, Positional:positional)<transacts>:logic=
    for (Bounds:AllBounds, Bounds.IsWithin[Positional]) {
        return true
    }
    false

RandomSphereOffset<public>(Radius:float)<transacts>:vector3=
    Theta := GetRandomFloat(0.0, 2 * PiFloat)
    Phi := GetRandomFloat(0.0, PiFloat)
    X := Radius * Sin(Phi) * Cos(Theta)
    Y := Radius * Sin(Phi) * Sin(Theta)
    Z := Radius * Cos(Phi)
    vector3{X := X, Y := Y, Z := Z}

ToString<public>(Bounds:bounds)<transacts>:string=
    "\{Lower: {Bounds.Lower}, Upper: {Bounds.Upper}, InfiniteZ: {LogicToString(Bounds.InfiniteZ)}\}"
