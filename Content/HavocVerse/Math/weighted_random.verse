using { /Verse.org/Random }

using { HavocVerse.Util }

weighted_random_entry<public>(t:type) := class<final> {
	Value<public>: t
	Weight<public>: float
}

# Get a random value out of a set of weighted entries.
# Fails if the entries array is empty.
WeightedRandom<public>(Entries:[]tuple(t,float) where t:type)<decides><transacts>: t ={
	Entries.Length > 0

	var TotalWeight: float = 0.0
	for (Entry:Entries) {
		set TotalWeight += Entry(1)
	}

	var Rand: float = GetRandomFloat(0.0, TotalWeight)
	ShuffledEntries := Shuffle(Entries)
	var Result: ?t = None()
	for (Entry:ShuffledEntries, not Result?) {
		set Rand -= Entry(1)
		if (Rand <= 0.0) {
			set Result = Some(Entry(0))
		}
	}
	Result?
}
