using { /Fortnite.com/Devices }
using { /Verse.org/Simulation }
using { /UnrealEngine.com/Temporary/Diagnostics }

using { HavocVerse.Debug }
using { HavocVerse.Util }

GlobalLogUnboundStackMaxLength: int = 10

var g_GlobalLogUnboundStack<public>: weak_map(session,[]tuple(string,log_level,float)) = map{}

GlobalLog_DebugStats<public>()<transacts>:aware_debug_stats=
    var Result: [string]int = map{}
    if (Stack := g_GlobalLogUnboundStack[GetSession()]) {
        set Result = ConcatenateMaps(Result, map{
            "unboundLogStack" => Stack.Length
        })
    }
    aware_debug_stats{
        Length := Result
    }

Log<public>(Device:creative_device, Message:string, ?Level:log_level = log_level.Normal)<transacts>:void=
    # TODO : fancier system to handle device context
    Log("{Device}: {Message}", ?Level := Level)

Log<public>(Device:creative_device, Message:message, ?Level:log_level = log_level.Normal):void=
    # TODO : fancier system to handle device context
    Log(ConcatMessage("{Device}: ", Message), ?Level := Level)

Log<public>(Message:string, ?Level:log_level = log_level.Normal)<transacts>:void=
    if (Device := g_DebugDevice[GetSession()]) {
        Device.LogMessage(Message, ?Level := Level)
    } else {
        Logger: log = log{Channel := havoc_log_chan}
        Logger.Print(Message, ?Level := Level)

        if (GlobalLogUnboundStackMaxLength <= 0) {
            return
        }
        LogEntry := (Message, Level, GetSimulationElapsedTime())
        var LogStack: []tuple(string,log_level,float) = if (l_LogStack := g_GlobalLogUnboundStack[GetSession()]) {
            l_LogStack + array{LogEntry}
        } else {
            array{LogEntry}
        }
        if (LogStack.Length > GlobalLogUnboundStackMaxLength, l_Slice := LogStack.Slice[LogStack.Length - GlobalLogUnboundStackMaxLength]) {
            set LogStack = l_Slice
        }
        if:
            set g_GlobalLogUnboundStack[GetSession()] = LogStack
    }

Log<public>(Message:message, ?Level:log_level = log_level.Normal)<transacts>:void=
    if (Device := g_DebugDevice[GetSession()]) {
        Device.LogMessage(Message, ?Level := Level)
    } else if (Level <> log_level.Verbose) {
        Logger: log = log{Channel := havoc_log_chan}
        Logger.Print("Attempted to Log a non-string `message`. This requires a debug device", ?Level := log_level.Error)
    }

DumpStackTrace<public>(?Level:log_level = log_level.Normal):void=
    if (Device := g_DebugDevice[GetSession()]) {
        Device.Logger.PrintCallStack(?Level := Level)
    } else {
        Logger: log = log{Channel := havoc_log_chan}
        Logger.PrintCallStack(?Level := Level)
    }

LogEnabled<public>(Level:log_level)<transacts><decides>:void=
    if (Device := g_DebugDevice[GetSession()]) {
        GetLogLevelIndex(Device.m_LogLevel) <= GetLogLevelIndex(Level)
    } else {
        Level <> log_level.Verbose
    }

GetLogLevelIndex<public>(LogLevel:log_level)<computes>:int=
    case(LogLevel):
        log_level.Verbose => 0
        log_level.Debug => 1
        log_level.Normal => 2
        log_level.Warning => 3
        log_level.Error => 4

GetLogLevelFromIndex<public>(Index:int)<computes><decides>:log_level=
    case(Index):
        0 => log_level.Verbose
        1 => log_level.Debug
        2 => log_level.Normal
        3 => log_level.Warning
        4 => log_level.Error
