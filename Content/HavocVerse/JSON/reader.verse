using { HavocVerse.Util }

json_reader_frame_kind := enum:
    Object
    Array

json_reader_frame := class:
    Kind: json_reader_frame_kind
    var m_Index: int = 0

CreateJsonReader<public><constructor>(Input:string)<transacts>:=
    json_reader{
        Source := Input
    }

# Wrapper type required because a function that <decides> cannot
# return a `logic` for "semantic reasons". Verse is such a joke.
json_boolean<public> := class:
    Value<public>: logic

# Basic JSON reader.
# Does not handle errors well (no reports) and does not comply with RFC 8259.
# It brings JSON support to Verse though, and that's good enough.
# It's intended purpose is to read JSON that was already checked to be valid (e.g. pass as input of a device).
json_reader<public> := class:
    Source<public>: string

    var m_Cursor: int = 0
    var m_Stack: []json_reader_frame = array{}
    var m_StackIndex: int = -1 # maintained as a separate variable because Verse is a flawed and/or bugged language

    EatWhitespace<private>()<transacts>:void=
        loop {
            if (Ch := Source[m_Cursor], Ch = ' ' or Ch = '\t') {
                set m_Cursor += 1
            } else {
                break
            }
        }

    CurrentFrame<private>()<transacts><decides>:json_reader_frame=
        Frame := m_Stack[m_StackIndex]
        if (m_Stack.Length-1 > m_StackIndex) {
            # Verse has a bug (or design flaw) with rolling back arrays
            NewStack := m_Stack.Slice[0, m_StackIndex+1]
            set m_Stack = NewStack
        }
        Frame

    BeginObject<public>()<transacts><decides>:void=
        StartNext[Some(json_reader_frame_kind.Array)]
        EatWhitespace()
        Source[m_Cursor] = '{'
        set m_Cursor += 1
        set m_StackIndex += 1
        set m_Stack += array{json_reader_frame{
            Kind := json_reader_frame_kind.Object
        }}

    EndObject<public>()<transacts><decides>:void=
        CurrentFrame[].Kind = json_reader_frame_kind.Object
        EatWhitespace()
        Source[m_Cursor] = '}'
        set m_Cursor += 1
        set m_StackIndex -= 1
        NewStack := m_Stack.Slice[0, m_Stack.Length-1]
        set m_Stack = NewStack

    BeginArray<public>()<transacts><decides>:void=
        StartNext[Some(json_reader_frame_kind.Array)]
        EatWhitespace()
        Source[m_Cursor] = '['
        set m_Cursor += 1
        set m_StackIndex += 1
        set m_Stack += array{json_reader_frame{
            Kind := json_reader_frame_kind.Array
        }}

    EndArray<public>()<transacts><decides>:void=
        CurrentFrame[].Kind = json_reader_frame_kind.Array
        EatWhitespace()
        Source[m_Cursor] = ']'
        set m_Cursor += 1
        set m_StackIndex -= 1
        NewStack := m_Stack.Slice[0, m_Stack.Length-1]
        set m_Stack = NewStack

    HasNext<public>()<transacts><decides>:void=
        EatWhitespace()
        Frame := CurrentFrame[]
        if (Frame.m_Index = 0) {
            case(Frame.Kind):
                json_reader_frame_kind.Object => Source[m_Cursor] = '"'
                json_reader_frame_kind.Array => Source[m_Cursor] <> ']'
        } else {
            Source[m_Cursor] = ','
        }

    StartNext<private>(ApplicableTo:?json_reader_frame_kind)<transacts><decides>:void=
        if ((not (ApplicableTo?)) or (ApplicableTo? = CurrentFrame[].Kind)) {
            EatWhitespace()
            Frame := CurrentFrame[]
            if (Frame.m_Index > 0) {
                Source[m_Cursor] = ','
                set m_Cursor += 1
            }
            set Frame.m_Index += 1
        }

    SkipNext<public>()<transacts><decides>:void=
        EatWhitespace()
        case(Source[m_Cursor]):
            '"' => NextString[]
            't' => NextBoolean[]
            'f' => NextBoolean[]
            '{' => {
                BeginObject[]
                loop {
                    if:
                        Name := NextName[]
                        SkipNext[]
                    else:
                        break
                }
                EndObject[]
            }
            '[' => {
                BeginArray[]
                loop {
                    if:
                        SkipNext[]
                    else:
                        break
                }
                EndArray[]
            }

    NextName<public>()<transacts><decides>:string=
        StartNext[false]
        EatWhitespace()
        Name := ReadString[]
        EatWhitespace()
        Source[m_Cursor] = ':'
        set m_Cursor += 1
        Name

    ReadString<private>()<transacts><decides>:string=
        Source[m_Cursor] = '"'
        set m_Cursor += 1

        var Result: string = ""
        var Escaping: logic = false
        loop {
            if (Char := Source[m_Cursor], Char <> '"' or Escaping?) {
                if (Char = '\\' and not (Escaping?)) {
                    set Escaping = true
                } else {
                    set Escaping = false
                    set Result += array{Char}
                }
                set m_Cursor += 1
            } else {
                break
            }
        }
        Source[m_Cursor] = '"'
        set m_Cursor += 1

        Result

    ReadIntWithLengthInfo<private>(?AllowNegative:logic = true)<transacts><decides>:tuple(int,int)=
        var IsNegative: logic = if (AllowNegative? and Source[m_Cursor] = '-') {
            set m_Cursor += 1
            true
        } else {
            false
        }

        var IntPart: int = 0
        var Length: int = 0
        loop {
            if (Char := Source[m_Cursor], Digit := GetDigit[Char]) {
                set IntPart = IntPart*10 + Digit
                set Length += 1
                set m_Cursor += 1
            } else {
                break
            }
        }
        Length > 0

        Result := if (IsNegative?) {
            -IntPart
        } else {
            IntPart
        }
        (Result, Length)

    ReadInt<private>(?AllowNegative:logic = true)<transacts><decides>:int=
        ReadIntWithLengthInfo[?AllowNegative := AllowNegative](0)

    ReadFloat<private>(?AllowNegative:logic = true)<transacts><decides>:float=
        var IsNegative: logic = if (AllowNegative? and Source[m_Cursor] = '-') {
            set m_Cursor += 1
            true
        } else {
            false
        }

        IntPart := ToFloat(ReadInt[])
        Result := if (Source[m_Cursor] = '.') {
            set m_Cursor += 1
            FracPartRes := ReadIntWithLengthInfo[?AllowNegative := false]
            FracPart := ToFloat(FracPartRes(0))
            FracPartLength := FracPartRes(1)
            IntPart + (FracPart / Pow(10.0, ToFloat(FracPartLength)))
        } else {
            IntPart
        }

        if (IsNegative?) {
            -Result
        } else {
            Result
        }

    NextString<public>()<transacts><decides>:string=
        StartNext[Some(json_reader_frame_kind.Array)]
        EatWhitespace()
        ReadString[]

    NextInt<public>()<transacts><decides>:int=
        StartNext[Some(json_reader_frame_kind.Array)]
        EatWhitespace()
        ReadInt[]

    NextFloat<public>()<transacts><decides>:float=
        StartNext[Some(json_reader_frame_kind.Array)]
        EatWhitespace()
        ReadFloat[]

    NextBoolean<public>()<transacts><decides>:json_boolean=
        StartNext[Some(json_reader_frame_kind.Array)]
        EatWhitespace()
        Expected := case(Source[m_Cursor]):
            't' => ("true", true)
            'f' => ("false", false)
        Source.Slice[m_Cursor, m_Cursor+Expected(0).Length] = Expected(0)
        json_boolean{
            Value := Expected(1)
        }

    DebugSnippet<public>(?ContextLength:int = 10)<transacts>:string=
        Start := Max(m_Cursor-ContextLength, 0)
        Context := if (Res := Source.Slice[Start, Min(m_Cursor+ContextLength, Source.Length)]) {
            Res
        } else {
            array{}
        }
        Ch := Source.ElementOrDefault(m_Cursor, '?')
        FrameKind := if (Frame := CurrentFrame[]) {
            case(Frame.Kind):
                json_reader_frame_kind.Object => "Object (index {Frame.m_Index})"
                json_reader_frame_kind.Array => "Array (index {Frame.m_Index})"
        } else {
            "None"
        }
        "\{>>{Context}<<\}[{m_Cursor - Start} - Ch '{Ch}', raw {m_Cursor}] - Frame {m_Stack.Length-1}:{FrameKind}"
