using { /Fortnite.com/Devices }
using { /Verse.org/Simulation }
using { /UnrealEngine.com/Temporary/Diagnostics }

using { HavocVerse.Debug }
using { HavocVerse.Devices.GameSpecific }
using { HavocVerse.Storage }
using { HavocVerse.Persistence }
using { HavocVerse.Logging }
using { HavocVerse.Util }

attachment_ref<public> := class<final><concrete> {
    @editable AttachmentKindId<public>: string = ""
    @editable Attachment<public>: string = ""
}

ToString<public>(Ref:attachment_ref): string ={
    "\{Kind:\"{Ref.AttachmentKindId}\", Attachment:\"{Ref.Attachment}\"\}"
}

attachments_manager_attachment_kind_config := class<final><concrete>:
    @editable AttachmentKindId: string = ""
    @editable DefaultActive: ?string = false
    @editable DefaultUnlocked: []string = array{}
    @editable PersistenceKeepAcrossTycoonRebirths: logic = false
    @editable VFXPerAttachment: []attachments_manager_attachment_vfx_config = array{}
    @editable VFXPerAttachmentClear: ?visual_effect_powerup_device = false

    @editable UserActions: []attachments_manager_attachment_kind_user_action_config = array{}

attachments_manager_attachment_vfx_config := class<final><concrete>:
    @editable Attachment: string = ""
    @editable VFX: visual_effect_powerup_device = visual_effect_powerup_device{}

attachments_manager_unlock_listener<internal> := class<final>:
    AttachmentKind: ?string
    Callback: type{__(:attachment_unlocked_event):void}

attachment_unlocked_event<public> := class<final>:
    Player<public>: player
    AttachmentKindId<public>: string
    Attachment<public>: string

attachments_manager_equip_listener<internal> := class<final>:
    AttachmentKind: ?string
    Callback: type{__(:attachment_equip_event):void}

attachment_equip_event<public> := class<final>:
    Player<public>: player
    AttachmentKindId<public>: string
    Attachment<public>: ?string

player_attachment_hook<public> := interface:
    EquipAttachment<public>(Player:player, AttachmentKindId:string, Attachment:string, Recall:logic):void

    UnequipAttachment<public>(Player:player, AttachmentKindId:string, Attachment:string, ReplacingWith:?string):void

    ShouldRecallOnRespawn<public>()<transacts>:logic

attachments_manager_hook_instance<internal> := class<final>:
    Hook: player_attachment_hook
    Filter: ?[string]void

    var m_ActivatedPlayers: [player]void = map{}

    IsApplicable(AttachmentKindId:string)<decides><transacts>:void=
        (not Filter?) or (Filter?[AttachmentKindId])

attachments_dynamic_unlock_source<public> := interface:
    IsAttachmentUnlocked<public>(Player:player, AttachmentKindId:string, Attachment:string)<transacts>:attachment_dynamic_unlock_state

attachment_dynamic_unlock_state<public> := enum:
    Unlocked
    Locked
    NotHandled

attachments_manager_device<public> := class<concrete><final>(creative_device, aware_device, aware_debug_device, persistence_resetable):
    @editable p_AttachmentKinds: []attachments_manager_attachment_kind_config = array{}

    var m_DynamicUnlockSources: [string][]attachments_dynamic_unlock_source = map{}
    var m_Hooks: []attachments_manager_hook_instance = array{}
    var m_SubscribedToPlayerSpawn: logic = false

    var m_UnlockListeners: []attachments_manager_unlock_listener = array{}
    var m_EquipListeners: []attachments_manager_equip_listener = array{}

    DeviceMeta<override>()<transacts>:aware_device_meta=
        aware_device_meta{
            DeviceName := "attachments_manager"
        }

    DebugStats<override>()<transacts>:aware_debug_stats=
        aware_debug_stats{
            Length := map{
                "listeners::unlock" => m_UnlockListeners.Length
                "listeners::equip" => m_EquipListeners.Length
            }
        }

    OnBegin<override>()<suspends>:void=
        AwareInitDevice(Self, OnInit)
    
    OnInit<private>()<suspends>:void=
        var AnyVFX: logic = false
        for (AttachmentKindConfig:p_AttachmentKinds) {
            for (UserActionConfig:AttachmentKindConfig.UserActions) {
                UserAction := attachments_manager_attachment_kind_user_action{
                    Device := Self
                    AttachmentKind := AttachmentKindConfig.AttachmentKindId
                    Config := UserActionConfig
                }
                UserActionConfig.OnTrigger.TriggeredEvent.Subscribe(UserAction.OnTriggered)
            }
            if (AttachmentKindConfig.VFXPerAttachment.Length > 0) {
                set AnyVFX = true
            }
        }

        GetPlayspace().PlayerAddedEvent().Subscribe(OnPlayerAdded)

        if (AnyVFX?) {
            RegisterHook(attachments_manager_vfx_player_attachment_hook{
                AttachmentsManager := Self
            })
        }

    # Registers a dynamic unlock source for a given attachment kind.
    #
    # WARNING: Dynamic sources make implementations need custom unlock events,
    #          and cannot rely on default unlock listeners.
    RegisterDynamicUnlockSource<public>(AttachmentKindId:string, Source:attachments_dynamic_unlock_source):void=
        if (PrevSources := m_DynamicUnlockSources[AttachmentKindId]) {
            if. set m_DynamicUnlockSources[AttachmentKindId] = PrevSources + array{Source}
        } else {
            if. set m_DynamicUnlockSources[AttachmentKindId] = array{Source}
        }

    # Registers a level hook.
    #
    # If AttachmentKindId is specified, the hook will be bound to that unique attachment kind.
    # Otherwise, the hook will be bound to all attachment kindsn.
    RegisterHook<public>(Hook:player_attachment_hook, ?AttachmentKindId:string = ""):void=
        if (AttachmentKindId.Length > 0, not GetAttachmentKindConfig[AttachmentKindId]) {
            Log("Registering attachment hook for unconfigured attachment kind \"{AttachmentKindId}\"", ?Level := log_level.Warning)
        }

        HookInstance := attachments_manager_hook_instance{
            Hook := Hook
            Filter := if (AttachmentKindId.Length > 0) {
                Some(map{AttachmentKindId => void})
            } else {
                None()
            }
        }
        set m_Hooks += array{HookInstance}
        HandlePostHookRegister(HookInstance)

    # Registers a level hook bound to specific level groups.
    RegisterHook<public>(Hook:player_attachment_hook, FilterKinds:[]string):void=
        for (
            AttachmentKindId:FilterKinds
            not GetAttachmentKindConfig[AttachmentKindId]
        ) {
            Log("Registering attachment hook for unconfigured attachment kind \"{AttachmentKindId}\"", ?Level := log_level.Warning)
        }

        HookInstance := attachments_manager_hook_instance{
            Hook := Hook
            Filter := Some(FilterKinds.ToSet())
        }
        set m_Hooks += array{HookInstance}
        HandlePostHookRegister(HookInstance)

    HandlePostHookRegister<private>(HookInstance:attachments_manager_hook_instance):void=
        for (Player:GetPlayspace().GetPlayers()) {
            ProcessHookPlayer(HookInstance, Player)
        }
        if (not m_SubscribedToPlayerSpawn?, Integration := AwareGetIntegration[]) {
            var AnyRequiresSpawnListener: logic = false
            for (i_HookInstance:m_Hooks, i_HookInstance.Hook.ShouldRecallOnRespawn()?) {
                set AnyRequiresSpawnListener = true
            }
            if (AnyRequiresSpawnListener?) {
                Integration.SubscribeToPlayerSpawnEvent(OnPlayerSpawned)
                set m_SubscribedToPlayerSpawn = true
            }
        }

    ProcessHookPlayer<private>(HookInstance:attachments_manager_hook_instance, Player:player):void=
        if (HookInstance.m_ActivatedPlayers[Player]) {
            return
        }
        if. set HookInstance.m_ActivatedPlayers[Player] = void

        for (
            Config:p_AttachmentKinds
            HookInstance.IsApplicable[Config.AttachmentKindId]
            ActiveAttachment := GetActiveAttachment[Player, Config.AttachmentKindId]
        ) {
            HookInstance.Hook.EquipAttachment(Player, Config.AttachmentKindId, ActiveAttachment, false)
        }

    OnPlayerAdded<private>(Player:player):void=
        for (HookInstance:m_Hooks) {
            ProcessHookPlayer(HookInstance, Player)
        }

    OnPlayerSpawned<private>(Player:player):void=
        for (HookInstance:m_Hooks, HookInstance.Hook.ShouldRecallOnRespawn()?) {
            for (
                Config:p_AttachmentKinds
                HookInstance.IsApplicable[Config.AttachmentKindId]
                ActiveAttachment := GetActiveAttachment[Player, Config.AttachmentKindId]
            ) {
                HookInstance.Hook.EquipAttachment(Player, Config.AttachmentKindId, ActiveAttachment, true)
            }
        }

    GetAttachmentKindConfig<internal>(AttachmentKindId:string)<transacts><decides>:attachments_manager_attachment_kind_config=
        var Result: ?attachments_manager_attachment_kind_config = false
        for (Config:p_AttachmentKinds, not (Result?), Config.AttachmentKindId = AttachmentKindId) {
            set Result = Some(Config)
        }
        Result?

    # Checks whether a player has unlocked an attachment for a given attachment kind.
    IsUnlocked<public>(Player:player, AttachmentKindId:string, Attachment:string)<transacts><decides>:void=
        if (not GetAttachmentKindConfig[AttachmentKindId]) {
            Log(Self, "Attachment kind \"{AttachmentKindId}\" is not configured: IsUnlocked(\"{Attachment}\")", ?Level := log_level.Warning)
        }
        IsUnlocked0(Player, AttachmentKindId, Attachment)?

    IsUnlocked0<private>(Player:player, AttachmentKindId:string, Attachment:string)<transacts>:logic=
        if (DynSources := m_DynamicUnlockSources[AttachmentKindId]) {
            for (DynSource:DynSources) {
                UnlockState := DynSource.IsAttachmentUnlocked(Player, AttachmentKindId, Attachment)
                case(UnlockState):
                    attachment_dynamic_unlock_state.Unlocked => return true
                    attachment_dynamic_unlock_state.Locked => return false
                    _ => {}
            }
        }
        if (
            PlayerData := PlayerDataGetAttachmentKindOrDefault(Player, AttachmentKindId)
            PlayerData.v0_Unlocked[Attachment] or PlayerData.v0_LastActive? = Attachment
        ) {
            return true
        }
        if (GetAttachmentKindConfig[AttachmentKindId].DefaultUnlocked.Find[Attachment]) {
            return true
        }
        false

    # Returns a player's unlocked attachments for a given attachment kind.
    GetUnlockedAttachments<public>(Player:player, AttachmentKindId:string)<transacts>:[]string=
        if (not GetAttachmentKindConfig[AttachmentKindId]) {
            Log(Self, "Attachment kind \"{AttachmentKindId}\" is not configured: GetUnlockedAttachments", ?Level := log_level.Warning)
        }
        PlayerData := PlayerDataGetAttachmentKindOrDefault(Player, AttachmentKindId)
        var Unlocked: []string = PlayerData.v0_Unlocked.Keys()
        if (Config := GetAttachmentKindConfig[AttachmentKindId]) {
            for (DefaultUnlocked:Config.DefaultUnlocked, not Unlocked.Find[DefaultUnlocked]) {
                set Unlocked = array{DefaultUnlocked} + Unlocked
            }
        }
        Unlocked

    # Unlocks a player's attachment for a given attachment kind.
    UnlockAttachment<public>(Player:player, AttachmentKindId:string, Attachment:string):logic=
        if (not Player.IsActive[]) {
            LogPlayerOffline("UnlockAttachment({AttachmentKindId}, {Attachment})", Player)
            return false
        }

        if (not GetAttachmentKindConfig[AttachmentKindId]) {
            Log(Self, "Attachment kind \"{AttachmentKindId}\" is not configured: UnlockAttachment(\"{Attachment}\")", ?Level := log_level.Warning)
        }
        PlayerData := PlayerDataGetAttachmentKindOrDefault(Player, AttachmentKindId)
        if (PlayerData.v0_Unlocked[Attachment] or PlayerData.v0_LastActive? = Attachment) {
            return false
        }
        PlayerDataSetAttachmentKind(Player, AttachmentKindId, attachment_kind_player_data{
            MakeAttachmentKindPlayerData<constructor>(PlayerData)
            v0_Unlocked := AddToMap(PlayerData.v0_Unlocked, Attachment, void)
        })
        Log(MsgUnlocked("{AttachmentKindId}:{Attachment}", Player), ?Level := log_level.Debug)
        for (
            Listener:m_UnlockListeners
            (not Listener.AttachmentKind?) or (Listener.AttachmentKind? = AttachmentKindId)
        ) {
            Event := attachment_unlocked_event{
                Player := Player
                AttachmentKindId := AttachmentKindId
                Attachment := Attachment
            }
            Listener.Callback(Event)
        }
        true

    # Undoes the unlock of a player's attachment for a given attachment kind.
    LockAttachment<public>(Player:player, AttachmentKindId:string, Attachment:string):void=
        if (not Player.IsActive[]) {
            LogPlayerOffline("LockAttachment({AttachmentKindId}, {Attachment})", Player)
            return
        }

        if (not GetAttachmentKindConfig[AttachmentKindId]) {
            Log(Self, "Attachment kind \"{AttachmentKindId}\" is not configured: LockAttachment", ?Level := log_level.Warning)
        }
        PlayerData := PlayerDataGetAttachmentKindOrDefault(Player, AttachmentKindId)
        PlayerDataSetAttachmentKind(Player, AttachmentKindId, attachment_kind_player_data{
            MakeAttachmentKindPlayerData<constructor>(PlayerData)
            v0_Unlocked := PlayerData.v0_Unlocked.RemoveByKey(Attachment)(0)
        })

    # Returns a player's active attachment for a given attachment kind.
    GetActiveAttachment<public>(Player:player, AttachmentKindId:string)<transacts><decides>:string=
        if (not GetAttachmentKindConfig[AttachmentKindId]) {
            Log(Self, "Attachment kind \"{AttachmentKindId}\" is not configured: GetActiveAttachment", ?Level := log_level.Warning)
        }
        if (CurrentActive := PlayerDataGetAttachmentKind[Player, AttachmentKindId].v0_LastActive?) {
            CurrentActive
        } else {
            Config := GetAttachmentKindConfig[AttachmentKindId]
            Config.DefaultActive?
        }

    # Sets a player's active attachment for a given attachment kind.
    SetActiveAttachment<public>(Player:player, AttachmentKindId:string, Attachment:?string):logic=
        if (not Player.IsActive[]) {
            AttachmentStr := Attachment? or "<none>"
            LogPlayerOffline("SetActiveAttachment({AttachmentKindId}, {AttachmentStr})", Player)
            return false
        }

        if (not GetAttachmentKindConfig[AttachmentKindId]) {
            Log(Self, "Attachment kind \"{AttachmentKindId}\" is not configured: SetActiveAttachment", ?Level := log_level.Warning)
        }
        PlayerData := PlayerDataGetAttachmentKindOrDefault(Player, AttachmentKindId)
        if (PlayerData.v0_LastActive = Attachment) {
            return false
        }
        LastActive := PlayerData.v0_LastActive
        PlayerDataSetAttachmentKind(Player, AttachmentKindId, attachment_kind_player_data{
            MakeAttachmentKindPlayerData<constructor>(PlayerData)
            v0_LastActive := Attachment
        })

        # hooks
        for (HookInstance:m_Hooks) {
            if (not HookInstance.m_ActivatedPlayers[Player]) {
                ProcessHookPlayer(HookInstance, Player)
            } else {
                if (LastActiveAttachment := LastActive?) {
                    HookInstance.Hook.UnequipAttachment(Player, AttachmentKindId, LastActiveAttachment, Attachment)
                }
                if (ActiveAttachment := Attachment?) {
                    HookInstance.Hook.EquipAttachment(Player, AttachmentKindId, ActiveAttachment, false)
                }
            }
        }

        # post listeners
        if (l_Attachment := Attachment?) {
            Log(MsgSetActive("{AttachmentKindId}:{l_Attachment}", Player), ?Level := log_level.Debug)
        } else {
            Log(MsgDelActive(AttachmentKindId, Player), ?Level := log_level.Debug)
        }
        for (
            Listener:m_EquipListeners
            (not Listener.AttachmentKind?) or (Listener.AttachmentKind? = AttachmentKindId)
        ) {
            Event := attachment_equip_event{
                Player := Player
                AttachmentKindId := AttachmentKindId
                Attachment := Attachment
            }
            Listener.Callback(Event)
        }
        true

    MsgUnlocked<private><localizes>(Id:string, Player:player):message=
        "Unlocked attachment \"{Id}\" for player \"{Player}\""

    MsgSetActive<private><localizes>(Id:string, Player:player):message=
        "Set active attachment \"{Id}\" for player \"{Player}\""
    MsgDelActive<private><localizes>(KindId:string, Player:player):message=
        "Set active attachment of kind \"{KindId}\" to none for player \"{Player}\""

    AddUnlockListener<public>(
        Callback:type{__(:attachment_unlocked_event):void},
        ?AttachmentKindId:?string = false
    ):void=
        set m_UnlockListeners += array{attachments_manager_unlock_listener{
            AttachmentKind := AttachmentKindId
            Callback := Callback
        }}

    AddEquipListener<public>(
        Callback:type{__(:attachment_equip_event):void},
        ?AttachmentKindId:?string = false
    ):void=
        set m_EquipListeners += array{attachments_manager_equip_listener{
            AttachmentKind := AttachmentKindId
            Callback := Callback
        }}

    PersistenceReset<override>(__:persistence_recordable_state, Agent:agent, Query:persistence_reset_query):void=
        for (Config:p_AttachmentKinds, Player := player[Agent]) {
            if (Config.PersistenceKeepAcrossTycoonRebirths?, tycoon_manager_rebirth_persistence_reset_query[Query]) {
                Log("Skipping reset of attachment kind \"{Config.AttachmentKindId}\" persistence state because it is initialized by a rebirth", ?Level := log_level.Debug)
            } else {
                LastActive := option{GetActiveAttachment[Player, Config.AttachmentKindId]}
                PlayerDataSetAttachmentKind(Player, Config.AttachmentKindId, attachment_kind_player_data{})
                NewActive := option{GetActiveAttachment[Player, Config.AttachmentKindId]}
                for (
                    HookInstance:m_Hooks
                    HookInstance.IsApplicable[Config.AttachmentKindId]
                    HookInstance.m_ActivatedPlayers[Player]
                ) {
                    if (LastActiveAttachment := LastActive?) {
                        HookInstance.Hook.UnequipAttachment(Player, Config.AttachmentKindId, LastActiveAttachment, NewActive)
                    }
                    if (ActiveAttachment := NewActive?) {
                        HookInstance.Hook.EquipAttachment(Player, Config.AttachmentKindId, ActiveAttachment, false)
                    }
                }
            }
        }

attachments_manager_vfx_player_attachment_hook<internal> := class<final>(player_attachment_hook):
    AttachmentsManager: attachments_manager_device

    EquipAttachment<override>(Player:player, AttachmentKindId:string, Attachment:string, Recall:logic):void=
        if (Config := AttachmentsManager.GetAttachmentKindConfig[AttachmentKindId]) {
            if (VFXClear := Config.VFXPerAttachmentClear?) {
                VFXClear.Pickup(Player)
            }
            for (VFXConfig:Config.VFXPerAttachment, VFXConfig.Attachment = Attachment) {
                VFXConfig.VFX.Pickup(Player)
            }
        }

    UnequipAttachment<override>(Player:player, AttachmentKindId:string, Attachment:string, ReplacingWith:?string):void=
        if (
            Config := AttachmentsManager.GetAttachmentKindConfig[AttachmentKindId]
            VFXClear := Config.VFXPerAttachmentClear?
        ) {
            VFXClear.Pickup(Player)
        }

    ShouldRecallOnRespawn<override>()<transacts>:logic = true

attachments_manager_attachment_kind_user_action_kind := enum:
    Unlock
    Lock
    SetActive

attachments_manager_attachment_kind_user_action_config := class<final><concrete>:
    @editable OnTrigger: trigger_device = trigger_device{}
    @editable Action: attachments_manager_attachment_kind_user_action_kind = attachments_manager_attachment_kind_user_action_kind.Unlock
    @editable Attachment: string = ""

attachments_manager_attachment_kind_user_action := class<final>:
    Device: attachments_manager_device
    AttachmentKind: string
    Config: attachments_manager_attachment_kind_user_action_config

    OnTriggered(Agent:?agent):void=
        if (Player := player[Agent?]) {
            case(Config.Action):
                attachments_manager_attachment_kind_user_action_kind.Unlock =>
                    Device.UnlockAttachment(Player, AttachmentKind, Config.Attachment)
                attachments_manager_attachment_kind_user_action_kind.Lock =>
                    Device.LockAttachment(Player, AttachmentKind, Config.Attachment)
                attachments_manager_attachment_kind_user_action_kind.SetActive =>
                    l_Attachment := if (Config.Attachment.Length > 0) {
                        Some(Config.Attachment)
                    } else {
                        None()
                    }
                    Device.SetActiveAttachment(Player, AttachmentKind, l_Attachment)
        }
