using { /Fortnite.com/Devices }
using { /Verse.org/Simulation }
using { /UnrealEngine.com/Temporary/Diagnostics }

using { HavocVerse.Logging }
using { HavocVerse.Util }

BulkItemGranterOpCountHardLimit: int = 10000

# Device that can grant any amount of a resource to an agent, backed
# by item granters and optimized with magnitudes inference.
bulk_resource_granter_device<public> := class(creative_device):
    @editable p_ResourceId<public>: string = ""
    # setup: each item granter is a power of ten magnitude
    #        each item per item granter is a unit of that magnitude (1-9)
    @editable p_ItemGranters: []item_granter_device = array{}

    var m_ComputedMagnitudes: []int = array{}

    SupportsGiveOperation<public>()<transacts><decides>:void=
        p_ItemGranters.Length > 0

    # Grants an amount of the resource to an agent.
    # WARNING: Silently fails if the item granter is not properly set up.
    Give<public>(Agent:agent, Amount:int, ?MaxOpCount:int = 80):void=
        if (Amount < 0) {
            Log("Cannot grant negative amounts of a resource. (from {Amount}x{p_ResourceId})", ?Level := log_level.Error)
            return
        }
        var MagnitudeIndex: int = FindAmountPositiveMagnitude(Amount)
        var Magnitude: int = if (l_Magnitude := m_ComputedMagnitudes[MagnitudeIndex]) {
            l_Magnitude
        } else {
            if (Amount > 0) {
                Log("Failed to grant {Amount} {p_ResourceId} items", ?Level := log_level.Error)
            }
            return
        }
        var Remaining: int = Amount
        var OpCount: int = 0
        var BailOut: ?tuple(int,int) = false
        loop {
            if (
                Remaining >= Magnitude
                ItemGranter := p_ItemGranters[MagnitudeIndex]
                Unit := Min(Floor[(Remaining*1.0) / (Magnitude*1.0)], 9)
            ) {
                set Remaining -= Magnitude * Unit
                set OpCount += 1
                if (OpCount >= BulkItemGranterOpCountHardLimit) {
                    set OpCount = -1
                    break
                } else if (not (BailOut?)) {
                    ItemGranter.SetNextItem(Unit-1)
                    ItemGranter.GrantItem(Agent)
                    if (OpCount >= MaxOpCount) {
                        set BailOut = Some((Remaining, OpCount))
                    }
                }
            } else if (MagnitudeIndex > 0, l_Magnitude := m_ComputedMagnitudes[MagnitudeIndex-1]) {
                set MagnitudeIndex -= 1
                set Magnitude = l_Magnitude
            } else {
                if (not (BailOut?) and Remaining > 0) {
                    Log("Failed to grant {Remaining} remaining {p_ResourceId} items (2)", ?Level := log_level.Error)
                }
                break
            }
        }
        if (BailState := BailOut?) {
            NeededOpCount := if (OpCount >= 0) {
                ToString(OpCount)
            } else {
                "(more than 10,000)"
            }
            Log(MsgFailTimeout(Agent, BailState(0), Amount, p_ResourceId, BailState(1), NeededOpCount), ?Level := log_level.Error)
        }

    MsgFailTimeout<private><localizes>(Agent:agent, Remaining:int, Total:int, ResourceId:string, OpCount:int, NeededOpCount:string):message=
        "Failed to grant {Remaining}/{Total} {ResourceId} items to '{Agent}'. Capped at {OpCount}/{NeededOpCount} grant operations."

    FindAmountPositiveMagnitude<private>(Amount:int):int=
        if (m_ComputedMagnitudes.Length <> p_ItemGranters.Length) {
            for (
                Index := 0..p_ItemGranters.Length-1
                not m_ComputedMagnitudes[Index]
                Magnitude := Floor[Pow(10.0, Index * 1.0)])
            {
                set m_ComputedMagnitudes += array{Magnitude}
            }
        }
        var IterIndex: int = m_ComputedMagnitudes.Length-1
        loop {
            if (Magnitude := m_ComputedMagnitudes[IterIndex], Amount >= Magnitude) {
                break
            }
            set IterIndex -= 1
            if (IterIndex < 0) {
                break
            }
        }
        IterIndex
