using { /Fortnite.com/Devices }
using { /Fortnite.com/Devices/CreativeAnimation }
using { /UnrealEngine.com/Assets }
using { /UnrealEngine.com/Temporary/Diagnostics }
using { /UnrealEngine.com/Temporary/SpatialMath }
using { /Verse.org/Assets }
using { /Verse.org/Colors }
using { /Verse.org/Simulation }
using { /Verse.org/Simulation/Tags }

using { HavocVerse.Debug }
using { HavocVerse.Logging }
using { HavocVerse.Util }

buy_plate_cost_met_tracking_state<public> := class<final>:
    MeetsCost<public>: logic

buy_plate_pooled_linked_prop_entry<internal> := class<final>:
    Applicable: logic
    CostReflectionDependent: logic
    var m_Prop: ?prop_pool_entry = false

buy_plate_device<public> := class<unique>(creative_device, aware_device, aware_debug_device, custom_device, centralizable, resetable):
    @editable p_DevId<public>: string = "N/A"
    @editable{ToolTip := TooltipRoutingIdOverride} p_RoutingIdOverride<public>: ?string = false
    @editable p_CustomData<public>: []string = array{}

    @editable p_StartEnabled: logic = true
    @editable p_AlwaysHideSelf: logic = true

    @editable p_StealthPlate<public>: logic = false
    @editable p_ExpectsManager<public>: logic = true

    @editable p_Button: []button_device = array{}
    @editable p_Trigger: []trigger_device = array{}
    @editable p_Volume: []volume_device = array{}

    @editable var p_AccessControl: []access_control_device = array{}
    @editable var p_MiddlewareDevices: []creative_device = array{} # discarded for m_Middlewares at runtime

    @editable p_CostButton: []conditional_button_device = array{}
    @editable p_OnFailHUD: []hud_message_device = array{}

    @editable p_ExpectsCostOverride<public>: logic = false
    @editable p_CostResource: buy_plate_cost_resource_config = buy_plate_cost_resource_config{}

    @editable p_LinkedProps: []creative_prop = array{}
    @editable p_DisposeLinkedProps: logic = false
    @editable p_Billboards: []billboard_device = array{}

    @editable p_UnlockBarriers: []barrier_device = array{}
    @editable p_UnlockProps: []buy_plate_unlock_display_prop_config = array{}
    @editable p_UnlockPropManipulators: []buy_plate_unlock_prop_manipulator_config = array{}

    @editable p_UnlockTriggers: []trigger_device = array{}
    @editable p_UnlockCinematics: []cinematic_sequence_device = array{}
    @editable p_UnlockButtons: []button_device = array{}
    @editable p_UnlockAudioPlayers: []audio_player_device = array{}
    @editable p_FirstUnlockTriggers: []trigger_device = array{}
    @editable p_UnlockCustomDevices: []custom_device_trigger_config = array{}
    @editable p_UnlockCustomDevicesQuickEnable: []creative_device = array{}

    UnlockedEvent<public>: event(agent) = event(agent){}
    var m_UnlockEventListeners<private>: []maybe_agent_callback = array{}
    var m_EnableDisableEventListeners<private>: []type{__(:logic):void} = array{}

    var m_Enabled<scoped {HavocVerse}>: logic = false
    var m_BuiltinSubscriptions<private>: []cancelable = array{}

    var m_Unlocked<scoped {HavocVerse}>: logic = false
    var m_NextActivateSilent<private>: logic = false
    var m_Middlewares<private>: []buy_plate_middleware = array{}
    var m_InjectCinematicPlayer<internal>: ?cinematics_player_device = false

    var m_Setup: logic = false
    var m_ManagerGroup: ?buy_plate_manager_active_group = false
    var m_PooledLinkedProps: []buy_plate_pooled_linked_prop_entry = array{}

    var m_CostOverride: ?buy_plate_cost_override = false
    var m_CostMetTrackingState<private>: ?buy_plate_cost_met_tracking_state = false

    DeviceMeta<override>()<transacts>:aware_device_meta=
        aware_device_meta{
            DeviceName := "buy_plate"
            DevId := p_DevId
            RoutingIdOverride := p_RoutingIdOverride
            CustomData := Some(p_CustomData)
        }

    DebugStats<override>()<transacts>:aware_debug_stats=
        aware_debug_stats{
            Length := map{
                "builtinSubscriptions" => m_BuiltinSubscriptions.Length
                "middlewares" => m_Middlewares.Length
                "pooledLinkedProps" => m_PooledLinkedProps.Length
                "listeners::unlockEvent" => m_UnlockEventListeners.Length
                "listeners::enableDisable" => m_EnableDisableEventListeners.Length
            }
        }

    # Returns whether the buy plate is unlocked.
    IsUnlocked<public>()<transacts><decides>:void=
        m_Unlocked?

    # Awaits an unlock, if not already unlocked.
    AwaitUnlock<public>()<suspends>:void=
        if (m_Unlocked?) {
            return
        }
        UnlockedEvent.Await()

    OnBegin<override>()<suspends>:void=
        AwareInitDevice(Self, OnInit)

    OnInit<private>()<suspends>:void=
        if (not (p_ExpectsManager?)) {
            Setup(None())
        } else {
            AwareSubscribeToPostInit(OnPostInit)
        }

    OnPostInit<private>():void=
        if (p_ExpectsManager?, not (m_ManagerGroup?)) {
            Log("{Self} was expecting a manager group, but wasn't assigned one", ?Level := log_level.Error)
        }

    Setup<internal>(ManagerGroup:?buy_plate_manager_active_group):logic=
        if (m_Setup?) {
            if (ManagerGroup? = m_ManagerGroup?) {
                Log(Self, "Attempted second setup with same manager group. Linked in ExtraBuyplates + within bounds?", ?Level := log_level.Warning)
                return false
            }
            FromStr := if (l_ManagerGroup := ManagerGroup?) {
                ToString(l_ManagerGroup)
            } else {
                "None"
            }
            CurrentStr := if (l_ManagerGroup := m_ManagerGroup?) {
                ToString(l_ManagerGroup)
            } else {
                "None"
            }
            Log("Attempting to setup {Self} a second time from {FromStr} (initially setup by {CurrentStr}). Managers target rules must be mutually exclusive, or use different priorities.", ?Level := log_level.Error)
            return false
        }
        set m_Setup = true
        set m_ManagerGroup = ManagerGroup
        if (l_ManagerGroup := ManagerGroup?) {
            if (
                p_AccessControl.Length > 0
                (not (l_ManagerGroup.Manager.p_LegacyAccessControls?)) or (not p_DevId.StartsWith["tx"])
            ) {
                Log("{Self} is configured with an access control device and a manager. The access control will be ignored in favor of the manager's.", ?Level := log_level.Warning)
            }
            set p_AccessControl = if (AccessControl := l_ManagerGroup.AccessControl?) {
                array{AccessControl}
            } else {
                array{}
            }
        } else {
            if (p_AccessControl.Length > 1) {
                Log("More than 1 access control device on {Self}, only the first one will be used", ?Level := log_level.Warning)
            }
        }
        if (AccessControl := p_AccessControl[0]) {
            AccessControl.RegisterUsingDevice(Self)
        }

        for (Index->Device:p_MiddlewareDevices, not buy_plate_middleware[Device]) {
            Log("{Self} is configured with a middleware device that is not a middleware (p_Middlewares, index {Index})", ?Level := log_level.Error)
        }
        set m_Middlewares = for (Device:m_Middlewares, Middleware := buy_plate_middleware[Device]) {
            Middleware
        }
        set p_MiddlewareDevices = array{}

        if (p_CostButton.Length > 1) {
            Log("More than 1 cost button on a buy plate, only the first one will be used", ?Level := log_level.Warning)
        }
        if (p_CostButton.Length >= 1 and p_CostResource.ResourceEndpoint[0]) {
            Log("Resource cost on a buy plate with a cost button", ?Level := log_level.Error)
        }

        ApplyInitialDevicesState()

        if (p_AlwaysHideSelf?) {
            Self.Hide()
        }
        if (p_StartEnabled?) {
            Self.Enable()
        } else {
            Self.HideSelfAndDeps()
        }
        true

    ApplyInitialDevicesState<private>(): void ={
        for (Barrier:p_UnlockBarriers). Barrier.Enable()
        for (Index->Display:p_UnlockProps). Display.ApplyLockedState(Self, Index)
        for (Index->Display:p_UnlockPropManipulators). Display.ApplyLockedState(Self, Index)

        for (Button:p_UnlockButtons) {
            Button.Disable()
        }
        for (AudioPlayer:p_UnlockAudioPlayers) {
            AudioPlayer.Stop()
            AudioPlayer.Disable()
        }
        for (SequenceDevice:p_UnlockCinematics) {
            SequenceDevice.Stop()
        }
    }

    Reset<override>(Agent:?agent):void=
        if (p_DisposeLinkedProps?) {
            Log("Resetting a buy plate with p_DisposeLinkedProps enabled. This will cause inconsistencies!", ?Level := log_level.Warning)
        }
        if (m_Unlocked?) {
            ApplyInitialDevicesState()
            set m_Unlocked = false
        }
        if (p_StartEnabled?) {
            Self.Enable()
        } else {
            Self.Disable()
        }

    AddMiddleware<public>(Middleware:buy_plate_middleware):void=
        set m_Middlewares += array{Middleware}

    ShowSelfAndDeps():void=
        if (not (p_AlwaysHideSelf?)) {
            Self.Show()
        }
        if (CostButton := p_CostButton[0]) {
            CostButton.Enable()
        }
        for (Prop:p_LinkedProps, Prop.IsValid[]) {
            Prop.Show()
        }
        for (Billboard:p_Billboards) {
            Billboard.ShowText()
        }
        if (
            not p_StealthPlate?
            ManagerGroup := m_ManagerGroup?
            ManagerGroup.Manager.p_PooledLinkedProps.Length <> m_PooledLinkedProps.Length
        ) {
            for (PropEntry:m_PooledLinkedProps, PoolEntry := PropEntry.m_Prop?) {
                PoolEntry.Release()
            }
            set m_PooledLinkedProps = for (Index->LinkedPropConfig:ManagerGroup.Manager.p_PooledLinkedProps) {
                if (LinkedPropConfig.MatchesFilter[ManagerGroup.GroupName]) {
                    Entry := buy_plate_pooled_linked_prop_entry{
                        Applicable := true
                        CostReflectionDependent := if (LinkedPropConfig.CostMetFilter?). true else false
                    }
                    UpdatePooledLinkedProp(Index, Entry, None())
                    Entry
                } else {
                    buy_plate_pooled_linked_prop_entry{
                        Applicable := false
                        CostReflectionDependent := false
                    }
                }
            }
        }

    UpdatePooledLinkedProp<private>(Index:int, Entry:buy_plate_pooled_linked_prop_entry, CostMetInstigator:?agent):void=
        if (
            Entry.Applicable?
            ManagerGroup := m_ManagerGroup?
            LinkedPropConfig := ManagerGroup.Manager.p_PooledLinkedProps[Index]
            PropPosition := ManagerGroup.Manager.GetBuyplatePosition[Self, ?ExtraOffset := LinkedPropConfig.Offset]
        ) {
            ExpectsVisible := if (
                CostMetFilter := LinkedPropConfig.CostMetFilter?
                Agent := (CostMetInstigator? or p_AccessControl[0].ListAllowedAgents()[0] or GetPlayspace().GetPlayers()[0])
            ) {
                case(CostMetFilter):
                    buy_plate_cost_met_filter.WhenCostMet => DoesAgentMeetsCostCached(Agent)
                    buy_plate_cost_met_filter.WhenCostUnmet => InvertLogic(DoesAgentMeetsCostCached(Agent))
            } else {
                true
            }

            if (ExpectsVisible?, not Entry.m_Prop?) {
                if (PoolEntry := LinkedPropConfig.ReserveProp[]) {
                    Success := SetupPooledLinkedProp(ManagerGroup, PoolEntry, PropPosition, Index, LinkedPropConfig)
                    if (Success?) {
                        set Entry.m_Prop = Some(PoolEntry)
                        Log("Enabled pooled linked prop index {Index} for {Self}", ?Level := log_level.Verbose)
                    } else {
                        PoolEntry.Release()
                    }
                } else {
                    Log("Could not reserve pooled linked prop index {Index} for {Self}: not enough in pool?", ?Level := log_level.Error)
                }
            } else if (not ExpectsVisible?, PoolEntry := Entry.m_Prop?) {
                PoolEntry.Release()
                set Entry.m_Prop = None()
                Log("Released pooled linked prop index {Index} for {Self}", ?Level := log_level.Verbose)
            }
        } else if (PoolEntry := Entry.m_Prop?) {
            PoolEntry.Release()
            set Entry.m_Prop = None()
        }

    SetupPooledLinkedProp<private>(
        ManagerGroup:buy_plate_manager_active_group,
        PoolEntry:prop_pool_entry, PropPosition:vector3,
        PoolIndex:int, LinkedPropConfig:buy_plate_manager_linked_props_pool_config
    ):logic=
        var Failed: logic = false
        if (LinkedPropConfig.DynamicMesh.Length > 0) {
            if (
                AssetsProvider := AwareDefaultAssetsProvider[]
                AssetName := TransformAssetName(LinkedPropConfig.DynamicMesh, ManagerGroup)
                Mesh := AssetsProvider.GetMeshAsset[AssetName]
            ) {
                PoolEntry.Prop.SetMesh(Mesh)
            } else {
                set Failed = true
            }
        }
        if (LinkedPropConfig.DynamicMaterial.Length > 0) {
            if (
                AssetsProvider := AwareDefaultAssetsProvider[]
                AssetName := TransformAssetName(LinkedPropConfig.DynamicMaterial, ManagerGroup)
                Material := AssetsProvider.GetMaterialAsset[AssetName]
            ) {
                PoolEntry.Prop.SetMaterial(Material)
            } else {
                set Failed = true
            }
        }
        if (Failed?) {
            Log("Ignoring pooled linked prop index {PoolIndex} on {Self} because the mesh or material couldn't be loaded", ?Level := log_level.Error)
            return false
        }
        TargetTransform := PoolEntry.Prop.GetTransform().WithTranslation(PropPosition)
        if. PoolEntry.Prop.TeleportTo[TargetTransform]
        PoolEntry.Prop.Show()
        true

    TransformAssetName<private>(AssetName:string, ManagerGroup:buy_plate_manager_active_group)<transacts>:string=
        AssetName
            .ReplaceAll("%DevId%", p_DevId)
            .ReplaceAll("%GroupName%", ManagerGroup.GroupName)

    HideSelfAndDeps():void=
        if (not (p_AlwaysHideSelf?)) {
            Self.Hide()
        }
        if (CostButton := p_CostButton[0]) {
            CostButton.Disable()
            if (m_Unlocked?) {
                # sometimes it doesn't hide properly
                Transform := CostButton.GetTransform()
                NewTransform := transform{
                    Translation := Transform.Translation
                    Rotation := Transform.Rotation
                    Scale := SafeScale(vector3{})
                }
                if (CostButton.TeleportTo[NewTransform]) {}
            }
        }
        for (Prop:p_LinkedProps, Prop.IsValid[]) {
            Prop.Hide()
        }
        for (Billboard:p_Billboards) {
            Billboard.HideText()
        }

        for (PropEntry:m_PooledLinkedProps, PoolEntry := PropEntry.m_Prop?) {
            PoolEntry.Release()
        }
        set m_PooledLinkedProps = array{}

    IsEnabled<public>()<transacts><decides>:void=
        m_Enabled?

    Enable<public>():void=
        if (m_Enabled? or m_Unlocked?) {
            return
        }
        set m_Enabled = true
        Self.ShowSelfAndDeps()

        if (ManagerGroup := m_ManagerGroup?, ManagerGroup.Manager.p_UseFakeVolumes?) {
            # handled by the manager
        } else if (p_Volume.Length > 0) {
            if (not p_StealthPlate?) {
                for (ZoneVolume:p_Volume) {
                    ZoneSubscription := ZoneVolume.AgentEntersEvent.Subscribe(Self.OnAgentInteract)
                    set m_BuiltinSubscriptions += array{ZoneSubscription}
                }
            } else {
                Log("{Self} has p_StealthPlate enabled, but also has a physical volume, it will be ignored", ?Level := log_level.Warning)
            }
        } else if (not p_StealthPlate?, p_Button.Length = 0, p_Trigger.Length = 0) {
            Log("{Self} is missing a physical volume (not using fake volume, nor button/trigger)", ?Level := log_level.Error)
        }

        for (Button:p_Button) {
            Subscription := Button.InteractedWithEvent.Subscribe(Self.OnAgentInteract)
            set m_BuiltinSubscriptions += array{Subscription}
        }
        for (Trigger:p_Trigger) {
            Subscription := Trigger.TriggeredEvent.Subscribe(Self.OnAgentInteractOpt)
            set m_BuiltinSubscriptions += array{Subscription}
        }
        if (CostButton := p_CostButton[0]) {
            set m_BuiltinSubscriptions += array{
                CostButton.ActivatedEvent.Subscribe(Self.OnPurchaseSuccess)
                CostButton.NotEnoughItemsEvent.Subscribe(Self.OnPurchaseFailed)
            }
        }

        for (Listener:m_EnableDisableEventListeners) {
            Listener(true)
        }

    Disable<public>():void=
        set m_Enabled = false
        if (m_BuiltinSubscriptions.Length > 0) {
            for (Subscription:m_BuiltinSubscriptions) {
                Subscription.Cancel()
            }
            set m_BuiltinSubscriptions = array{}
        }
        Self.HideSelfAndDeps()
        for (Listener:m_EnableDisableEventListeners) {
            Listener(false)
        }

    GetPosition<public>(?ExtraOffset:vector3 = vector3{})<transacts>:vector3=
        if (
            ManagerGroup := m_ManagerGroup?
            Position := ManagerGroup.Manager.GetBuyplatePosition[Self, ?ExtraOffset := ExtraOffset]
        ) {
            Position
        } else {
            GetTransform().Translation
        }

    SetCostOverride<public>(CostOverride:buy_plate_cost_override)<transacts>:void=
        set m_CostOverride = Some(CostOverride)

    DoesAgentMeetsCost<public>(Agent:agent):logic=
        if (l_CostOverride := m_CostOverride?) {
            l_CostOverride.HasEnough(Agent)
        } else if (p_ExpectsCostOverride?) {
            false
        } else if (CostButton := p_CostButton[0]) {
            if (CostButton.HasAllItems[Agent]) {
                true
            } else {
                false
            }
        } else if (ResourceEndpoint := p_CostResource.ResourceEndpoint[0]) {
            Amount := ResourceEndpoint.GetAmount(Some(Agent))
            if (Amount >= p_CostResource.Amount) {
                true
            } else {
                false
            }
        } else {
            true
        }

    UpdateCostMetTracking<public>(Agent:agent):?buy_plate_cost_met_tracking_state=
        if (not (m_Enabled?) or m_Unlocked?) {
            return None()
        }

        MeetsCost := DoesAgentMeetsCost(Agent)
        if (State := m_CostMetTrackingState?, State.MeetsCost = MeetsCost) {
            return None()
        }
        set m_CostMetTrackingState = Some(buy_plate_cost_met_tracking_state{
            MeetsCost := MeetsCost
        })

        Log("{Self} cost met tracking state changed to {LogicToString(MeetsCost)}", ?Level := log_level.Debug)
        if (ManagerGroup := m_ManagerGroup?, ManagerGroup.Manager.p_EnableCostMetReflection?) {
            UpdateCostMetReflection(Agent, ManagerGroup.Manager.p_CostMetReflectionConfig, MeetsCost)
        }
        for (Index->PropEntry:m_PooledLinkedProps, PropEntry.CostReflectionDependent?) {
            UpdatePooledLinkedProp(Index, PropEntry, Some(Agent))
        }
        m_CostMetTrackingState

    DoesAgentMeetsCostCached<public>(Agent:agent):logic=
        if (State := m_CostMetTrackingState?) {
            State.MeetsCost
        } else {
            DoesAgentMeetsCost(Agent)
        }

    UpdateCostMetReflection<private>(Agent:agent, Config:buy_plate_cost_met_reflection_config, MeetsCost:logic):void=
        if (
            Prop := case(Config.LinkedPropKind):
                buy_plate_linked_prop_kind.StaticLinkedProp =>
                    p_LinkedProps[Config.LinkedPropIndex]
                buy_plate_linked_prop_kind.PooledLinkedProp =>
                    m_PooledLinkedProps[Config.LinkedPropIndex].m_Prop?.Prop
            AssetsProvider := AwareDefaultAssetsProvider[]
        ) {
            AssetNames := if (MeetsCost?) {
                (Config.MetMesh, Config.MetMaterial, Config.MetAdditionalMaterials)
            } else {
                (Config.UnmetMesh, Config.UnmetMaterial, Config.UnmetAdditionalMaterials)
            }
            if (AssetNames(0).Length > 0, Mesh := AssetsProvider.GetMeshAsset[AssetNames(0)]) {
                Prop.SetMesh(Mesh)
            }
            if (AssetNames(1).Length > 0, Material := AssetsProvider.GetMaterialAsset[AssetNames(1)]) {
                Prop.SetMaterial(Material)
            }
            for (Index->MaterialName:AssetNames(2), Material := AssetsProvider.GetMaterialAsset[MaterialName]) {
                Prop.SetMaterial(Material, ?Index := 1 + Index)
            }
        }

    TryUnlock<private>(Agent:agent, ?Silent:logic = false):void=
        if (CostOverride := m_CostOverride?) {
            Success := CostOverride.TryUnlock(Agent, Silent)
            if (Success?) {
                Self.Unlock(Agent, ?UserTriggered := true)
            }
            return
        } else if (p_ExpectsCostOverride?) {
            Log("Trying to unlock a buy plate with p_ExpectsCostOverride enabled, but no cost override was set", ?Level := log_level.Warning)
            return
        }
        if (CostButton := p_CostButton[0]) {
            CostButton.Activate(Agent)
            set m_NextActivateSilent = Silent
        } else if (ResourceEndpoint := p_CostResource.ResourceEndpoint[0], p_CostResource.CheckDontTake?) {
            Amount := ResourceEndpoint.GetAmount(Some(Agent))
            if (Amount >= p_CostResource.Amount) {
                Self.Unlock(Agent, ?UserTriggered := true)
            } else if (p_CostResource.NotifyFail?) {
                ResourceEndpoint.SendNotEnoughNotification(Agent)
            }
        } else if (ResourceEndpoint := p_CostResource.ResourceEndpoint[0]) {
            Success := ResourceEndpoint.Take(Some(Agent), p_CostResource.Amount,
                ?NotifyNotEnough := p_CostResource.NotifyFail, ?SendFeedback := p_CostResource.TakeFeedback)
            if (Success?) {
                Self.Unlock(Agent, ?UserTriggered := true)
            }
        } else {
            Self.Unlock(Agent, ?UserTriggered := true)
        }

    OnAgentInteractOpt<private>(Agent:?agent):void=
        if (l_Agent := Agent?) {
            OnAgentInteract(l_Agent)
        }

    OnAgentInteract<internal>(Agent:agent):void=
        if (AccessControl := p_AccessControl[0], not AccessControl.IsAllowed[Agent]) {
            return
        }
        for (Middleware:m_Middlewares) {
            Allowed := Middleware.MayTryUnlockBuyplate(Agent, Self)
            if (not (Allowed?)) {
                return
            }
        }
        TryUnlock(Agent)

    OnPurchaseSuccess(Agent:agent):void=
        Self.Unlock(Agent, ?UserTriggered := true)

    OnPurchaseFailed(Agent:agent):void=
        if (m_NextActivateSilent?) {
            set m_NextActivateSilent = false
        } else {
            for (OnFailHUD:p_OnFailHUD) {
                OnFailHUD.Show(Agent)
            }
        }

    # Unlocks the buy plate.
    # This will not check for resource requirements, etc.
    Unlock<public>(
        Agent:agent, ?UserTriggered:logic,
        ?FirstUnlock:logic = true, ?Force:logic = false
    ):void=
        if (m_Unlocked? and not (Force?)) {
            return
        }

        set m_Unlocked = true
        Disable()

        if (p_DisposeLinkedProps?) {
            for (Prop:p_LinkedProps, Prop.IsValid[]) {
                Prop.Dispose()
            }
        }

        if (InjectCinematic := m_InjectCinematicPlayer?, Player := player[Agent]) {
            spawn:
                PlayCinematicThenContinueUnlock(Player, InjectCinematic, ?UserTriggered := UserTriggered,
                    ?FirstUnlock := FirstUnlock, ?Force := Force)
        } else {
            Unlock0(Agent, ?UserTriggered := UserTriggered, ?FirstUnlock := FirstUnlock, ?Force := Force)
        }

    PlayCinematicThenContinueUnlock<private>(
        Player:player, CinematicPlayer:cinematics_player_device, ?UserTriggered:logic,
        ?FirstUnlock:logic = true, ?Force:logic = false
    )<suspends>:void=
        Sleep(0.0)
        CinematicPlayer.Play(Player, true)
        if (Player.IsActive[]) {
            Unlock0(Player, ?UserTriggered := UserTriggered, ?FirstUnlock := FirstUnlock, ?Force := Force)
        } else {
            Log(Self, "Player quit during injected cinematic, reverting state", ?Level := log_level.Debug)
            Reset(None())
        }

    Unlock0<private>(
        Agent:agent, ?UserTriggered:logic,
        ?FirstUnlock:logic = true, ?Force:logic = false
    ):void=
        for (Barrier:p_UnlockBarriers). Barrier.Disable()
        for (Index->Display:p_UnlockProps). Display.ApplyUnlockedState(Self, Index)
        for (Index->Display:p_UnlockPropManipulators). Display.ApplyUnlockedState(Self, Index)

        UnlockedEvent.Signal(Agent)
        for (Middleware:m_Middlewares) {
            Middleware.OnUnlockBuyplate(Agent, Self)
        }

        for (Index->TriggerConfig:p_UnlockCustomDevices) {
            TriggerConfig.AttemptTrigger(Some(Agent), "buyplate unlock: {Self} - [{Index}]")
        }
        for (Index->Device:p_UnlockCustomDevicesQuickEnable) {
            Context := "buyplate unlock: {Self} - quick enable [{Index}]"
            if (CustomDevice := custom_device[Device]) {
                Success := CustomDevice.CustomTrigger("enable", Some(Agent))
                if (not (Success?)) {
                    Log("Failed to enable a custom device ({Context})]", ?Level := log_level.Warning)
                }
            } else {
                Log("Trying to trigger a custom device with a non-custom device ({Context})", ?Level := log_level.Error)
            }
        }
        for (Listener:m_UnlockEventListeners) {
            Listener(Some(Agent))
        }

        for (Trigger:p_UnlockTriggers) {
            Trigger.Trigger(Agent)
        }
        if (ManagerGroup := m_ManagerGroup?) {
            for (
                TriggerConfig:ManagerGroup.Manager.p_UnlockTriggers
                TriggerConfig.MatchesFilter[p_CustomData, FirstUnlock]
            ) {
                TriggerConfig.Trigger.Trigger(Agent)
            }
            if (UserTriggered?, UnlockEffect := ManagerGroup.Manager.GetUnlockEffect[p_CustomData]) {
                if (
                    AssetsProvider := AwareDefaultAssetsProvider[]
                    UnlockEffect.VFX.Length > 0
                    ParticleAsset := AssetsProvider.GetParticleSystemAsset[UnlockEffect.VFX]
                ) {
                    spawn:
                        SpawnVFX(ParticleAsset, UnlockEffect)
                }
                if (EffectPropAsset := UnlockEffect.EffectProp?) {
                    spawn:
                        SpawnEffectProp(EffectPropAsset, UnlockEffect)
                }
            }
        }
        for (SequenceDevice:p_UnlockCinematics) {
            SequenceDevice.Play(Agent)
        }
        for (Button:p_UnlockButtons) {
            Button.Enable()
        }
        for (AudioPlayer:p_UnlockAudioPlayers) {
            AudioPlayer.Enable()
            AudioPlayer.Play()
        }
        if (FirstUnlock?) {
            for (Trigger:p_FirstUnlockTriggers) {
                Trigger.Trigger(Agent)
            }
        }

    SpawnVFX<private>(
        ParticleAsset:particle_system,
        UnlockEffect:buy_plate_manager_unlock_effect_config
    )<suspends>:void=
        Particle := SpawnParticleSystem(ParticleAsset, GetPosition(?ExtraOffset := UnlockEffect.ParticleSpawnOffset))
        Sleep(UnlockEffect.VFXDespawnDelay)
        Particle.Cancel()

    SpawnEffectProp<private>(
        EffectPropAsset:creative_prop_asset,
        UnlockEffect:buy_plate_manager_unlock_effect_config
    )<suspends>:void=
        Transform := transform{
            Translation := GetPosition(?ExtraOffset := UnlockEffect.ParticleSpawnOffset)
            Scale := MakeUniformVec3(1.0)
        }
        PropRes := SpawnProp(EffectPropAsset, Transform)(0)
        if (Prop := PropRes?) {
            Sleep(UnlockEffect.PropDespawnDelay)
            Prop.Dispose()
        }

    CustomTrigger<override>(Action:string, Agent:?agent):logic=
        if (Action = "enable") {
            Enable()
            true
        } else if (Action = "disable") {
            Disable()
            true
        } else {
            false
        }

    GetRoutingId<override>()<computes>:string=
        p_DevId

    GetCustomData<override>()<computes><decides>:[]string=
        p_CustomData

    SubscribeToEvent<override>(EventName:string, Callback:maybe_agent_callback)<transacts><decides>:void=
        if (EventName = "unlock") {
            set m_UnlockEventListeners += array{Callback}
        } else if (EventName = "enable" or EventName = "disable") {
            Closure := buy_plate_enable_disable_closure{
                Callback := Callback
                ExpectedState := if (EventName = "enable"). true else false
            }
            set m_EnableDisableEventListeners += array{Closure.Call}
        }

buy_plate_cost_resource_config := class<concrete>:
    @editable ResourceEndpoint: []resource_endpoint_device = array{}
    @editable Amount: int = 0
    @editable NotifyFail: logic = true
    @editable TakeFeedback: logic = true
    @editable CheckDontTake: logic = false

buy_plate_display_action := enum:
    Show
    Hide

buy_plate_unlock_display := interface:
    ApplyLockedState(BuyPlate:buy_plate_device, Index:int):void

    ApplyUnlockedState(BuyPlate:buy_plate_device, Index:int):void

buy_plate_unlock_display_prop_config := class<concrete>(buy_plate_unlock_display):
    @editable Prop: creative_prop = creative_prop{}
    @editable Action: buy_plate_display_action = buy_plate_display_action.Show
    @editable HideScaleDown: logic = false
    @editable UnmanagedLockedState: logic = false
    @editable CustomData: []string = array{}
    @editable AllowAnimation: logic = true

    var m_InitialTransform: ?transform = false
    var m_PlayedAnimation: logic = false
    var m_DelayedShow: ?event() = false

    Hide<private>(BuyPlate:buy_plate_device, Index:int):void=
        if (m_PlayedAnimation?, AnimationController := Prop.GetAnimationController[]) {
            AnimationController.Pause()
        }
        if (not Prop.IsValid[]) {
            Log(BuyPlate, "Could not hide unlock prop index {Index} because it is invalid", ?Level := log_level.Error)
            return
        }
        if (AbortEvent := m_DelayedShow?) {
            AbortEvent.Signal()
            set m_DelayedShow = None()
        }
        if (not (m_InitialTransform?)) {
            set m_InitialTransform = Some(Prop.GetTransform())
        }
        Prop.Hide()
        if:
            HideScaleDown?
            NewTransform := m_InitialTransform?.Rescale(0.00001)
            Prop.TeleportTo[NewTransform]
        else if:
            m_PlayedAnimation?
            NewTransform := m_InitialTransform?
            Prop.TeleportTo[NewTransform]
        set m_PlayedAnimation = false

    Show<private>(BuyPlate:buy_plate_device, Index:int):void=
        if (not Prop.IsValid[]) {
            Log(BuyPlate, "Could not show unlock prop index {Index} because it is invalid", ?Level := log_level.Error)
            return
        }
        if (not (m_InitialTransform?)) {
            set m_InitialTransform = Some(Prop.GetTransform())
        }
        if (m_PlayedAnimation?, AnimationController := Prop.GetAnimationController[]) {
            AnimationController.Pause()
        }
        if (
            AllowAnimation?
            not (UnmanagedLockedState?)
            ManagerGroup := BuyPlate.m_ManagerGroup?
            UnlockAnimationConfig := ManagerGroup.Manager.GetUnlockAnimation[CustomData]
            AnimationController := Prop.GetAnimationController[]
        ) {
            InitialTransform := if (Tr := m_InitialTransform?). Tr else return
            Origin := InitialTransform
                .AddTranslation(InitialTransform.Rotation.RotateVector(UnlockAnimationConfig.StartOffset))
                .WithScale(SafeScale(UnlockAnimationConfig.StartScale))
            if:
                Prop.TeleportTo[Origin]
            set m_PlayedAnimation = true

            if (PrevAbortEvent := m_DelayedShow?) {
                PrevAbortEvent.Signal()
            }
            AbortEvent := event(){}
            set m_DelayedShow = Some(AbortEvent)
            spawn:
                ScheduleDelayedShow(AbortEvent)

            Keyframes := UnlockAnimationConfig.ToKeyframeDeltas(InitialTransform, Origin)
            AnimationController.SetAnimation(Keyframes, ?Mode := animation_mode.OneShot)
            AnimationController.Play()
        } else {
            if:
                (HideScaleDown?) or (m_PlayedAnimation?)
                NewTransform := m_InitialTransform?
                Prop.TeleportTo[NewTransform]
            Prop.Show()
            set m_PlayedAnimation = false
        }

    ScheduleDelayedShow<private>(AbortEvent:event())<suspends>:void=
        race:
            AbortEvent.Await()
            block:
                Sleep(0.0)
                Prop.Show()

    ApplyLockedState<override>(BuyPlate:buy_plate_device, Index:int):void=
        if (UnmanagedLockedState?) {
            return
        }
        if (Action = buy_plate_display_action.Show) {
            Hide(BuyPlate, Index)
        } else if (Action = buy_plate_display_action.Hide, Prop.IsValid[]) {
            Show(BuyPlate, Index)
        }

    ApplyUnlockedState<override>(BuyPlate:buy_plate_device, Index:int):void=
        if (Action = buy_plate_display_action.Show, Prop.IsValid[]) {
            Show(BuyPlate, Index)
        } else if (Action = buy_plate_display_action.Hide) {
            Hide(BuyPlate, Index)
        }

buy_plate_unlock_prop_manipulator_config := class<concrete>(buy_plate_unlock_display):
    @editable PropManipulator: prop_manipulator_device = prop_manipulator_device{}
    @editable Action: buy_plate_display_action = buy_plate_display_action.Show
    @editable UnmanagedLockedState: logic = false

    ApplyLockedState<override>(BuyPlate:buy_plate_device, Index:int):void=
        if (UnmanagedLockedState?) {
            return
        }
        if (Action = buy_plate_display_action.Show) {
            PropManipulator.HideProps()
        } else if (Action = buy_plate_display_action.Hide) {
            PropManipulator.ShowProps()
        }

    ApplyUnlockedState<override>(BuyPlate:buy_plate_device, Index:int):void=
        if (Action = buy_plate_display_action.Show) {
            PropManipulator.ShowProps()
        } else if (Action = buy_plate_display_action.Hide) {
            PropManipulator.HideProps()
        }

buy_plate_cost_override<public> := interface:
    TryUnlock<public>(Agent:agent, Silent:logic):logic

    HasEnough<public>(Agent:agent):logic

buy_plate_free_cost_override := class(buy_plate_cost_override):
    TryUnlock<override>(Agent:agent, Silent:logic):logic=
        true

    HasEnough<override>(Agent:agent):logic=
        true

buy_plate_enable_disable_closure := class:
    Callback: maybe_agent_callback
    ExpectedState: logic

    Call(State:logic):void=
        if (State = ExpectedState) {
            Callback(None())
        }
