using { /Fortnite.com/Devices }
using { /Verse.org/Simulation }
using { /UnrealEngine.com/Temporary/Diagnostics }

using { HavocVerse.Debug }
using { HavocVerse.Util }
using { HavocVerse.Logging }

centralizable<public> := interface:
    GetRoutingId<public>()<computes>:string

    GetCustomData<public>()<computes><decides>:[]string

    SubscribeToEvent<public>(EventName:string, Callback:maybe_agent_callback)<transacts><decides>:void

centralized_actions_filter_kind := enum:
    AnyCompatible
    DevId
    HasTag

centralized_actions_filter_config := struct<concrete>:
    @editable Type: centralized_actions_filter_kind = centralized_actions_filter_kind.AnyCompatible
    @editable Value: string = ""

centralized_actions_mapping_key_source_kind := enum:
    DevId
    TagPrefix

centralized_actions_mapping_key_source_actions := struct<concrete>:
    @editable Type: centralized_actions_mapping_key_source_kind = centralized_actions_mapping_key_source_kind.DevId
    @editable Prefix: string = ""

centralized_actions_mapping_config := struct<concrete>:
    @editable Match: string = ""
    @editable Data: string = ""

centralized_actions_device := class(creative_device, aware_device, aware_debug_device):
    @editable p_DevId: string = "N/A"
    @editable p_AccessControl: access_control_device = access_control_device{}

    @editable p_DeviceFilter: centralized_actions_filter_config = centralized_actions_filter_config{}
    @editable p_DeviceEvent: string = "unlock"
    @editable p_ActionDevice: creative_device = creative_device{}
    @editable p_ActionTemplate: string = "%Data%"
    @editable p_MappingKeySource: centralized_actions_mapping_key_source_actions = centralized_actions_mapping_key_source_actions{}
    @editable p_SingleMatch: logic = true
    @editable p_Mappings: []centralized_actions_mapping_config = array{}

    @editable p_QuickMappingsSeparator: string = ";"
    @editable p_QuickMappings: []string = array{}

    var m_AllMappings: []centralized_actions_mapping_config = array{}

    DeviceMeta<override>()<transacts>:aware_device_meta=
        aware_device_meta{
            DeviceName := "centralized_actions"
            DevId := p_DevId
        }

    DebugStats<override>()<transacts>:aware_debug_stats=
        aware_debug_stats{
            Length := map{
                "allMappings" => m_AllMappings.Length
            }
        }

    OnBegin<override>()<suspends>:void=
        AwareInitDevice(Self, OnInit)

    OnInit<private>()<suspends>:void=
        if (p_QuickMappingsSeparator.Length > 1) {
            Log("p_QuickMappingsSeparator on centralized actions device \"{p_DevId}\" is not a single character. Multi-char separators are not yet supported", ?Level := log_level.Error)
        }

        set m_AllMappings = p_Mappings
        for (Index->Entry:p_QuickMappings, Sep := p_QuickMappingsSeparator[0]) {
            Parts := Entry.Split(Sep)
            for (SplitIndex->QuickMapping:Parts, QuickMapping.Length > 0) {
                HandleQuickMapping(QuickMapping, Index, SplitIndex)
            }
        }

        Subscribed := p_AccessControl.SubscribeToDevicesBackrefs(OnDeviceRegistered)
        if (not (Subscribed?)) {
            Log("Centralized actions device \"{p_DevId}\" could not subscribe to access control device because backrefs building is not enabled", ?Level := log_level.Error)
        }
    
    HandleQuickMapping(QuickMapping:string, Index:int, SplitIndex:int):void=
        if (SplitRes := QuickMapping.SplitOnce['=']) {
            set m_AllMappings += array{centralized_actions_mapping_config{
                Match := SplitRes(0)
                Data := SplitRes(1)
            }}
        } else {
            Log("Centralized actions device \"{p_DevId}\" quick mapping \"{QuickMapping}\" (index {Index}-{SplitIndex}) could not be turned into a mapping (missing equal sign?)", ?Level := log_level.Error)
        }

    OnDeviceRegistered(Device:creative_device):void=
        if (
            CustomDevice := centralizable[Device]
            MatchesFilter[CustomDevice]
            # TODO : device filter
        ) {
            Closure := centralized_actions_mapping_closure{
                Device := Self
                OriginDevice := CustomDevice
            }
            if (CustomDevice.SubscribeToEvent[p_DeviceEvent, Closure.Call]) {
            }
            # TODO : warn if matching device (other than AnyCompatible) could not subscribe to event
        }

    OnTrigger(OriginDevice:centralizable, Agent:?agent):void=
        if (MappingKey := GetMappingKey[OriginDevice]) {
            for (
                Mapping:m_AllMappings
                Mapping.Match = MappingKey
            ) {
                TriggerConfig := custom_device_trigger_config{
                    Device := array{p_ActionDevice}
                    Data := p_ActionTemplate.ReplaceAll("%Data%", Mapping.Data)
                }
                TriggerConfig.AttemptTrigger(Agent, "centralized actions device \"{p_DevId}\"")
                if (p_SingleMatch?) {
                    return
                }
            }
        }

    MatchesFilter<private>(CustomDevice:centralizable)<computes><decides>:void=
        case(p_DeviceFilter.Type):
            centralized_actions_filter_kind.AnyCompatible => {}
            centralized_actions_filter_kind.DevId => CustomDevice.GetRoutingId() = p_DeviceFilter.Value
            centralized_actions_filter_kind.HasTag => CustomDevice.GetCustomData[].Find[p_DeviceFilter.Value]

    GetMappingKey<private>(CustomDevice:centralizable)<transacts><decides>:string=
        case(p_MappingKeySource.Type):
            centralized_actions_mapping_key_source_kind.DevId => CustomDevice.GetRoutingId()
            centralized_actions_mapping_key_source_kind.TagPrefix => {
                Prefix := p_MappingKeySource.Prefix
                var Result: ?string = false
                for (

                    Tag:CustomDevice.GetCustomData[]
                    not (Result?)
                    Tag.StartsWith[Prefix]
                    MappingKey := Tag.Slice[Prefix.Length]
                ) {
                    set Result = Some(MappingKey)
                }
                Result?
            }

centralized_actions_mapping_closure := class:
    Device: centralized_actions_device
    OriginDevice: centralizable

    Call(Agent:?agent):void=
        Device.OnTrigger(OriginDevice, Agent)
