using { /Fortnite.com/Devices }
using { /Fortnite.com/Characters }
using { /Verse.org/Simulation }
using { /Verse.org/Random }
using { /UnrealEngine.com/Temporary/Diagnostics }
using { /UnrealEngine.com/Temporary/SpatialMath }

using { HavocVerse.Debug }
using { HavocVerse.Devices }
using { HavocVerse.Logging }
using { HavocVerse.Util }

elimination_ext_device := class(creative_device, aware_device):
    @editable p_CreatureSpawners: []creature_spawner_device = array{}
    @editable p_CreaturePlacers: []creature_placer_device = array{}

    @editable p_DropChance: float = 1.0
    @editable p_DropResourceEndpoint: resource_endpoint_device = resource_endpoint_device{}
    @editable p_DropAmount: int = 1
    @editable p_SendFeedback: logic = true
    @editable p_ExtraTrigger: []trigger_device = array{}

    @editable p_ParticlePropAsset: creative_prop_asset = DefaultCreativePropAsset
    @editable p_ParticleCreatureOffset: vector3 = vector3{Z := 50.0}
    @editable p_ParticleCreatureOffsetUseLocalRotation: logic = true
    @editable p_ParticleDeleteAfter: float = 0.5

    DeviceMeta<override>()<transacts>:aware_device_meta=
        aware_device_meta{
            DeviceName := "elimination_ext"
        }

    OnBegin<override>()<suspends>:void=
        AwareInitDevice(Self, OnInit)

    OnInit<private>()<suspends>:void=
        for (CreatureSpawner:p_CreatureSpawners) {
            CreatureSpawner.EliminatedEvent.Subscribe(OnEliminated)
        }
        for (CreaturePlacer:p_CreaturePlacers) {
            Closure := elimination_ext_placer_closure{
                Device := Self
            }
            CreaturePlacer.SpawnedEvent.Subscribe(Closure.OnSpawned)
            CreaturePlacer.EliminatedEvent.Subscribe(Closure.OnEliminated)
        }

    OnEliminated<private>(Interaction:device_ai_interaction_result):void=
        if (Player := player[Interaction.Source?]) {
            CreatureCharacter := option{Interaction.Target?.GetFortCharacter[]}
            HandleElimination(Player, CreatureCharacter)
        }

    HandleElimination<internal>(Player:player, CreatureCharacter:?fort_character):void=
        if (GetRandomFloat(0.0, 1.0) >= p_DropChance) {
            return
        }
        p_DropResourceEndpoint.Give(Some(Player), p_DropAmount, ?SendFeedback := p_SendFeedback)
        for (Trigger:p_ExtraTrigger) {
            Trigger.Trigger(Player)
        }
        if (Creature := CreatureCharacter?) {
            CreatureTransform := Creature.GetTransform()
            ParticleTransform := CreatureTransform
                .WithTranslation(if (p_ParticleCreatureOffsetUseLocalRotation?) {
                        CreatureTransform.Translation + CreatureTransform.Rotation.RotateVector(p_ParticleCreatureOffset)
                    } else {
                        CreatureTransform.Translation + p_ParticleCreatureOffset
                    })
            SpawnedPropRes := SpawnProp(p_ParticlePropAsset, ParticleTransform)
            if (SpawnedProp := SpawnedPropRes(0)?) {
                spawn:
                    HandleParticleCleanup(SpawnedProp)
            }
        }

    HandleParticleCleanup<private>(Prop:creative_prop)<suspends>:void=
        Sleep(p_ParticleDeleteAfter)
        Prop.Dispose()

elimination_ext_placer_closure := class:
    Device: elimination_ext_device

    var m_SpawnedCreature: ?agent = false

    OnSpawned<internal>(Creature:agent):void=
        set m_SpawnedCreature = Some(Creature)

    OnEliminated<internal>(Agent:?agent):void=
        CreatureCharacter := option{m_SpawnedCreature?.GetFortCharacter[]}
        set m_SpawnedCreature = None()
        if (Player := player[Agent?]) {
            Device.HandleElimination(Player, CreatureCharacter)
        }
