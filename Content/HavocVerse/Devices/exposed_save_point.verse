using { /Fortnite.com/Devices }
using { /Verse.org/Simulation }
using { /UnrealEngine.com/Temporary/Diagnostics }

exposed_save_point_device<public> := class(creative_device):
    @editable p_LoadPlayerTrigger: trigger_device = trigger_device{}
    @editable p_SavePlayerTrigger: trigger_device = trigger_device{}
    @editable p_ClearPlayerTrigger: trigger_device = trigger_device{}

    LoadPlayer<public>(Player:player):void=
        p_LoadPlayerTrigger.Trigger(Player)

    SavePlayer<public>(Player:player):void=
        p_SavePlayerTrigger.Trigger(Player)

    ClearDataForPlayer<public>(Player:player):void=
        p_ClearPlayerTrigger.Trigger(Player)
