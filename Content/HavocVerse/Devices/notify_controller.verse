using { /Fortnite.com/Devices }
using { /Verse.org/Simulation }
using { /UnrealEngine.com/Temporary/Diagnostics }

using { HavocVerse.Debug }
using { HavocVerse.Util }
using { HavocVerse.Logging }

notify_template<public> := class<final><concrete>:
    @editable TemplateId: string = "default"
    @editable HUDMessage: hud_message_device = hud_message_device{}
    @editable DefaultText: string = ""
    @editable DefaultImageIndex: int = 0
    @editable DefaultStyle: int = 0
    @editable DefaultDuration: float = 1.5

notify_controller_device<public> := class<final><concrete>(creative_device, aware_device, aware_debug_device):
    @editable p_DevId: string = "N/A"

    @editable p_Templates: []notify_template = array{}
    @editable p_QueueProcessInterval: float = 0.1
    @editable p_ColdStartDelay: float = 0.2

    @editable p_PackedDataSeqControllers: []cinematic_sequence_device = array{}
    @editable p_PackImageBits: int = 6
    @editable p_PackStyleBits: int = 2

    var m_Players<private>: [player]notify_controller_player_state = map{}

    DeviceMeta<override>()<transacts>:aware_device_meta=
        aware_device_meta{
            DeviceName := "notify_controller_device"
            DevId := p_DevId
        }

    DebugStats<override>()<transacts>:aware_debug_stats=
        aware_debug_stats{
            Length := map{
                "players" => m_Players.Length
            }
        }

    OnBegin<override>()<suspends>:void=
        AwareInitDevice(Self, OnInit)
    
    OnInit<private>()<suspends>:void=
        GetPlayspace().PlayerAddedEvent().Subscribe(OnPlayerAdded)
        GetPlayspace().PlayerRemovedEvent().Subscribe(OnPlayerRemoved)
        for (Player:GetPlayspace().GetPlayers()) {
            OnPlayerAdded(Player)
        }

    OnPlayerAdded<private>(Player:player):void=
        for (__->PlayerState:m_Players) {
            set PlayerState.m_ColdStart = true
        }

    OnPlayerRemoved<private>(Player:player):void=
        RemoveRes := m_Players.RemoveByKey(Player)
        set m_Players = RemoveRes(0)
        if (PlayerState := RemoveRes(1)?) {
            set PlayerState.m_Queue = array{}
            if (OnScreen := PlayerState.m_OnScreen?) {
                OnScreen.MessageDevice.Hide(Player)
            }
            set PlayerState.m_OnScreen = None()
        }

    GetPlayerState<private>(Player:player):notify_controller_player_state=
        if (PlayerState := m_Players[Player]) {
            PlayerState
        } else {
            PlayerState := notify_controller_player_state{}
            if (set m_Players[Player] = PlayerState) {}
            PlayerState
        }

    Notify<public>(
        Player:player, ?TemplateId:string,
        ?Message:?message = false, ?ImageIndex:?int = false, ?Style:?int = false,
        ?Duration:float = -1.0
    ):void=
        for (Template:p_Templates, Template.TemplateId = TemplateId) {
            l_ImageIndex := ImageIndex? or Template.DefaultImageIndex
            Notify(Player, Template.HUDMessage,
                ?Message := Message? or ToMessage(Template.DefaultText),
                ?ImageIndex := l_ImageIndex,
                ?Style := Style? or Template.DefaultStyle,
                ?Duration := if (Duration < 0.0). Template.DefaultDuration else Duration)
            Log("Pushed notification to '{Player}'; template: \"{TemplateId}\", img index: {l_ImageIndex}", ?Level := log_level.Debug)
            return
        }
        Log("Could not push notification to '{Player}': template \"{TemplateId}\" isn't configured", ?Level := log_level.Error)

    Notify<public>(
        Player:player, MessageDevice:hud_message_device,
        ?Message:message, ?ImageIndex:int, ?Style:int, ?Duration:float
    ):void=
        PlayerState := GetPlayerState(Player)
        Msg := notify_controller_message{
            MessageDevice := MessageDevice
            Message := Message
            ImageIndex := ImageIndex
            Style := Style
            Duration := Duration
        }
        if (not PlayerState.m_OnScreen?) {
            set PlayerState.m_OnScreen = Some(Msg)
            spawn:
                ProcessQueue(Player, PlayerState)
        } else {
            set PlayerState.m_Queue += array{Msg}
        }

    ProcessQueue<private>(Player:player, PlayerState:notify_controller_player_state)<suspends>:void=
        loop {
            if (Msg := PlayerState.m_OnScreen?) {
                SendNotification(PlayerState, Player, Msg)
            } else {
                break
            }
            set PlayerState.m_OnScreen = None()
            if (PopRes := PlayerState.m_Queue.Pop[0]) {
                set PlayerState.m_Queue = PopRes(0)
                NextMsg := PopRes(1)
                set PlayerState.m_OnScreen = Some(NextMsg)
            }
        }

    SendNotification<private>(PlayerState:notify_controller_player_state, Player:player, Msg:notify_controller_message)<suspends>:void=
        ShiftImageData := DefaultingFloor(Pow(2.0, ToFloat(p_PackStyleBits)))
        PackedData := Clamp(Msg.ImageIndex, 0, FromBits(array{true}.Repeat(p_PackImageBits))) * ShiftImageData
                    + Clamp(Msg.Style, 0, FromBits(array{true}.Repeat(p_PackStyleBits)))
        if (PlayerState.m_ColdStart?) {
            set PlayerState.m_ColdStart = false
            Log(Self, "Handling packedData cold start for '{Player}'", ?Level := log_level.Verbose)
            GetPlayspace().SetMPCValue(p_PackedDataSeqControllers, Player, 0, ?Context := ToString(Self) + " - cold start")
            Sleep(p_ColdStartDelay)
        }
        GetPlayspace().SetMPCValue(p_PackedDataSeqControllers, Player, PackedData, ?Context := ToString(Self))
        Msg.MessageDevice.Show(Player, Msg.Message, ?DisplayTime := Msg.Duration)
        Log(Self, "Showing notification to '{Player}'; packed data: {PackedData}", ?Level := log_level.Verbose)
        Sleep(Msg.Duration + p_QueueProcessInterval)

notify_controller_player_state<internal> := class<final>:
    var m_ColdStart: logic = true
    var m_Queue: []notify_controller_message = array{}
    var m_OnScreen: ?notify_controller_message = false

notify_controller_message<internal> := class<final>:
    MessageDevice: hud_message_device
    Message: message
    ImageIndex: int
    Style: int
    Duration: float
