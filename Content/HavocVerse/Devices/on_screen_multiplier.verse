using { /Fortnite.com/Devices }
using { /Verse.org/Simulation }
using { /UnrealEngine.com/Temporary/Diagnostics }
using { /UnrealEngine.com/Temporary/SpatialMath }
using { /UnrealEngine.com/Temporary/UI }

using { HavocVerse.Debug }
using { HavocVerse.UI }
using { HavocVerse.Util }
using { HavocVerse.Logging }

on_screen_multiplier_source_kind := enum:
    ResourceManager
    AutoGenerationManager

on_screen_multiplier_entry := class<concrete>:
    @editable RangeFrom: float = 1.0
    @editable RangeTo: float = 1.0
    @editable ImageTexture: string = ""

on_screen_multiplier_device := class(creative_device, aware_device):
    @editable p_DevId<public>: string = "N/A"

    @editable p_SourceKind: on_screen_multiplier_source_kind = on_screen_multiplier_source_kind.ResourceManager
    @editable p_SourceResourceManager: ?resource_manager_device = false
    @editable p_SourceAutoGenerationManager: ?auto_generation_manager = false

    @editable p_ResourceId: string = ""
    @editable p_Entries: []on_screen_multiplier_entry = array{}

    @editable p_HUDAccessControl: ?access_control_device = false
    @editable p_HUDManagerConfig: hud_manager_base_config = hud_manager_base_config{
        RefreshRate := 1.0
    }
    @editable p_HUDImageSize: vector2 = vector2{X := 100.0, Y := 50.0}
    @editable p_HUDImagePlacement: ui_canvas_placement_config = ui_canvas_placement_config{}

    var m_HudManager: ?on_screen_multiplier_hud_manager = false

    DeviceMeta<override>()<transacts>:aware_device_meta=
        aware_device_meta{
            DeviceName := "on_screen_multiplier"
            DevId := p_DevId
        }

    OnBegin<override>()<suspends>:void=
        AwareInitDevice(Self, OnInit)

    OnInit<private>()<suspends>:void=
        case(p_SourceKind):
            on_screen_multiplier_source_kind.ResourceManager =>
                if (not (p_SourceResourceManager?)) {
                    Log("{Self} is set to source kind ResourceManager but no resource manager is set", ?Level := log_level.Error)
                }
                if (p_SourceAutoGenerationManager?) {
                    Log("{Self} is set to source kind ResourceManager but an auto generation manager is set", ?Level := log_level.Warning)
                }
            on_screen_multiplier_source_kind.AutoGenerationManager =>
                if (not (p_SourceAutoGenerationManager?)) {
                    Log("{Self} is set to source kind AutoGenerationManager but no auto generation manager is set", ?Level := log_level.Error)
                }
                if (p_SourceResourceManager?) {
                    Log("{Self} is set to source kind AutoGenerationManager but a resource manager is set", ?Level := log_level.Warning)
                }

        HudManager := on_screen_multiplier_hud_manager{
            AccessControl := p_HUDAccessControl
            RefreshRate := p_HUDManagerConfig.RefreshRate
            ZOrder := p_HUDManagerConfig.ZOrder
            Device := Self
        }
        set m_HudManager = Some(HudManager)
        spawn:
            HudManager.Run(GetPlayspace())

    GetImageTexture<internal>(Multiplier:float)<transacts>:?string=
        for (Entry:p_Entries) {
            if (Multiplier >= Entry.RangeFrom and Multiplier <= Entry.RangeTo) {
                return Some(Entry.ImageTexture)
            }
        }
        None()

on_screen_multiplier_hud_manager := class(hud_manager_base):
    Device: on_screen_multiplier_device

    CreateHUD<override>(Player:player):?hud_manager_player=
        Image := CreateDefaultingTextureBlock(Device.p_HUDImageSize, None())
        Canvas := canvas{
            Slots := array{
                Device.p_HUDImagePlacement.ToCanvasSlot(Image.Widget())
            }
        }
        Element := on_screen_multiplier_hud_element{
            Widget := Canvas
            Device := Device
            Player := Player
            Image := Image
        }
        Some(MakeSingleElementHUDPlayer(Player, Canvas, Element))

on_screen_multiplier_hud_element := class(hud_element):
    Device: on_screen_multiplier_device
    Player: player

    Image: defaulting_texture_block

    var m_LastMultiplier: ?float = false

    Update<override>():void=
        Multiplier: float = if (
            Device.p_SourceKind = on_screen_multiplier_source_kind.ResourceManager
            ResourceManager := Device.p_SourceResourceManager?
        ) {
            if (Device.p_ResourceId.Length > 0) {
                ResourceManager.GetEffectiveMultiplier(Player, Device.p_ResourceId)
            } else {
                ResourceManager.GetTotalMultiplier(Player)
            }
        } else if (
            Device.p_SourceKind = on_screen_multiplier_source_kind.AutoGenerationManager
            AutoGenerationManager := Device.p_SourceAutoGenerationManager?
        ) {
            if (Device.p_ResourceId.Length > 0) {
                AutoGenerationManager.GetEffectiveMultiplier(Device.p_ResourceId)
            } else {
                AutoGenerationManager.GetGlobalMultiplierInstance().GetTotalMultiplier()
            }
        } else {
            return
        }
        if (PrevMultiplier := m_LastMultiplier?, PrevMultiplier = Multiplier) {
            return
        }
        Log("{Device} updating to {Multiplier}x multiplier", ?Level := log_level.Verbose)
        set m_LastMultiplier = Some(Multiplier)
        TextureName := Device.GetImageTexture(Multiplier)
        Image.SetCachingTexture(TextureName)
