using { /Fortnite.com/Devices }
using { /Fortnite.com/Characters }
using { /Verse.org/Simulation }
using { /Verse.org/Random }
using { /UnrealEngine.com/Temporary/Diagnostics }
using { /UnrealEngine.com/Temporary/SpatialMath }

using { HavocVerse.Debug }
using { HavocVerse.Storage }
using { HavocVerse.Logging }
using { HavocVerse.Util }

pet_condition_provider<public> := interface:
    IsPetUnlocked<public>(Player:player, PetId:string)<transacts><decides>:void

pet_ambient_noise_config<public> := class<concrete><final>:
    @editable MinInterval: float = 10.0
    @editable MaxInterval: float = 30.0
    @editable AudioPlayer: audio_player_device = audio_player_device{}
    @editable AudioPlayerTeleport: logic = false
    @editable PlayToInstigatorOnly: logic = true

pet_group_config := class<concrete><final>:
    @editable PetName: string = ""
    @editable GroupWithOverride: string = ""
    @editable PlayerOffset: vector3 = vector3{X := 150.0, Y := -100.0, Z := -80.0}
    @editable Speed: float = 600.0
    @editable AlwaysFacePlayer: logic = false
    @editable ClosenessThreshold: float = 400.0
    @editable Tiers: []pet_tier_config = array{}
    @editable AmbientNoises: []pet_ambient_noise_config = array{}

pet_tier_config := class<concrete><final>:
    @editable TierName: string = ""
    @editable ExpectsConditionProvider: logic = false

    @editable StateIdle: pet_tier_anim_state_config = pet_tier_anim_state_config{}
    @editable StateRunning: ?pet_tier_anim_state_config = false 

    @editable UserActions: []pet_user_action_config = array{}
    @editable EventTriggers: []pet_event_trigger_config = array{}

pet_tier_anim_state_config := class<concrete><final>:
    @editable PropAsset: ?creative_prop_asset = false
    @editable PropPool: ?shared_prop_pool_device = false
    @editable PropScale: vector3 = vector3{X := 1.0, Y := 1.0, Z := 1.0}

pet_user_action_kind := enum:
    Unlock
    Activate
    Deactivate
    Toggle

pet_user_action_config := class<concrete><final>:
    @editable Action: pet_user_action_kind = pet_user_action_kind.Unlock
    @editable OnTrigger: trigger_device = trigger_device{}

pet_event_trigger_kind := enum:
    Unlocked
    FirstActivation
    Activated
    Deactivated

pet_event_trigger_config := class<concrete><final>:
    @editable OnEvent: pet_event_trigger_kind = pet_event_trigger_kind.Unlocked
    @editable Trigger: trigger_device = trigger_device{}

pet_config := class<final>:
    GroupName: string
    PetId: string
    PlayerOffset: vector3
    Speed: float
    AlwaysFacePlayer: logic
    ClosenessThreshold: float
    AmbientNoises: []pet_ambient_noise_config
    TierConfig: pet_tier_config
    ExpectsConditionProvider: logic

    var m_ConditionProvider: ?pet_condition_provider = false

    TriggerEvent<public>(Event:pet_event_trigger_kind, Agent:agent):void=
        for (TriggerConfig:TierConfig.EventTriggers, TriggerConfig.OnEvent = Event) {
            TriggerConfig.Trigger.Trigger(Agent)
        }

pet_constraint_resolve_strategy := enum:
    Fail
    ReplaceOldest

pet_constraint_group_config := class<concrete>(pet_constraint_config):
    @editable GroupName: string = ""

pet_constraint_config := class<concrete>:
    @editable MaxActivePets: int = 1
    @editable ResolutionStrategy: pet_constraint_resolve_strategy = pet_constraint_resolve_strategy.ReplaceOldest
    @editable FailHUDMessage: []hud_message_device = array{}

pet_controller_device<public> := class<concrete><final>(creative_device, aware_device, aware_debug_device, custom_device, persistence_recordable):
    @editable p_Pets: []pet_group_config = array{}

    @editable p_Constraints: []pet_constraint_group_config = array{}
    @editable p_GlobalConstraint: pet_constraint_config = pet_constraint_config{}
 
    @editable p_TeleportAtDistance: float = 3000.0
    @editable p_MaxDistancePerTick: float = 500.0
    @editable p_GroundInferenceTimeout: float = 3.0

    var m_PetConfigs: [string]pet_config = map{}

    var m_Players: [player]pet_controller_player = map{}

    var m_EquippedListeners: []type{__(:player,:string):void} = array{}
    var m_UnequippedListeners: []type{__(:player,:string):void} = array{}
    var m_UnlockListeners: []type{__(:player,:string):void} = array{}

    var m_PersistenceControl: persistence_recordable_control = persistence_recordable_control{}

    DeviceMeta<override>()<transacts>:aware_device_meta=
        aware_device_meta{
            DeviceName := "pet_controller"
        }

    DebugStats<override>()<transacts>:aware_debug_stats=
        aware_debug_stats{
            Length := map{
                "petConfigs" => m_PetConfigs.Length
                "players" => m_Players.Length
                "listeners::equip" => m_EquippedListeners.Length
                "listeners::unequip" => m_UnequippedListeners.Length
                "listeners::unlock" => m_UnlockListeners.Length
            }
        }

    OnBegin<override>()<suspends>:void=
        AwareInitDevice(Self, OnInit)

    EnsurePreInit<private>():void=
        if (p_Pets.Length = 0 or m_PetConfigs.Length > 0) {
            return
        }
        for (PetConfig:p_Pets) {
            for (Tier:PetConfig.Tiers) {
                PetId := "{PetConfig.PetName}:{Tier.TierName}"

                if (not (Tier.StateIdle.PropAsset?), not (Tier.StateIdle.PropPool?)) {
                    Log("Pet \"{PetId}\" has no prop asset or prop pool configured for idle state (default)", ?Level := log_level.Error)
                }

                if (set m_PetConfigs[PetId] = pet_config{
                    GroupName := if (PetConfig.GroupWithOverride.Length > 0) {
                        PetConfig.GroupWithOverride
                    } else {
                        PetConfig.PetName
                    }
                    PetId := PetId
                    PlayerOffset := PetConfig.PlayerOffset
                    Speed := PetConfig.Speed
                    AlwaysFacePlayer := PetConfig.AlwaysFacePlayer
                    ClosenessThreshold := PetConfig.ClosenessThreshold
                    AmbientNoises := PetConfig.AmbientNoises
                    TierConfig := Tier
                    ExpectsConditionProvider := Tier.ExpectsConditionProvider
                }) {}
            }
        }

    OnInit<private>()<suspends>:void=
        EnsurePreInit()

        for (PetConfig:p_Pets) {
            for (Tier:PetConfig.Tiers) {
                PetId := "{PetConfig.PetName}:{Tier.TierName}"
                for (UserActionConfig:Tier.UserActions) {
                    UserAction := pet_controller_user_action{
                        Controller := Self
                        Config := UserActionConfig
                        PetId := PetId
                    }
                    UserActionConfig.OnTrigger.TriggeredEvent.Subscribe(UserAction.OnTriggered)
                }
            }
        }

        GetPlayspace().PlayerRemovedEvent().Subscribe(OnPlayerRemoved)
        AwareSubscribeToPostInit(OnPostInit)

    OnPostInit<private>():void=
        for (
            PetId->PetConfig:m_PetConfigs
            PetConfig.ExpectsConditionProvider?
            not (PetConfig.m_ConditionProvider?)
        ) {
            Log("Pet \"{PetId}\" expects a condition provider but none was assigned", ?Level := log_level.Error)
        }

    SetPetConditionProvider<public>(PetId:string, ConditionProvider:pet_condition_provider):void=
        EnsurePreInit()
        if (PetConfig := m_PetConfigs[PetId]) {
            if (PetConfig.ExpectsConditionProvider?) {
                if (PetConfig.m_ConditionProvider?) {
                    Log("Overwriting condition provider for pet \"{PetId}\"", ?Level := log_level.Warning)
                }
                set PetConfig.m_ConditionProvider = Some(ConditionProvider)
                Log("Set condition provider for pet \"{PetId}\"", ?Level := log_level.Debug)
            } else {
                Log("Couldn't set condition provider for pet \"{PetId}\": pet not expecting condition", ?Level := log_level.Error)
            }
        } else {
            Log("Couldn't set condition provider for pet \"{PetId}\": pet not found", ?Level := log_level.Error)
        }

    GetPlayerState<private>(Player:player):pet_controller_player=
        if (PlayerState := m_Players[Player]) {
            PlayerState
        } else {
            PlayerState := pet_controller_player{}
            if (set m_Players[Player] = PlayerState) {}
            PlayerState
        }
    
    OnPlayerRemoved<private>(Player:player):void=
        RemoveRes := m_Players.RemoveByKey(Player)
        set m_Players = RemoveRes(0)
        if (PetPlayer := RemoveRes(1)?) {
            for (__->PetInstance:PetPlayer.m_Active) {
                PetInstance.Discard()
            }
            set PetPlayer.m_Active = map{}
        }

    GetPetConfig<internal>(PetId:string)<transacts><decides>:pet_config=
        m_PetConfigs[PetId]

    GetPetConfigs<internal>():[]pet_config=
        for (__->PetConfig:m_PetConfigs) {
            PetConfig
        }

    IsPetUnlocked<public>(Player:player, PetId:string)<transacts><decides>:void=
        PetConfig := m_PetConfigs[PetId]
        if (PetConfig.ExpectsConditionProvider?) {
            PetConfig.m_ConditionProvider?.IsPetUnlocked[Player, PetId]
        } else {
            m_Players[Player].m_Unlocked[PetId]
        }

    IsPetActive<public>(Player:player, PetId:string)<transacts><decides>:void=
        m_Players[Player].m_Active[PetId]

    GetActivePets<public>(Player:player)<transacts>:[]string=
        for (PetId->__:m_Players[Player].m_Active) {
            PetId
        }

    UnlockPet<public>(Player:player, PetId:string, ?AutoSave:logic = true):logic=
        PetConfig := if (l_PetConfig := m_PetConfigs[PetId]) {
            l_PetConfig
        } else {
            Log(MsgPetNotFound("unlock", Player, PetId), ?Level := log_level.Error)
            return false
        }
        if (PetConfig.ExpectsConditionProvider?) {
            Log("Trying to unlock pet \"{PetId}\" but it is backed by a condition provider", ?Level := log_level.Error)
        }
        PetPlayer := GetPlayerState(Player)
        if (not PetPlayer.m_Unlocked[PetId], set PetPlayer.m_Unlocked[PetId] = false) {
            PetConfig.TriggerEvent(pet_event_trigger_kind.Unlocked, Player)
            Log(MsgSuccess("Unlocked", Player, PetId), ?Level := log_level.Debug)
            for (Listener:m_UnlockListeners) {
                Listener(Player, PetId)
            }
        }
        if (AutoSave?) {
            m_PersistenceControl.Save(Player)
        }
        true

    ActivatePet<public>(Player:player, PetId:string, ?AutoUnlock:logic = false, ?AutoSave:logic = true):void=
        PetConfig := if (l_PetConfig := m_PetConfigs[PetId]) {
            l_PetConfig
        } else {
            Log(MsgPetNotFound("activate", Player, PetId), ?Level := log_level.Error)
            return
        }
        if (not IsPetUnlocked[Player, PetId]) {
            if (AutoUnlock?) {
                if (not (PetConfig.ExpectsConditionProvider?)) {
                    UnlockPet(Player, PetId, ?AutoSave := false)
                }
            } else {
                return
            }
        }
        PetPlayer := GetPlayerState(Player)
        if (PetPlayer.m_Active[PetId]) {
            return
        }
        for (
            Constraint:p_Constraints
            Constraint.GroupName = PetConfig.GroupName
        ) {
            GroupResolved := EnforceConstraint(Player, PetPlayer, Constraint, Constraint.GroupName)
            if (not (GroupResolved?)) {
                return
            }
        }
        Resolved := EnforceConstraint(Player, PetPlayer, p_GlobalConstraint, "")
        if (not (Resolved?)) {
            return
        }
        if (not (PetPlayer.m_Unlocked[PetId] = true), set PetPlayer.m_Unlocked[PetId] = true) {
            PetConfig.TriggerEvent(pet_event_trigger_kind.FirstActivation, Player)
        }
        if (FortCharacter := Player.GetFortCharacter[]) {
            PetInstance := pet_instance{
                Controller := Self
                Owner := FortCharacter
                OwnerAgent := Player
                PetConfig := PetConfig
                ActivatedAt := GetSimulationElapsedTime()
            }
            if (set PetPlayer.m_Active[PetId] = PetInstance) {
                spawn:
                    PetInstance.PetLoop()
                for (NoiseConfig:PetConfig.AmbientNoises) {
                    spawn:
                        PetInstance.StartAmbientNoiseLoop(NoiseConfig)
                }
                PetConfig.TriggerEvent(pet_event_trigger_kind.Activated, Player)
                Log(MsgSuccess("Activated", Player, PetId), ?Level := log_level.Debug)
                for (Listener:m_EquippedListeners) {
                    Listener(Player, PetId)
                }
            }
        }
        if (AutoSave?) {
            m_PersistenceControl.Save(Player)
        }

    EnforceConstraint<private>(
        Player:player,
        PetPlayer:pet_controller_player,
        Constraint:pet_constraint_config,
        GroupName:string
    ):logic=
        ActivePets := PetPlayer.GetPetInstances(GroupName)
        if (ActivePets.Length < Constraint.MaxActivePets) {
            return true
        }
        case(Constraint.ResolutionStrategy):
            pet_constraint_resolve_strategy.Fail =>
                for (HudMessage:Constraint.FailHUDMessage) {
                    HudMessage.Show(Player)
                }
                return false
            pet_constraint_resolve_strategy.ReplaceOldest =>
                OrderedActivePets := ActivePets.QuickSortDesc(CmpPetInstanceByActivateTime)
                if (ToDeactivate := OrderedActivePets[0]) {
                    DeactivatePet(Player, ToDeactivate.PetConfig.PetId, ?AutoSave := false)
                    Log(MsgConstraintResolvedReplaceOldest(Player, ToDeactivate.PetConfig.PetId), ?Level := log_level.Debug)
                }
        true

    MsgConstraintResolvedReplaceOldest<private><localizes>(Agent:agent, PetId:string):message=
        "Resolved pet constraint for '{Agent}' by replacing oldest pet '{PetId}'"

    DeactivatePet<public>(Player:player, PetId:string, ?AutoSave:logic = true):void=
        PetConfig := if (l_PetConfig := m_PetConfigs[PetId]) {
            l_PetConfig
        } else {
            Log(MsgPetNotFound("deactivate", Player, PetId), ?Level := log_level.Error)
            return
        }
        PetPlayer := GetPlayerState(Player)
        if (PetInstance := PetPlayer.m_Active[PetId]) {
            PetInstance.Discard()
            set PetPlayer.m_Active = PetPlayer.m_Active.RemoveByKey(PetId)(0)
            PetConfig.TriggerEvent(pet_event_trigger_kind.Deactivated, Player)
            Log(MsgSuccess("Deactivated", Player, PetId), ?Level := log_level.Debug)
            for (Listener:m_UnequippedListeners) {
                Listener(Player, PetId)
            }
            if (AutoSave?) {
                m_PersistenceControl.Save(Player)
            }
        }

    MsgPetNotFound<private><localizes>(Action:string, Player:player, PetId:string):message=
        "Couldn't {Action} pet '{PetId}' for '{Player}' because it doesn't exist"
    MsgSuccess<private><localizes>(Action:string, Player:player, PetId:string):message=
        "{Action} pet '{PetId}' for '{Player}'"

    SubscribeToEquipEvents<public>(
        PetEquippedCallback:type{__(:player,:string):void},
        PetUnequippedCallback:type{__(:player,:string):void}
    ):void=
        set m_EquippedListeners += array{PetEquippedCallback}
        set m_UnequippedListeners += array{PetUnequippedCallback}

    SubscribeToUnlockEvent<public>(Callback:type{__(:player,:string):void}):void=
        set m_UnlockListeners += array{Callback}

    CustomTrigger<override>(Input:string, Agent:?agent):logic=
        if (Player := player[Agent?], SplitRes := Input.SplitOnce[':']) {
            if (SplitRes(0) = "unlock", PetId := SplitRes(1)) {
                UnlockPet(Player, PetId)
                return true
            } else if (SplitRes(0) = "activate", PetId := SplitRes(1)) {
                ActivatePet(Player, PetId)
                return true
            } else if (SplitRes(0) = "deactivate", PetId := SplitRes(1)) {
                DeactivatePet(Player, PetId)
                return true
            } else if (SplitRes(0) = "toggle", PetId := SplitRes(1)) {
                if (IsPetActive[Player, PetId]) {
                    DeactivatePet(Player, PetId)
                } else {
                    ActivatePet(Player, PetId)
                }
                return true
            }
        }
        false

    GetPersistenceControl<override>()<transacts><decides>:persistence_recordable_control=
        m_PersistenceControl

    PersistenceReserve<override>(StorageMem:storage_memory)<transacts><decides>:persistence_recordable_state=
        var PetChunks: [string]storage_memory_chunk = map{}
        for (PetId->__:m_PetConfigs) {
            if:
                MemChunk := StorageMem.Reserve[3, "pet_{PetId}"]
                set PetChunks[PetId] = MemChunk
        }
        pet_controller_persistence_state{
            PetChunks := PetChunks
        }

    PersistenceSave<override>(i_State:persistence_recordable_state, Agent:agent, SaveCause:persistence_save_cause):void=
        if (State := pet_controller_persistence_state[i_State], PlayerState := m_Players[Agent]) {
            for (PetId->MemChunk:State.PetChunks) {
                IsUnlocked := if (PlayerState.m_Unlocked[PetId]). true else false
                EverActivated := if (l_EverActivated := PlayerState.m_Unlocked[PetId]). l_EverActivated else false
                IsActive := if (PlayerState.m_Active[PetId]). true else false
                MemChunk.Write(Agent, array{
                    IsUnlocked,
                    EverActivated,
                    IsActive
                })
            }
        }

    PersistenceLoad<override>(i_State:persistence_recordable_state, Agent:agent, Query:persistence_load_query):void=
        if (State := pet_controller_persistence_state[i_State], Player := player[Agent]) {
            PlayerState := GetPlayerState(Player)
            for (PetId->MemChunk:State.PetChunks) {
                Data := MemChunk.Read(Agent)
                if (Data[0]?) {
                    if:
                        EverActivated := Data[1]
                        set PlayerState.m_Unlocked[PetId] = EverActivated
                    else if:
                        set PlayerState.m_Unlocked[PetId] = false
                    if (Data[2]?) {
                        ActivatePet(Player, PetId)
                    }
                } else if (Query.Overwrite?) {
                    set PlayerState.m_Unlocked = PlayerState.m_Unlocked.RemoveByKey(PetId)(0)
                }
                if (Query.Overwrite?, not (Data[2]?), PlayerState.m_Active[PetId]) {
                    DeactivatePet(Player, PetId, ?AutoSave := false)
                }
            }
        }

pet_controller_persistence_state := class(persistence_recordable_state):
    PetChunks: [string]storage_memory_chunk

    Reset<override>(Agent:agent):void=
        for (__->MemChunk:PetChunks) {
            MemChunk.Reset(Agent)
        }

pet_controller_player := class:
    var m_Unlocked: [string]logic = map{}
    var m_Active: [string]pet_instance = map{}

    GetPetInstances<internal>(GroupName:string)<transacts>:[]pet_instance=
        if (GroupName.Length > 0) {
            for (__->PetInstance:m_Active, PetInstance.PetConfig.GroupName = GroupName) {
                PetInstance
            }
        } else {
            m_Active.Values()
        }

pet_state := enum:
    Idle
    Running

pet_instance := class<final>:
    Controller: pet_controller_device
    Owner: fort_character
    OwnerAgent: agent
    PetConfig: pet_config

    ActivatedAt: float

    var m_State<private>: pet_state = pet_state.Idle
    var m_SpawnedProp<private>: ?creative_prop = false
    var m_PropPoolEntry<private>: ?prop_pool_entry = false

    var m_LastGroundZ: ?tuple(float,float) = false

    var m_Removed: logic = false

    RemovedEvent: event() = event(){}

    Discard<internal>():void=
        if (m_Removed?) {
            return
        }
        set m_Removed = true
        RemovedEvent.Signal()

    GetProp<private>()<transacts><decides>:creative_prop=
        if (Prop := m_SpawnedProp?) {
            Prop
        } else {
            m_PropPoolEntry?.Prop
        }

    SwitchState<private>(NewState:pet_state):?creative_prop=
        PetProp := if (m_State = NewState, Prop := GetProp[]) {
            Prop
        } else {
            set m_State = NewState
            AnimState := (case(m_State):
                pet_state.Running => PetConfig.TierConfig.StateRunning?
            ) or PetConfig.TierConfig.StateIdle

            SpawnTransform := if (Prop := m_SpawnedProp?) {
                l_SpawnTransform := Prop.GetTransform()
                Prop.Dispose()
                l_SpawnTransform
            } else if (PoolEntry := m_PropPoolEntry?) {
                l_SpawnTransform := PoolEntry.Prop.GetTransform()
                PoolEntry.Release()
                l_SpawnTransform
            } else {
                PlayerPosition := Owner.GetTransform().Translation
                SpawnPosition := PlayerPosition + ApplyPlayerOffset(Owner.GetViewRotation(), PetConfig.PlayerOffset)
                transform{
                    Translation := SpawnPosition
                    Rotation := GetPlayerFacingRotation(PlayerPosition, SpawnPosition)
                    Scale := AnimState.PropScale
                }
            }

            if (PropPool := AnimState.PropPool?, PoolEntry := PropPool.Reserve[SpawnTransform]) {
                PoolEntry.Prop.Show()
                set m_PropPoolEntry = Some(PoolEntry)
                PoolEntry.Prop
            } else if (PropAsset := AnimState.PropAsset?) {
                SpawnRes := SpawnProp(PropAsset, SpawnTransform)
                if (l_Prop := SpawnRes(0)?) {
                    set m_SpawnedProp = Some(l_Prop)
                    l_Prop
                } else {
                    Log("Could not spawn prop of pet {AsString()} for state {m_State}")
                    return None()
                }
            } else {
                Log("Could not spawn prop of pet {AsString()} for state {m_State} because the pool is empty (or not set) and no prop asset was set either")
                return None()
            }
        }
        Some(PetProp)

    PetLoop<internal>()<suspends>:void=
        SwitchState(pet_state.Idle) # spawn first instance of prop

        if (PetConfig.AlwaysFacePlayer?) {
            branch:
                PlayerFaceLoop()
        }
 
        TeleportAtDistanceSquared := Controller.p_TeleportAtDistance * Controller.p_TeleportAtDistance
        ClosenessThresholdSquared := PetConfig.ClosenessThreshold * PetConfig.ClosenessThreshold

        var FacePlayerIfClose: logic = false
        var StrictAwayCheck: logic = false
        loop {
            if (m_Removed?) {
                break
            }
            PropTransform := if (PetProp := GetProp[], PetProp.IsValid[]) {
                PetProp.GetTransform()
            } else {
                Log("Pet {AsString()} is being discarded because its prop is not valid", ?Level := log_level.Warning)
                Discard()
                break
            }
            PlayerPosition := if (
                LastGroundZ := m_LastGroundZ?
                not Owner.IsOnGround[]
                not Owner.IsFlying[]
                not Owner.IsGliding[]
                not Owner.IsInWater[]
                GetSimulationElapsedTime() - LastGroundZ(1) <= Controller.p_GroundInferenceTimeout
            ) {
                Position := Owner.GetTransform().Translation
                vector3{
                    X := Position.X
                    Y := Position.Y
                    Z := LastGroundZ(0)
                }
            } else {
                Position := Owner.GetTransform().Translation
                set m_LastGroundZ = Some(Position.Z, GetSimulationElapsedTime())
                Position
            }
            TargetPosition := PlayerPosition + ApplyPlayerOffset(Owner.GetViewRotation(), PetConfig.PlayerOffset)
            MoveDistanceSquared := DistanceSquared(TargetPosition, PropTransform.Translation) 

            if ((StrictAwayCheck? and MoveDistanceSquared > 0.0) or MoveDistanceSquared >= ClosenessThresholdSquared) {
                # away from player
                if (MoveDistanceSquared >= TeleportAtDistanceSquared) {
                    # tp to player
                    NewTransform := transform{
                        Translation := TargetPosition
                        Rotation := GetPlayerFacingRotation(PlayerPosition, TargetPosition)
                        Scale := PropTransform.Scale
                    }
                    if. GetProp[].TeleportTo[NewTransform]
                    SwitchState(pet_state.Idle)
                    set FacePlayerIfClose = false
                    set StrictAwayCheck = false
                } else {
                    # move towards player
                    SwitchState(pet_state.Running)
                    PreTeleportTransform := transform{
                        Translation := PropTransform.Translation
                        Rotation := GetPlayerFacingRotation(PlayerPosition, PropTransform.Translation)
                        Scale := PropTransform.Scale
                    }
                    if. GetProp[].TeleportTo[PreTeleportTransform]
                    MoveDistance := Sqrt(MoveDistanceSquared)
                    Direction := (TargetPosition - PropTransform.Translation) / MoveDistance
                    NewPosition := if (MoveDistance > Controller.p_MaxDistancePerTick) {
                        # multi part move, make sure closeness check doesn't get the pet stuck in a weird position
                        set StrictAwayCheck = true
                        PositionOnLine := PropTransform.Translation + (Direction * Controller.p_MaxDistancePerTick)
                        vector3{
                            X := PositionOnLine.X
                            Y := PositionOnLine.Y
                            # to move Z axis faster
                            Z := TargetPosition.Z
                        }
                    } else {
                        set StrictAwayCheck = false
                        TargetPosition
                    }
                    NewTransform := transform{
                        Translation := NewPosition
                        Rotation := PreTeleportTransform.Rotation
                        Scale := PropTransform.Scale
                    }
                    ActualMoveDistance := Distance(NewTransform.Translation, PropTransform.Translation)
                    Time := ActualMoveDistance / PetConfig.Speed
                    if (LogEnabled[log_level.Verbose]) {
                        Log("Moving pet {AsString()} {ActualMoveDistance}cm over {Time}s. StrictAwayCheck: {LogicToString(StrictAwayCheck)}", ?Level := log_level.Verbose)
                    }
                    if (PetProp := GetProp[]) {
                        race:
                            PetProp.MoveTo(NewTransform, Time)
                            RemovedEvent.Await()
                    }
                    set FacePlayerIfClose = true
                }
            } else {
                # close to player
                SwitchState(pet_state.Idle)
                if (FacePlayerIfClose?) {
                    NewTransform := transform{
                        Translation := PropTransform.Translation
                        Rotation := GetPlayerFacingRotation(PlayerPosition, PropTransform.Translation)
                        Scale := PropTransform.Scale
                    }
                    if (GetProp[].TeleportTo[NewTransform]) {
                        set FacePlayerIfClose = false
                    }
                }
            }
            Sleep(0.0)
        }

        if (Prop := m_SpawnedProp?) {
            Prop.Dispose()
        } else if (PoolEntry := m_PropPoolEntry?) {
            PoolEntry.Release()
        }

    PlayerFaceLoop<private>()<suspends>:void=
        loop {
            Sleep(0.0)
            if:
                PetProp := GetProp[]
                PlayerPosition := Owner.GetTransform().Translation
                PropPosition := PetProp.GetTransform().Translation
                NewRotation := GetPlayerFacingRotation(PlayerPosition, PropPosition)
                PetProp.TeleportTo[PropPosition, NewRotation]
        }

    StartAmbientNoiseLoop<internal>(NoiseConfig:pet_ambient_noise_config)<suspends>:void=
        loop {
            Delay := GetRandomFloat(NoiseConfig.MinInterval, NoiseConfig.MaxInterval)
            Log("Will play ambient noise for {AsString()} in {Delay}s", ?Level := log_level.Verbose)
            race:
                Sleep(Delay)
                RemovedEvent.Await()
            if (m_Removed?) {
                break
            }
            if (NoiseConfig.AudioPlayerTeleport?, PetProp := GetProp[]) {
                if (NoiseConfig.AudioPlayer.TeleportTo[PetProp.GetTransform()]) {}
            }
            Log("Playing ambient noise for {AsString()}", ?Level := log_level.Verbose)
            if (NoiseConfig.PlayToInstigatorOnly?) {
                NoiseConfig.AudioPlayer.Play(OwnerAgent)
            } else {
                NoiseConfig.AudioPlayer.Play()
            }
        }

    GetPlayerFacingRotation<private>(PlayerPosition:vector3, PropPosition:vector3)<transacts>:rotation=
        DiffVector := PlayerPosition - PropPosition
        TargetRotRadians := ArcTan(DiffVector.X, DiffVector.Y)
        MakeRotation(vector3{Z := -1.0}, TargetRotRadians)

    PlayerOffsetRotateAxis<private>: vector3 = vector3{X := 1.0, Y := 1.0, Z := 0.0}

    ApplyPlayerOffset<private>(ViewRotation:rotation, PlayerOffset:vector3):vector3=
        RotateRes := ViewRotation.RotateVector(PlayerOffsetRotateAxis)
        vector3{
            X := RotateRes.X * PlayerOffset.X
            Y := RotateRes.Y * PlayerOffset.Y
            Z := PlayerOffset.Z
        }

    AsString<public>()<transacts>:string=
        PetConfig.PetId

ToString(State:pet_state)<transacts>:string=
    case(State):
        pet_state.Idle => "Idle"
        pet_state.Running => "Running"

CmpPetInstanceByActivateTime<internal>(A:pet_instance, B:pet_instance)<computes>:int=
    CmpFloat(A.ActivatedAt, B.ActivatedAt)

pet_controller_user_action := class:
    Controller: pet_controller_device
    Config: pet_user_action_config
    PetId: string

    OnTriggered(Agent:?agent):void=
        if (Player := player[Agent?]) {
            case(Config.Action):
                pet_user_action_kind.Unlock => Controller.UnlockPet(Player, PetId)
                pet_user_action_kind.Activate => Controller.ActivatePet(Player, PetId)
                pet_user_action_kind.Deactivate => Controller.DeactivatePet(Player, PetId)
                pet_user_action_kind.Toggle =>
                    if (Controller.IsPetActive[Player, PetId]) {
                        Controller.DeactivatePet(Player, PetId)
                    } else {
                        Controller.ActivatePet(Player, PetId)
                    }
        }
