using { /Fortnite.com/Devices }
using { /Fortnite.com/Characters }
using { /Fortnite.com/Game }
using { /Verse.org/Simulation }
using { /UnrealEngine.com/Temporary/Diagnostics }

using { HavocVerse.Debug }
using { HavocVerse.Logging }
using { HavocVerse.Util }
using { HavocVerse.Util.Task }

player_health_player_state := class:
    InitialMaxHealth: float
    InitialMaxShield: float

    var MaxHealth: ?float = false
    var MaxShield: ?float = false

player_health_respawn_snapshot := class:
    MaxHealth: float
    MaxShield: float

player_health_restore_maximums := enum:
    No
    EliminationSnapshot
    LastSet

player_health_device := class(creative_device, aware_device, aware_debug_device, custom_device, resetable):
    @editable p_ExposedPlayerSpawn: []exposed_player_spawn_device = array{}
    @editable p_RestoreMaximumsOnRespawn: player_health_restore_maximums = player_health_restore_maximums.LastSet

    @editable p_CancelHealthScaleWithMaxChange: logic = true
    @editable p_CancelShieldScaleWithMaxChange: logic = true

    @editable p_UserActions: []player_health_user_action_config = array{}

    var m_Players: [player]player_health_player_state = map{}
    var m_RespawnSnapshots: [player]player_health_respawn_snapshot = map{}

    DeviceMeta<override>()<transacts>:aware_device_meta=
        aware_device_meta{
            DeviceName := "player_health_device"
        }

    DebugStats<override>()<transacts>:aware_debug_stats=
        aware_debug_stats{
            Length := map{
                "players" => m_Players.Length
                "respawnSnapshots" => m_RespawnSnapshots.Length
            }
        }

    OnBegin<override>()<suspends>:void=
        AwareInitDevice(Self, OnInit)

    OnInit<private>()<suspends>:void=
        if (p_RestoreMaximumsOnRespawn <> player_health_restore_maximums.No) {
            GetPlayspace().PlayerRemovedEvent().Subscribe(OnPlayerRemoved)
            if (p_RestoreMaximumsOnRespawn = player_health_restore_maximums.EliminationSnapshot) {
                GetPlayspace().PlayerAddedEvent().Subscribe(AddEliminationHook)
                for (Player:GetPlayspace().GetPlayers()) {
                    AddEliminationHook(Player)
                }
            }
            if (ExposedPlayerSpawn := p_ExposedPlayerSpawn[0]) {
                if (ExposedPlayerSpawn.IsMisconfigured[]) {
                    Log("p_RestoreMaximumsOnRespawn is enabled on a player health device but p_ExposedPlayerSpawn seems misconfigured", ?Level := log_level.Warning)
                }
                SubscribeToEvent(ExposedPlayerSpawn.PlayerSpawnedEvent, OnPlayerSpawned)
            } else {
                Log("Player health device defaults to using alive check loop to reapply maximums on respawn")
                ExposedPlayerSpawn := exposed_player_spawn_device{
                    p_DetectionMethod := exposed_player_spawn_detection_method.AliveCheckLoop
                }
                ExposedPlayerSpawn.OnInit()
                SubscribeToEvent(ExposedPlayerSpawn.PlayerSpawnedEvent, OnPlayerSpawned)
            }
        }
        for (UserActionConfig:p_UserActions) {
            UserAction := player_health_user_action{
                Device := Self
                Config := UserActionConfig
            }
            UserActionConfig.OnTrigger.TriggeredEvent.Subscribe(UserAction.OnTriggered)
        }

    Reset<override>(Agent:?agent):void=
        if (l_Agent := Agent?) {
            if (Player := player[l_Agent]) {
                RollbackToInitialMaximums(Player)
                OnPlayerRemoved(Player)
            }
        } else {
            for (Player->__:m_Players) {
                RollbackToInitialMaximums(Player)
            }
            set m_Players = map{}
            set m_RespawnSnapshots = map{}
        }

    RollbackToInitialMaximums<private>(Player:player):void=
        if (State := m_Players[Player], FortCharacter := Player.GetFortCharacter[]) {
            FortCharacter.SetMaxHealth(State.InitialMaxHealth)
            FortCharacter.SetMaxShield(State.InitialMaxShield)
        }

    OnPlayerRemoved<private>(Player:player):void=
        set m_Players = m_Players.RemoveByKey(Player)(0)
        if (p_RestoreMaximumsOnRespawn = player_health_restore_maximums.EliminationSnapshot) {
            set m_RespawnSnapshots = m_RespawnSnapshots.RemoveByKey(Player)(0)
        }

    AddEliminationHook<private>(Player:player):void=
        if (FortCharacter := Player.GetFortCharacter[]) {
            Closure := player_health_eliminated_closure{
                Device := Self
                FortPlayer := FortCharacter
            }
            FortCharacter.EliminatedEvent().Subscribe(Closure.OnEliminated)
        }

    OnPlayerEliminated<internal>(FortPlayer:fort_character):void=
        if (
            Player := player[FortPlayer.GetAgent[]]
            Snapshot := player_health_respawn_snapshot{
                MaxHealth := FortPlayer.GetMaxHealth()
                MaxShield := FortPlayer.GetMaxShield()
            }
            set m_RespawnSnapshots[Player] = Snapshot
        ) {
            Log("Taking respawn snapshot on elimination: {Snapshot}", ?Level := log_level.Debug)
        }

    OnPlayerSpawned<private>(Event:player_spawn_event):void=
        case(p_RestoreMaximumsOnRespawn):
            player_health_restore_maximums.EliminationSnapshot => 
                RemoveRes := m_RespawnSnapshots.RemoveByKey(Event.Player)
                set m_RespawnSnapshots = RemoveRes(0)
                if (Snapshot := RemoveRes(1)?, FortCharacter := Event.Player.GetFortCharacter[]) {
                    Log("Player respawned, applying respawn snapshot: {Snapshot}", ?Level := log_level.Debug)
                    FortCharacter.SetMaxHealth(Snapshot.MaxHealth)
                    FortCharacter.SetMaxShield(Snapshot.MaxShield)
                }
            player_health_restore_maximums.LastSet => 
                if (State := m_Players[Event.Player], FortCharacter := Event.Player.GetFortCharacter[]) {
                    Log("Player respawned, applying last set state: {HealthLastSetToString(State)}", ?Level := log_level.Debug)
                    if (MaxHealth := State.MaxHealth?) {
                        FortCharacter.SetMaxHealth(MaxHealth)
                    }
                    if (MaxShield := State.MaxShield?) {
                        FortCharacter.SetMaxShield(MaxShield)
                    }
                }
            _ => {}

    GetOrInitState<private>(FortCharacter:fort_character):?player_health_player_state=
        if (Player := player[FortCharacter.GetAgent[]]) {
            if (State := m_Players[Player]) {
                return Some(State)
            } else if (
                State := player_health_player_state{
                    InitialMaxHealth := FortCharacter.GetMaxHealth()
                    InitialMaxShield := FortCharacter.GetMaxShield()
                }
                set m_Players[Player] = State
            ) {
                return Some(State)
            }
        }
        None()

    SetMaxHealth<public>(FortCharacter:fort_character, NewMaxHealth:float, OnlyIfHigher:logic):void=
        if (OnlyIfHigher? and FortCharacter.GetMaxHealth() >= NewMaxHealth) {
            return
        }
        State := GetOrInitState(FortCharacter)
        PrevHealth := FortCharacter.GetHealth()
        FortCharacter.SetMaxHealth(NewMaxHealth)
        Log("Setting max health to {NewMaxHealth}, prev health: {PrevHealth}", ?Level := log_level.Debug)
        if (p_CancelHealthScaleWithMaxChange?) {
            FortCharacter.SetHealth(PrevHealth)
        }
        if (p_RestoreMaximumsOnRespawn = player_health_restore_maximums.LastSet, l_State := State?) {
            set l_State.MaxHealth = Some(NewMaxHealth)
        }

    SetMaxShield<public>(FortCharacter:fort_character, NewMaxShield:float, OnlyIfHigher:logic):void=
        if (OnlyIfHigher? and FortCharacter.GetMaxShield() >= NewMaxShield) {
            return
        }
        State := GetOrInitState(FortCharacter)
        PrevShield := FortCharacter.GetShield()
        FortCharacter.SetMaxShield(NewMaxShield)
        Log("Setting max shield to {NewMaxShield}, prev shield: {PrevShield}", ?Level := log_level.Debug)
        if (p_CancelShieldScaleWithMaxChange?) {
            FortCharacter.SetShield(PrevShield)
        }
        if (p_RestoreMaximumsOnRespawn = player_health_restore_maximums.LastSet, l_State := State?) {
            set l_State.MaxShield = Some(NewMaxShield)
        }

    CustomTrigger<override>(Input:string, Agent:?agent):logic=
        SplitRes := Input.Split(':')
        if (FortCharacter := Agent?.GetFortCharacter[]) {
            if (SplitRes[0] = "set_max_health", NewMaxHealth := ParseFloat[SplitRes[1]]) {
                SetMaxHealth(FortCharacter, NewMaxHealth, false)
                return true
            } else if (SplitRes[0] = "set_max_health_if_higher", NewMaxShield := ParseFloat[SplitRes[1]]) {
                SetMaxShield(FortCharacter, NewMaxShield, true)
                return true
            } else if (SplitRes[0] = "set_max_shield", NewMaxShield := ParseFloat[SplitRes[1]]) {
                SetMaxShield(FortCharacter, NewMaxShield, false)
                return true
            } else if (SplitRes[0] = "set_max_shield_if_higher", NewMaxShield := ParseFloat[SplitRes[1]]) {
                SetMaxShield(FortCharacter, NewMaxShield, true)
                return true
            } else if (SplitRes[0] = "set_health", NewHealth := ParseFloat[SplitRes[1]]) {
                FortCharacter.SetHealth(NewHealth)
                return true
            } else if (SplitRes[0] = "set_shield", NewShield := ParseFloat[SplitRes[1]]) {
                FortCharacter.SetShield(NewShield)
                return true
            }
        }
        false

player_health_eliminated_closure := class:
    Device: player_health_device
    FortPlayer: fort_character

    OnEliminated(EliminationResult:elimination_result):void=
        Device.OnPlayerEliminated(FortPlayer)

player_health_user_action_kind := enum:
    SetMaxHealth
    SetMaxShield
    SetHealth
    SetShield

player_health_user_action_config := class<concrete>:
    @editable Action: player_health_user_action_kind = player_health_user_action_kind.SetMaxHealth
    @editable OnTrigger: trigger_device = trigger_device{}
    @editable Value: float = 100.0
    @editable OnlyIfHigher: logic = false

player_health_user_action := class:
    Device: player_health_device
    Config: player_health_user_action_config

    OnTriggered(Agent:?agent):void=
        if (Player := player[Agent?], FortCharacter := Player.GetFortCharacter[]) {
            case(Config.Action):
                player_health_user_action_kind.SetMaxHealth => Device.SetMaxHealth(FortCharacter, Config.Value, Config.OnlyIfHigher)
                player_health_user_action_kind.SetMaxShield => Device.SetMaxShield(FortCharacter, Config.Value, Config.OnlyIfHigher)
                player_health_user_action_kind.SetHealth => FortCharacter.SetHealth(Config.Value)
                player_health_user_action_kind.SetShield => FortCharacter.SetShield(Config.Value)
        }

ToString(RespawnSnapshot:player_health_respawn_snapshot)<transacts>:string=
    "\{MaxHealth: {RespawnSnapshot.MaxHealth}, MaxShield: {RespawnSnapshot.MaxShield}\}"

HealthLastSetToString(State:player_health_player_state)<transacts>:string=
    StrMaxHealth := if (Val := State.MaxHealth?) {
        ToString(Val)
    } else {
        "Unmanaged"
    }
    StrMaxShield := if (Val := State.MaxShield?) {
        ToString(Val)
    } else {
        "Unmanaged"
    }
    "\{MaxHealth: {StrMaxHealth}, MaxShield: {StrMaxShield}\}"
