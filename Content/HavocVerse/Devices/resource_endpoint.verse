using { /Fortnite.com/Devices }
using { /Verse.org/Simulation }
using { /UnrealEngine.com/Temporary/Diagnostics }

using { HavocVerse.Debug }
using { HavocVerse.Resource }
using { HavocVerse.Logging }
using { HavocVerse.Util }

# Device that exposes a specific resource from a resource manager.
resource_endpoint_device<public> := class(creative_device, resource, aware_device, custom_device):
    @editable p_ResourceManager<public>: resource_manager_device = resource_manager_device{}
    @editable p_ResourceId<public>: string = ""

    @editable p_GiveUserActions: []resource_endpoint_user_action_give_config = array{}

    DeviceMeta<override>()<transacts>:aware_device_meta=
        aware_device_meta{
            DeviceName := "resource_endpoint"
        }

    OnBegin<override>()<suspends>:void=
        AwareInitDevice(Self, OnInit)

    OnInit<private>()<suspends>:void=
        p_ResourceManager.Setup()
        if (not p_ResourceManager.GetResource[p_ResourceId]) {
            Log("A resource endpoint references an unsupported resource by its resource manager: {p_ResourceId}", ?Level := log_level.Warning)
        }

        for (UserActionConfig:p_GiveUserActions) {
            if (UserActionConfig.GiveAmount >= 0.0) {
                UserAction := resource_endpoint_user_action_give{
                    EndpointDevice := Self
                    Config := UserActionConfig
                }
                UserActionConfig.OnTrigger.TriggeredEvent.Subscribe(UserAction.OnTriggered)
            } else if (UserActionConfig.GiveAmount < 0.0) {
                Log("A resource endpoint has a user action that tries to give a negative amount (of resource {p_ResourceId}, amount {UserActionConfig.GiveAmount})", ?Level := log_level.Error)
            }
        }

    GetAmount<override>(Agent:?agent):int=
        p_ResourceManager.GetAmount(Agent, p_ResourceId)

    Give<override>(Agent:?agent, Amount:int, ?SendFeedback:logic = false):logic=
        p_ResourceManager.Give(Agent, p_ResourceId, Amount, ?SendFeedback := SendFeedback)

    Take<override>(Agent:?agent, Amount:int, ?NotifyNotEnough:logic = false, ?SendFeedback:logic = false):logic=
        p_ResourceManager.Take(Agent, p_ResourceId, Amount, ?NotifyNotEnough := NotifyNotEnough, ?SendFeedback := SendFeedback)

    CustomAmountFormatter<override>()<transacts>:?number_formatter=
        if (Res := p_ResourceManager.GetResource[p_ResourceId]). Res.CustomAmountFormatter() else false

    SendNotEnoughNotification<public>(Agent:agent):logic=
        p_ResourceManager.SendNotEnoughNotification(Agent, p_ResourceId)

    CustomTrigger<override>(Input:string, Agent:?agent):logic=
        if (SplitRes := Input.SplitOnce[':']) {
            Action := SplitRes(0)
            Data := SplitRes(1)
            if (Action = "give", Amount := ParseInt[Data]) {
                Give(Agent, Amount)
                return true
            }
        }
        false

resource_endpoint_user_action_give_multiplier := enum:
    NoMultiplier
    EffectiveMultiplier
    AutoGenerationRate

resource_endpoint_user_action_give_round_method := enum:
    R_Round
    R_Floor
    R_Ceil

TooltipMultiplierSourceOverride<internal><localizes>:message = "Resource to use as the source of the multiplier"

resource_endpoint_user_action_give_config := class<concrete>:
    @editable OnTrigger: trigger_device = trigger_device{}
    @editable ShowFeedback: logic = true
    @editable GiveAmount: float = 1.0
    @editable MultiplierSourceOverride: ?string = false
    @editable Multiplier: resource_endpoint_user_action_give_multiplier = resource_endpoint_user_action_give_multiplier.NoMultiplier
    @editable RoundingMethod : resource_endpoint_user_action_give_round_method = resource_endpoint_user_action_give_round_method.R_Ceil
    @editable ConstrainAmountMin: int = 0
    @editable ConstrainAmountMax: int = 2147483647

resource_endpoint_user_action_give := class:
    EndpointDevice: resource_endpoint_device
    Config: resource_endpoint_user_action_give_config

    OnTriggered(Agent:?agent):void=
        var Amount: int = if (l_Agent := Agent?) {
            MultiplierSource := if (ResourceId := Config.MultiplierSourceOverride?) {
                ResourceId
            } else {
                EndpointDevice.p_ResourceId
            }
            FloatAmount := case(Config.Multiplier):
                resource_endpoint_user_action_give_multiplier.NoMultiplier => Config.GiveAmount
                resource_endpoint_user_action_give_multiplier.AutoGenerationRate =>
                    Rate := EndpointDevice.p_ResourceManager.GetAutoGenerationRate(l_Agent, MultiplierSource)
                    Log(MsgUsingGenerationRateMultiplier(l_Agent, MultiplierSource, EndpointDevice.p_ResourceId, Rate), ?Level := log_level.Verbose)
                    Config.GiveAmount * Rate
                resource_endpoint_user_action_give_multiplier.EffectiveMultiplier =>
                    EffectiveMultiplier := EndpointDevice.p_ResourceManager.GetEffectiveMultiplier(l_Agent, MultiplierSource)
                    Log(MsgUsingEffectiveMultiplier(l_Agent, MultiplierSource, EndpointDevice.p_ResourceId, EffectiveMultiplier), ?Level := log_level.Verbose)
                    Config.GiveAmount * EffectiveMultiplier
            case(Config.RoundingMethod):
                resource_endpoint_user_action_give_round_method.R_Round => DefaultingRound(FloatAmount)
                resource_endpoint_user_action_give_round_method.R_Ceil => DefaultingCeil(FloatAmount)
                resource_endpoint_user_action_give_round_method.R_Floor => DefaultingFloor(FloatAmount)
        } else {
            DefaultingFloor(Config.GiveAmount)
        }
        set Amount = Clamp(Amount, Config.ConstrainAmountMin, Config.ConstrainAmountMax)
        EndpointDevice.p_ResourceManager.Give(Agent, EndpointDevice.p_ResourceId, Amount, ?SendFeedback := Config.ShowFeedback)

    MsgUsingGenerationRateMultiplier<private><localizes>(Agent:agent, ResourceId:string, MultiplierSource:string, Rate:float):message=
        "Using generation rate {Rate} (source: {MultiplierSource}) as multiplier when giving resource {ResourceId} to {Agent}"
    MsgUsingEffectiveMultiplier<private><localizes>(Agent:agent, ResourceId:string, MultiplierSource:string, Multiplier:float):message=
        "Using effective multiplier {Multiplier} (source: {MultiplierSource}) when giving resource {ResourceId} to {Agent}"
