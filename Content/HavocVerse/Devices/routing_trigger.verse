using { /Fortnite.com/Devices }
using { /Verse.org/Simulation }
using { /UnrealEngine.com/Temporary/Diagnostics }

using { HavocVerse.Debug }
using { HavocVerse.Util }

routing_trigger_device_out_entry := class<concrete>:
    @editable AccessControl: ?access_control_device = false
    @editable Trigger: ?trigger_device = false
    @editable SwitchTeamTo: ?int = false

routing_trigger_device := class(creative_device, aware_device):
    @editable p_DevId: string = ""

    @editable p_InTrigger: trigger_device = trigger_device{}
    @editable p_OutTriggers: []routing_trigger_device_out_entry = array{}
    @editable p_SingleMatchOnly: logic = false
    @editable p_FallbackOutTrigger: ?trigger_device = false
    @editable p_FallbackSwitchTeamTo: ?int = false

    DeviceMeta<override>()<transacts>:aware_device_meta=
        aware_device_meta{
            DeviceName := "routing_trigger"
            DevId := p_DevId
        }

    OnBegin<override>()<suspends>:void=
        p_InTrigger.TriggeredEvent.Subscribe(OnInTriggered)

    OnInTriggered<private>(Agent:?agent):void=
        if (Player := player[Agent?]) {
            TeamCollection := GetPlayspace().GetTeamCollection()
            var Matched: logic = false
            for (
                Entry:p_OutTriggers
                (not (p_SingleMatchOnly?)) or (not (Matched?))
                Entry.AccessControl?.IsAllowed[Player]
            ) {
                set Matched = true
                if (Trigger := Entry.Trigger?) {
                    Trigger.Trigger(Player)
                }
                if:
                    SwitchTeamTo := Entry.SwitchTeamTo?
                    TeamCollection.SetTeam[Player, SwitchTeamTo]
            }
            if:
                not (Matched?)
                SwitchTeamTo := p_FallbackSwitchTeamTo?
                TeamCollection.SetTeam[Player, SwitchTeamTo]
        } else if (FallbackTrigger := p_FallbackOutTrigger?) {
            FallbackTrigger.Trigger()
        }
