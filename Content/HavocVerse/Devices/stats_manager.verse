using { /Fortnite.com/Devices }
using { /Verse.org/Simulation }
using { /UnrealEngine.com/Temporary/Diagnostics }
using { /UnrealEngine.com/Temporary/UI }
using { /UnrealEngine.com/Temporary/SpatialMath }
using { /Verse.org/Colors }

using { HavocVerse.Debug }
using { HavocVerse.Devices }
using { HavocVerse.Devices.GameSpecific }
using { HavocVerse.Persistence }
using { HavocVerse.UI }
using { HavocVerse.Util }
using { HavocVerse.Math }
using { HavocVerse.Logging }

event_score_type<public> := enum:
	HighScore
	TimeRaceShortest
	TimeRaceLongest

stats_manager_event_config := class<final><concrete> {
	@editable EventName: string = ""
	@editable DisplayName: ?string = false
	@editable ScoreType: event_score_type = event_score_type.HighScore

	FormatScore<internal>(Score:float)<transacts>: string ={
		case(ScoreType):
			event_score_type.HighScore => FormatNumber(Score)
			_ => ToTimerMMSS(Score, ?Millis := true) + "s"
	}
}

stats_manager_test_granter := class<final><concrete> {
	@editable Button: button_device = button_device{}
	@editable EventName: string = ""
	@editable Score: float_range = float_range{}
}

stats_manager_test_granter_closure := class<final> {
	Device: stats_manager_device
	Granter: stats_manager_test_granter

	OnButtonInteractedWith<internal>(Agent:agent): void = if (Player := player[Agent]) {
		Device.RegisterEventScore(Granter.EventName, Player, Granter.Score.GetRandomValue())
	}
}

stats_manager_device<public> := class<final><concrete>(creative_device, aware_device) {
	@editable p_GameTitle: string = ""
	@editable p_EventKinds: []stats_manager_event_config = array{}

	@editable p_ResourceManager: resource_manager_device = resource_manager_device{}
	@editable p_ScoreResourceId: string = "score"
	@editable p_LevelGroupsManager: ?level_groups_manager_device = false
	@editable p_LevelGroupId: string = ""
	@editable p_TycoonManager: ?tycoon_manager_device = false

	@editable p_OverviewOpenButtons: []button_device = array{}
	@editable p_TestGranters: []stats_manager_test_granter = array{}

	NotificationGroup: ui_notification_group = ui_notification_group{}

	DeviceMeta<override>()<transacts>: aware_device_meta ={
		aware_device_meta{
			DeviceName := "stats_manager"
		}
	}

	### DOMAIN
	OnBegin<override>()<suspends>: void = AwareInitDevice(Self, OnInit)
	
	OnInit<private>()<suspends>: void ={
		GetPlayspace().PlayerAddedEvent().Subscribe(OnPlayerAdded)

		for (Button:p_OverviewOpenButtons) {
			Button.InteractedWithEvent.Subscribe(OnOpenButtonInteractedWith)
		}

		for (TestGranter:p_TestGranters) {
			Closure := stats_manager_test_granter_closure{
				Device := Self
				Granter := TestGranter
			}
			TestGranter.Button.InteractedWithEvent.Subscribe(Closure.OnButtonInteractedWith)
		}

		spawn. StartSaveLoop()
	}

	OnPlayerAdded<private>(Player:player): void ={
		Document := GetPlayerDocument(Player)
		Now := UtcDateTimeNow()
		if (
			LastJoinedEpoch := Document.GetFloat["last_join"]
			LastJoinedAt := DateTimeFromEpoch(LastJoinedEpoch)
			LastJoinedAt.Day <> Now.Day
		) {
			Document.SetInt("dj", Document.GetInt("dj", 0) + 1)
		} else if (not Document.GetInt["dj"]) {
			Document.SetInt("dj", 1)
		}
		Document.SetFloat("last_join", Now.ToEpoch())
	}

	StartSaveLoop<private>()<suspends>: void ={
        var LastTick: float = GetSimulationElapsedTime()
		loop {
			Sleep(1.0)

            Now := GetSimulationElapsedTime()
            Elapsed := Now - LastTick
            set LastTick = Now

			for (Player:GetPlayspace().GetPlayers()) {
				IncreasePlayerPlaytime(Player, Elapsed)
			}
		}
	}

	GetPlayerDocument<internal>(Player:player)<transacts>: player_data_document ={
		PlayerDataGetDocument(Player, "stats")
	}

	IncreasePlayerPlaytime<private>(Player:player, Delta:float): void ={
		Document := GetPlayerDocument(Player)
		Playtime := Document.GetFloat["pt"] or (
			# if we actually plan to use it, TODO import playtime from gameplay rewards device
			0.0
		);
		Document.SetFloat("pt", Playtime + Delta)
	}

	OnOpenButtonInteractedWith<private>(Agent:agent): void = if (Player := player[Agent]) {
        OpenSimpleMenu("stats_overview", Player, stats_manager_overview_menu_component{
            Device := Self
			Player := Player
        })
	}

	GetEventConfig<private>(EventName:string)<decides><transacts>: stats_manager_event_config ={
		var Found: ?stats_manager_event_config = None()
		for (Config:p_EventKinds, not Found?, Config.EventName = EventName) {
			set Found = Some(Config)
		}
		Found?
	}

	### API
	# Returns the total play time of a player in seconds.
	GetTotalPlaytime<public>(Player:player)<transacts>: float ={
		GetPlayerDocument(Player).GetFloat("pt", 0.0)
	}

	# Janky thing to manually fix scores set by glitches -Ice
	DebugOverrideEventScore<public>(EventName:string, Player:player, Score:float): void ={
		Config := option{GetEventConfig[EventName]}
		if (not Config?) {
			Log("Unregistered event kind '{EventName}'.", ?Level := log_level.Warning)
		}

		Document := GetPlayerDocument(Player)
		Key := "evt_{EventName}"
		Document.SetFloat(Key, Score)
	}

	# Registers a player's score for an event kind.
	# Won't overwrite the current player's high score if the given `Score` is lower.
	RegisterEventScore<public>(EventName:string, Player:player, Score:float, ?SendFeedback:logic=true): void ={
		Config := option{GetEventConfig[EventName]}
		if (not Config?) {
			Log("Registering event score of unregistered event kind '{EventName}'. Don't know how to display its score", ?Level := log_level.Warning)
		}

		Document := GetPlayerDocument(Player)
		Key := "evt_{EventName}"
		NewHighScore := if (HighScore := Document.GetFloat[Key]) {
			option{IsNewHighScore[Config, HighScore, Score]}.ToL()
		} else {
			true
		}
		if (NewHighScore?) {
			Document.SetFloat(Key, Score)
		}
		
		if (SendFeedback?, EventConfig := Config?) {
			Notification := event_score_notification_component{
				EventConfig := EventConfig
				Score := Score
				NewHighScore := NewHighScore
			}
			NotificationGroup.Send(
				Player,
				Notification,
				ui_notification_position{
					Anchor := vector2{X := 1.0, Y := 0.5},
					Alignment := vector2{X := 1.0, Y := 0.5}
				}
			)
		}
	}

	IsNewHighScore<private>(Config:?stats_manager_event_config, HighScore:float, Score:float)<decides><transacts>: void ={
		if (EventConfig := Config?) {
			case(EventConfig.ScoreType):
				event_score_type.HighScore => Score > HighScore
				event_score_type.TimeRaceShortest => Score < HighScore
				event_score_type.TimeRaceLongest => Score > HighScore
		} else {
			Score > HighScore
		}
	}
}

stats_manager_overview_menu_component := class<final>(ui_menu_component) {
	Device: stats_manager_device
	Player: player

	Create<override>(): widget ={
		Background := overlay:
			Slots := array{
				overlay_slot:
					Widget := color_block:
						DefaultColor := NamedColors.Yellow
						DefaultDesiredSize := vector2{X := 10.0}
					VerticalAlignment := vertical_alignment.Fill
					HorizontalAlignment := horizontal_alignment.Left
				overlay_slot:
					Widget := color_block:
						DefaultColor := NamedColors.Black
						DefaultOpacity := 0.8
					VerticalAlignment := vertical_alignment.Fill
					HorizontalAlignment := horizontal_alignment.Fill
					Padding := margin{
						Left := 10.0
					}
			}

		var MainStack: []stack_box_slot = array{}

		Score := Device.p_ResourceManager.GetAmount(Player, Device.p_ScoreResourceId)

		Document := Device.GetPlayerDocument(Player)
		Playtime := Document.GetFloat("pt", 0.0)
		DaysJoined := Document.GetInt("dj", 1)
		PlaytimeStr := ToHumanDuration(Playtime, human_duration_parts{
			Hours := option{Playtime >= TimeUnits.Hour}.ToL()
			Minutes := true
			Seconds := false
		}, human_duration_opts{
			Delimiter := " "
			HoursFmt := ("hr", "hrs")
			MinutesFmt := ("min", "mins")
			SecondsFmt := ("s", "s")
		})

		Now := UtcDateTimeNow()
		TimeStamp := stack_box:
			Orientation := orientation.Vertical
			Slots := array{
				StackSlot of (CreateTextBlock(FormatDateTime("%d %h %Y", Now), ?TextColor := NamedColors.Silver), ?Horiz := horizontal_alignment.Right)
				StackSlot of (CreateTextBlock(FormatDateTime("%I:%M %p", Now), ?TextColor := NamedColors.Silver), ?Horiz := horizontal_alignment.Right)
			}

		set MainStack += array{
			StackSlot of (CreateTextBlock("Your Stats - {Device.p_GameTitle}", ?TextColor := NamedColors.Yellow), ?Horiz := horizontal_alignment.Center)
			StackSlot of (CreateTextBlock(MsgUserHandle(Player), ?TextColor := NamedColors.Salmon), ?Top := 20.0)
			StackSlot of (ScoreLine("Days Joined", "{DaysJoined} days"), ?Top := 20.0)
			StackSlot of (ScoreLine("Time Played", PlaytimeStr), ?Top := 5.0)
			StackSlot of (ScoreLine("Score", FormatNumber(Score)), ?Top := 5.0)
		}
		if (LevelGroupsManager := Device.p_LevelGroupsManager?) {
			LevelExpPair := LevelGroupsManager.GetLevelExp(Player, Device.p_LevelGroupId)
			ExpectedExp := LevelGroupsManager.GetTargetExp[Device.p_LevelGroupId, LevelExpPair] or -1
			RawProgress := ToFloat(LevelExpPair.ExpRest) / ToFloat(ExpectedExp)

			Level := LevelExpPair.Level
			ProgressPercent := DefaultingFloor(Clamp(RawProgress, 0.0, 1.0) * 100.0)
			set MainStack += array{
				StackSlot of (ScoreLine("Level", "{Level} ({ProgressPercent}%)"), ?Top := 5.0)
			}
		}
		if (
			TycoonManager := Device.p_TycoonManager?
			RebirthLevel := TycoonManager.GetRebirthLevel[Player]
		) {
			set MainStack += array{
				StackSlot of (ScoreLine("Rebirths", ToString(RebirthLevel)), ?Top := 5.0)
			}
		}

		if (Device.p_EventKinds.Length > 0) {
			set MainStack += array{
				StackSlot of (CreateTextBlock("Event High Scores", ?TextColor := MakeColorFromHex("#ffff69")), ?Top := 30.0)
			}
			for (Index->EventKind:Device.p_EventKinds) {
				Name := EventKind.DisplayName? or EventKind.EventName
				EventScoreStr := if (EventScore := Document.GetFloat["evt_{EventKind.EventName}"]) {
					EventKind.FormatScore(EventScore)
				} else {
					"N/A"
				}
				Top := if (Index = 0). 20.0 else 5.0
				set MainStack += array{
					StackSlot of (ScoreLine(Name, EventScoreStr), ?Top := Top)
				}
			}
		}

		CloseButton := CreateFortButton(fort_button_type.Regular, ToMessage("Close"))
		BindToViewerClose(CloseButton.OnClick())
		set MainStack += array{
			StackSlot of (
				ApplySizeHint(CloseButton, vector2{X := 240.0, Y := 50.0}),
				?Top := 30.0, ?Horiz := horizontal_alignment.Center
			);
		}

		overlay:
			Slots := array{
				overlay_slot:
					Widget := Background
					HorizontalAlignment := horizontal_alignment.Fill
					VerticalAlignment := vertical_alignment.Fill
				overlay_slot:
					Widget := TimeStamp
					HorizontalAlignment := horizontal_alignment.Right
					VerticalAlignment := vertical_alignment.Top
					Padding := margin{
						Left := 460.0
						Top := 86.0
						Right := 40.0
					}
				overlay_slot:
					Widget := stack_box:
						Orientation := orientation.Vertical
						Slots := MainStack
					HorizontalAlignment := horizontal_alignment.Fill
					VerticalAlignment := vertical_alignment.Fill
					Padding := margin{
						Top := 30.0
						Right := 40.0
						Bottom := 30.0
						Left := 50.0
					}
			}
	}

	ScoreLine<private>(Label:string, Value:string): widget ={
		MakeTextRow(array{
			CreateTextBlock("{Label}: ")
			CreateTextBlock(Value, ?TextColor := NamedColors.LightSkyBlue)
		})
	}

	MsgUserHandle<private><localizes>(Agent:player):message = "@{Agent}"
}	

event_score_notification_component := class<final>(ui_notification_component) {
	EventConfig: stats_manager_event_config
	Score: float
	NewHighScore: logic

	Create<override>(): widget ={
		Background := overlay:
			Slots := array{
				overlay_slot:
					Widget := color_block:
						DefaultColor := NamedColors.Yellow
						DefaultDesiredSize := vector2{X := 6.0}
					VerticalAlignment := vertical_alignment.Fill
					HorizontalAlignment := horizontal_alignment.Left
				overlay_slot:
					Widget := color_block:
						DefaultColor := NamedColors.Black
						DefaultOpacity := 0.8
					VerticalAlignment := vertical_alignment.Fill
					HorizontalAlignment := horizontal_alignment.Fill
					Padding := margin{
						Left := 6.0
					}
			}

		DisplayName := EventConfig.DisplayName? or EventConfig.EventName
		NameWidget := CreateTextBlock(DisplayName, ?TextColor := MakeColorFromHex("#ffff69"))
		Description := if (EventConfig.ScoreType = event_score_type.HighScore) {
			MakeTextRow(array{
				CreateTextBlock("Scored ")
				CreateTextBlock(EventConfig.FormatScore(Score), ?TextColor := NamedColors.LightSkyBlue)
				CreateTextBlock(" in ")
				NameWidget
			})
		} else {
			MakeTextRow(array{
				CreateTextBlock("Completed ")
				NameWidget
				CreateTextBlock(" in ")
				CreateTextBlock(EventConfig.FormatScore(Score), ?TextColor := NamedColors.LightSkyBlue)
			})
		}

		Title := if (NewHighScore?) {
			"New Personal Best!"
		} else {
			"Event completed"
		}
		var MainStack: []stack_box_slot = array{}
		set MainStack += array{
			StackSlot of CreateTextBlock(Title, ?TextColor := NamedColors.Yellow)
			StackSlot of (Description, ?Top := 5.0)
		}

		ScheduleTask(Animate)
		overlay:
			Slots := array{
				overlay_slot:
					Widget := Background
					HorizontalAlignment := horizontal_alignment.Fill
					VerticalAlignment := vertical_alignment.Fill
				overlay_slot:
					Widget := color_block:
						DefaultDesiredSize := vector2{X := 0.0, Y := 0.0}
					Padding := margin{
						Left := 280.0
					}
				overlay_slot:
					Widget := stack_box:
						Orientation := orientation.Vertical
						Slots := MainStack
					HorizontalAlignment := horizontal_alignment.Fill
					VerticalAlignment := vertical_alignment.Fill
					Padding := margin{
						Top := 10.0
						Right := 20.0
						Bottom := 10.0
						Left := 26.0
					}
			}
	}

	Animate<private>()<suspends>: void ={
		HiddenOffset := vector2{X := -560.0}
		VisibleOffset := vector2{X := 0.0}
		SetOffset := Self.OffsetSetter(?LeftAligned := false)
		sync:
			TweenValue(HiddenOffset, VisibleOffset, 0.8, SetOffset)
			TweenValue(0.7, 1.0, 1.0, Self.SetOpacity)
		Sleep(5.0)
		sync:
			TweenValue(VisibleOffset, HiddenOffset, 0.5, SetOffset)
			TweenValue(1.0, 0.4, 0.5, Self.SetOpacity)
		Hide()
	}
}
