using { /Fortnite.com/Devices }
using { /Fortnite.com/UI }
using { /Verse.org/Simulation }
using { /UnrealEngine.com/Temporary/Diagnostics }
using { /UnrealEngine.com/Temporary/UI }

using { HavocVerse.Debug }
using { HavocVerse.Logging }
using { HavocVerse.UI }
using { HavocVerse.Util }

ui_fix_device := class(creative_device, aware_device, aware_debug_device):
    @editable p_RefreshRate: float = 0.05

    var m_Players: [player]ui_fix_refresh_hud_element = map{}

    DeviceMeta<override>()<transacts>:aware_device_meta=
        aware_device_meta{
            DeviceName := "ui_fix"
        }

    DebugStats<override>()<transacts>:aware_debug_stats=
        aware_debug_stats{
            Length := map{
                "players" => m_Players.Length
            }
        }

    OnBegin<override>()<suspends>:void=
        AwareInitDevice(Self, OnInit)

    OnInit<private>()<suspends>:void=
        if (p_RefreshRate < 0.0) {
            Log("p_RefreshRate is below 0 on {Self}", ?Level := log_level.Error)
            return
        }

        Playspace := GetPlayspace()
        Playspace.PlayerAddedEvent().Subscribe(OnPlayerAdded)
        Playspace.PlayerRemovedEvent().Subscribe(OnPlayerRemoved)
        for (Player:Playspace.GetPlayers()) {
            OnPlayerAdded(Player)
        }

        if (Integration := AwareGetIntegration[]) {
            Integration.SubscribeToPlayerSpawnEvent(OnPlayerSpawned)
        }

        spawn:
            StartUpdateLoop()

    StartUpdateLoop<private>()<suspends>:void=
        loop {
            Sleep(p_RefreshRate)
            for (__->Element: m_Players) {
                Element.Update()
            }
        }

    OnPlayerAdded<private>(Player:player):void=
        Message1 := ToMessage(" ")
        Message2 := ToMessage("  ")
        TextWidget := text_block{
            DefaultText := Message1
        }
        Element := ui_fix_refresh_hud_element{
            Widget := TextWidget
            TextWidget := TextWidget
            Message1 := Message1
            Message2 := Message2
        }
        if (PlayerUI := GetPlayerUI[Player]) {
            PlayerUI.AddWidget(Element.Widget)
            set m_Players = AddToMap(m_Players, Player, Element)
        }

    OnPlayerRemoved<private>(Player:player):void=
        set m_Players = m_Players.RemoveByKey(Player)(0)

    OnPlayerSpawned<private>(Player:player):void=
        spawn:
            HandleUIResendOnSpawn(Player)

    HandleUIResendOnSpawn<private>(Player:player)<suspends>:void=
        Sleep(0.0)
        GlobalUIForceResend(Player)

ui_fix_refresh_hud_element := class(hud_element):
    TextWidget: text_block

    Message1: message
    Message2: message

    var Oscillator: logic = false

    Update<override>():void=
        if (Oscillator?) {
            set Oscillator = false
            TextWidget.SetText(Message1)
        } else {
            set Oscillator = true
            TextWidget.SetText(Message2)
        }
