using { /Verse.org/Simulation }
using { /UnrealEngine.com/Temporary/Diagnostics }

using { HavocVerse.JSON }
using { HavocVerse.Logging }
using { HavocVerse.UI }
using { HavocVerse.Util }

tycoon_manager_config := class<concrete>:
    BeaconLines: [string]tycoon_manager_beacon_line_config = map{}
    Rebirth: ?tycoon_manager_rebirth_config = false

tycoon_manager_beacon_line_config := class<final>:
    InOrderBuyPlates: []tycoon_manager_beacon_line_config_entry

    ContainsBuyPlate(BuyPlateId:string)<decides><transacts>:void=
        var Found: logic = false
        for (LineEntry:InOrderBuyPlates, not Found?, LineEntry.ContainsBuyPlate[BuyPlateId]) {
            set Found = true
        }
        Found?

tycoon_manager_beacon_line_config_entry := class<final>:
    AvailableOptions: [string]void

    ContainsBuyPlate(BuyPlateId:string)<decides><transacts>:void=
        AvailableOptions[BuyPlateId]

    ContainsAnyBuyPlateOf(Ids:[]string)<decides><transacts>:void=
        var Found: logic = false
        for (BuyPlateId:Ids, not Found?, AvailableOptions[BuyPlateId]) {
            set Found = true
        }
        Found?

tycoon_manager_buyplates_config := struct:
    Entries: []tycoon_manager_buyplates_entry = array{}

tycoon_manager_buyplates_entry := class<final>:
    BuyplateId: string
    Cost: ?tycoon_manager_buyplates_entry_resource
    Generates: ?tycoon_manager_buyplates_entry_resource
    Unlocks: []tycoon_manager_unlock_entry

    var m_StorageIndex: int = -1

tycoon_manager_unlock_entry := struct:
    TargetRoutingId: string
    Payload: ?string

tycoon_manager_buyplates_entry_resource := struct:
    ResourceId: string
    Amount: int
    CheckDontTake: logic

tycoon_manager_rebirth_config := class<final>:
    MaxLevel: int
    DefaultToLast: logic
    LevelConfigs: [int]tycoon_manager_rebirth_level_config

    GetLevelConfig(Level:int)<transacts><decides>:tycoon_manager_rebirth_level_config=
        GetLevelConfig0(Level)?

    GetLevelConfig0<private>(Level:int)<transacts>:?tycoon_manager_rebirth_level_config=
        if (DefaultToLast?) {
            var Index: int = Level
            loop {
                if (LevelConfig := LevelConfigs[Index]) {
                    return Some(LevelConfig)
                }
                if (Index <= 0) {
                    break
                }
                set Index -= 1
            }
        } else if (LevelConfig := LevelConfigs[Level]) {
            Some(LevelConfig)
        }
        None()

tycoon_manager_rebirth_level_config := class:
    HudImageTexture: ?string = false
    ResourceMultipliers: [string]float = map{}

## Parse Tycoon Config {{{
ParseTycoonConfig<internal>(InputJson:string)<transacts><decides>:tycoon_manager_config=
    Reader := CreateJsonReader(InputJson)
    Reader.BeginObject[]
    var BeaconLines: [string]tycoon_manager_beacon_line_config = map{}
    var Rebirth: ?tycoon_manager_rebirth_config = None()
    loop {
        if:
            Name := Reader.NextName[]
            case(Name):
                "beacon_lines" =>
                    ParsedBeaconLines := ParseTycoonConfigBeaconLines[Reader]
                    set BeaconLines = ParsedBeaconLines
                "rebirth" =>
                    ParsedRebirthConfig := ParseTycoonConfigRebirth[Reader]
                    set Rebirth = Some(ParsedRebirthConfig)
                _ => Reader.SkipNext[]
        else:
            if (Reader.HasNext[]) {
                FieldPos := if (Name := Reader.NextName[]) {
                    "field \"{Name}\""
                } else {
                    "first field"
                }
                Log("Tycoon manager failed reading {FieldPos} for tycoon config", ?Level := log_level.Warning)
            }
            break
    }
    Reader.EndObject[]
    tycoon_manager_config{
        BeaconLines := BeaconLines
        Rebirth := Rebirth
    }
## }}}

## Parse Rebirth Config {{{
ParseTycoonConfigRebirth<internal>(Reader:json_reader)<transacts><decides>:tycoon_manager_rebirth_config=
    var MaxLevel: int = 0
    var DefaultToLast: logic = true
    var HudImageTextures: [int]string = map{}
    var PerLevelResourceMultipliers: [int][string]float = map{}
    Reader.BeginObject[]
    var LastName: ?string = false
    loop {
        if:
            Name := Reader.NextName[]
            case(Name):
                "max_level" =>
                    ParsedMaxLevel := Reader.NextInt[]
                    set MaxLevel = ParsedMaxLevel
                "default_to_last" =>
                    ParsedDefaultToLast := Reader.NextBoolean[]
                    set DefaultToLast = ParsedDefaultToLast.Value
                "hud_images" =>
                    ParsedHudImages := ParseTycoonConfigRebirthHudImages[Reader]
                    set HudImageTextures = ParsedHudImages
                "resource_multipliers" =>
                    ParsedResourceMultipliers := ParseTycoonConfigRebirthPerLevelResourceMultipliers[Reader]
                    set PerLevelResourceMultipliers = ParsedResourceMultipliers
                _ => Reader.SkipNext[]
        else:
            if (Reader.HasNext[]) {
                FieldPos := if (Name := Reader.NextName[]) {
                    "field \"{Name}\""
                } else {
                    "first field"
                }
                Log("Tycoon manager failed reading {FieldPos} for rebirth config", ?Level := log_level.Warning)
            }
            break
    }
    Reader.EndObject[]

    var ConfiguredLevels: [int]void = map{}
    for (Level->__:HudImageTextures) {
        set ConfiguredLevels[Level] = void
    }
    for (Level->__:PerLevelResourceMultipliers) {
        set ConfiguredLevels[Level] = void
    }

    var LevelConfigs: [int]tycoon_manager_rebirth_level_config = map{}
    for (Level->__:ConfiguredLevels) {
        LevelConfig := tycoon_manager_rebirth_level_config{
            HudImageTexture := option{HudImageTextures[Level]}
            ResourceMultipliers := if (l_ResourceMultipliers := PerLevelResourceMultipliers[Level]) {
                l_ResourceMultipliers
            } else {
                map{}
            }
        }
        set LevelConfigs[Level] = LevelConfig
    }

    tycoon_manager_rebirth_config{
        MaxLevel := MaxLevel
        DefaultToLast := DefaultToLast
        LevelConfigs := LevelConfigs
    }

ParseTycoonConfigRebirthHudImages<internal>(Reader:json_reader)<transacts><decides>:[int]string=
    var HudImageTextures: [int]string = map{}
    Reader.BeginObject[]
    var LastName: ?string = false
    loop {
        if:
            Name := Reader.NextName[]
            Level := ParseInt[Name]
            TextureName := Reader.NextString[]
            set HudImageTextures[Level] = TextureName
            set LastName = Some(Name)
        else:
            if (Reader.HasNext[]) {
                FieldPos := if (Name := Reader.NextName[]) {
                    "object \"{Name}\""
                } else if (l_LastName := LastName?) {
                    "object after \"{l_LastName}\""
                } else {
                    "first object"
                }
                Log("Tycoon manager failed reading {FieldPos} for rebirth hud images config", ?Level := log_level.Warning)
            }
            break
    }
    Reader.EndObject[]
    HudImageTextures

ParseTycoonConfigRebirthPerLevelResourceMultipliers<internal>(Reader:json_reader)<transacts><decides>:[int][string]float=
    var PerLevelResourceMultipliers: [int][string]float = map{}
    Reader.BeginObject[]
    var LastName: ?string = false
    loop {
        if:
            Name := Reader.NextName[]
            Level := ParseInt[Name]
            ParsedResourceMultipliers := ParseTycoonConfigRebirthResourceMultipliers[Reader, Level]
            set PerLevelResourceMultipliers[Level] = ParsedResourceMultipliers
            set LastName = Some(Name)
        else:
            if (Reader.HasNext[]) {
                FieldPos := if (Name := Reader.NextName[]) {
                    "object \"{Name}\""
                } else if (l_LastName := LastName?) {
                    "object after \"{l_LastName}\""
                } else {
                    "first object"
                }
                Log("Tycoon manager failed reading {FieldPos} for rebirth resource multipliers", ?Level := log_level.Warning)
            }
            break
    }
    Reader.EndObject[]
    PerLevelResourceMultipliers

ParseTycoonConfigRebirthResourceMultipliers<internal>(Reader:json_reader, Level:int)<transacts><decides>:[string]float=
    var ResourceMultipliers: [string]float = map{}
    Reader.BeginObject[]
    var LastName: ?string = false
    loop {
        if:
            Name := Reader.NextName[]
            Multiplier := Reader.NextFloat[]
            set ResourceMultipliers[Name] = Multiplier
            set LastName = Some(Name)
        else:
            if (Reader.HasNext[]) {
                FieldPos := if (Name := Reader.NextName[]) {
                    "object \"{Name}\""
                } else if (l_LastName := LastName?) {
                    "object after \"{l_LastName}\""
                } else {
                    "first object"
                }
                Log("Tycoon manager failed reading {FieldPos} for rebirth resource multipliers for level {Level}", ?Level := log_level.Warning)
            }
            break
    }
    Reader.EndObject[]
    ResourceMultipliers
## }}}

## Parse Beacon Lines {{{
ParseTycoonConfigBeaconLines<internal>(Reader:json_reader)<transacts><decides>:[string]tycoon_manager_beacon_line_config=
    var BeaconLines: [string]tycoon_manager_beacon_line_config = map{}
    Reader.BeginObject[]
    var LastName: ?string = false
    loop {
        if:
            Name := Reader.NextName[]
            BeaconLineConfig := ParseTycoonConfigBeaconLine[Reader]
            set BeaconLines[Name] = BeaconLineConfig
            set LastName = Some(Name)
        else:
            if (Reader.HasNext[]) {
                FieldPos := if (Name := Reader.NextName[]) {
                    "object \"{Name}\""
                } else if (l_LastName := LastName?) {
                    "object after \"{l_LastName}\""
                } else {
                    "first object"
                }
                Log("Tycoon manager failed reading {FieldPos} for tycoon config beacon lines", ?Level := log_level.Warning)
            }
            break
    }
    Reader.EndObject[]
    BeaconLines

parse_tycoon_config_beacon_line_state := class<final>:
    var InOrderBuyPlates: []tycoon_manager_beacon_line_config_entry = array{}
    var AvailableOptions: []string = array{}

ParseTycoonConfigBeaconLine<internal>(Reader:json_reader)<transacts><decides>:tycoon_manager_beacon_line_config=
    State := parse_tycoon_config_beacon_line_state{}
    Reader.BeginArray[]
    loop {
        set State.AvailableOptions = array{}
        if (BuyPlateId := Reader.NextString[]) {
            set State.AvailableOptions += array{BuyPlateId}
        } else if (
            Reader.BeginArray[]
            loop {
                if (BuyPlateId := Reader.NextString[]) {
                    set State.AvailableOptions += array{BuyPlateId}
                } else {
                    break
                }
            }
            Reader.EndArray[]
        ) {
        } else {
            break
        }
        set State.InOrderBuyPlates += array{tycoon_manager_beacon_line_config_entry{
            AvailableOptions := State.AvailableOptions.ToSet()
        }}
    }
    Reader.EndArray[]
    tycoon_manager_beacon_line_config{
        InOrderBuyPlates := State.InOrderBuyPlates
    }

## }}}

## Parse Buyplates {{{
tycoon_manager_buyplates_config_entry_builder := class:
    BuyplateId: string

    var m_Cost: ?string = false
    var m_Generates: ?string = false
    var Unlocks: []tycoon_manager_unlock_entry = array{}

ParseBuyplatesJsonConfig<internal>(InputJson:string)<transacts><decides>:tycoon_manager_buyplates_config=
    Reader := CreateJsonReader(InputJson)
    Reader.BeginObject[]
    var Entries: []tycoon_manager_buyplates_entry = array{}
    var LastName: ?string = false
    loop {
        if:
            Name := Reader.NextName[]
            Entry := ParseBuyplatesJsonConfigEntry[Reader, Name]
            set Entries += array{Entry}
            set LastName = Some(Name)
        else:
            if (Reader.HasNext[]) {
                FieldPos := if (Name := Reader.NextName[]) {
                    "object \"{Name}\""
                } else if (l_LastName := LastName?) {
                    "object after \"{l_LastName}\""
                } else {
                    "first object"
                }
                Log("Tycoon manager failed reading {FieldPos} for buyplate configs", ?Level := log_level.Warning)
            }
            break
    }
    Reader.EndObject[]
    tycoon_manager_buyplates_config{
        Entries := Entries
    }

ParseBuyplatesJsonConfigEntry<internal>(Reader:json_reader, BuyplateId:string)<transacts><decides>:tycoon_manager_buyplates_entry=
    Reader.BeginObject[]
    Builder := tycoon_manager_buyplates_config_entry_builder{
        BuyplateId := BuyplateId
    }
    var LastName: ?string = false
    loop {
        if:
            Name := Reader.NextName[]
            case(Name):
                "cost" => {
                    Value := Reader.NextString[]
                    set Builder.m_Cost = Some(Value)
                }
                "generates" => {
                    Value := Reader.NextString[]
                    set Builder.m_Generates = Some(Value)
                }
                "unlock" => {
                    Reader.BeginArray[]
                    loop {
                        if (UnlockStr := Reader.NextString[]) {
                            Unlock := if (SplitRes := UnlockStr.SeqSplitOnce["->"]) {
                                tycoon_manager_unlock_entry{
                                    TargetRoutingId := SplitRes(0)
                                    Payload := Some(SplitRes(1))
                                }
                            } else {
                                tycoon_manager_unlock_entry{
                                    TargetRoutingId := UnlockStr
                                    Payload := None()
                                }
                            }
                            set Builder.Unlocks += array{Unlock}
                        } else {
                            break
                        }
                    }
                    Reader.EndArray[]
                }
                _ => Reader.SkipNext[]
            set LastName = Some(Name)
        else:
            if (Reader.HasNext[]) {
                FieldPos := if (Name := Reader.NextName[]) {
                    "field \"{Name}\""
                } else if (l_LastName := LastName?) {
                    "field after \"{l_LastName}\""
                } else {
                    "first field"
                }
                Log("Tycoon manager failed reading {FieldPos} for buyplate config \"{BuyplateId}\"", ?Level := log_level.Warning)
            }
            break
    }
    Reader.EndObject[]
    var Cost: ?tycoon_manager_buyplates_entry_resource = false
    if (var CostStr: string = Builder.m_Cost?) {
        var CheckFlag: logic = false
        if:
            NewCostStr := CostStr.TryReplace["<check>", ""]
            set CostStr = NewCostStr
            set CheckFlag = true

        if (
            SplitRes := CostStr.SplitOnce[':']
            ResourceId := SplitRes(0)
            Amount := ParseInt[SplitRes(1)]
        ) {
            set Cost = Some(tycoon_manager_buyplates_entry_resource{
                ResourceId := ResourceId
                Amount := Amount
                CheckDontTake := CheckFlag
            })
        } else {
            Log("Could not parse cost for buyplate \"{BuyplateId}\": invalid format \"{CostStr}\"", ?Level := log_level.Error)
        }
    } else if (CostStr := Builder.m_Cost?) {
        Log("Could not parse cost for buyplate \"{BuyplateId}\": invalid format \"{CostStr}\"", ?Level := log_level.Error)
    }
    var Generates: ?tycoon_manager_buyplates_entry_resource = false
    if (
        i_Generates := Builder.m_Generates?, SplitRes := i_Generates.SplitOnce[':'],
        Amount := ParseInt[SplitRes(1)]
    ) {
        set Generates = Some(tycoon_manager_buyplates_entry_resource{
            ResourceId := SplitRes(0)
            Amount := Amount
            CheckDontTake := false
        })
    } else if (Builder.m_Generates?) {
        Log("Could not parse resource generation for buyplate \"{BuyplateId}\"", ?Level := log_level.Error)
    }
    Entry := tycoon_manager_buyplates_entry{
        BuyplateId := BuyplateId
        Cost := Cost
        Generates := Generates
        Unlocks := Builder.Unlocks
    }
## }}}

## ToString {{{
ToString(Obj:tycoon_manager_buyplates_config):string=
    var Entries: string = ""
    for (Index->Entry:Obj.Entries) {
        if (Index > 0) {
            set Entries += ", "
        }
        set Entries += ToString(Entry)
    }
    "\{Entries: [{Entries}]\}"

ToString(Obj:tycoon_manager_buyplates_entry):string=
    Cost := if (l_Cost := Obj.Cost?) {
        ToString(l_Cost)
    } else {
        "None"
    }
    var Unlocks: string = ""
    for (Index->Unlock:Obj.Unlocks) {
        if (Index > 0) {
            set Unlocks += ", "
        }
        set Unlocks += "{Unlock}"
    }
    "\{BuyplateId: \"{Obj.BuyplateId}\", Cost: {Cost}, Unlocks: [{Unlocks}]\}"

ToString(Obj:tycoon_manager_buyplates_entry_resource):string=
    "\{ResourceId: \"{Obj.ResourceId}\", Amount: {Obj.Amount}\}"

ToString(Obj:tycoon_manager_unlock_entry):string=
    PayloadStr := if (Payload := Obj.Payload?) {
        "Some(\"{Payload}\")"
    } else {
        "None"
    }
    "\{TargetRoutingId: \"{Obj.TargetRoutingId}\", Payload: {PayloadStr}\}"
## }}}
