using { /Fortnite.com/Devices }
using { /Fortnite.com/Characters }
using { /Verse.org/Simulation }
using { /Verse.org/Random }
using { /UnrealEngine.com/Temporary/Diagnostics }
using { /UnrealEngine.com/Temporary/SpatialMath }
using { /UnrealEngine.com/Temporary/UI }

using { HavocVerse.Debug }
using { HavocVerse.Devices }
using { HavocVerse.Storage }
using { HavocVerse.JSON }
using { HavocVerse.UI }
using { HavocVerse.Util }
using { HavocVerse.Util.Task }
using { HavocVerse.Logging }

tycoon_manager_rebirth_reset_trigger_config := class<concrete>:
    @editable Trigger: trigger_device = trigger_device{}
    @editable ForLevel: int = -1

    IsApplicable<public>(RebirthLevel:int)<transacts><decides>:void=
        ForLevel < 0 or ForLevel = RebirthLevel

tycoon_manager_progress_line<public> := class<concrete>:
    @editable Name<public>: string = "N/A"
    @editable ClassSlot: int = -1
    @editable AccessControl<public>: ?access_control_device = false
    @editable FrontCheckpoint: ?player_checkpoint_device = false
    @editable FrontTeleporter: ?teleporter_device = false
    @editable AutoGenerationManager: ?auto_generation_manager = false
    @editable ClaimBuyplate: ?buy_plate_device = false
    @editable ExtraBuyplates: []creative_device = array{}
    @editable ExtraRollbackDevices: []creative_device = array{}
    @editable ExtraRollbackTriggers: []trigger_device = array{}
    @editable BeaconLines: []tycoon_manager_progress_line_beacon_line = array{}

    @editable RebirthResetTriggers: []tycoon_manager_rebirth_reset_trigger_config = array{}
    @editable RebirthBillboard: []billboard_device = array{}
    @editable RebirthBillboardTemplate: string = "Rebirths: %CurrentLevel%"
    @editable RebirthBillboardHideOnZero: logic = false

    var m_Index: int = -1
    var m_AutoGenerationTickets: [string]auto_generation_ticket = map{}
    var m_PersistenceRecorder: ?persistence_recorder_device = false
    var m_RebirthRecordableHandle: ?persistence_recorder_recordable_handle = false
    var m_BuyPlatesBackrefCache: [string]buy_plate_device = map{}

    var m_RebirthLevel: int = 0
    var m_SetResourceMultipliers: [string]float = map{}

tycoon_manager_progress_line_beacon_line := class<concrete>:
    @editable BeaconLineId: string = "beaconline1"
    @editable MapIndicator: map_indicator_device = map_indicator_device{}
    @editable MapIndicatorUnmetCost: ?map_indicator_device = false

    var m_CurrentEntry: int = 0
    var m_LockedTargetBuyPlate: ?string = false

    var m_MapIndicatorState: ?tycoon_manager_beacon_line_buy_plate_state = false
    var m_MapIndicatorEnabled: ?tycoon_manager_beacon_line_map_indicator_variant = false
    var m_MapIndicatorPulseActive: logic = false
    var m_MapIndicatorPulseToAgents: [agent]event() = map{}

    GetActiveMapIndicator()<decides><transacts>:map_indicator_device=
        case(m_MapIndicatorEnabled?):
            tycoon_manager_beacon_line_map_indicator_variant.Primary => MapIndicator
            tycoon_manager_beacon_line_map_indicator_variant.UnmetCost => MapIndicatorUnmetCost?

    DisablePulse(ProgressLine:tycoon_manager_progress_line):void=
        if (not m_MapIndicatorPulseActive?) {
            return
        }
        set m_MapIndicatorPulseActive = false
        if (l_MapIndicator := GetActiveMapIndicator[]) {
            var ConcernedAgents: [agent]void = map{}
            for (Agent->AbortEvent:m_MapIndicatorPulseToAgents) {
                AbortEvent.Signal()
                if. set ConcernedAgents[Agent] = void
            }
            for (Agent:ProgressLine.AccessControl?.ListAllowedAgents()) {
                if. set ConcernedAgents[Agent] = void
            }
            for (Agent->__:ConcernedAgents) {
                l_MapIndicator.DeactivateObjectivePulse(Agent)
            }
        }
        set m_MapIndicatorPulseToAgents = map{}

tycoon_manager_beacon_line_map_indicator_variant := enum:
    Primary
    UnmetCost

ToString(Variant:tycoon_manager_beacon_line_map_indicator_variant)<transacts>:string=
    case(Variant):
        tycoon_manager_beacon_line_map_indicator_variant.Primary => "Primary"
        tycoon_manager_beacon_line_map_indicator_variant.UnmetCost => "UnmetCost"

tycoon_manager_beacon_line_buy_plate_state := class<final>:
    BuyPlate: buy_plate_device
    UnmetCost: logic

tycoon_manager_player_state := class<final>:
    ProgressLine: tycoon_manager_progress_line

tycoon_manager_rebirth_pet_config<public> := class<concrete>:
    @editable PetId<public>: string = ""
    @editable RebirthLevel<public>: int = 1

tycoon_manager_rebirth_attachment_config<public> := class<concrete>:
    @editable RebirthLevel<public>: int = 1
    @editable AttachmentKindId<public>: string = ""
    @editable Attachment<public>: string = ""

tycoon_manager_beacon_line_area_config := class<concrete>:
    @editable Volume: volume_device = volume_device{}
    @editable OuterVolume: ?volume_device = false

    var m_InVolume: [agent]void = map{}

tycoon_claim_or_teleport_result<public> := enum:
    Claimed
    Teleported
    Failed

tycoon_rebirth_event<public> := class<final>:
    Player<public>: player
    NewRebirthLevel<public>: int
    Virtual<public>: logic

tycoon_manager_device<public> := class(creative_device, aware_device, aware_debug_device, pet_condition_provider, attachments_dynamic_unlock_source):
    @editable p_ResourceManager<public>: resource_manager_device = resource_manager_device{}
    @editable p_AnalyticsController<public>: analytics_controller_device = analytics_controller_device{}
    @editable p_ClassManager<public>: class_manager_device = class_manager_device{}
    @editable p_ProgressLines: []tycoon_manager_progress_line = array{}
    @editable p_TimeRewardsDevice: []time_rewards_device = array{}
    @editable p_PetController: []pet_controller_device = array{}
    @editable p_AttachmentsManager: ?attachments_manager_device = false
    @editable p_IntroCinematic: ?cinematics_player_device = false

    @editable var p_TycoonJsonConfig: string = ""
    @editable var p_BuyplatesJsonConfig: string = ""
    @editable p_RoutingIdSingleMatch: logic = true
    @editable p_AlwaysSetCostOverride: logic = true

    @editable p_ClaimFailHUDMessage: []hud_message_device = array{}
    @editable p_MiscPlayerHeight<public>: float = 180.0

    @editable p_BeaconLinesHideAreasInvert: logic = false
    @editable p_BeaconLinesHideAreas: []tycoon_manager_beacon_line_area_config = array{}
    @editable p_BeaconLinesMapIndicatorOffset: vector3 = vector3{}

    @editable p_EnableQuitRollback: logic = true
    @editable p_QuitRollbackGranularity: quit_rollback_device_granularity = quit_rollback_device_granularity.Player

    @editable p_EnablePersistence: logic = false
    @editable p_PersistenceSaveSlots: []shared_storage_chunk_device = array{}
    @editable p_PersistenceSavePoints: []exposed_save_point_device = array{}
    @editable p_PersistenceAutoSaveInternal: float = 3.0
    @editable p_PersistenceBuyPlatesReserveBits: int = -1
    @editable p_PersistenceIncludeResourceManager: logic = true
    @editable p_PersistenceExtraDevices: []creative_device = array{}
    @editable p_PersistenceSelectMenuEnabled: logic = true
    @editable p_PersistenceUseDialogSelectMenu: []persistence_recorder_popup_dialog_menu_config = array{}
    @editable p_PersistenceSelectMenuConfig: default_persistence_recorder_load_select_menu_config = default_persistence_recorder_load_select_menu_config{}
    @editable p_PersistenceBuyPlateRecoveryChunkSize: int = -1
    @editable p_PersistenceBuyPlateRecoveryChunkDelay: float = 0.1

    @editable p_RebirthRetainResources<public>: []string = array{}
    @editable p_RebirthPets<public>: []tycoon_manager_rebirth_pet_config = array{}
    @editable p_RebirthAttachments<public>: []tycoon_manager_rebirth_attachment_config = array{}
    @editable p_RebirthLevelReserveBits: int = 4
    @editable p_RebirthNextLevelOnTrigger: []trigger_device = array{}
    @editable p_RebirthNextLevelDialog: []popup_dialog_device = array{}
    @editable p_RebirthNextLevelDialogPerLevel: logic = false
    @editable p_RebirthNextLevelDialogPerLevelDefaultLast: logic = true
    @editable p_RebirthNextLevelDialogConfirmButtonIndex: int = 0
    @editable p_RebirthPlaceholdersMultiplierFormatter: number_formatter_config = number_formatter_config{
        Template := "%Val%x"
    }
    @editable p_RebirthResetTriggers: []tycoon_manager_rebirth_reset_trigger_config = array{}
    @editable p_RebirthMultipliersChannel: string = "tycoon_rebirth"
    @editable p_RebirthHUDManagerConfig: hud_manager_base_config = hud_manager_base_config{
        RefreshRate := -1.0
    }
    @editable p_RebirthHUDImageSize: vector2 = vector2{X := 100.0, Y := 50.0}
    @editable p_RebirthHUDImagePlacement: ui_canvas_placement_config = ui_canvas_placement_config{}

    @editable p_BuyPlateCostMetTrackingRefreshRate: float = 0.5

    PlayerSaveSlotSelectedEvent<public>: event(tuple(tycoon_manager_progress_line,player_save_slot_selected_event)) = event(tuple(tycoon_manager_progress_line,player_save_slot_selected_event)){}
    PlayerJoinProgressLineEvent<public>: event(tuple(player,tycoon_manager_progress_line)) = event(tuple(player,tycoon_manager_progress_line)){}

    var m_TycoonConfig<private>: tycoon_manager_config = tycoon_manager_config{}
    var m_BuyplatesConfigs<private>: [string]tycoon_manager_buyplates_entry = map{}
    var m_Initialized<private>: logic = false

    var m_Players<private>: [player]tycoon_manager_player_state = map{}
    var m_RebirthHudManager<internal>: ?hud_manager_base = false
    var m_RebirthDialogs<private>: [int]tuple(string,dialog_device_wrapper) = map{}
    var m_RebirthPets<private>: [string]int = map{}
    var m_RebirthAttachments<private>: [string]int = map{}

    var m_PersistenceControl<private>: persistence_recordable_control = persistence_recordable_control{}

    var m_RebirthListeners<private>: []type{__(:tycoon_rebirth_event):void} = array{}

    DeviceMeta<override>()<transacts>:aware_device_meta=
        aware_device_meta{
            DeviceName := "tycoon_manager"
        }

    DebugStats<override>()<transacts>:aware_debug_stats=
        aware_debug_stats{
            Length := map{
                "buyplatesConfigs" => m_BuyplatesConfigs.Length
                "players" => m_Players.Length
                "rebirthDialogs" => m_RebirthDialogs.Length
                "listeners::rebirth" => m_RebirthListeners.Length
            }
        }

    OnBegin<override>()<suspends>:void=
        for (BeaconLineHideArea:p_BeaconLinesHideAreas) {
            Closure := tycoon_manager_beacon_line_area_closure{
                Device := Self
                Config := BeaconLineHideArea
            }

            BeaconLineHideArea.Volume.AgentEntersEvent.Subscribe(Closure.OnAgentEntered)
            OuterVolume := BeaconLineHideArea.OuterVolume? or BeaconLineHideArea.Volume
            OuterVolume.AgentExitsEvent.Subscribe(Closure.OnAgentExited)
        }

        AwareInitDevice(Self, OnInit)

    OnInit<private>()<suspends>:void=
        if (Res := ParseTycoonConfig[p_TycoonJsonConfig]) {
            Log("Tycoon manager parsed tycoon config succesfully.")
            set p_TycoonJsonConfig = ""
            set m_TycoonConfig = Res
        } else {
            Log("Tycoon manager could not parse tycoon config :(", ?Level := log_level.Error)
        }

        if (
            RebirthConfig := m_TycoonConfig.Rebirth?
            MaxPersistenceRebirthLevel := FromBits(array{true}.Repeat(p_RebirthLevelReserveBits))
            RebirthConfig.MaxLevel > MaxPersistenceRebirthLevel
        ) {
            Log("Tycoon managed has max rebirth level to {RebirthConfig.MaxLevel} but only {p_RebirthLevelReserveBits} bits (max {MaxPersistenceRebirthLevel}) are allocated for rebirth persistence", ?Level := log_level.Error)
        }

        if (Res := ParseBuyplatesJsonConfig[p_BuyplatesJsonConfig]) {
            Log("Tycoon manager parsed {Res.Entries.Length} buyplate configs succesfully. Setting up...")
            set p_BuyplatesJsonConfig = ""
            for (Index->Entry:Res.Entries, set m_BuyplatesConfigs[Entry.BuyplateId] = Entry) {
                set Entry.m_StorageIndex = Index
            }
        } else {
            Log("Tycoon manager could not parse buyplate configs :(", ?Level := log_level.Error)
        }

        for (Index->ProgressLine:p_ProgressLines) {
            set ProgressLine.m_Index = Index
            if (AccessControl := ProgressLine.AccessControl?, AccessControl.p_Mode = access_control_mode.Whitelist) {
                JoinClosure := tycoon_manager_player_closure{
                    Device := Self
                    ProgressLine := ProgressLine
                }
                # since access control is in whitelist mode, and these events are backed by ordered arrays,
                # we can make the assumption that OnPlayerAdded is processed before persistence load
                AccessControl.SubscribeToPlayerJoinEvents(JoinClosure.OnPlayerAdded, JoinClosure.OnPlayerRemoved)

                if (ClaimBuyplate := ProgressLine.ClaimBuyplate?) {
                    Middleware := tycoon_manager_claim_buyplate_middleware{
                        Device := Self
                        ProgressLine := ProgressLine
                    }
                    ClaimBuyplate.AddMiddleware(Middleware)
                    set ClaimBuyplate.m_InjectCinematicPlayer = p_IntroCinematic
                }
                if (p_BuyPlateCostMetTrackingRefreshRate > 0.0) {
                    spawn:
                        StartCostMetTrackingLoop(ProgressLine, AccessControl)
                    Closure := tycoon_manager_buyplate_cost_met_reflection_closure{
                        Device := Self
                        ProgressLine := ProgressLine
                    }
                    p_ResourceManager.SubscribeToResourceTake(Closure.OnResourceTake)
                }

                if (p_EnablePersistence?) {
                    PersistenceRecorder := persistence_recorder_device{
                        p_DevId := "progress line {ProgressLine.Name}"
                        p_PlayerTrackingMode := persistence_recorder_player_tracking_mode.AccessControlWhitelist
                        p_AccessControl := AccessControl
                        p_StorageSaveSlots := p_PersistenceSaveSlots
                        p_SavePoints := p_PersistenceSavePoints
                        p_AutoSaveInterval := p_PersistenceAutoSaveInternal
                        p_UseDefaultSelectMenu := p_PersistenceSelectMenuEnabled
                        p_DefaultSelectMenuConfig := p_PersistenceSelectMenuConfig
                        p_UseDialogSelectMenu := p_PersistenceUseDialogSelectMenu
                        p_SkipSelectMenuOnFirstPlay := true
                        p_CustomBuyPlatesReserveBits := p_PersistenceBuyPlatesReserveBits
                        p_BuyPlateRecoveryChunkSize := p_PersistenceBuyPlateRecoveryChunkSize
                        p_BuyPlateRecoveryChunkDelay := p_PersistenceBuyPlateRecoveryChunkDelay
                    }
                    if (m_TycoonConfig.Rebirth?) {
                        RebirthDummyDevice := tycoon_manager_rebirth_dummy_device{
                            TycoonManager := Self
                            ProgressLine := ProgressLine
                        }
                        Handle := PersistenceRecorder.RegisterRecordable(RebirthDummyDevice)
                        set ProgressLine.m_RebirthRecordableHandle = Some(Handle)
                    }
                    if (p_PersistenceIncludeResourceManager?) {
                        PersistenceRecorder.RegisterDeviceToRecord(p_ResourceManager)
                    }
                    for (TimeRewardsDevice:p_TimeRewardsDevice) {
                        PersistenceRecorder.RegisterDeviceToRecord(TimeRewardsDevice)
                    }
                    for (Device:p_PersistenceExtraDevices) {
                        PersistenceRecorder.RegisterDeviceToRecord(Device)
                    }
                    spawn:
                        PersistenceRecorder.OnBegin()
                    set ProgressLine.m_PersistenceRecorder = Some(PersistenceRecorder)

                    Closure := tycoon_manager_player_persistence_load_closure{
                        Callback := OnPlayerSaveSlotSelected
                        ProgressLine := ProgressLine
                    }
                    SubscribeToEvent(PersistenceRecorder.PlayerSaveSlotSelectedEvent, Closure.Call)
                }

                DeviceRegisterClosure := tycoon_manager_device_register_closure{
                    Device := Self
                    ProgressLine := ProgressLine
                }
                Subscribed := AccessControl.SubscribeToDevicesBackrefs(DeviceRegisterClosure.Call)
                if (Subscribed?) {
                    if (ClaimBuyplate := ProgressLine.ClaimBuyplate?) {
                        OnDeviceRegistered(ProgressLine, ClaimBuyplate)
                    }
                    for (BuyPlate:ProgressLine.ExtraBuyplates) {
                        OnDeviceRegistered(ProgressLine, BuyPlate)
                    }
                } else {
                    Log("Could not subscribe to access control device of progress line \"{ProgressLine.Name}\" because backrefs building is not enabled", ?Level := log_level.Error)
                }

                ResetBeaconLines(ProgressLine)
            } else if (ProgressLine.AccessControl?) {
                Log("Progress line \"{ProgressLine.Name}\" has an access control, but its mode is not Whitelist", ?Level := log_level.Error)
            } else {
                Log("Progress line \"{ProgressLine.Name}\" is missing an access control device", ?Level := log_level.Error)
            }
        }

        GetPlayspace().PlayerRemovedEvent().Subscribe(OnPlayerQuit)

        if (p_EnablePersistence?) {
            for (
                Index->Device:p_PersistenceExtraDevices
                (not persistence_recordable_base[Device]) and (not persistence_resetable[Device])
            ) {
                Log("Tycoon manager is configured with an extra persistence device that does not support persistence (index {Index})", ?Level := log_level.Error)
            }
        }

        if (RebirthConfig := m_TycoonConfig.Rebirth?) {
            for (Trigger:p_RebirthNextLevelOnTrigger) {
                Trigger.TriggeredEvent.Subscribe(OnRebirthNextLevelTriggerCalled)
            }

            HudManager := tycoon_manager_rebirth_hud_manager{
                RefreshRate := p_RebirthHUDManagerConfig.RefreshRate
                ZOrder := p_RebirthHUDManagerConfig.ZOrder
                Device := Self
                Config := RebirthConfig
            }
            set m_RebirthHudManager = Some(HudManager)
            spawn:
                HudManager.Run(GetPlayspace())
        }

        for (
            PetRebirthConfig:p_RebirthPets
            set m_RebirthPets[PetRebirthConfig.PetId] = PetRebirthConfig.RebirthLevel
        ) {}
        for (PetController:p_PetController) {
            for (PetId->__:m_RebirthPets) {
                PetController.SetPetConditionProvider(PetId, Self)
            }
        }

        for (
            AttachmentRebirthConfig:p_RebirthAttachments
            AttachmentSlug := AttachmentRebirthConfig.AttachmentKindId + ":" + AttachmentRebirthConfig.Attachment
            set m_RebirthAttachments[AttachmentSlug] = AttachmentRebirthConfig.RebirthLevel
        ) {}
        if (AttachmentsManager := p_AttachmentsManager?) {
            for (AttachmentRebirthConfig:p_RebirthAttachments) {
                AttachmentsManager.RegisterDynamicUnlockSource(AttachmentRebirthConfig.AttachmentKindId, Self)
            }
        } else if (p_RebirthAttachments.Length > 0) {
            Log("Tycoon manager is configured with rebirth attachments but no attachments manager is set", ?Level := log_level.Error)
        }

        AwareSubscribeToPostInit(OnPostInit)

    OnPostInit<private>():void=
        set m_Initialized = true
        for (ProgressLine:p_ProgressLines) {
            for (BeaconLine:ProgressLine.BeaconLines) {
                set BeaconLine.m_CurrentEntry = 0
                set BeaconLine.m_LockedTargetBuyPlate = None()
                UpdateBeaconLine(ProgressLine, BeaconLine)
            }
        }

    OnPlayerQuit<private>(Player:player):void=
        for (BeaconLineArea:p_BeaconLinesHideAreas) {
            set BeaconLineArea.m_InVolume = BeaconLineArea.m_InVolume.RemoveByKey(Player)(0)
        }
        for (ProgressLine:p_ProgressLines) {
            for (BeaconLine:ProgressLine.BeaconLines) {
                RemoveRes := BeaconLine.m_MapIndicatorPulseToAgents.RemoveByKey(Player)
                set BeaconLine.m_MapIndicatorPulseToAgents = RemoveRes(0)
                if (AbortEvent := RemoveRes(1)?) {
                    AbortEvent.Signal()
                }
            }
        }
        if (PlayerState := m_Players[Player]) {
            OnPlayerRemoved(PlayerState.ProgressLine, Player)
            set m_Players = m_Players.RemoveByKey(Player)(0)
        }

    OnPlayerAdded<internal>(ProgressLine:tycoon_manager_progress_line, Player:player):void=
        PlayerState := tycoon_manager_player_state{
            ProgressLine := ProgressLine
        }
        if. set m_Players[Player] = PlayerState
        if (ProgressLine.ClassSlot > 0) {
            p_ClassManager.SetClassSlot(Player, ProgressLine.ClassSlot)
        }
        SyncPlayerBeaconLines(Player)
        if (AccessControl := ProgressLine.AccessControl?) {
            UpdateCostMetTracking(ProgressLine, AccessControl)
        }
        PlayerJoinProgressLineEvent.Signal(Player, ProgressLine)

    OnPlayerRemoved<internal>(ProgressLine:tycoon_manager_progress_line, Player:player):void=
        SyncPlayerBeaconLines(Player)
        DoQuitRollback(ProgressLine, Player)
        if (ProgressLine.ClassSlot > 0) {
            p_ClassManager.ResetClass(Player)
        }
        if (PlayerState := m_Players[Player], PlayerState.ProgressLine.m_Index = ProgressLine.m_Index) {
            set m_Players = m_Players.RemoveByKey(Player)(0)
        }

    SubscribeToRebirthEvent<public>(Listener:type{__(:tycoon_rebirth_event):void}):void=
        set m_RebirthListeners += array{Listener}

    IsPlaying<public>(Player:player)<transacts><decides>:void=
        m_Players[Player]

    ForceExpelPlayer<public>(Player:player):logic=
        if (
            PlayerState := m_Players[Player]
            AccessControl := PlayerState.ProgressLine.AccessControl?
            AccessControl.IsAllowed[Player]
        ) {
            AccessControl.RemoveFromWhitelist(Player, false)
            true
        } else {
            false
        }

    TeleportHome<public>(Player:player):logic=
        if (ProgressLine := GetPlayerState[Player].ProgressLine, Teleporter := ProgressLine.FrontTeleporter?) {
            Teleporter.Teleport(Player)
            true
        } else if (TeleportHome0[Player]) {
            true
        } else {
            false
        }

    # Attempts to teleport the player to the front checkpoint of their progress line.
    # Fails if the player is not part of a progress line, if no front checkpoint device was linked, or if .TeleportTo[] fails.
    TeleportHome0<private>(Player:player)<decides><transacts>:void=
        ProgressLine := GetPlayerState[Player].ProgressLine
        Checkpoint := ProgressLine.FrontCheckpoint?
        FortCharacter := Player.GetFortCharacter[]
        TargetTransform := Checkpoint.GetTransform()
        TargetPosition := TargetTransform.Translation.AddZ(p_MiscPlayerHeight)
        FortCharacter.StallingTeleportTo[TargetPosition, TargetTransform.Rotation]

    ClaimOrTeleportHome<public>(Player:player):tycoon_claim_or_teleport_result=
        if (IsPlaying[Player]) {
            TPSuccess := TeleportHome(Player)
            if (TPSuccess?) {
                return tycoon_claim_or_teleport_result.Teleported
            } else {
                Log("Couldn't teleport '{Player}' to their progress line, check editor logs for reason", ?Level := log_level.Error)
            }
        } else {
            var InUseProgressLines: [int]void = map{}
            for (__->PlayerState:m_Players) {
                if. set InUseProgressLines[PlayerState.ProgressLine.m_Index] = void
            }
            for (ProgressLine:p_ProgressLines, not InUseProgressLines[ProgressLine.m_Index]) {
                if (ClaimBuyplate := ProgressLine.ClaimBuyplate?) {
                    ClaimBuyplate.Unlock(Player, ?UserTriggered := false, ?FirstUnlock := false, ?Force := true)
                } else {
                    ClaimProgressLineFromBuyplate(ProgressLine, Player)
                }
                TPSuccess := TeleportHome(Player)
                if (not TPSuccess?) {
                    Log("Couldn't teleport '{Player}' to their progress line after claiming, check editor logs for reason", ?Level := log_level.Error)
                }
                return tycoon_claim_or_teleport_result.Claimed
            }
            Log("Couldn't auto claim a progress line for '{Player}': all progress lines are in use", ?Level := log_level.Error)
        }
        tycoon_claim_or_teleport_result.Failed

    ClaimProgressLineFromBuyplate<internal>(ProgressLine:tycoon_manager_progress_line, Player:agent):void=
        if (AccessControl := ProgressLine.AccessControl?) {
            AccessControl.AddToWhitelist(Player)
        } else {
            Log("Couldn't claim progress line #{ProgressLine.m_Index} for '{Player}': missing access control", ?Level := log_level.Error)
        }

    DoQuitRollback<private>(ProgressLine:tycoon_manager_progress_line, Player:player):void=
        if (not (p_EnableQuitRollback?) or ProgressLine.AccessControl?.ListAllowedAgents().Length > 0) {
            return
        }
        set ProgressLine.m_RebirthLevel = 0
        RollbackAgent: ?agent = case(p_QuitRollbackGranularity):
            quit_rollback_device_granularity.Player => Some(Player)
            quit_rollback_device_granularity.Omitted => false
        DoRollback(ProgressLine, RollbackAgent)
        if (ClaimBuyplate := ProgressLine.ClaimBuyplate?) {
            ClaimBuyplate.Reset(RollbackAgent)
            set ClaimBuyplate.m_InjectCinematicPlayer = p_IntroCinematic
        }
        EnsureRebirthEffects(ProgressLine)

    DoRollback<private>(ProgressLine:tycoon_manager_progress_line, RollbackAgent:?agent):void=
        for (Ticket:ProgressLine.m_AutoGenerationTickets) {
            Ticket.Disable()
        }
        var RollbackDeviceCount: int = 0
        if (DeviceBackrefs := ProgressLine.AccessControl?.GetUsingDevices[]) {
            for (Device:DeviceBackrefs, Resetable := resetable[Device]) {
                Resetable.Reset(RollbackAgent)
                set RollbackDeviceCount += 1
            }
        }
        for (BuyPlateDevice:ProgressLine.ExtraBuyplates, BuyPlate := buy_plate_device[BuyPlateDevice]) {
            BuyPlate.Reset(RollbackAgent)
            set RollbackDeviceCount += 1
        }
        for (Device:ProgressLine.ExtraRollbackDevices, Resetable := resetable[Device]) {
            Resetable.Reset(RollbackAgent)
            set RollbackDeviceCount += 1
        }
        ResetBeaconLines(ProgressLine)
        if (l_Agent := RollbackAgent?) {
            for (Trigger:ProgressLine.ExtraRollbackTriggers) {
                Trigger.Trigger(l_Agent)
            }
        } else {
            for (Trigger:ProgressLine.ExtraRollbackTriggers) {
                Trigger.Trigger()
            }
        }
        Log("Rolled back progress line \"{ProgressLine.Name}\", {RollbackDeviceCount} affected devices", ?Level := log_level.Debug)

    StartCostMetTrackingLoop<private>(
        ProgressLine:tycoon_manager_progress_line,
        AccessControl:access_control_device
    )<suspends>:void=
        loop {
            UpdateCostMetTracking(ProgressLine, AccessControl)
            Sleep(p_BuyPlateCostMetTrackingRefreshRate)
        }

    UpdateCostMetTracking<internal>(
        ProgressLine:tycoon_manager_progress_line,
        AccessControl:access_control_device
    ):void=
        if (Agent := AccessControl.ListAllowedAgents()[0]) {
            var NewStateBy: []string = array{}
            for (Device:AccessControl.GetUsingDevices[], BuyPlate := buy_plate_device[Device]) {
                NewState := BuyPlate.UpdateCostMetTracking(Agent)
                if (NewState?) {
                    set NewStateBy += array{BuyPlate.GetRoutingId()}
                }
            }
            for (
                BeaconLine:ProgressLine.BeaconLines
                BeaconLineConfig := m_TycoonConfig.BeaconLines[BeaconLine.BeaconLineId]
                LineEntry := BeaconLineConfig.InOrderBuyPlates[BeaconLine.m_CurrentEntry]
                LineEntry.ContainsAnyBuyPlateOf[NewStateBy]
            ) {
                UpdateBeaconLine(ProgressLine, BeaconLine)
            }
        }

    ResetBeaconLines<private>(ProgressLine:tycoon_manager_progress_line):void=
        for (BeaconLine:ProgressLine.BeaconLines) {
            set BeaconLine.m_CurrentEntry = 0
            set BeaconLine.m_LockedTargetBuyPlate = None()
            UpdateBeaconLine(ProgressLine, BeaconLine)
        }

    IsBeaconLineEntryComplete<private>(
        ProgressLine:tycoon_manager_progress_line,
        LineEntry:tycoon_manager_beacon_line_config_entry
    )<decides><transacts>:void=
        var AllUnlocked: logic = true
        for (BuyPlateId->__:LineEntry.AvailableOptions, AllUnlocked?) {
            if (BuyPlate := FindBuyPlateById[ProgressLine, BuyPlateId]) {
                if (not BuyPlate.IsUnlocked[]) {
                    set AllUnlocked = false
                }
            } else {
                set AllUnlocked = false
            }
        }
        AllUnlocked?

    BestRankingBeaconLineEntryBuyPlate<private>(
        ProgressLine:tycoon_manager_progress_line,
        LineEntry:tycoon_manager_beacon_line_config_entry
    ):?buy_plate_device=
        Agent := option{ProgressLine.AccessControl?.ListAllowedAgents()[0]}

        var AnyCostMet: logic = false
        EligibleEntries := for (
            BuyPlateId->__:LineEntry.AvailableOptions
            BuyPlate := FindBuyPlateById[ProgressLine, BuyPlateId]
            BuyPlate.IsEnabled[]
            not BuyPlate.IsUnlocked[]
        ) {
            CostMet := if (l_Agent := Agent?) {
                BuyPlate.DoesAgentMeetsCostCached(l_Agent)
            } else {
                false
            }
            if (CostMet?, not AnyCostMet?) {
                set AnyCostMet = true
            }
            (BuyPlate, CostMet)
        }

        if (
            AnyCostMet?
            Player := player[Agent?]
            FortPlayer := Player.GetFortCharacter[]
            AgentPosition := FortPlayer.GetTransform().Translation
        ) {
            var TopRanking: ?tuple(buy_plate_device,float) = false
            for (Entry:EligibleEntries, Entry(1)?) {
                BuyPlate := Entry(0)
                Dist := Distance(AgentPosition, BuyPlate.GetPosition())
                if ((not TopRanking?) or (Dist < TopRanking?(1))) {
                    set TopRanking = Some(Entry(0), Dist)
                }
            }
            option{TopRanking?(0)}
        } else {
            option{EligibleEntries[0](0)}
        }

    UpdateBeaconLine<private>(
        ProgressLine:tycoon_manager_progress_line,
        BeaconLine:tycoon_manager_progress_line_beacon_line
    ):void=
        var NewEntry: logic = false
        loop {
            if (
                BeaconLineConfig := m_TycoonConfig.BeaconLines[BeaconLine.BeaconLineId]
                LineEntry := BeaconLineConfig.InOrderBuyPlates[BeaconLine.m_CurrentEntry]
                IsBeaconLineEntryComplete[ProgressLine, LineEntry]
            ) {
                # handle unordered unlocking
                set BeaconLine.m_CurrentEntry += 1
                set BeaconLine.m_LockedTargetBuyPlate = None()
                set NewEntry = true
            } else {
                break
            }
        }

        var MapIndicatorPointTo: ?tycoon_manager_beacon_line_buy_plate_state = false
        if (
            BeaconLineConfig := m_TycoonConfig.BeaconLines[BeaconLine.BeaconLineId]
            LineEntry := BeaconLineConfig.InOrderBuyPlates[BeaconLine.m_CurrentEntry]
        ) {
            if (NewEntry?) {
                for (BuyPlateId->__:LineEntry.AvailableOptions, not FindBuyPlateById[ProgressLine, BuyPlateId]) {
                    Log("Buy plate \"{BuyPlateId}\" is part of the now active entry in beacon line \"{BeaconLine.BeaconLineId}\" but is not registered", ?Level := log_level.Warning)
                }
            }

            TargetBuyplate := if (
                LockedTargetBuyPlateId := BeaconLine.m_LockedTargetBuyPlate?
                LineEntry.ContainsBuyPlate[LockedTargetBuyPlateId]
                BuyPlate := FindBuyPlateById[ProgressLine, LockedTargetBuyPlateId]
                BuyPlate.IsEnabled[]
                not BuyPlate.IsUnlocked[]
            ) {
                Some(BuyPlate)
            } else {
                TopRanking := BestRankingBeaconLineEntryBuyPlate(ProgressLine, LineEntry)
                TopRankingStr := if (BuyPlate := TopRanking?) {
                    ToString(BuyPlate)
                } else {
                    "None"
                }
                Log(MsgRankedBeaconLineBuyPlates(BeaconLine.BeaconLineId, TopRankingStr), ?Level := log_level.Debug)
                TopRanking
            }

            if (BuyPlate := TargetBuyplate?) {
                UnmetCost := if (Agent := ProgressLine.AccessControl?.ListAllowedAgents()[0]) {
                    InvertLogic(BuyPlate.DoesAgentMeetsCostCached(Agent))
                } else {
                    false
                }
                if (not UnmetCost?) {
                    set BeaconLine.m_LockedTargetBuyPlate = Some(BuyPlate.GetRoutingId())
                }

                if (
                    BuyPlate = BeaconLine.m_MapIndicatorState?.BuyPlate
                    UnmetCost = BeaconLine.m_MapIndicatorState?.UnmetCost
                ) {
                    return
                }
                set MapIndicatorPointTo = Some(tycoon_manager_beacon_line_buy_plate_state{
                    BuyPlate := BuyPlate
                    UnmetCost := UnmetCost
                })
            }
        }

        if (
            BuyPlateState := MapIndicatorPointTo?
            Position := BuyPlateState.BuyPlate.GetPosition(?ExtraOffset := p_BeaconLinesMapIndicatorOffset)
            Transform := transform{
                Translation := Position
                Rotation := IdentityRotation()
                Scale := MakeUniformVec3(1.0)
            }
            MapIndicatorVariant := if (BuyPlateState.UnmetCost?, BeaconLine.MapIndicatorUnmetCost?) {
                tycoon_manager_beacon_line_map_indicator_variant.UnmetCost
            } else {
                tycoon_manager_beacon_line_map_indicator_variant.Primary
            }
            MapIndicator := case(MapIndicatorVariant):
                tycoon_manager_beacon_line_map_indicator_variant.Primary => BeaconLine.MapIndicator
                tycoon_manager_beacon_line_map_indicator_variant.UnmetCost => BeaconLine.MapIndicatorUnmetCost?
            MapIndicator.TeleportTo[Transform]
        ) {
            set BeaconLine.m_MapIndicatorState = Some(BuyPlateState)

            if (PrevVariant := BeaconLine.m_MapIndicatorEnabled?, PrevVariant <> MapIndicatorVariant) {
                BeaconLine.DisablePulse(ProgressLine)
                if (PrevMapIndicator := BeaconLine.GetActiveMapIndicator[]) {
                    PrevMapIndicator.Disable()
                }
                set BeaconLine.m_MapIndicatorEnabled = false
            }
            if (not BeaconLine.m_MapIndicatorEnabled?) {
                set BeaconLine.m_MapIndicatorEnabled = Some(MapIndicatorVariant)
                MapIndicator.Enable()
                Log(MsgBeaconLineMapIndicatorVariantEnabled(BeaconLine.BeaconLineId, ToString(MapIndicatorVariant)),
                    ?Level := log_level.Verbose)
            }
            if (not BeaconLine.m_MapIndicatorPulseActive?) {
                set BeaconLine.m_MapIndicatorPulseActive = true
                for (
                    Agent:ProgressLine.AccessControl?.ListAllowedAgents()
                    ShouldSeeBeaconLine[Agent]
                ) {
                    ActivateObjectivePulseWithWarmup(Agent, BeaconLine)
                }
            } else {
                for (Agent->__1:BeaconLine.m_MapIndicatorPulseToAgents) {
                    for (__2 := 0..5) {
                        ActivateObjectivePulseWithWarmup(Agent, BeaconLine)
                    }
                }
            }
        } else {
            set BeaconLine.m_MapIndicatorState = false

            BeaconLine.DisablePulse(ProgressLine)
            if (MapIndicator := BeaconLine.GetActiveMapIndicator[], MapIndicatorVariant := BeaconLine.m_MapIndicatorEnabled?) {
                MapIndicator.Disable()
                set BeaconLine.m_MapIndicatorEnabled = false
                Log(MsgBeaconLineMapIndicatorVariantDisabled(BeaconLine.BeaconLineId, ToString(MapIndicatorVariant)),
                    ?Level := log_level.Verbose)
            }
        }

    MsgRankedBeaconLineBuyPlates<private><localizes>(BeaconLineId:string, BuyPlateStr:string):message=
        "Beacon line \"{BeaconLineId}\" elected {BuyPlateStr} as best ranking buy plate"
    MsgBeaconLineMapIndicatorVariantEnabled<private><localizes>(BeaconLineId:string, VariantStr:string):message=
        "Beacon line \"{BeaconLineId}\" enabled map indicator variant {VariantStr}"
    MsgBeaconLineMapIndicatorVariantDisabled<private><localizes>(BeaconLineId:string, VariantStr:string):message=
        "Beacon line \"{BeaconLineId}\" disabled map indicator"

    IsWithinBeaconLineHideArea<private>(Agent:agent)<transacts><decides>:void=
        var Within: logic = false
        for (BeaconLineArea:p_BeaconLinesHideAreas, BeaconLineArea.m_InVolume[Agent]) {
            set Within = true
        }
        Within?

    ShouldSeeBeaconLine<private>(Player:agent)<decides><transacts>: void ={
        if (p_BeaconLinesHideAreasInvert?) {
            IsWithinBeaconLineHideArea[Player]
        } else {
            not IsWithinBeaconLineHideArea[Player]
        }
    }

    SyncPlayerBeaconLines<internal>(Player:player):void=
        if (
            PlayerState := m_Players[Player]
            ProgressLine := PlayerState.ProgressLine
            AccessControl := ProgressLine.AccessControl?
            AccessControl.IsAllowed[Player]
            ShouldSeeBeaconLine[Player]
        ) {
            for (
                BeaconLine:ProgressLine.BeaconLines
                BeaconLine.m_MapIndicatorPulseActive?
            ) {
                if (
                    BeaconLine.m_MapIndicatorPulseToAgents[Player]
                    MapIndicator := BeaconLine.GetActiveMapIndicator[]
                ) {
                    MapIndicator.ActivateObjectivePulse(Player)
                } else if (not BeaconLine.m_MapIndicatorPulseToAgents[Player]) {
                    ActivateObjectivePulseWithWarmup(Player, BeaconLine)
                }
            }
            Log(MsgSyncPlayerBeaconLines(Player, "active"), ?Level := log_level.Verbose)
        } else if (
            PlayerState := m_Players[Player]
            ProgressLine := PlayerState.ProgressLine
        ) {
            for (
                BeaconLine:ProgressLine.BeaconLines
                BeaconLine.m_MapIndicatorPulseToAgents[Player]
                MapIndicator := BeaconLine.GetActiveMapIndicator[]
            ) {
                MapIndicator.DeactivateObjectivePulse(Player)
                set BeaconLine.m_MapIndicatorPulseToAgents = BeaconLine.m_MapIndicatorPulseToAgents.RemoveByKey(Player)(0)
            }
            Log(MsgSyncPlayerBeaconLines(Player, "inactive"), ?Level := log_level.Verbose)
        }

    MsgSyncPlayerBeaconLines<private><localizes>(Agent:agent, StateStr:string):message=
        "Syncing beacon lines for '{Agent}'. New State: {StateStr}"

    ActivateObjectivePulseWithWarmup<private>(
        Agent:agent,
        BeaconLine:tycoon_manager_progress_line_beacon_line
    ):void=
        RemoveRes := BeaconLine.m_MapIndicatorPulseToAgents.RemoveByKey(Agent)
        set BeaconLine.m_MapIndicatorPulseToAgents = RemoveRes(0)
        if (PrevAbortEvent := RemoveRes(1)?) {
            PrevAbortEvent.Signal()
        }

        AbortEvent := event(){}
        if (set BeaconLine.m_MapIndicatorPulseToAgents[Agent] = AbortEvent) {
            spawn:
                ActivateObjectivePulseWithWarmup0(Agent, BeaconLine, AbortEvent)
        }

    ActivateObjectivePulseWithWarmup0<private>(
        Agent:agent,
        BeaconLine:tycoon_manager_progress_line_beacon_line,
        AbortEvent:event()
    )<suspends>:void=
        race:
            AbortEvent.Await()
            PraiseOurLordThatThePulseActivates(Agent, BeaconLine)

    PraiseOurLordThatThePulseActivates<private>(
        Agent:agent,
        BeaconLine:tycoon_manager_progress_line_beacon_line
    )<suspends>:void=
        if (MapIndicator := BeaconLine.GetActiveMapIndicator[]) {
            MapIndicator.DeactivateObjectivePulse(Agent)
        }
        Sleep(0.0)
        var It: int = 0
        var Interval: float = 5.0
        loop {
            if (
                BeaconLine.m_MapIndicatorPulseToAgents[Agent]
                MapIndicator := BeaconLine.GetActiveMapIndicator[]
            ) {
                # let's out-random the game's randomness /s
                OutrandomingTheVeryBadGame := GetRandomInt(10, 20)
                for (__ := 1..OutrandomingTheVeryBadGame) {
                    MapIndicator.ActivateObjectivePulse(Agent)
                    Sleep(0.0)
                }
            } else {
                return
            }
            Sleep(Interval)
            set Interval = Min(20.0, Interval + 5.0)
            set It += 1
            if (It > 5) {
                return
            }
        }

    FindBuyPlateById<private>(ProgressLine:tycoon_manager_progress_line, BuyPlateId:string)<transacts><decides>:buy_plate_device=
        ProgressLine.m_BuyPlatesBackrefCache[BuyPlateId] or FindBuyPlateById0[ProgressLine, BuyPlateId]

    FindBuyPlateById0<private>(ProgressLine:tycoon_manager_progress_line, BuyPlateId:string)<transacts><decides>:buy_plate_device=
        var Result: ?buy_plate_device = false
        if (DeviceBackrefs := ProgressLine.AccessControl?.GetUsingDevices[]) {
            for (
                Device:DeviceBackrefs,
                not (Result?),
                BuyPlate := buy_plate_device[Device],
                BuyPlate.GetRoutingId() = BuyPlateId
            ) {
                set Result = Some(BuyPlate)
            }
        }
        Result?

    OnDeviceRegistered<internal>(ProgressLine:tycoon_manager_progress_line, Device:creative_device):void=
        if (BuyPlate := buy_plate_device[Device]) {
            BuyPlateId := BuyPlate.GetRoutingId()
            if. set ProgressLine.m_BuyPlatesBackrefCache[BuyPlateId] = BuyPlate
            if (Config := m_BuyplatesConfigs[BuyPlate.GetRoutingId()]) {
                Closure := tycoon_manager_configured_buy_plate_closure{
                    Device := Self
                    ProgressLine := ProgressLine
                    Config := Config
                }
                if. BuyPlate.SubscribeToEvent["unlock", Closure.OnUnlock]
                if. BuyPlate.SubscribeToEvent["enable", Closure.OnEnable]

                if (Cost := Config.Cost?) {
                    CostOverride := tycoon_manager_buy_plate_cost_override{
                        Device := Self
                        Cost := Cost
                    }
                    BuyPlate.SetCostOverride(CostOverride)
                } else if (BuyPlate.p_ExpectsCostOverride? or p_AlwaysSetCostOverride?) {
                    BuyPlate.SetCostOverride(buy_plate_free_cost_override{})
                }
            } else {
                Closure := tycoon_manager_any_buy_plate_closure{
                    Device := Self
                    ProgressLine := ProgressLine
                    BuyPlateId := BuyPlate.GetRoutingId()
                }
                if. BuyPlate.SubscribeToEvent["unlock", Closure.OnUnlock]
                if. BuyPlate.SubscribeToEvent["enable", Closure.OnEnable]
            }
            if (m_Initialized?) {
                for (
                    BeaconLineId->BeaconLineConfig:m_TycoonConfig.BeaconLines
                    LineEntry := BeaconLineConfig.InOrderBuyPlates[0]
                    LineEntry.ContainsBuyPlate[BuyPlateId]
                ) {
                    for (BeaconLine:ProgressLine.BeaconLines, BeaconLine.BeaconLineId = BeaconLineId) {
                        UpdateBeaconLine(ProgressLine, BeaconLine)
                    }
                }
            }
        }
        if (PersistenceRecorder := ProgressLine.m_PersistenceRecorder?) {
            if (BuyPlate := buy_plate_device[Device], BuyPlateId := BuyPlate.GetRoutingId()) {
                FixedIndex := if (Index := m_BuyplatesConfigs[BuyPlateId].m_StorageIndex) {
                    Index
                } else {
                    Log("Tycoon manager is to persist buyplate \"{BuyPlateId}\" but its ordering cannot be ensured as it is not present in the config", ?Level := log_level.Warning)
                    -1
                }
                PersistenceRecorder.RegisterBuyPlateToRecord(BuyPlate, ?FixedIndex := FixedIndex)
            } else {
                PersistenceRecorder.RegisterDeviceToRecord(Device, ?Silent := true)
            }
        }

    OnAnyBuyPlateEnabled<internal>(ProgressLine:tycoon_manager_progress_line, BuyplateId:string):void=
        for (
            BeaconLine:ProgressLine.BeaconLines
            BeaconLineConfig := m_TycoonConfig.BeaconLines[BeaconLine.BeaconLineId]
            LineEntry := BeaconLineConfig.InOrderBuyPlates[BeaconLine.m_CurrentEntry]
            LineEntry.ContainsBuyPlate[BuyplateId]
        ) {
            UpdateBeaconLine(ProgressLine, BeaconLine)
        }
        if (AccessControl := ProgressLine.AccessControl?) {
            UpdateCostMetTracking(ProgressLine, AccessControl)
        }

    OnAnyBuyPlateUnlocked<internal>(ProgressLine:tycoon_manager_progress_line, BuyplateId:string, Agent:?agent):void=
        for (
            BeaconLine:ProgressLine.BeaconLines
            BeaconLineConfig := m_TycoonConfig.BeaconLines[BeaconLine.BeaconLineId]
        ) {
            if (
                (LineEntry := BeaconLineConfig.InOrderBuyPlates[BeaconLine.m_CurrentEntry],
                 LineEntry.ContainsBuyPlate[BuyplateId])
            ) {
                set BeaconLine.m_LockedTargetBuyPlate = None()
                UpdateBeaconLine(ProgressLine, BeaconLine)
            } else if (
                BeaconLine.m_LockedTargetBuyPlate?
                BeaconLineConfig.ContainsBuyPlate[BuyplateId]
            ) {
                set BeaconLine.m_LockedTargetBuyPlate = None()
                UpdateBeaconLine(ProgressLine, BeaconLine)
            }
        }

    OnConfiguredBuyPlateUnlocked<internal>(ProgressLine:tycoon_manager_progress_line, Config:tycoon_manager_buyplates_entry, Agent:?agent):void=
        if (Generation := Config.Generates?) {
            if (Ticket := ProgressLine.m_AutoGenerationTickets[Generation.ResourceId]) {
                Ticket.ModifyRate(ToFloat(Generation.Amount))
            } else if (
                Ticket := ProgressLine.AutoGenerationManager?.CreateTicket[Generation.ResourceId, false]
                set ProgressLine.m_AutoGenerationTickets[Generation.ResourceId] = Ticket
            ) {
                Ticket.SetRate(ToFloat(Generation.Amount))
            } else {
                Log("Failed to create auto generation ticket for resource \"{Generation.ResourceId}\" for generation of progress line \"{ProgressLine.Name}\". Is auto generation manager set?", ?Level := log_level.Error)
            }
        }
        if (Config.Unlocks.Length > 0, DeviceBackrefs := ProgressLine.AccessControl?.GetUsingDevices[]) {
            var Unlocked: [string]logic = map{}
            for (
                Device:DeviceBackrefs
                DeviceRoutingId := aware_device[Device].DeviceMeta().GetRoutingId()
            ) {
                for (
                    Unlock:Config.Unlocks
                    DeviceRoutingId = Unlock.TargetRoutingId
                    (not Unlocked[DeviceRoutingId]) or (not p_RoutingIdSingleMatch?) or (Unlock.Payload?)
                    set Unlocked[DeviceRoutingId] = true
                ) {
                    TriggerData := Unlock.Payload? or "enable"
                    Device.AttemptCustomTrigger(TriggerData, Agent, ?Context := "progress line \"{ProgressLine.Name}\"")
                }
            }
            for (Unlock:Config.Unlocks, not Unlocked[Unlock.TargetRoutingId]) {
                Log("Failed to process unlock \"{Unlock}\" after \"{Config.BuyplateId}\" unlocked for progress line \"{ProgressLine.Name}\". No matching device found in backrefs.", ?Level := log_level.Error)
            }
        }

    OnPlayerSaveSlotSelected<private>(ProgressLine:tycoon_manager_progress_line, Event:player_save_slot_selected_event):void=
        PlayerSaveSlotSelectedEvent.Signal(ProgressLine, Event)

        EventName := if ((Event.AlreadyPlayed?) and (Event.Reset?)) {
            "claim_reset"
        } else if (Event.AlreadyPlayed?) {
            "claim_continue"
        } else {
            "claim_first"
        }
        p_AnalyticsController.Emit(Event.Player, EventName)

    OnRebirthNextLevelTriggerCalled<private>(Agent:?agent):void=
        Player := if (l_Player := player[Agent?]) {
            l_Player
        } else {
            Log("Tycoon manager rebirth next level OnTrigger does nothing without a player instigator", ?Level := log_level.Warning)
            return
        }
        PromptRebirth(Player)

    PromptRebirth<public>(Player:player):void=
        if (RebirthConfig := m_TycoonConfig.Rebirth?, PlayerState := GetPlayerState[Player]) {
            RebirthLevel := PlayerState.ProgressLine.m_RebirthLevel
            if (RebirthLevel >= RebirthConfig.MaxLevel) {
                Log(MsgRebirthMaxLevelOnTrigger(Player), ?Level := log_level.Debug)
                return
            }
            DialogIndex := if (p_RebirthNextLevelDialogPerLevel?) {
                if (p_RebirthNextLevelDialogPerLevelDefaultLast?) {
                    var Index: int = RebirthLevel
                    loop {
                        if (Index <= 0 or p_RebirthNextLevelDialog[Index]) {
                            break
                        }
                        set Index -= 1
                    }
                    Index
                } else {
                    RebirthLevel
                }
            } else {
                0
            }
            if (DialogDevice := p_RebirthNextLevelDialog[DialogIndex]) {
                StateRes := if (l_StateRes := m_RebirthDialogs[DialogIndex]) {
                    l_StateRes
                } else {
                    InitialTemplate := DialogDevice.GetDescriptionText()
                    Wrapper := dialog_device_wrapper{
                        DialogDevice := DialogDevice
                    }
                    l_StateRes := (InitialTemplate, Wrapper)
                    if (set m_RebirthDialogs[DialogIndex] = l_StateRes) {}
                    l_StateRes
                }
                var Template: string = StateRes(0)
                set Template = Template
                    .ReplaceAll("%CurrentLevel%", ToString(RebirthLevel))
                    .ReplaceAll("%NextLevel%", ToString(RebirthLevel + 1))
                if (NextLevelConfig := RebirthConfig.GetLevelConfig[RebirthLevel + 1]) {
                    NumberFormatter := p_RebirthPlaceholdersMultiplierFormatter.CreateFormatter()
                    for (ResourceId->Multiplier:NextLevelConfig.ResourceMultipliers) {
                        set Template = Template.ReplaceAll("%Multiplier:{ResourceId}%", NumberFormatter.Format(Multiplier))
                    }
                }
                Title := DialogDevice.GetTitleText()
                DialogDevice.SetTitleText(ToMessage(Title)) 
                DialogDevice.SetDescriptionText(ToMessage(Template))
                Closure := tycoon_manager_rebirth_dialog_closure{
                    Device := Self
                    Player := Player
                    ProgressLine := PlayerState.ProgressLine
                    ExpectedCurrentLevel := RebirthLevel
                }
                StateRes(1).Show(Player, Some(Closure.Call))
            } else {
                RebirthPlayer(Player)
            }
        } else {
            Log("Tycoon manager rebirth next level OnTrigger does nothing when rebirth is not configured", ?Level := log_level.Warning)
        }

    GetPlayerState<internal>(Player:player)<transacts><decides>:tycoon_manager_player_state=
        m_Players[Player]

    GetRebirthLevel<public>(Player:player)<transacts><decides>:int=
        m_Players[Player].ProgressLine.m_RebirthLevel

    RebirthPlayer<internal>(Player:player):void=
        if (RebirthConfig := m_TycoonConfig.Rebirth?, PlayerState := m_Players[Player]) {
            ProgressLine := PlayerState.ProgressLine
            if (ProgressLine.m_RebirthLevel >= RebirthConfig.MaxLevel) {
                Log(MsgRebirthMaxLevel(Player), ?Level := log_level.Debug)
                return
            }
            set ProgressLine.m_RebirthLevel += 1
            RebirthLevel := ProgressLine.m_RebirthLevel
            Log(MsgRebirth(Player, RebirthLevel, "manual"), ?Level := log_level.Debug)
            if (PersistenceRecorder := PlayerState.ProgressLine.m_PersistenceRecorder?) {
                # save rebirth level
                if (Handle := PlayerState.ProgressLine.m_RebirthRecordableHandle?) {
                    Handle.Save(Player)
                } else {
                    PersistenceRecorder.SaveAgent(Player)
                }
                # reset persistence state
                PersistenceRecorder.ResetAgent(Player, tycoon_manager_rebirth_persistence_reset_query{
                    SourceDevice := Self
                    ResetFilter := Some(RebirthPersistenceResetFilter)
                })
            }
            DoRollback(PlayerState.ProgressLine, Some(Player))
            if (ClaimBuyplate := ProgressLine.ClaimBuyplate?) {
                ClaimBuyplate.Unlock(Player, ?UserTriggered := false, ?FirstUnlock := false, ?Force := true)
            }
            for (TriggerConfig:p_RebirthResetTriggers, TriggerConfig.IsApplicable[RebirthLevel]) {
                TriggerConfig.Trigger.Trigger(Player)
            }
            for (TriggerConfig:ProgressLine.RebirthResetTriggers, TriggerConfig.IsApplicable[RebirthLevel]) {
                TriggerConfig.Trigger.Trigger(Player)
            }
            if (m_RebirthListeners.Length > 0) {
                Event := tycoon_rebirth_event{
                    Player := Player
                    NewRebirthLevel := RebirthLevel
                    Virtual := false
                }
                for (Listener:m_RebirthListeners) {
                    Listener(Event)
                }
            }
            if (PersistenceRecorder := PlayerState.ProgressLine.m_PersistenceRecorder?) {
                # save persistence in case triggers caused relevant changes
                PersistenceRecorder.SaveAgent(Player)
            }
            EnsureRebirthEffects(PlayerState.ProgressLine)
            if (AccessControl := ProgressLine.AccessControl?) {
                UpdateCostMetTracking(ProgressLine, AccessControl)
            }
        }

    ForceVirtualRebirth<public>(Player:player):void=
        if (RebirthConfig := m_TycoonConfig.Rebirth?, PlayerState := m_Players[Player]) {
            ProgressLine := PlayerState.ProgressLine
            if (ProgressLine.m_RebirthLevel >= RebirthConfig.MaxLevel) {
                Log(MsgRebirthMaxLevel(Player), ?Level := log_level.Debug)
                return
            }
            set ProgressLine.m_RebirthLevel += 1
            RebirthLevel := ProgressLine.m_RebirthLevel
            Log(MsgRebirth(Player, RebirthLevel, "forced virtual"), ?Level := log_level.Debug)
            if (PersistenceRecorder := PlayerState.ProgressLine.m_PersistenceRecorder?) {
                # save rebirth level
                if (Handle := PlayerState.ProgressLine.m_RebirthRecordableHandle?) {
                    Handle.Save(Player)
                } else {
                    PersistenceRecorder.SaveAgent(Player)
                }
            }
            if (m_RebirthListeners.Length > 0) {
                Event := tycoon_rebirth_event{
                    Player := Player
                    NewRebirthLevel := RebirthLevel
                    Virtual := true
                }
                for (Listener:m_RebirthListeners) {
                    Listener(Event)
                }
            }
            EnsureRebirthEffects(PlayerState.ProgressLine)
        }

    RebirthPersistenceResetFilter<private>(Recordable:persistence_recordable_base)<transacts><decides>:void=
        not pet_controller_device[Recordable]

    EnsureRebirthEffects<internal>(ProgressLine:tycoon_manager_progress_line):void=
        if (HudManager := m_RebirthHudManager?) {
            HudManager.Refresh()
        }
        if (RebirthConfig := m_TycoonConfig.Rebirth?, AutoGenerationManager := ProgressLine.AutoGenerationManager?) {
            ResourceMultipliers := if (LevelConfig := RebirthConfig.GetLevelConfig[ProgressLine.m_RebirthLevel]) {
                LevelConfig.ResourceMultipliers
            } else {
                map{}
            }
            for (ResourceId->Multiplier:ResourceMultipliers) {
                AutoGenerationManager.GetResourceMultiplierInstance(ResourceId).SetMultiplier(p_RebirthMultipliersChannel, Multiplier)
                if (set ProgressLine.m_SetResourceMultipliers[ResourceId] = Multiplier) {}
            }
            # remove multipliers that are no longer wanted
            for (ResourceId->__:ProgressLine.m_SetResourceMultipliers, not ResourceMultipliers[ResourceId]) {
                AutoGenerationManager.GetResourceMultiplierInstance(ResourceId).SetMultiplier(p_RebirthMultipliersChannel, None())
            }
        }
        if (Billboard := ProgressLine.RebirthBillboard[0]) {
            if (ProgressLine.m_RebirthLevel > 0 or not (ProgressLine.RebirthBillboardHideOnZero?)) {
                Text := ProgressLine.RebirthBillboardTemplate
                    .ReplaceAll("%CurrentLevel%", ToString(ProgressLine.m_RebirthLevel))
                Billboard.ShowText()
                Billboard.SetText(ToMessage(Text))
            } else {
                Billboard.HideText()
                Billboard.SetText(ToMessage(""))
            }
        }

    MsgRebirth<private><localizes>(Player:player, NewLevel:int, Context:string):message=
        "Rebirthing '{Player}' to level {NewLevel} ({Context})"
    MsgRebirthMaxLevel<private><localizes>(Player:player):message=
        "Ignoring next rebirth level for '{Player}' as they reached max level"
    MsgRebirthMaxLevelOnTrigger<private><localizes>(Player:player):message=
        "Ignoring next rebirth level OnTrigger for '{Player}' as they reached max level"

    IsPetUnlocked<override>(Player:player, PetId:string)<transacts><decides>:void=
        RebirthLevel := m_Players[Player].ProgressLine.m_RebirthLevel
        RebirthLevel >= m_RebirthPets[PetId]

    IsAttachmentUnlocked<override>(Player:player, AttachmentKindId:string, Attachment:string)<transacts>:attachment_dynamic_unlock_state=
        RebirthLevel := if (l_RebirthLevel := m_Players[Player].ProgressLine.m_RebirthLevel) {
            l_RebirthLevel
        } else {
            return attachment_dynamic_unlock_state.NotHandled
        }
        AttachmentSlug := AttachmentKindId + ":" + Attachment
        if (RebirthLevel >= m_RebirthAttachments[AttachmentSlug]) {
            attachment_dynamic_unlock_state.Unlocked
        } else {
            # could be Locked, but NotHandled leaves the option to unlock it alternatively before rebirth level
            attachment_dynamic_unlock_state.NotHandled
        }

tycoon_manager_beacon_line_area_closure := class<final>:
    Device: tycoon_manager_device
    Config: tycoon_manager_beacon_line_area_config

    OnAgentEntered<internal>(Agent:agent):void=
        if. set Config.m_InVolume[Agent] = void
        if (Player := player[Agent]) {
            Device.SyncPlayerBeaconLines(Player)
        }

    OnAgentExited<internal>(Agent:agent):void=
        set Config.m_InVolume = Config.m_InVolume.RemoveByKey(Agent)(0)
        if (Player := player[Agent]) {
            Device.SyncPlayerBeaconLines(Player)
        }

tycoon_manager_player_closure := class:
    Device: tycoon_manager_device
    ProgressLine: tycoon_manager_progress_line

    OnPlayerAdded(Player:player):void=
        Device.OnPlayerAdded(ProgressLine, Player)

    OnPlayerRemoved(Player:player, GameQuit:logic):void=
        Device.OnPlayerRemoved(ProgressLine, Player)

## Persistence {{{
tycoon_manager_player_persistence_load_closure := class:
    Callback: type{__(:tycoon_manager_progress_line,:player_save_slot_selected_event):void}
    ProgressLine: tycoon_manager_progress_line

    Call(Event:player_save_slot_selected_event):void=
        Callback(ProgressLine, Event)

tycoon_manager_rebirth_dummy_device := class(aware_device, persistence_recordable, persistence_resetable):
    TycoonManager: tycoon_manager_device
    ProgressLine: tycoon_manager_progress_line

    DeviceMeta<override>()<transacts>: aware_device_meta ={
        aware_device_meta{
            DeviceName := "tycoon_manager_rebirth"
        }
    }

    GetPersistenceControl<override>()<transacts><decides>:persistence_recordable_control=
        None()?

    PersistenceReserve<override>(StorageMem:storage_memory)<transacts><decides>:tycoon_manager_rebirth_persistence_state=
        TycoonManager.p_EnablePersistence?
        tycoon_manager_rebirth_persistence_state{
            RebirthLevelChunk := StorageMem.Reserve[TycoonManager.p_RebirthLevelReserveBits, "tycoon_rebirth_level"]
        }

    PersistenceSave<override>(i_State:persistence_recordable_state, Agent:agent, SaveCause:persistence_save_cause):void=
        if (SaveCause = persistence_save_cause.AutoSave) {
            return
        }
        if (
            State := tycoon_manager_rebirth_persistence_state[i_State]
            PlayerState := TycoonManager.GetPlayerState[player[Agent]]
            PlayerState.ProgressLine.m_Index = ProgressLine.m_Index
        ) {
            State.RebirthLevelChunk.WriteInt(Agent, ProgressLine.m_RebirthLevel)
        }

    PersistenceLoad<override>(i_State:persistence_recordable_state, Agent:agent, Query:persistence_load_query):void=
        if (
            State := tycoon_manager_rebirth_persistence_state[i_State]
            PlayerState := TycoonManager.GetPlayerState[player[Agent]]
            PlayerState.ProgressLine.m_Index = ProgressLine.m_Index
        ) {
            SavedLevel := State.RebirthLevelChunk.ReadInt(Agent)
            PrevRebirthLevel := ProgressLine.m_RebirthLevel
            set ProgressLine.m_RebirthLevel = if (Query.Overwrite?) {
                SavedLevel
            } else {
                Max(ProgressLine.m_RebirthLevel, SavedLevel)
            }
            if (PrevRebirthLevel <> ProgressLine.m_RebirthLevel) {
                TycoonManager.EnsureRebirthEffects(ProgressLine)
            }
        }

    PersistenceReset<override>(i_State:persistence_recordable_state, Agent:agent, Query:persistence_reset_query):void=
        if (tycoon_manager_rebirth_persistence_reset_query[Query]) {
            Log("Skipping reset of rebirth persistence state because it is initialized by a rebirth", ?Level := log_level.Debug)
        } else {
            i_State.Reset(Agent)
            PersistenceLoad(i_State, Agent, persistence_load_query{
                Overwrite := true
                FirstLoad := false
            })
        }

tycoon_manager_rebirth_persistence_reset_query<public> := class(persistence_reset_query):

tycoon_manager_rebirth_persistence_state<public> := class(persistence_recordable_state):
    RebirthLevelChunk<public>: storage_memory_chunk

    Reset<override>(Agent:agent):void=
        RebirthLevelChunk.Reset(Agent)
## }}}

## Buy plate features {{{
tycoon_manager_buyplate_cost_met_reflection_closure := class:
    Device: tycoon_manager_device
    ProgressLine: tycoon_manager_progress_line

    OnResourceTake(Event:resource_take_event):void=
        if (AccessControl := ProgressLine.AccessControl?, AccessControl.IsAllowed[Event.Agent?]) {
            Device.UpdateCostMetTracking(ProgressLine, AccessControl)
        }

tycoon_manager_claim_buyplate_middleware := class(buy_plate_middleware):
    Device: tycoon_manager_device
    ProgressLine: tycoon_manager_progress_line

    MayTryUnlockBuyplate<override>(Agent:agent, BuyPlate:buy_plate_device):logic=
        if (Player := player[Agent], Device.GetPlayerState[Player]) {
            for (HUDMessage:Device.p_ClaimFailHUDMessage) {
                HUDMessage.Show(Agent)
            }
            false
        } else {
            true
        }

    OnUnlockBuyplate<override>(Agent:agent, BuyPlate:buy_plate_device):void=
        if (Player := player[Agent]) {
            set BuyPlate.m_InjectCinematicPlayer = None() # to not play again on rebirth
            Device.ClaimProgressLineFromBuyplate(ProgressLine, Player)
        }
        # if:
        #     FortCharacter := Agent.GetFortCharacter[]
        #     Transform := FortCharacter.GetTransform()
        #     FortCharacter.TeleportTo[Transform.Translation, Transform.Rotation]
        # then:
        #     Log("Teleported '{Agent}' to themselves on claim because the Lord asked for it", ?Level := log_level.Debug)

tycoon_manager_buy_plate_cost_override := class(buy_plate_cost_override):
    Device: tycoon_manager_device
    Cost: tycoon_manager_buyplates_entry_resource

    TryUnlock<override>(Agent:agent, Silent:logic):logic=
        if (Cost.Amount <= 0) {
            return true
        }
        Notify := InvertLogic(Silent)
        if (Cost.CheckDontTake?) {
            Amount := Device.p_ResourceManager.GetAmount(Agent, Cost.ResourceId)
            if (Amount >= Cost.Amount) {
                return true
            } else if (Notify?) {
                Device.p_ResourceManager.SendNotEnoughNotification(Agent, Cost.ResourceId)
            }
        } else {
            Success := Device.p_ResourceManager.Take(Some(Agent), Cost.ResourceId, Cost.Amount,
                ?NotifyNotEnough := Notify, ?SendFeedback := Notify)
            return Success
        }
        false

    HasEnough<override>(Agent:agent):logic=
        if (Cost.Amount <= 0) {
            return true
        }
        Amount := Device.p_ResourceManager.GetAmount(Agent, Cost.ResourceId)
        if (Amount >= Cost.Amount) {
            true
        } else {
            false
        }
## }}}

tycoon_manager_device_register_closure := class:
    Device: tycoon_manager_device
    ProgressLine: tycoon_manager_progress_line

    Call(RegisteredDevice:creative_device):void=
        Device.OnDeviceRegistered(ProgressLine, RegisteredDevice)

## Buy plates closures {{{
tycoon_manager_any_buy_plate_closure := class:
    Device: tycoon_manager_device
    ProgressLine: tycoon_manager_progress_line
    BuyPlateId: string

    OnUnlock(Agent:?agent):void=
        Device.OnAnyBuyPlateUnlocked(ProgressLine, BuyPlateId, Agent)

    OnEnable(Agent:?agent):void=
        Device.OnAnyBuyPlateEnabled(ProgressLine, BuyPlateId)

tycoon_manager_configured_buy_plate_closure := class:
    Device: tycoon_manager_device
    ProgressLine: tycoon_manager_progress_line
    Config: tycoon_manager_buyplates_entry

    OnUnlock(Agent:?agent):void=
        Device.OnConfiguredBuyPlateUnlocked(ProgressLine, Config, Agent)
        Device.OnAnyBuyPlateUnlocked(ProgressLine, Config.BuyplateId, Agent)

    OnEnable(Agent:?agent):void=
        Device.OnAnyBuyPlateEnabled(ProgressLine, Config.BuyplateId)
## }}}
## Rebirth UI {{{
tycoon_manager_rebirth_dialog_closure := class:
    Device: tycoon_manager_device
    Player: player

    ProgressLine: tycoon_manager_progress_line
    ExpectedCurrentLevel: int

    Call(Event:dialog_response_event):void=
        Log("Rebirth dialog event {Event}", ?Level := log_level.Debug)
        if (
            Event.ResponseType = dialog_response_type.Clicked
            Event.ButtonIndex? = Device.p_RebirthNextLevelDialogConfirmButtonIndex
            player[Event.Agent] = Player
            PlayerState := Device.GetPlayerState[Player]
            PlayerState.ProgressLine.m_Index = ProgressLine.m_Index
            PlayerState.ProgressLine.m_RebirthLevel = ExpectedCurrentLevel
        ) {
            Device.RebirthPlayer(Player)
        }

tycoon_manager_rebirth_hud_manager := class(hud_manager_base):
    Device: tycoon_manager_device
    Config: tycoon_manager_rebirth_config

    CreateHUD<override>(Player:player):?hud_manager_player=
        Image := CreateDefaultingTextureBlock(Device.p_RebirthHUDImageSize, None())
        Canvas := canvas{
            Slots := array{
                Device.p_RebirthHUDImagePlacement.ToCanvasSlot(Image.Widget())
            }
        }
        Element := tycoon_manager_rebirth_hud_image{
            Widget := Canvas
            Manager := Self
            Player := Player
            Image:= Image
        }
        Some(MakeSingleElementHUDPlayer(Player, Canvas, Element))

tycoon_manager_rebirth_hud_image := class(hud_element):
    Manager: tycoon_manager_rebirth_hud_manager
    Player: player

    Image: defaulting_texture_block

    var m_LastLevel: ?int = false

    Update<override>():void=
        Level := option{Manager.Device.GetRebirthLevel[Player]}
        if (m_LastLevel = Level) {
            return
        }
        set m_LastLevel = Level
        if (
            l_Level := Level?
            TextureName := Manager.Config.GetLevelConfig[l_Level].HudImageTexture?
            Texture := AwareDefaultAssetsProvider[].GetTextureAsset[TextureName]
        ) {
            Show()
            Image.SetTexture(Texture)
            Log("Rebirth HUD image switching to level {l_Level}", ?Level := log_level.Debug)
        } else {
            Hide()
            Log("Rebirth HUD image switching to no level", ?Level := log_level.Debug)
        }
## }}}
