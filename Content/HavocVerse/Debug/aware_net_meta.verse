using { /Fortnite.com/Devices }

using { HavocVerse.Util }

aware_device<public> := interface:
    DeviceMeta<public>()<transacts>:aware_device_meta

aware_debug_device<public> := interface:
    DebugStats<public>()<transacts>:aware_debug_stats

aware_device_meta<public> := class:
    DeviceName<public>: string

    DevId<public>: string = ""
    RoutingIdOverride<public>: ?string = false
    CustomData<public>: ?[]string = false

    PreferSequentialLoad<public>: logic = false

    # Returns the Routing ID, defaulting to the Dev Id.
    GetRoutingId<public>()<transacts>:string=
        RoutingIdOverride? or DevId

aware_debug_stats<public> := class<final>:
    Length<public>: [string]int = map{}

operator'+'<public>(Input:aware_debug_stats, Other:aware_debug_stats)<transacts>:aware_debug_stats=
    var Length: [string]int = Input.Length
    for (Key->Value:Other.Length) {
        if. set Length[Key] = (Length[Key] + Value) or Value
    }
    aware_debug_stats{
        Length := Length
    }

ToString<public>(DeviceMeta:aware_device_meta)<transacts>:string=
    var Result: string = if (DeviceMeta.DeviceName.Length > 0) {
        DeviceMeta.DeviceName
    } else {
        "<unknown>"
    }
    if (DeviceMeta.DevId.Length > 0) {
        set Result += ":{DeviceMeta.DevId}"
    }

    var Fields: []string = array{}
    if (RoutingIdOverride := DeviceMeta.RoutingIdOverride?, RoutingIdOverride.Length > 0) {
        set Fields += array{"RoutingId:{RoutingIdOverride}"}
    }
    if (CustomData := DeviceMeta.CustomData?) {
        set Fields += array{"CustomData:{StringArrayToString(CustomData)}"}
    }
    if (Fields.Length > 0) {
        set Result += "\{"
        for (Index->Field:Fields) {
            if (Index > 0) {
                set Result += ", "
            }
            set Result += Field
        }
        set Result += "\}"
    }
    Result

ToString<public>(Device:creative_device)<transacts>:string=
    if (AwareDevice := aware_device[Device]) {
        ToString(AwareDevice.DeviceMeta())
    } else {
        "<unknown>"
    }

TooltipRoutingIdOverride<public><localizes>:message = "By default, p_RoutingId is the same as p_DevId. You may override this."
