using { /Fortnite.com/Devices }
using { /Fortnite.com/Devices/CreativeAnimation }
using { /Verse.org/Simulation }
using { /Verse.org/Random }
using { /UnrealEngine.com/Temporary/Diagnostics }
using { /UnrealEngine.com/Temporary/SpatialMath }

using { HavocVerse.Debug }
using { HavocVerse.Devices }
using { HavocVerse.Logging }
using { HavocVerse.Math }
using { HavocVerse.Persistence }
using { HavocVerse.Util }

prize_wheel_sound_variant<internal> := enum {
	Win
	WinBig
	Meh
}

stand_games_device := class<final><concrete>(creative_device, aware_device, custom_device, resetable) {
	@editable p_DevId: string = ""
    @editable{ToolTip := TooltipRoutingIdOverride} p_RoutingIdOverride: ?string = false
	@editable p_DevTp: ?teleporter_device = false

	@editable p_AccessControl: access_control_device = access_control_device{}
	@editable p_ResourceManager: resource_manager_device = resource_manager_device{}

	# DUCKS STAND
	@editable p_ResourceAmnt: int = 5
	@editable p_ResourceID: string = ""
	@editable p_DuckAccolade: accolades_device = accolades_device{}
	@editable p_DuckHitAudioPlayer: audio_player_device = audio_player_device{}
	@editable p_Ducks: []creative_prop = array{}
	@editable p_DucksPropMani: prop_manipulator_device = prop_manipulator_device{}
	@editable p_DuckedDuckPropAsset: creative_prop_asset = DefaultCreativePropAsset

	## PRIZE WHEEL
	@editable p_PrizeWheelProp: creative_prop = creative_prop{}
	@editable p_PrizeWheelTickerProp: ?creative_prop = false
	@editable p_PrizeWheelButtons: []button_device = array{}
	@editable p_WheelExplosive: explosive_device = explosive_device{}
	@editable p_WheelCreaturePlacer: creature_placer_device = creature_placer_device{}

	@editable p_WheelSoundEffectSpin: audio_player_device = audio_player_device{}
	@editable p_WheelSoundEffectWin: audio_player_device = audio_player_device{}
	@editable p_WheelSoundEffectWinBig: audio_player_device = audio_player_device{}
	@editable p_WheelSoundEffectMeh: audio_player_device = audio_player_device{}

	DeviceMeta<override>()<transacts>: aware_device_meta ={
		aware_device_meta{
			DeviceName := "stand_games"
			DevId := p_DevId
			RoutingIdOverride := p_RoutingIdOverride
		}
	}

	OnBegin<override>()<suspends>: void ={
		AwareInitDevice(Self, OnInit)
	}
	
	OnInit<private>()<suspends>: void ={
		p_AccessControl.RegisterUsingDevice(Self)

		if (Teleporter := p_DevTp?, Player := GetPlayspace().GetPlayers()[0]) {
			Teleporter.Teleport(Player)
		}

		if (p_Ducks.Length > 0) {
			p_DucksPropMani.DamagedEvent.Subscribe(OnDuckPropHit)
		} else {
			Log(Self, "The ducks are missing!")
		}

		for (Button:p_PrizeWheelButtons) {
			Button.InteractedWithEvent.Subscribe(OnPrizeWheelInteract)
		}

		ApplyInitialState()
	}

	ApplyInitialState<private>(): void ={
		set m_DuckIndex = -1
		for (DuckProp:p_Ducks) {
			DuckProp.Hide()
		}
		for (DuckedDuck:m_DuckedDucks, Prop := DuckedDuck?, Prop.IsValid[]) {
			Prop.Dispose()
		}
		set m_DuckedDucks = array{}

		p_PrizeWheelProp.Hide()
		if (TickerProp := p_PrizeWheelTickerProp?) {
			TickerProp.Hide()
		}
		for (Button:p_PrizeWheelButtons) {
			Button.Disable()
		}
	}

	Reset<override>(Agent:?agent): void ={
		if (AbortEvent := m_WheelSpinning?) {
			AbortEvent.Signal()
		}
		if:
			BaseRotation := m_WheelBaseRotation?
		    p_PrizeWheelProp.TeleportTo[p_PrizeWheelProp.GetTransform().WithRotation(BaseRotation)]
		ApplyInitialState()
	}

    CustomTrigger<override>(Input:string, Agent:?agent): logic ={
		case(Input):
			"enable_ducks" => if (p_Ducks.Length > 0) {
				spawn:
					ShuffleDucks(?Animate := false)
			}
			"enable_wheel" => {
				p_PrizeWheelProp.Show()
				if (TickerProp := p_PrizeWheelTickerProp?) {
					TickerProp.Show()
				}
				for (Button:p_PrizeWheelButtons) {
					Button.Enable()
				}
			}
			_ => return false
		true
	}

	## DUCKS STAND {{{
	var m_DuckIndex: int = -1
	var m_DuckedDucks: []?creative_prop = array{}
	var m_ShufflingDucks: logic = false

	ShuffleDucks<private>(?Animate:logic, ?AvoidIndex:int=-1)<suspends>: void ={
		set m_ShufflingDucks = true
		loop {
			set m_DuckIndex = GetRandomInt(0, p_Ducks.Length-1)
			if (AvoidIndex <> m_DuckIndex or p_Ducks.Length <= 1) {
				break
			}
		}
		for (Index->Prop:p_Ducks, Prop.IsValid[]) {
			Prop.Hide()
		}
		set m_DuckedDucks = m_DuckedDucks.EnforceMinLength(p_Ducks.Length, None())
		for (
			Index->Prop:m_DuckedDucks
			Index <> m_DuckIndex
			not Prop?
			DuckProp := p_Ducks[Index]
			DuckProp.IsValid[]
		) {
			SpawnRes := SpawnProp(p_DuckedDuckPropAsset, GetDuckedDuckTransform(DuckProp))
			if. set m_DuckedDucks[Index] = SpawnRes(0)
			if (not SpawnRes(0)?) {
				Log(Self, "Failed to spawn duck prop: {SpawnRes(1)}")
			}
		}
		if (Animate?, DuckProp := p_Ducks[m_DuckIndex], DuckProp.IsValid[]) {
			if (not m_DuckedDucks[m_DuckIndex]?) {
				SpawnRes := SpawnProp(p_DuckedDuckPropAsset, GetDuckedDuckTransform(DuckProp))
				if. set m_DuckedDucks[m_DuckIndex] = SpawnRes(0)
				if (not SpawnRes(0)?) {
					Log(Self, "Failed to spawn animate duck prop: {SpawnRes(1)}")
				}
			}
			if (Prop := m_DuckedDucks[m_DuckIndex]?, AnimationController := Prop.GetAnimationController[]) {
				PropTransform := Prop.GetTransform()
				Rot1 := PropTransform.Rotation.RotateBy(MakeRotation(PropTransform.Rotation.GetLocalForward(), DegreesToRadians(90.0)))
				AnimationController.PlayNewAnimation(array{
					keyframe_delta{
						DeltaLocation := vector3{}
						DeltaRotation := MakeShortestRotationBetween(PropTransform.Rotation, Rot1)
						Time := 0.3
						Interpolation := InterpolationTypes.EaseIn
					}
				})
				AnimationController.AwaitDone()
			}
		}
		if (Prop := m_DuckedDucks[m_DuckIndex]?, Prop.IsValid[]) {
			Prop.Dispose()
			if. set m_DuckedDucks[m_DuckIndex] = None()
		}
		if (Prop := p_Ducks[m_DuckIndex], Prop.IsValid[]) {
			Prop.Show()
		}
		set m_ShufflingDucks = false
	}

	OnDuckPropHit<private>(Agent:agent): void ={
		if (m_DuckIndex < 0 or m_ShufflingDucks?) {
			return
		}
		DuckIndex := m_DuckIndex
		set m_DuckIndex = -1
		p_ResourceManager.Give(Agent, p_ResourceID, p_ResourceAmnt, ?SendFeedback := true)
		p_DuckAccolade.Award(Agent)
		p_DuckHitAudioPlayer.Play(Agent)
		spawn:
			HandleDuckHit(DuckIndex)
	}

	HandleDuckHit<private>(DuckIndex:int)<suspends>: void ={
		if (DuckProp := p_Ducks[DuckIndex], DuckProp.IsValid[]) {
			DuckProp.Hide()
			DuckTransform := DuckProp.GetTransform()
			SpawnRes := SpawnProp(p_DuckedDuckPropAsset, DuckTransform)
			if. set m_DuckedDucks[DuckIndex] = SpawnRes(0)
			if (Prop := SpawnRes(0)?, AnimationController := Prop.GetAnimationController[]) {
				InitialTransform := Prop.GetTransform()
				Rot1 := InitialTransform.Rotation.RotateBy(MakeRotation(InitialTransform.Rotation.GetLocalForward(), DegreesToRadians(-120.0)))
				Rot2 := Rot1.RotateBy(MakeRotation(InitialTransform.Rotation.GetLocalForward(), DegreesToRadians(30.0)))
				AnimationController.PlayNewAnimation(array{
					keyframe_delta{
						DeltaLocation := vector3{}
						DeltaRotation := MakeShortestRotationBetween(InitialTransform.Rotation, Rot1)
						Time := 0.3
					}
					keyframe_delta{
						DeltaLocation := vector3{}
						DeltaRotation := MakeShortestRotationBetween(Rot1, Rot2)
						Time := 0.25
						Interpolation := InterpolationTypes.EaseOut
					}
				})
				# AnimationController.AwaitDone()
			} else {
				Log(Self, "Failed to spawn post-hit duck prop: {SpawnRes(1)}")
			}
		}
		ShuffleDucks(?Animate := true, ?AvoidIndex := DuckIndex)
	}

	GetDuckedDuckTransform<private>(DuckProp:creative_prop)<transacts>: transform ={
		DuckTransform := DuckProp.GetTransform()
		ToRotateBy := MakeRotation(DuckTransform.Rotation.GetLocalForward(), DegreesToRadians(-90.0))
		NewRotation := DuckTransform.Rotation.RotateBy(ToRotateBy)
		DuckTransform.WithRotation(NewRotation)
	}
	## }}}

	## PRIZE WHEEL {{{
	var m_WheelBaseRotation: ?rotation = false
	var m_WheelSpinning: ?event() = false

	OnPrizeWheelInteract<private>(Agent:agent): void = if (Player := player[Agent]) {
		if (m_WheelSpinning?) {
			return
		}
		if (p_PrizeWheelProp.IsValid[], AnimationController := p_PrizeWheelProp.GetAnimationController[]) {
			AbortEvent := event(){}
			set m_WheelSpinning = Some(AbortEvent)
			spawn:
				SpinWheel(Player, AnimationController, AbortEvent)
		} else {
			Log(Self, "Prize wheel prop is invalid")
		}
	}

	SpinWheel<private>(
		Player:player,
		AnimationController:animation_controller,
		AbortEvent:event()
	)<suspends>: void ={
		race:
			AbortEvent.Await()
			SpinWheel0(Player, AnimationController)
	}

	SpinWheel0<private>(
		Player:player,
		AnimationController:animation_controller
	)<suspends>: void ={
		Rotation := p_PrizeWheelProp.GetTransform().Rotation
		BaseRotation := m_WheelBaseRotation? or (
			set m_WheelBaseRotation = Some(Rotation)
			Rotation
		);
		if. p_PrizeWheelProp.TeleportTo[p_PrizeWheelProp.GetTransform().WithRotation(BaseRotation)]
		Axis := BaseRotation.GetLocalRight()
		RewardIndex := PickWheelReward(Player)

		Builder := spinning_wheel_builder{
			Axis := Axis
			TargetEndSegmentIndex := RewardIndex
			BaseRotation := BaseRotation
			m_PrevRotation := Rotation
		}
		Keyframes := Builder.Build()

		for (Button:p_PrizeWheelButtons) {
			Button.Disable()
		}
		p_WheelSoundEffectSpin.Play()
		AnimationController.PlayNewAnimation(Keyframes)

		AnimationController.AwaitDone()
		HandleWheelReward(Player, RewardIndex)

		Sleep(0.1)
		set m_WheelSpinning = false
		for (Button:p_PrizeWheelButtons) {
			Button.Enable()
		}
	}

	PickWheelReward<private>(Player:player): int ={
		TopTierRewardSet := array{0, 5}.ToSet()
		var RewardIndex: int = WeightedRandom[array{
			(0, 4.0)
			(1, 5.0)
			(2, 5.0)
			(3, 5.0)
			(4, 21.0)
			(5, 4.0)
			(6, 15.0)
			(7, 4.0)
			(8, 22.0)
			(9, 15.0)
		}] or 0
		var IsRewardTopTier: logic = option{TopTierRewardSet[RewardIndex]}.ToL()

		Document := PlayerDataGetDocument(Player, "przwhl")
		NewSpinCount := Document.GetInt("spcnt", 0) + 1
		Document.SetInt("spcnt", NewSpinCount)
		GotHookingReward := Document.GetBool("ghkrd", false)
		if (not IsRewardTopTier?, NewSpinCount > 3, not GotHookingReward?) {
			set RewardIndex = WeightedRandom[array{
				(0, 50.0)
				(5, 50.0)
			}] or 0
			set IsRewardTopTier = option{TopTierRewardSet[RewardIndex]}.ToL()
			Log(Self, "Picked prize wheel reward #{RewardIndex} (hook re-roll) for '{Player}'", ?Level := log_level.Debug)
		} else {
			Log(Self, "Picked prize wheel reward #{RewardIndex} for '{Player}'", ?Level := log_level.Debug)
		}
		if (IsRewardTopTier?, not GotHookingReward?) {
			Document.SetBool("ghkrd", true)
		}

		RewardIndex
	}

	HandleWheelReward<private>(Player:player, RewardIndex:int): void ={
		Log(Self, "Giving prize wheel reward #{RewardIndex} to '{Player}'", ?Level := log_level.Debug)
		case(RewardIndex):
			0 => p_ResourceManager.Give(Player, "boost", 500, ?SendFeedback := true)
			1 => WheelGiveTickets(Player, 60)
			2 => p_ResourceManager.Give(Player, "boost", 50, ?SendFeedback := true)
			3 => p_ResourceManager.Give(Player, "boost", 25, ?SendFeedback := true)
			4 => WheelGiveTickets(Player, 15)
			5 => p_ResourceManager.Give(Player, "boost", 250, ?SendFeedback := true)
			6 => p_WheelCreaturePlacer.Spawn()
			7 => WheelGiveTickets(Player, 120)
			8 => p_ResourceManager.Give(Player, "boost", 10, ?SendFeedback := true)
			9 => SpawnWheelExplode(Player)
			_ => {}

		SoundVariant := case(RewardIndex):
			0 => prize_wheel_sound_variant.WinBig
			1 => prize_wheel_sound_variant.Win
			2 => prize_wheel_sound_variant.Win
			3 => prize_wheel_sound_variant.Win
			4 => prize_wheel_sound_variant.Win
			5 => prize_wheel_sound_variant.WinBig
			6 => prize_wheel_sound_variant.Meh
			7 => prize_wheel_sound_variant.WinBig
			8 => prize_wheel_sound_variant.Win
			_ => prize_wheel_sound_variant.Meh
		AudioPlayer := case(SoundVariant):
			prize_wheel_sound_variant.Win => p_WheelSoundEffectWin
			prize_wheel_sound_variant.WinBig => p_WheelSoundEffectWinBig
			prize_wheel_sound_variant.Meh => p_WheelSoundEffectMeh
		AudioPlayer.Play(Player)
	}

	SpawnWheelExplode(Agent:agent):void=
		spawn:
			WheelExplode(Agent)

	WheelExplode(Agent:agent)<suspends>:void=
		Sleep(0.5)
		p_WheelExplosive.Explode(Agent)
		Sleep(1.0)
		p_WheelExplosive.Reset()

	WheelGiveTickets(Agent:agent, GiveAmount:int):void=
		var Amount:int =
			FloatAmount :=
				Rate := p_ResourceManager.GetAutoGenerationRate(Agent, "money")
				GiveAmount * Rate
			DefaultingFloor(FloatAmount)
		set Amount = Clamp(Amount, 5, 2000000000000)
		p_ResourceManager.Give(Agent, "money", Amount, ?SendFeedback := true)

	## }}}
}

spinning_wheel_builder<internal> := class<final> {
	Axis: vector3
	TargetEndSegmentIndex: int
	WheelSegments: int = 10
	TopSpeed: float = 8.0 # seg/s
	TopSpeedDuration: float = 3.0
	DecelerationSegments: int = 36

	BaseRotation: rotation
	var m_PrevRotation: rotation
	var m_InitialDegreesOffset: float = 0.0
	var m_Keyframes: []keyframe_delta = array{}

	Build<internal>(): []keyframe_delta ={
		CursorDegreesOffset := 18.0 # be at middle-ish
		# var RecoverDegrees: float = RadiansToDegrees(MakeComponentWiseDeltaRotation(m_PrevRotation, BaseRotation).GetAngle())
		var SpinCount: int = 0
		# if (not IsAlmostEqual[RecoverDegrees, 0.0, 0.01]) {
		# 	BaseRotWithOffset := BaseRotation.RotateBy(MakeRotation(Axis, DegreesToRadians(CursorDegreesOffset)))
		# 	set RecoverDegrees = RadiansToDegrees(MakeComponentWiseDeltaRotation(m_PrevRotation, BaseRotation).GetAngle())
		# 	set SpinCount += DefaultingFloor(RecoverDegrees / 36.0)
		# 	set m_InitialDegreesOffset = RecoverDegrees - (SpinCount * 36.0)
		# 	Print("Recovered from previous spin from {RecoverDegrees}deg: {SpinCount}x spins, {m_InitialDegreesOffset}deg")
		# } else {
			set m_InitialDegreesOffset = CursorDegreesOffset
		# }

		SpinCountFromBase := DefaultingFloor(TopSpeed * TopSpeedDuration)
		set SpinCount += SpinCountFromBase
		 
		WouldBeSpins := SpinCountFromBase + DecelerationSegments
		if (FixingSpins := Mod[(TargetEndSegmentIndex+1) - WouldBeSpins, WheelSegments], FixingSpins > 0) {
			set SpinCount += WheelSegments - FixingSpins
		}

		for (__ := 1..SpinCount) {
			PushSegmentRotation(1.0 / TopSpeed)
		}

		for (SegNo := 1..DecelerationSegments) {
			DecelerateSpeed := Lerp(TopSpeed, 1.0, ToFloat(SegNo) / ToFloat(DecelerationSegments))
			PushSegmentRotation(1.0 / DecelerateSpeed)
		}

		m_Keyframes
	}

	PushSegmentRotation<private>(Time:float): void ={
		Degrees := if (m_Keyframes.Length > 0) {
			36.0
		} else {
			36.0 + m_InitialDegreesOffset
		}
		NextRotation := m_PrevRotation.RotateBy(MakeRotation(Axis, DegreesToRadians(Degrees)))
		set m_Keyframes += array{keyframe_delta{
			DeltaLocation := vector3{}
			DeltaRotation := MakeShortestRotationBetween(m_PrevRotation, NextRotation)
			Time := Time
		}}
		set m_PrevRotation = NextRotation
	}
}
