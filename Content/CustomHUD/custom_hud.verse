using { /Fortnite.com/Devices }
using { /Fortnite.com/UI }
using { /Fortnite.com/Characters }
using { /Fortnite.com/Game }
using { /Verse.org/Simulation }
using { /Verse.org/Colors }
using { /Verse.org/Assets }
using { /UnrealEngine.com/Temporary/Diagnostics }
using { /UnrealEngine.com/Temporary/UI }
using { /UnrealEngine.com/Temporary/SpatialMath }

using { HavocVerse.Debug }
using { HavocVerse.Devices }
using { HavocVerse.UI }
using { HavocVerse.Util }
using { HavocVerse.Logging }

custom_hud_device := class<concrete><final>(creative_device, aware_device, aware_debug_device):
    @editable p_AccessControl: access_control_device = access_control_device{}

    @editable p_ResourceManager: resource_manager_device = resource_manager_device{}
    @editable p_LevelGroupsManager: level_groups_manager_device = level_groups_manager_device{}
    @editable p_LevelGroupId: string = "levels"

    @editable p_LevelLabelNumberFormatter: number_formatter_config = number_formatter_config{}
    @editable p_LevelInfoHUD: hud_message_device = hud_message_device{}
    @editable p_HealthInfoHUD: hud_message_device = hud_message_device{}

    @editable p_TicketCountFormatter: number_formatter_config = number_formatter_config{}
    @editable p_TicketRateFormatter: number_formatter_config = number_formatter_config{
        Template := "%Val%/s"
    }
    @editable p_JackoCountFormatter: number_formatter_config = number_formatter_config{}

    @editable p_LevelProgresssSeqControllers: []cinematic_sequence_device = array{}
    @editable p_HealthBarLevelSeqControllers: []cinematic_sequence_device = array{}
    @editable p_ShieldBarLevelSeqControllers: []cinematic_sequence_device = array{}

    @editable p_HauntedHouseEventDevice: wave_survival_device = wave_survival_device{}

    var m_VerseHUDInstances<private>: [player]verse_hud_instance = map{}
    var m_ResourceBoosterSources<internal>: []resource_booster_device = array{}

    LevelProgressMPC<private>: mpc_handle = mpc_handle{Context := "Level Progress"}
    HealthBarMPC<private>: mpc_handle = mpc_handle{Context := "Health Bar"}
    ShieldBarMPC<private>: mpc_handle = mpc_handle{Context := "Shield Bar"}

    CorrectionScale<internal>: float = 1.2

    DeviceMeta<override>()<transacts>:aware_device_meta=
        aware_device_meta{
            DeviceName := "custom_hud_device"
        }

    DebugStats<override>()<transacts>:aware_debug_stats=
        aware_debug_stats{
            Length := map{
                "verseHudInstances" => m_VerseHUDInstances.Length
                "resourceBoosterSources" => m_ResourceBoosterSources.Length
            }
        }

    OnBegin<override>()<suspends>:void=
        AwareInitDevice(Self, OnInit)

        p_HauntedHouseEventDevice.SetUpdatePlayerHudFn(HandlePlayerEventInfoChange)
    
    OnInit<private>()<suspends>:void=
        LevelProgressMPC.Setup(GetPlayspace(), p_LevelProgresssSeqControllers)
        HealthBarMPC.Setup(GetPlayspace(), p_HealthBarLevelSeqControllers)
        ShieldBarMPC.Setup(GetPlayspace(), p_ShieldBarLevelSeqControllers)

        p_AccessControl.SubscribeToPlayerJoinEvents(OnPlayerAdded, OnPlayerRemoved)
        p_LevelGroupsManager.SubscribeToPlayerUpdateEvent(OnPlayerExpStateChanged)

        GetPlayspace().PlayerAddedEvent().Subscribe(OnPlayerGameJoin)
        for (Player:GetPlayspace().GetPlayers()) {
            OnPlayerGameJoin(Player)
        }

        spawn:
            StartUpdateLoop()

        AwareSubscribeToPostInit(OnPostInit)

    OnPostInit<private>():void=
        set m_ResourceBoosterSources = array{}
        for (AwareDevice:AwareListDevices(), ResourceBoosterDevice := resource_booster_device[AwareDevice]) {
            set m_ResourceBoosterSources += array{ResourceBoosterDevice}
        }

    StartUpdateLoop<private>()<suspends>:void=
        loop {
            Sleep(0.5)
            for (Player:GetPlayspace().GetPlayers()) {
                UpdatePlayer(Player)
            }
        }

    OnPlayerGameJoin<private>(Player:player): void ={
        if (FortCharacter := Player.GetFortCharacter[]) {
            Closure := player_events_closure{
                Device := Self
                Player := Player
            }
            FortCharacter.DamagedEvent().Subscribe(Closure.OnDamaged)
            FortCharacter.HealedEvent().Subscribe(Closure.OnHealed)
            FortCharacter.DamagedShieldEvent().Subscribe(Closure.OnShieldDamaged)
            FortCharacter.HealedShieldEvent().Subscribe(Closure.OnShieldHealed)
        }
    }

    OnPlayerAdded<private>(Player:player): void ={
        if (not m_VerseHUDInstances[Player]) {
            VerseHUDInstance := CreateVerseHUDInstance(Player)
            set m_VerseHUDInstances = AddToMap(m_VerseHUDInstances, Player, VerseHUDInstance)
            GlobalUIAddWidget(Player, VerseHUDInstance.Canvas)
        }
        UpdatePlayer(Player)
    }

    OnPlayerRemoved<private>(Player:player, GameQuit:logic): void ={
        HideUMGWidgets(Player)
        RemoveRes := m_VerseHUDInstances.RemoveByKey(Player)
        set m_VerseHUDInstances = RemoveRes(0)
        if (Instance := RemoveRes(1)?) {
            GlobalUIRemoveWidget(Player, Instance.Canvas)
        }
    }

    OnPlayerExpStateChanged<private>(Player:player, __:level_exp_pair):void=
        if (p_AccessControl.IsAllowed[Player]) {
            UpdateLevelInfo(Player)
        }

    HandlePlayerHealthInfoChange<internal>(Player:player): void ={
        if (p_AccessControl.IsAllowed[Player]) {
            UpdateHealthInfo(Player)
        }
    }

    HandlePlayerEventInfoChange<internal>(Player:player): void ={
        if (p_AccessControl.IsAllowed[Player], Instance := m_VerseHUDInstances[Player]) {
            Instance.UpdateEventInfo()
        }
    }

    UpdatePlayer<private>(Player:player):void=
        if (not p_AccessControl.IsAllowed[Player]) {
            return
        }
        if (Instance := m_VerseHUDInstances[Player]) {
            Instance.UpdateResources()
            Instance.UpdateBoosterInfo()
            Instance.UpdateEventInfo()
        }
        UpdateLevelInfo(Player)
        UpdateHealthInfo(Player)

    HideUMGWidgets<private>(Player:player):void=
        p_LevelInfoHUD.Hide(Player)
        p_HealthInfoHUD.Hide(Player)

    UpdateLevelInfo<private>(Player:player):void=
        LevelExpPair := p_LevelGroupsManager.GetLevelExp(Player, p_LevelGroupId)
        ExpectedExp := p_LevelGroupsManager.GetTargetExp[p_LevelGroupId, LevelExpPair] or -1

        RawProgress := ToFloat(LevelExpPair.ExpRest) / ToFloat(ExpectedExp)
        LevelProgressMPC.SetProgress(Player, RawProgress)

        LevelTitle := {
            "LEVEL {LevelExpPair.Level}"
        }
        LevelProgressInfo := if (RawProgress > 1.0) {
            "(MAX)"
        } else {
            CurrentExpStr := p_LevelLabelNumberFormatter.Format(LevelExpPair.ExpRest)
            ExpectedExpStr := p_LevelLabelNumberFormatter.Format(ExpectedExp)
            "{CurrentExpStr}/{ExpectedExpStr}"
        }

        Payload := "{LevelTitle}\n{LevelProgressInfo}"
        p_LevelInfoHUD.Show(Player, ToMessage(Payload))

    UpdateHealthInfo<private>(Player:player): void ={
        if (FortCharacter := Player.GetFortCharacter[]) {
            HealthValue := DefaultingFloor(FortCharacter.GetHealth())
            ShieldValue := DefaultingFloor(FortCharacter.GetShield())

            HealthProgress := FortCharacter.GetHealth() / FortCharacter.GetMaxHealth()
            ShieldProgress := FortCharacter.GetShield() / FortCharacter.GetMaxShield()
            HealthBarMPC.SetProgress(Player, HealthProgress)
            ShieldBarMPC.SetProgress(Player, ShieldProgress)

            Payload := "{ShieldValue}\n{HealthValue}"
            p_HealthInfoHUD.Show(Player, ToMessage(Payload))
        }
    }

    CreateVerseHUDInstance<private>(Player:player): verse_hud_instance ={
        FontBeastBB := GetCustomFontOrDefault("FeastBB")
        GapSize := 6.0 * CorrectionScale

        TicketsCountTextBlock := CreateCustomFontTextBlock(FontBeastBB, "1,000,000", ?FontSize := 30.0 * CorrectionScale)
        JackoCountTextBlock := CreateCustomFontTextBlock(FontBeastBB, "10,000", ?FontSize := 30.0 * CorrectionScale, ?TextColor := MakeColorFromHex("#e38a00"))
        TicketsRateTextBlock := CreateCustomFontTextBlock(FontBeastBB, "4,000/s", ?FontSize := 16.0 * CorrectionScale, ?TextColor := MakeColorFromHex("#ac75e4"))
        ResourcesRow := overlay:
            Slots := array{
                overlay_slot:
                    Widget := overlay:
                        Slots := array{
                            overlay_slot:
                                Widget := texture_block:
                                    DefaultImage := CustomHUD.Src.res_tickets_bg
                                    DefaultDesiredSize := vector2{X := 233.0 * CorrectionScale, Y := 60.0 * CorrectionScale}
                            overlay_slot:
                                Widget := TicketsCountTextBlock.Widget
                                Padding := margin{
                                    Left := 66.0 * CorrectionScale
                                    Top := 8.0
                                }
                                HorizontalAlignment := horizontal_alignment.Left
                                VerticalAlignment := vertical_alignment.Top
                            overlay_slot:
                                Widget := TicketsRateTextBlock.Widget
                                Padding := margin{
                                    Left := 66.0 * CorrectionScale
                                    Top := 45.0
                                }
                                HorizontalAlignment := horizontal_alignment.Left
                                VerticalAlignment := vertical_alignment.Top
                        }
                    HorizontalAlignment := horizontal_alignment.Left
                    VerticalAlignment := vertical_alignment.Fill
                overlay_slot:
                    Widget := overlay:
                        Slots := array{
                            overlay_slot:
                                Widget := texture_block:
                                    DefaultImage := CustomHUD.Src.res_jacko_bg
                                    DefaultDesiredSize := vector2{X := 200.5 * CorrectionScale, Y := 60.0 * CorrectionScale}
                            overlay_slot:
                                Widget := JackoCountTextBlock.Widget
                                Padding := margin{
                                    Left := 66.0 * CorrectionScale
                                    Top := 18.0
                                }
                                HorizontalAlignment := horizontal_alignment.Left
                                VerticalAlignment := vertical_alignment.Top
                        }
                    Padding := margin{
                        Left := GapSize
                    }
                    HorizontalAlignment := horizontal_alignment.Right
                    VerticalAlignment := vertical_alignment.Fill
            }
        
        BarsGroup := overlay:
            Slots := array{
                overlay_slot:
                    Widget := texture_block:
                        DefaultImage := CustomHUD.Src.bars_group_bg
                        DefaultDesiredSize := vector2{X := 441.5 * CorrectionScale, Y := 92.5 * CorrectionScale}
                overlay_slot:
                    Widget := texture_block:
                        DefaultImage := CustomHUD.Src.heart
                        DefaultDesiredSize := vector2{X := 44.5 * CorrectionScale, Y := 39.5 * CorrectionScale}
                    Padding := margin{
                        Left := 14.0 * CorrectionScale
                        Top := 43.0 * CorrectionScale
                    }
                    HorizontalAlignment := horizontal_alignment.Left
                    VerticalAlignment := vertical_alignment.Top
            }

        BoostInfoPlaceholder := widget_placeholder{}
        BottomLeftStack := stack_box:
            Orientation := orientation.Vertical
            Slots := array{
                stack_box_slot:
                    Widget := BoostInfoPlaceholder.Overlay
                    HorizontalAlignment := horizontal_alignment.Left
                    VerticalAlignment := vertical_alignment.Top
                stack_box_slot:
                    Widget := ResourcesRow
                    Padding := margin{
                        Top := GapSize
                    }
                    HorizontalAlignment := horizontal_alignment.Fill
                    VerticalAlignment := vertical_alignment.Top
                stack_box_slot:
                    Widget := BarsGroup
                    Padding := margin{
                        Top := GapSize
                    }
                    HorizontalAlignment := horizontal_alignment.Left
                    VerticalAlignment := vertical_alignment.Top
            }

        Canvas := canvas:
            Slots := array{canvas_slot{
                Widget := BottomLeftStack
                Anchors := anchors{
                    Minimum := vector2{X := 0.0, Y := 1.0}
                    Maximum := vector2{X := 0.0, Y := 1.0}
                }
                Alignment := vector2{X := 0.0, Y := 1.0}
                Offsets := margin{
                    Left := 64.0
                    Top := -75.0
                }
            }}

        verse_hud_instance{
            Device := Self
            Player := Player
            Canvas := Canvas
            TicketsCountTextBlock := TicketsCountTextBlock
            TicketsRateTextBlock := TicketsRateTextBlock
            JackoCountTextBlock := JackoCountTextBlock
            BoostInfoPlaceholder := BoostInfoPlaceholder
        }
    }

verse_hud_instance<internal> := class<final> {
    Device: custom_hud_device
    Player: player
    Canvas: canvas

    TicketsCountTextBlock: custom_font_text_block
    TicketsRateTextBlock: custom_font_text_block
    JackoCountTextBlock: custom_font_text_block
    BoostInfoPlaceholder: widget_placeholder

    var m_BoostInfo: ?boost_info = false
    var m_HHEventInfo: ?hh_event_info = false

    UpdateResources<internal>(): void ={
        ResourceManager := Device.p_ResourceManager

        TicketCount := ResourceManager.GetAmount(Player, "money")
        TicketRate := ResourceManager.GetAutoGenerationRate(Player, "money")
        JackoCount := ResourceManager.GetAmount(Player, "boost")

        TicketsCountTextBlock.SetText(Device.p_TicketCountFormatter.Format(TicketCount))
        TicketsRateTextBlock.SetText(Device.p_TicketRateFormatter.Format(TicketRate))
        JackoCountTextBlock.SetText(Device.p_JackoCountFormatter.Format(JackoCount))
    }

    UpdateBoosterInfo<internal>(): void ={
        Multiplier := Device.p_ResourceManager.GetEffectiveMultiplier(Player, "money")
        BoosterLevel := Clamp(DefaultingFloor(Multiplier), 0, 99)

        if (BoosterLevel <= 1) {
            if (m_BoostInfo?) {
                BoostInfoPlaceholder.SetWidget(None())
                set m_BoostInfo = None()
            }
            return
        }

        var FoundRemainingTime: ?float = None()
        for (
            ResourceBoosterDevice:Device.m_ResourceBoosterSources
            ResourceBoosterDevice.p_MultiplierChannel = "default"
            ResourceBoosterDevice.p_AutoGenerationManager.IncludesPlayer[Player]
        ) {
            DeviceRemainingTime := ResourceBoosterDevice.GetRemainingTime(Player)
            if (PrevRemainingTime := FoundRemainingTime?) {
                set FoundRemainingTime = Some(PrevRemainingTime + DeviceRemainingTime)
            } else {
                set FoundRemainingTime = Some(DeviceRemainingTime)
            }
        }

        TimerText := if (RemainingTime := FoundRemainingTime?, RemainingTime > 0.0) {
            ToTimerMMSS(DefaultingCeil(RemainingTime))
        } else { 
            ""
        }
        BoostInfo := if (l_BoostInfo := m_BoostInfo?) {
            l_BoostInfo
        } else {
            CreateBoostInfo()
        }
        BoostInfo.BoostLabelTextBlock.SetText("{BoosterLevel}x Boost!")
        BoostInfo.BoostTimerTextBlock.SetText(TimerText)
    }

    CreateBoostInfo<private>(): boost_info ={
        FontBeastBB := GetCustomFontOrDefault("FeastBB")

        BoostLabelTextBlock := CreateCustomFontTextBlock(FontBeastBB, "", ?FontSize := 17.0 * Device.CorrectionScale, ?TextColor := MakeColorFromHex("#dfdfdf"))
        BoostTimerTextBlock := CreateCustomFontTextBlock(FontBeastBB, "", ?FontSize := 17.0 * Device.CorrectionScale, ?TextColor := MakeColorFromHex("#ffea57"))
        Widget := overlay:
            Slots := array{
                overlay_slot:
                    Widget := texture_block:
                        DefaultImage := CustomHUD.Src.resource_boost_bg
                        DefaultDesiredSize := vector2{X := 159.0 * Device.CorrectionScale, Y := 47.0 * Device.CorrectionScale}
                overlay_slot:
                    Widget := BoostLabelTextBlock.Widget
                    Padding := margin{
                        Left := 60.0 * Device.CorrectionScale
                        Top := 7.0
                    }
                    HorizontalAlignment := horizontal_alignment.Left
                    VerticalAlignment := vertical_alignment.Top
                overlay_slot:
                    Widget := BoostTimerTextBlock.Widget
                    Padding := margin{
                        Left := 60.0 * Device.CorrectionScale
                        Top := 30.0
                    }
                    HorizontalAlignment := horizontal_alignment.Left
                    VerticalAlignment := vertical_alignment.Top
            }
        BoostInfo := boost_info{
            BoostLabelTextBlock := BoostLabelTextBlock
            BoostTimerTextBlock := BoostTimerTextBlock
        }
        set m_BoostInfo = Some(BoostInfo)
        BoostInfoPlaceholder.SetWidget(Widget)
        BoostInfo
    }

    UpdateEventInfo<internal>(): void ={
        if (
            Device.p_HauntedHouseEventDevice.IsPlaying[Player]
            ProgressInfo := Device.p_HauntedHouseEventDevice.GetProgressInfo[]
        ) {
            EventInfo := m_HHEventInfo? or CreateHHEventInfo()
            EventInfo.WaveTextBlock.SetText("Wave {ProgressInfo.CurrentWaveIndex + 1}")
            # intentionally using WaveIndex/WaveCount so the current in-progress isn't counted as completed
            EventInfo.HouseProgressBar.SetProgress(ToFloat(ProgressInfo.CurrentWaveIndex) / ToFloat(ProgressInfo.TotalWaves))
            EventInfo.WaveProgressBar.SetProgress(ProgressInfo.WaveProgress)
        } else if (EventInfo := m_HHEventInfo?) {
            Canvas.RemoveWidget(EventInfo.CanvasWidget)
            set m_HHEventInfo = None()
        }
    }

    CreateHHEventInfo<private>(): hh_event_info ={
        HouseProgressBar := MakeProgressBar(
            ?Width := 795.0,
            ?Height := 22.0,
            ?FillTexture := CustomHUD.Src.HauntedHouse.hh_house_pg_fill,
            ?FillStart := Some(CustomHUD.Src.HauntedHouse.hh_house_pg_start, 10.5),
            ?FillEnd := Some(CustomHUD.Src.HauntedHouse.hh_house_pg_end, 10.5),
            ?Tailing := false
        )
        HouseProgressBar.SetProgress(0.5)

        HouseBar := overlay:
            Slots := array{
                overlay_slot:
                    Widget := texture_block:
                        DefaultImage := CustomHUD.Src.HauntedHouse.hh_house_progress_bg
                        DefaultDesiredSize := HouseProgressBar.Size
                    HorizontalAlignment := horizontal_alignment.Left
                    VerticalAlignment := vertical_alignment.Top
                overlay_slot:
                    Widget := HouseProgressBar.Widget
                    HorizontalAlignment := horizontal_alignment.Left
                    VerticalAlignment := vertical_alignment.Top
                overlay_slot:
                    Widget := texture_block:
                        DefaultImage := CustomHUD.Src.HauntedHouse.hh_house_pg_label
                        DefaultDesiredSize := vector2{X := 114.0, Y := 12.0}
                    HorizontalAlignment := horizontal_alignment.Center
                    VerticalAlignment := vertical_alignment.Center
            }

        WaveProgressBar := MakeProgressBar(
            ?Width := 794.0,
            ?Height := 28.5,
            ?FillTexture := CustomHUD.Src.HauntedHouse.hh_wave_pg_fill,
            ?FillStart := Some(CustomHUD.Src.HauntedHouse.hh_wave_pg_start, 13.5),
            ?FillEnd := Some(CustomHUD.Src.HauntedHouse.hh_wave_pg_end, 13.5),
            ?Tailing := false
        )
        WaveProgressBar.SetProgress(0.7)

        FontBeastBB := GetCustomFontOrDefault("FeastBB")
        WaveTextBlock := CreateCustomFontTextBlock(FontBeastBB, "", ?FontSize := 24.0)

        WaveBar := overlay:
            Slots := array{
                overlay_slot:
                    Widget := texture_block:
                        DefaultImage := CustomHUD.Src.HauntedHouse.hh_wave_progress_bg
                        DefaultDesiredSize := WaveProgressBar.Size
                    HorizontalAlignment := horizontal_alignment.Left
                    VerticalAlignment := vertical_alignment.Top
                overlay_slot:
                    Widget := WaveProgressBar.Widget
                    HorizontalAlignment := horizontal_alignment.Left
                    VerticalAlignment := vertical_alignment.Top
                overlay_slot:
                    Widget := WaveTextBlock.Widget
                    Padding := margin{
                        Left := 10.0
                    }
                    HorizontalAlignment := horizontal_alignment.Left
                    VerticalAlignment := vertical_alignment.Center
            }

        CanvasWidget := overlay:
            Slots := array{
                overlay_slot:
                    Widget := texture_block:
                        DefaultImage := CustomHUD.Src.HauntedHouse.hh_wave_progress_shadow
                        DefaultDesiredSize := vector2{X := 855.0, Y := 92.5}
                    Padding := margin{
                        Top := 5.0
                    }
                    HorizontalAlignment := horizontal_alignment.Center
                    VerticalAlignment := vertical_alignment.Top
                overlay_slot:
                    Widget := HouseBar
                    HorizontalAlignment := horizontal_alignment.Center
                    VerticalAlignment := vertical_alignment.Top
                overlay_slot:
                    Widget := WaveBar
                    Padding := margin{
                        Top := 35.0
                    }
                    HorizontalAlignment := horizontal_alignment.Center
                    VerticalAlignment := vertical_alignment.Top
            }

        Canvas.AddWidget(canvas_slot{
            Widget := CanvasWidget
            Anchors := anchors{
                Minimum := vector2{X := 0.5, Y := 0.0}
                Maximum := vector2{X := 0.5, Y := 0.0}
            }
            Alignment := vector2{X := 0.5, Y := 0.0}
            Offsets := margin{
                Top := 80.0
            }
        })
        EventInfo := hh_event_info{
            CanvasWidget := CanvasWidget
            WaveTextBlock := WaveTextBlock
            HouseProgressBar := HouseProgressBar
            WaveProgressBar := WaveProgressBar
        }
        set m_HHEventInfo = Some(EventInfo)
        EventInfo
    }
}

hh_event_info<internal> := class<final> {
    CanvasWidget: widget
    WaveTextBlock: custom_font_text_block
    HouseProgressBar: progress_bar
    WaveProgressBar: progress_bar
}

progress_bar<internal> := class<final> {
    Widget: widget
    Size: vector2
    Tailing: logic
    FillTail: ?widget
    TailWidth: float
    FillHead: ?widget
    HeadWidth: float
    FillWidget: widget

    SetProgress<public>(Progress:float): void ={
        ClampedProgress := Clamp(Progress, 0.0, 1.0)

        var TargetWidth: float = Size.X * ClampedProgress
        if (Tail := FillTail?) {
            Tail.SetVisibility(if (TargetWidth > 0.0) {
                widget_visibility.Visible
            } else {
                widget_visibility.Collapsed
            })
            set TargetWidth -= TailWidth
        }

        if (Tailing?) {
            if (Head := FillHead?) {
                Head.SetVisibility(if (TargetWidth > 0.0) {
                    widget_visibility.Visible
                } else {
                    widget_visibility.Collapsed
                })
                set TargetWidth -= HeadWidth
            }
            SetFillWidth(Max(TargetWidth, 0.0))
        } else {
            EndAtThreshold := Size.X - HeadWidth
            if (TargetWidth > EndAtThreshold) {
                SetFillWidth(1.0)
                if (Head := FillHead?) {
                    Head.SetVisibility(widget_visibility.Visible)
                }
            } else {
                SetFillWidth(Max(TargetWidth, 0.0))
                if (Head := FillHead?) {
                    Head.SetVisibility(widget_visibility.Collapsed)
                }
            }
        }
    }

    SetFillWidth<private>(FillWidth:float): void ={
        if (ColorBlock := color_block[FillWidget]) {
            NewSize := ColorBlock.GetDesiredSize().WithX(FillWidth)
            ColorBlock.SetDesiredSize(NewSize)
        } else if (TextureBlock := texture_block[FillWidget]) {
            NewSize := TextureBlock.GetDesiredSize().WithX(FillWidth)
            TextureBlock.SetDesiredSize(NewSize)
        }
    }
}

MakeProgressBar<internal>(
    ?Width: float,
    ?Height: float,
    ?FillTexture: texture,
    ?FillStart: ?tuple(texture,float) = false,
    ?FillEnd: ?tuple(texture,float) = false,
    ?Tailing: logic = true
): progress_bar ={
    Stack := stack_box:
        Orientation := orientation.Horizontal

    FillTail := if (Tup := FillStart?) {
        Image := texture_block:
            DefaultImage := Tup(0)
            DefaultDesiredSize := vector2{X := Tup(1), Y := Height}
        Stack.AddWidget(stack_box_slot{
            Widget := Image
            HorizontalAlignment := horizontal_alignment.Left
            VerticalAlignment := vertical_alignment.Center
        })
        Some(Image)
    } else {
        None()
    }

    FillWidget := texture_block:
        DefaultImage := FillTexture
        DefaultDesiredSize := vector2{X := 0.0, Y := Height}
        DefaultHorizontalTiling := image_tiling.Repeat
        DefaultVerticalTiling := image_tiling.Stretch
    Stack.AddWidget(stack_box_slot{
        Widget := FillWidget
        HorizontalAlignment := horizontal_alignment.Left
        VerticalAlignment := vertical_alignment.Center
    })

    FillHead := if (Tup := FillEnd?) {
        Image := texture_block:
            DefaultImage := Tup(0)
            DefaultDesiredSize := vector2{X := Tup(1), Y := Height}
        Stack.AddWidget(stack_box_slot{
            Widget := Image
            HorizontalAlignment := horizontal_alignment.Left
            VerticalAlignment := vertical_alignment.Center
        })
        Some(Image)
    } else {
        None()
    }

    progress_bar{
        Widget := Stack
        Size := vector2{X := Width, Y := Height}
        Tailing := Tailing
        FillTail := FillTail
        TailWidth := (Tup := FillStart?; Tup(1)) or 0.0
        FillHead := FillHead
        HeadWidth := (Tup := FillEnd?; Tup(1)) or 0.0
        FillWidget := FillWidget
    }
}

boost_info<internal> := class<final> {
    BoostLabelTextBlock: custom_font_text_block
    BoostTimerTextBlock: custom_font_text_block
}

player_events_closure<internal> := class<final> {
    Device: custom_hud_device
    Player: player

    OnDamaged<public>(Result:damage_result): void ={
        Device.HandlePlayerHealthInfoChange(Player)
    }

    OnHealed<public>(Result:healing_result): void ={
        Device.HandlePlayerHealthInfoChange(Player)
    }

    OnShieldDamaged<public>(Result:damage_result): void ={
        Device.HandlePlayerHealthInfoChange(Player)
    }

    OnShieldHealed<public>(Result:healing_result): void ={
        Device.HandlePlayerHealthInfoChange(Player)
    }
}

hh_recap_menu_component<public> := class<final>(ui_menu_component) {
    Won<public>: logic
    MaxWave<public>: int
    AliveDuration<public>: float
    ZombieKills<public>: int
    EarnedJackos<public>: int

    Create<override>(): widget ={
        FontBeastBB := GetCustomFontOrDefault("FeastBB")
        ClrPurple := MakeColorFromHex("#9d6bd0")
        ClrYellow := MakeColorFromHex("#f9e355")
        ClrRed := MakeColorFromHex("#ff5f5f")
        ClrGreen := MakeColorFromHex("#67ff5f")
        ClrOrange := MakeColorFromHex("#e38a00")
        ClrGray := MakeColorFromHex("#dad9d9")

        InfoLines := stack_box:
            Orientation := orientation.Vertical
            Slots := array{
                stack_box_slot:
                    Widget := MakeValueTextLine("Reached", "Wave {MaxWave}", ClrPurple)
                    HorizontalAlignment := horizontal_alignment.Left
                    VerticalAlignment := vertical_alignment.Top
                stack_box_slot:
                    Widget := MakeValueTextLine("Fought For", "{ToTimerMMSS(DefaultingFloor(AliveDuration))}s", ClrYellow)
                    Padding := margin{
                        Top := 8.0
                    }
                    HorizontalAlignment := horizontal_alignment.Left
                    VerticalAlignment := vertical_alignment.Top
                stack_box_slot:
                    Widget := MakeValueTextLine("Killed", "{ZombieKills} zombies", ClrPurple)
                    Padding := margin{
                        Top := 8.0
                    }
                    HorizontalAlignment := horizontal_alignment.Left
                    VerticalAlignment := vertical_alignment.Top
                stack_box_slot:
                    Widget := if (Won?) {
                        CreateCustomFontTextBlock(FontBeastBB, "You Won!", ?FontSize := 30.0, ?TextColor := ClrGreen)
                    } else {
                        CreateCustomFontTextBlock(FontBeastBB, "Zombies Won, You Lost!", ?FontSize := 30.0, ?TextColor := ClrRed)
                    }.Widget
                    Padding := margin{
                        Top := 8.0
                    }
                    HorizontalAlignment := horizontal_alignment.Left
                    VerticalAlignment := vertical_alignment.Top
            }

        JackosFormatter := number_formatter_config{}
        RewardAmount := stack_box:
            Orientation := orientation.Horizontal
            Slots := array{
                stack_box_slot:
                    Widget := CreateCustomFontTextBlock(FontBeastBB, JackosFormatter.Format(EarnedJackos), ?FontSize := 40.0, ?TextColor := ClrOrange).Widget
                    HorizontalAlignment := horizontal_alignment.Left
                    VerticalAlignment := vertical_alignment.Top
                stack_box_slot:
                    Widget := texture_block:
                        DefaultImage := CustomHUD.Icons.pumpkin
                        DefaultDesiredSize := vector2{X := 34.0, Y := 33.5}
                    Padding := margin{
                        Left := 10.0
                    }
                    HorizontalAlignment := horizontal_alignment.Left
                    VerticalAlignment := vertical_alignment.Top
            }

        RewardInfo := stack_box:
            Orientation := orientation.Vertical
            Slots := array{
                stack_box_slot:
                    Widget := CreateCustomFontTextBlock(FontBeastBB, "Rewards", ?FontSize := 30.0, ?TextColor := ClrGray).Widget
                    HorizontalAlignment := horizontal_alignment.Center
                    VerticalAlignment := vertical_alignment.Top
                stack_box_slot:
                    Widget := RewardAmount
                    Padding := margin{
                        Top := 6.0
                    }
                    HorizontalAlignment := horizontal_alignment.Center
                    VerticalAlignment := vertical_alignment.Top
            }

        CloseButton := CreateFortButton(fort_button_type.Loud, ToMessage("Close"))
        BindToViewerClose(CloseButton.OnClick())
        Body := canvas:
            Slots := array{
                canvas_slot:
                    Widget := InfoLines
                    Anchors := anchors{
                        Minimum := vector2{X := 0.0, Y := 0.0}
                        Maximum := vector2{X := 0.0, Y := 0.0}
                    }
                    Alignment := vector2{X := 0.0, Y := 0.0}
                    Offsets := margin{
                        Top := 90.0
                        Left := 150.0
                    }
                canvas_slot:
                    Widget := RewardInfo
                    Anchors := anchors{
                        Minimum := vector2{X := 0.5, Y := 0.0}
                        Maximum := vector2{X := 0.5, Y := 0.0}
                    }
                    Alignment := vector2{X := 0.5, Y := 0.0}
                    Offsets := margin{
                        Top := 265.0
                    }
                canvas_slot:
                    Widget := ApplySizeHint(CloseButton, vector2{X := 230.0, Y := 40.0})
                    Anchors := anchors{
                        Minimum := vector2{X := 0.5, Y := 0.0}
                        Maximum := vector2{X := 0.5, Y := 0.0}
                    }
                    Alignment := vector2{X := 0.5, Y := 0.0}
                    Offsets := margin{
                        Top := 370.0
                    }
            }

        overlay:
            Slots := array{
                overlay_slot:
                    Widget := texture_block:
                        DefaultImage := CustomHUD.Src.HauntedHouse.hh_recap_bg
                        DefaultDesiredSize := vector2{X := 604.5, Y := 502.0}
                    HorizontalAlignment := horizontal_alignment.Left
                    VerticalAlignment := vertical_alignment.Top
                overlay_slot:
                    Widget := Body
                    Padding := margin{
                        Top := 25.0
                        Left := 25.0
                        Right := 29.0
                        Bottom := 29.0
                    }
                    HorizontalAlignment := horizontal_alignment.Fill
                    VerticalAlignment := vertical_alignment.Fill
            }
    }

    MakeValueTextLine<private>(Label:string, Value:string, ValueColor:color): widget ={
        FontBeastBB := GetCustomFontOrDefault("FeastBB")
		MakeTextRow(array{
            CreateCustomFontTextBlock(FontBeastBB, "{Label} ", ?FontSize := 30.0, ?TextColor := NamedColors.White).Widget
            CreateCustomFontTextBlock(FontBeastBB, Value, ?FontSize := 30.0, ?TextColor := ValueColor).Widget
		})
    }
}
